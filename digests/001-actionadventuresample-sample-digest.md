---
id: digest-001
kind: digest
---

# action_adventure_sample — sample digest

> **Extractor provenance.** `tools/extract-template.py` (Pass A) + UnrealMCP on
> `localhost:8000` (Pass B). Default level
> `/c1b682b0-4d00-41ab-8ca2-8444fa187773/action_adventure_sample` (**Observed**).
> Hub: Building BRICK Islands
> + index BRICK Templates
> + island tutorial BRICK Action Adventure
> + companion Collectible Devices
> + Assembly Devices
> + Working with BRICK Islands
> + theme Rules (publishing, not a device syllabus).
> `docsUrl` empty. Not on the spreadsheet. Compat **43.00**. `featureSet`: **BrickGameplay**.

| | |
|---|---|
| Classification | **BRICK learning island** — class-path action adventure (Warrior / Blaster / Thief) with persistable Verse quests, stud economy, Assembly puzzles, boss attacks |
| Content type | **kit + wrap** (123 authored vs 1591 OFPA; `/BRICK_Cane` + `/SPACE_ASSETS` + `/NEIGHBORHOOD_Assets` + `/ENV_PACK_UnplayableTestMap`) |
| Verse | **13 files** (18 devices placed) |
| Fab `.uref` | **0** |

## What it is

The picker name is **Action Adventure Template** (BRICK). Docs H1 is **BRICK Action Adventure Template**. Disk folder is `action_adventure_sample`. The island is a **Starting HUB plus three class worlds and a quest-giver boss boss**: talk to a quest giver, collect / assemble / fight, return for studs. Teaching that survives deletion is **linear persistable Verse quest dialog**, **Item Spawner teleport pool for stud bursts**, **tagged Collectible → Item Granter**, **npc_attack_controller stun/AOE**, and **destructible-prop `IsValid` poll**. Do not duplicate: homestead-claim-spawn.md (this is **not** `village_starter_template`; **2** BRICK spawners both `useAsIslandStart=true`, no claim pads), phone-copy-paste-prefab-quickbar.md / BRICK-fortnite-mix-no-scale.md (Phone sandbox; Home Builder has **0** Assembly), sequential-collect-race.md (ordered Verse spawners), collectible-tracker-endgame.md / prop-mover-device-wiring.md (**0** End Game / Prop Mover; **2** Prop Manipulation), cubegrid-then-color-grids.md (**0** CubeGrid), conversation-device-verse-quest.md (**0** Conversation Device), beacon-volume-popup-trail.md (15 beacons are quest `AddToShowList`, not a Patchwork trail). ExtraAssets HUD stays **four** — this HUD Controller overrides PlayerInfo from **in-project** `Content/UI/WBP_BRICK_PlayerInfoStack`.

## Scale (**Observed** — `summary.txt` + live)

| | |
|---|---|
| Authored assets | **123** (Audio, Cinematics, BRICKQuests, BRICKModule_StudRewards, Materials, NPCs, UI, VFX + GameFeatureData / HLOD) |
| OFPA | **1591** |
| Maps | `action_adventure_sample.umap` (only) |
| Verse files | **13** |
| Verse devices placed | **18** |
| Data layers | **0** |
| Fab `.uref` | **0** |
| Compatibility | **43.00** |
| Matchmaking | **16** / live `teamSize=Dynamic` / `teamFillOption=MustFill` |
| Streaming | Partitioned + actor folders + OFPA. Umap **omits** `LevelHasStreamingDisabled` (IceBase/Social camp, **not** Music/Obstacle/Home Builder). Sampled `bIsSpatiallyLoaded=false`. `load_actors` required. `find_actors` by `"Verse"` / `"Character Spawner"` returned **0** — labels are **NPC Spawner**, **Assembly Device**, **Quest Giver**; types `ScriptDevice` / `Device_CharacterSpawner_C`. `folderPath` often **None**. Live `get_folders` returned the **full** Outliner (Starting HUB / Level_Warrior / Level_Thief / Level_Blaster / Level_Boss / Demo Area) unlike other BRICK Background-only. |

Empty `Developers/felixkupis/Collections/`. First-party scratch folder.

`experimental.*` (Python / toolsets / Scene Graph) in `.uefnproject` is **Unverified**. Content-side: **13** `.verse`, **0** `__ExternalEntities__`. Plugin `EnableSceneGraph: true` with no Scene Graph entities.

Largest file well under GHES 100 MB. Whole folder ~**88.5 MB**.

## Level organisation (**Observed** — live `get_folders` + finds)

```
Starting HUB/     spawn, class selectors, PlayerCharacter quest, achievements
Level_Warrior/    candy-cane house (`/BRICK_Cane`)
Level_Thief/      NEIGHBORHOOD disguise house
Level_Blaster/    space station (`/SPACE_ASSETS`)
Level_Boss/       quest-giver boss arena
Demo Area/        Verse class toolboxes (8 BP_DemoDisplay)
```

**36** `Device_Billboard_V2`. Syllabus is both the public Action Adventure page **and** the boards (`_raw/action_adventure_sample/billboard-texts.md`). Demo boards name Verse classes; world boards are FantasyAdventure/DesktopAssistant-style hints. `text` is on the **actor**. Event graph: `_raw/action_adventure_sample/event-graph.md`.

## Device inventory (the design)

| Count | Class | Role |
|---|---|---|
| 129 | `Device_BRICKItemSpawner` | Stud pool (toss / gravity tags) + PLAY pickups |
| 86 | `BRICK_Device_Collectible` | World studs; live `BRICKCollectibleObject=Coin` |
| 36 | `Device_Billboard_V2` | Demo toolboxes + in-world hints |
| 30 | `Device_Barrier_V2_Placed` | Class / hub gates |
| 29 | `Device_BRICKAssemble` | Puzzle / trophy builds; `On Assembled` → Tracker |
| 27 | `BP_Device_DamageVolume_V2` | Lasers / pits (not the stud grant) |
| 26 | `Device_ConditionalButton_V2` | Quest / inventory gates; stud-count poll |
| 24 | `Device_CinematicSequence` | Hints, fireplace, boss intro, Blaster panels |
| 24 | `Device_CRD_AudioPlayer` | Per-zone beds / stingers |
| 22 | `Device_Trigger_V2` | Resets, assembly, progress |
| 18 | `VerseDevice` | Quest / stud / destructible / boss |
| 17 | `Device_Button_V2` | Interact + demo reset |
| 16 | `Device_Switch_V2` | Persistent achievements vs UI/NPC toggles |
| 16 | `Device_Teleporter_V2` | Hub ↔ class worlds |
| 15 | `Device_Beacon_V2` | Verse `AddToShowList` while a quest is active |
| 13 | `Device_CharacterSpawner` | QuestGiver ×2, Boss, Enemies, BlastEnemy |
| 10 | `Device_Tracker_V2` | `statToTrack=None`; titles empty at rest |
| 8 | `BP_DemoDisplay` | Construction-script demo toolboxes |
| 3 | `Device_ClassDesigner_V2` | Warrior slot1 HP500; Blaster slot2 HP250/shield300; Thief slot3 HP200; `enabled=false` (definitions) |
| 3 | `Device_ClassSelector_V2` | Gameplay Only; slots 1/2/3 |
| 2 | `BP_BRICK_Player_Spawner_Prop` | Both `useAsIslandStart=true` |
| 2 | `Device_ExplosiveBarrel_V2` | Boss AOE (`_Attack`, `_BossLevel`) |
| 2 | `Device_PropManipulation_V2` | **Not** Prop Mover |
| 1 | `Device_ItemGranter_V2` | `Item Granter_GoldForCollectibles` — Verse tag Subscribe, Details **[]** |
| 1 | `Device_HUDController_V2` | `HUD Controller_BRICK` |
| 1 | `Device_ExperienceSettings_BRICK` | `allowBuilding=None`, `enableJam=Off`, infinite resources, `environmentDamagePreset=All` |
| 1 | `BP_DaySequenceModifier_Device` | Priority **2000**, FixedTime **12**, `sunlightIntensity=10`, `fogDensity=0.05`, **`use Volume=true`** |
| 0 | Conversation / End Game / Prop Mover / CubeGrid / Playset / Persona | Narrative quest / Obstacle win / Home Builder Phone |

Extractor histogram named spawners `BP_Creative_Player_Spawner_Prop`; live class is **BRICK**. Use live.

Dressing: `/BRICK_Cane` 923, `/SPACE_ASSETS` 880, `/NEIGHBORHOOD_Assets` 619, `/ENV_PACK_UnplayableTestMap` 160 (dressing from skipped UnplayableTestMap kit), `/CRD_BRICKAssemble` 110.

## Verse (13 files)

Copied to `KB/templates/action_adventure_sample/snippets/verse/`:

| File | Role |
|---|---|
| `BRICK_quest.verse` | `BRICK_quest` / `BRICK_quest_repeatable` + tracker/beacon/reward |
| `BRICK_quest_giver_device.verse` | Button interact → persistable log → canvas dialog |
| `BRICK_quest_dialog.verse` | `button_loud` / `button_quiet` canvas, `ZOrder := 3` |
| `BRICK_persistent_quest_log.verse` | `class<persistable><final>` `CompletedQuests` / `ActiveQuests` |
| `npc_behavior_quest_giver.verse` | Stasis in place, AllowTurning |
| `BRICK_stud_spawner.verse` | Event → `SpawnStudsAtLocation` |
| `BRICK_stud_spawner_pool.verse` | Tag-find pool, `TeleportTo` + `SpawnItem` |
| `BRICK_collectibe_stud_granter.verse` | Filename typo **collectibe**; class `BRICK_collectible_stud_granter` |
| `BRICK_stud_counter_device.verse` | Poll Conditional Button item count → HUD Message |
| `BRICK_destructible_prop_definition.verse` | Cached transform + `creative_prop_asset` |
| `BRICK_destructible_prop_manager.verse` | `IsValid` poll → studs + `SpawnProp` |
| `npc_attack.verse` | Stun / ranged AOE variants |
| `npc_attack_controller.verse` | `race` elim vs timed attacks on nearest player |

## Authored kit (**Observed** — `assets.txt` + live)

```
Content/Audio/{Blaster,Boss,HUB,Thief,Tutorial,Warrior}/
Content/Cinematics/          24 LevelSequences (hints, fireplace, boss intro, Blaster panels)
Content/BRICKQuests/          NPCDef_QuestGiver + quest-giver UMG
Content/BRICKModule_StudRewards/
Content/Materials/
Content/NPCs/                NPCDef_Boss, Enemy_0, Enemy_Blaster, Enemy_Warrior
Content/UI/                  WBP_BRICK_PlayerInfo{,Stack}, WBP_HUDMessage_StudCounter, heart/stud icons
Content/VFX/                 BP_Prop_NPCAttack_{AOE,Stun} + Niagara
```

**Tier 2 mounts:** `/c1b682b0-…` 4173, `/Game` 3632, `/CreativeCoreDevices` 1070, `/BRICK_Cane` 923, `/SPACE_ASSETS` 880, `/NEIGHBORHOOD_Assets` 619. No Fab `.uref`.

## What this sample actually teaches

1. **A linear persistable quest line is Verse canvas dialog + Tracker/Beacon, not Conversation Device.** `weak_map(player, BRICK_persistent_quest_log)`. `GetQuestLineProgression` walks the array in order. techniques/persistable-quest-line-verse-dialog.md.
2. **Stud bursts teleport a tagged pool of Item Spawners, then `SpawnItem`.** One spawner = one pickup. techniques/stud-itemspawner-teleport-pool.md.
3. **World studs are BRICK Collectible `Coin`s tagged into one Item Granter.** HUD count is Conditional Button `GetItemCount`, not Collectible score. techniques/tagged-collectible-stud-granter.md.
4. **Boss interrupts are a `creative_device` racing stun/AOE on the nearest player**, while the Guard CD keeps fighting. techniques/npc-attack-stasis-aoe.md.
5. **Destructible stud props poll `IsValid` and `SpawnProp` the cached transform.** No `DestroyedEvent` on `creative_prop`. techniques/destructible-prop-validity-poll.md.
6. **Assembly `On Assembled` Increments a Tracker** (`Tracker_Quest_AssembleLab`). Persistent Switch (`usePersistence=true`) keeps trophy builds across sessions (Billboard_Demo_Achievement). Assembly companion HideThenBuild Verse is **homework**, not shipped.
7. **Classes are Designer definitions (`enabled=false`) + Selector Gameplay Only.** Warrior / Blaster / Thief. Island `defaultClass` NoClass.
8. **In-project HUD is `WBP_BRICK_PlayerInfoStack`** (portraits / hearts). Equipped + Quickbar overrides **null**. HideElement minimap/team; ShowElement HUD messages; health/shields DoNotOverride. ExtraAssets stays **four**.
9. **`allowBuilding=None` + `enableJam=Off`** like Music / Obstacle / Home Builder. PLAY destruction `environmentDamagePreset=All`. brick_editor_sample is the BrickGameplay exception (`ExperienceSettings_V2`, `allowBuilding=All`, `enableJam=Default`). BrickAssembled meshes **cannot** feed these 29 Assembly devices.
10. **DSD is 2000-only FixedTime 12**, even with `use Volume=true` — not a priority stack. Intensity **10** vs Home Builder **8**.
11. **36 billboards + 8 DemoDisplay** are the in-island editor syllabus (quest, pool, granter, boss, achievement, assembly). Tutorial prose is the walkthrough; boards name the Verse classes.

## Official tutorial vs shipped

**External** hub + BRICK Templates index + Action Adventure + Collectible Devices + Assembly Devices + Working with BRICK Islands + theme Rules (headed Chrome 2026-09-15). Verbatim: `_raw/action_adventure_sample/tutorial.txt`. Spreadsheet: no row. `docsUrl` empty. Live slug `BRICK-action-adventure-in-unreal-editor-for-fortnite` (H1 **BRICK Action Adventure Template**). Guessed `BRICK-action-adventure-template-in-fortnite` was not the live URL.

### Lines up

| Tutorial | Shipped |
|---|---|
| Picker / templates index **Action Adventure**; quests, combat, puzzles, Assembly showcase | picker **Action Adventure Template**; **29** Assembly devices |
| Warrior / Blaster / Thief class hats | Class Designer + Selector slots 1/2/3 |
| quest-giver boss boss, Verse-based attacks | `npc_attack_controller` + 2 Explosive devices |
| Persistable quests; BRICK_quest_giver; Trackers; NPC Spawner Guard + `npc_behavior_quest_giver` | 2 quest givers (PlayerCharacter `autoProgressDialog=true`); 10 Trackers; QuestGiver CDs |
| Object pool of Item Spawners + `SpawnStudsAtLocation` | `BRICK_stud_pool` + toss/gravity tags |
| Custom UI portraits / heart bars | `WBP_BRICK_PlayerInfoStack` + heart materials |
| Collectible companion points at this island | **86** BRICK Collectibles |

### Does not line up

| Tutorial | Shipped |
|---|---|
| “Progress through in **any order**” | `GetQuestLineProgression` is **linear** (first `CanBeStarted`) |
| `Stat to Track = Events` | live **None**; Increment bindings are the bus |
| `CP_Ingredient_GoldStud` / companion **Gold Stud** | world collectibles **Coin**; granter item class waived (empty spawn struct) |
| Tag `BRICK_stud_itemspawner_tag` | shipped **toss** vs **gravity** tags |
| Assembly + cine + **Prop Mover** | **0** Prop Mover; **2** Prop Manipulation; **24** cine |
| Assembly companion: “Test … in the **BRICK Home Builder** template” | Home Builder has **0** Assembly; this island has **29** |
| Assembly HideThenBuild 88-line Verse | **not** in this project |
| Tutorial 212-line `BRICK_quest_giver_device` paste | train **`snippets/`**; filenames and `autoProgressDialog` differ from the paste |

Training rule: shipped bindings win. Tutorial *prose* is the syllabus (plus boards). Tutorial *Verse* is untrusted if it conflicts. Do not emit Conversation Device trees, Parkour ordered spawners, Obstacle Gold Coin→End Game, Home Builder Phone/mix, Village homestead, CubeGrid, or ExtraAssets HUD as this island’s setup.

## Negative findings

- Picker **Action Adventure Template** / docs **BRICK Action Adventure Template** / folder `action_adventure_sample` are the same tile — not Home Builder, not Brick Editor, not MartialArts, not `village_starter_template`.
- Name does **not** mean Conversation Device quests or an obby End Game.
- **0** Prop Mover — do not copy Obstacle movers here.
- `find_actors` `"Collectible"` also hits Billboard_Demo_Studs_CollectibleGranter.
- `GetDeviceProperties` on Verse `@editable` device refs returns **subobject** `refPath`s (`__verse_0x…`).
- `get_properties` fails the whole call if any name is missing (`equippedItemWidgetOverride` is wrong; live is `equippedItemInfoWidgetOverride` / `quickbar Slot Widget Override Class`; DSD `useVolume` is **`use Volume`**).
- `spawn Item` live struct is empty — same Home Builder trap.
- HUD `get_referencers` on widgets mostly **[]** — widget graph Unverified.

## Capture backlog

Waived: per-quest array names via GetDeviceProperties (subobject trap); which of 129 spawners are pool vs PLAY; per-collectible tag membership; Assembly radius/time per device; NPC Character Definition internals (`enabledDuringPhase` unread on spawners). `experimental.*` Unverified.

## Live notes

- Cursor `user-unreal-mcp` was down; `tools/mcp.py` on `:8000` worked.
- Mount `/c1b682b0-4d00-41ab-8ca2-8444fa187773` from `ListFiles("")`. Do not recurse `ListFiles("")`.
- Billboard `text` is on the actor, not ToyOptions.
- Wrong `refPath` `/mount/map.ActorName` is invalid; use `actorPath`.
- Did not flip experimental flags. Did not switch maps.
