---
id: digest-002
kind: digest
---

# animation_template — sample digest

> Offline `_raw/animation_template/`. Live
> `/04397672-f6f8-4070-849c-592746c0c6d7/animation_template` (**Observed**).
> Docs: Animation 101 (spreadsheet Keep=Yes).

| | |
|---|---|
| Classification | **Animation authoring museum** — Animated Mesh, Control Rig, Sequencer; one Verse cinematic-stasis helper |
| Content type | kit + gallery wrap: **45 authored vs 1508 placed** |

## Scale

| | |
|---|---|
| Authored | 45 (`Mannequin/` 29, `Fortnite/ControlRigs` + faceboards/skeletons, `Sequences/` 3) |
| OFPA | 1508 |
| Verse | **1** (`player_cinematic_mode_device.verse`) |
| Animated Mesh devices | 2 |
| Skeletal mesh actors | 10 |
| Cinematic Sequence | 1 |
| Switches | 4 |
| Matchmaking | 16 |
| Compatibility | 43.00 |

Empty `Developers/felixkupis/Collections/`. Twentieth first-party hit. **Wave 2 starts here.**

Live folders (**Observed**): `Animation Examples` (+ `Cameras`), `Core Devices`, `LAYOUT/{Courtyard,Garden,Hallway,Interior Setdressing,Gameplay}`.

## What it is *for*

Walkable Animation 101: dance mannequins (Animated Mesh), Control Rig / Sequencer exhibits, a courtyard cinematic. Verse is **not** the product — it only hides+freezes the player while the museum cinematic plays so they aren’t in the shot.

## Device inventory (the design)

| Count | Class | Role |
|---|---|---|
| 2 | `Device_AnimatedMesh` | loop emotes on mannequins |
| 10 | `FortSkeletalMeshActor` | Control Rig / pose demos |
| 1 | `Device_CinematicSequence` | museum flythrough |
| 4 | `Device_Switch_V2` | start cinematic / exhibits |
| 1 | `VerseDevice` | `player_cinematic_mode_device` |
| 1 | `Device_Timer_V2` | exhibit timing |
| 9 | billboards | captions |
| rest | Art Deco / sidewalks | throwaway gallery |

Authored kit to copy: `Content/Mannequin/` (meshes, 9 animations, 11 emotes) and `Content/Fortnite/ControlRigs/` + faceboards.

## Verse (1 file)

Snippet: `snippets/verse/player_cinematic_mode_device.verse`

Switch `TurnedOffEvent` → `PutInStasis(stasis_args{AllowTurning/Falling/Emotes := false})` + `Hide()`. Sequence `StoppedEvent` → `Show()` + `ReleaseFromStasis()`. Stores `?agent` from the switch instigator.

Same family as Stand Up / Prop Hunt / Triad cinematics, but the *mechanism* is stasis+hide, not chair-gate or flythrough camera.

## What this sample actually teaches

1. **Three authoring surfaces, one island:** Animated Mesh device (stock emote, no Sequencer), Control Rig on `FortSkeletalMeshActor`, Sequencer cinematic. Don’t mix them unless you mean to.
2. **Hide the pawn for the shot:** `PutInStasis` + `Hide` on start, reverse on `StoppedEvent`. See techniques/cinematic-hide-stasis.md.
3. **Switch polarity:** they subscribe `TurnedOffEvent` to *start* the cinematic — wiring is Observed in Verse, not inferred from the device name.
4. **Mannequin folder is the reusable kit**; 1508 sidewalk tiles are not.

## Official tutorial vs shipped

**External:** Animation 101 Template

**Single page**, four in-page sections (not a chapter hub): Skeletal Mesh → Animated Mesh Device → Cinematic Sequence Device → Control Rig.

Spreadsheet `/en-us/uefn/animation-101-template-in-unreal-editor-for-fortnite` redirects to the same page. `docsUrl` empty.

Retrieved 2026-09-09 after Epic ID login in headed Chrome (anonymous / headless → 403).

### Lines up

| Tutorial | Shipped |
|---|---|
| Museum of skeletal meshes | live folders `Animation Examples` + `LAYOUT` |
| Drag `FN_Mannequin` → `FortSkeletalMeshActor` | 10 of those; `Content/Mannequin/` |
| Hallway = instanced anim (Looping / Playing / Initial Position / Play Rate) | authored `Mannequin/Animations/` + emotes |
| Theater = Animated Mesh device (play/pause/reverse in-game) | 2 `Device_AnimatedMesh` |
| Meshes without that device are frozen in PIE | matches: 10 actors vs 2 devices |
| Skeletal meshes don’t collide with player/world/each other | tutorial claim — train as authored intent |
| Stair cinematic = Cinematic Sequence device + trigger; `MuseumFlyThrough` | 1 `Device_CinematicSequence`; `Content/Sequences/MuseumFlyThrough.uasset` |
| Control Rig: camera-cut track + 3 rigs; drag `FN_Mannequin_ControlRig` | `Fortnite/ControlRigs/` + mannequin ControlRig mesh |
| Bake To Control Rig / attach props | authoring syllabus, not Verse |

### Does not line up

| Topic | Tutorial | Shipped |
|---|---|---|
| Trigger widget | **Button** devices (desk + stair) | **Switch** (`Device_Switch_V2` ×4). Verse listens to `TurnedOffEvent` to *start* |
| Hide player for the shot | **not mentioned** | `player_cinematic_mode_device` Hide + `PutInStasis` / `ReleaseFromStasis` |
| Sequence assets | names `MuseumFlyThrough` | also `TheaterFlyThrough`, `MainSequenceDemo` |
| Verse | none | 1 file — extra, not contradicted |

Train on: Mannequin kit + Animated Mesh vs Control Rig/Sequencer split from the **prose**; stasis/hide from the **snippet**. Do not emit Animation 101 as a Button-triggered Verse minigame. Emitting `button_device` for this island is doc regurgitation.

## Negative findings

- Gallery dressing is Art Deco bank / jewelry — not an animation lesson.
- `using {/Fortnite.com/Characters }` missing space after `{` — ships; don’t copy as style.
- Scratch `Developers/` folder.

## Capture backlog

1. Live Animated Mesh clip names vs `Mannequin/Emotes`. *Waive — authoring, not Verse.*
2. Billboard 9 captions vs tutorial hallway/theater/stair labels. *Waive — tutorial is the syllabus.*

## Live notes

Level confirmed. Folders named Animation Examples + LAYOUT. Streaming off.
