---
id: digest-003
kind: digest
---

# characters_cameras_npc_template — sample digest

> Offline `_raw/characters_cameras_npc_template/`. Live
> `/d8b0d010-f1e3-407c-ae0a-4e9ffe11effc/characters_cameras_npc_template` (**Observed**).
> Spreadsheet Keep=Yes, Docs=**No** (link `-`). **No island tutorial.** Discipline pages exist (Cameras overview, Camera Types) — they are not this island.

| | |
|---|---|
| Classification | **Walkable Character / Camera / NPC curriculum** — numbered halls 00–07 |
| Content type | kit + lab: **55 authored vs 96 placed** |

## Scale

| | |
|---|---|
| Authored | 55 (`StackOBot` SK/anim/rig, 3 `CharDef_*`, 4 `CS_*` sequences) |
| OFPA | 96 |
| Verse | **1** — `npc_behavior_usingverse.verse` |
| Fab | 0 |
| `Device_Character_V2` | **21** |
| Compatibility | **42.30** |
| Matchmaking | 16 |

Empty `Developers/felixkupis/Collections/`. Twenty-eighth first-party hit.

## Halls (the syllabus)

Live Outliner (**Observed**):

| Hall | Teaches |
|---|---|
| `00_Start` | lobby |
| `01_CharacterDevice` | single Character device |
| `02_CharacterGroup` | group of Character devices |
| `03_FortniteSpawners` | Fortnite creature/guard spawners vs Character |
| `04_NPCSpawner` | NPC Spawner + Verse `npc_behavior` override |
| `05_CinematicSequence_Character` | Sequencer on a Character (wake-up / idle) |
| `06_FullCinematic` | multi-camera cine (`CineCameraActor` + `CS_Bot_WakeUp_Camera`) |
| `07_PlayerCamera` | gameplay cameras: Perspective + Fixed Point + Custom Controls 3P |

Same numbered-hall convention as Modeling / Materials / Greyboxing.

## Device inventory (the design)

| Count | Class | Role |
|---|---|---|
| 21 | `Device_Character_V2` | halls 01–02 |
| 3 | `Device_CharacterSpawner` | hall 04 |
| 1 | `Device_GuardSpawner_V2` | hall 03 |
| 1 | `Device_CreatureSpawner_V2` | hall 03 |
| 1 | `Device_CreaturePlacer_V2` | hall 03 |
| 1 | `Device_CreatureManager_V2` | hall 03 |
| 4 | `AIPatrolPath_Device` | patrol vs Verse wander |
| 1 | `Device_Character_Controller_V2` | possess / drive a Character |
| 3 | `Device_CinematicSequence` | hall 05–06 |
| 3 | `CineCameraActor` | hall 06 |
| 1 | `CP_Movie_Set_Camera` | hall 06 |
| 2 | `Device_CameraModes_Perspective` | hall 07 |
| 1 | `Device_CameraModes_FixedPoint` | hall 07 |
| 1 | `Device_CustomControls_ThirdPerson` | hall 07 |
| 5 | `Device_MutatorVolume_V2` | swap exhibit (Materials pattern) |
| rest | HUD / billboard / audio mixer / day-seq modifier | dressing |

Authored kit: `StackOBot/` (SKM_Bot + 15 `A_Bot_*` + `PhA_Bot` rig + backpack + jetpack FX), `Character/CharDef_NPC_{Patrol,Verse,Wildlife}`, `Cinematics/CS_Bot_{Idle,WakeUp,WakeUp_Camera,WakeUp_ControlRig}`.

## Verse (the design)

`npc_behavior_usingverse` **subclasses `npc_behavior`**, not `creative_device`. Wired as NPC Spawner **Behavior Script Override** (or on a Character Definition). See techniques/npc-behavior-navigatable-focus.md.

```
GetAgent[] → GetFortCharacter[] → GetNavigatable[] + GetFocusInterface[]
loop:
  NavigateTo(random offset from spawn, Walking)
  race:
    NavigateTo(spawn)          # wins
    MaintainFocus(look-at)     # never completes → walk backwards
    DrawDebugLookAt            # never completes
```

`ShowAIDebug` draws via `debug_draw_channel`. Island Settings → **Verse Debug Draw ON** or you see nothing. Comment in source.

Three Character Definitions split the hall-04 exhibit:

| Asset | Meant for |
|---|---|
| `CharDef_NPC_Patrol` | AIPatrolPath |
| `CharDef_NPC_Verse` | this script |
| `CharDef_NPC_Wildlife` | wildlife / `CustomCharacter_Wildlife` sequence |

## What this sample actually teaches

1. **Character device ≠ Fortnite creature/guard spawner** — halls 01–03 are the comparison.
2. **Verse NPC control is `npc_behavior`, not a creative_device sitting in the world.**
3. **Navigatable + Focus are separate interfaces.** Focus without killing Navigate = look-at while walking (the `race` demo).
4. **Cinematic on a Character is Sequencer + Control Rig** (`CS_Bot_WakeUp_ControlRig`), same family as Animation 101, then hall 06 adds real cine cameras.
5. **Gameplay camera is devices** (Perspective / Fixed Point / Custom Controls 3P) — **not** Scene Graph camera components. Do not mix with Camera Types docs.
6. StackOBot is the custom mesh kit (anim + jetpack FX). Not a MetaHuman (that's character_environment).

## Official tutorial vs shipped

**No island tutorial.** Spreadsheet Docs=No, `docsUrl` empty, link `-`. Guessed `/documentation/fortnite/characters-cameras-and-npcs-in-unreal-editor-for-fortnite` is not this island. Search 2026-09-11: Cameras overview + Camera Types (Scene Graph) + community NPC posts — none name this sample.

Do not train as if those pages walk these halls. The Outliner **is** the syllabus.

Related but **not this island:**

| Page | Why not |
|---|---|
| Cameras | option catalog |
| Camera Types | Scene Graph cameras — this island uses **devices** |

## Negative findings

- Scratch `Developers/` folder.
- Compatibility 42.30.
- Debug-draw default `ShowAIDebug = true` will spam in PIE if Island Settings debug is on.

## Capture backlog

1. Which Character Definition is on which hall-04 spawner. *Open, low — names already map.*
2. Live camera-device properties on hall 07. *Waive — devices are the lesson.*

## Live notes

Level + numbered halls confirmed. Streaming off (not re-verified). Chrome still signed in.
