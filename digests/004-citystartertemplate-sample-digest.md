---
id: digest-004
kind: digest
---

# city_starter_template — sample digest

> **Extractor provenance.** `tools/extract-template.py` (Pass A) + UnrealMCP on
> `localhost:8000` (Pass B). Default level
> `/2128341c-f5b4-4107-ba06-40bb501d7847/city_starter_template` (**Observed**).
> Docs: TURTLE City Starter Visual Styles
> (single page). `docsUrl` empty; guessed unsuffixed `city-starter-visual-styles-…` **hub-redirects**.
> Compat **43.00**. `featureSet`: **CustomGameplayFeature**.

| | |
|---|---|
| Classification | **theme visual-styles starter** — NYC slice, cartoon vs comic-noir |
| Content type | **authored PP/sky/windows over a mounted city kit** (66 authored vs 9219 OFPA) |
| Verse | **0 files** |
| Fab `.uref` | **0** |

## What it is

The picker name is **City Starter**. The island is a **16-player look lab**, not Arcade's beat-em-up: sewer lair style picker, streets, Channel 6, rooftops. Teaching that survives deletion is the dual unbound PPV stacks (cine-toggled) and the fake-interior window kit. TURTLE spawners / sidescroller / Verse gather live in turtle_arcade_sample. alternate dimension ooze / SciFiFacility live in dimension_x_template.

## Scale (**Observed** — `summary.txt` + live)

| | |
|---|---|
| Authored assets | **66** |
| OFPA | **9219** |
| Maps | `city_starter_template.umap` (only) |
| Verse files | **0** (`VerseToolset.ListFiles` on `/2128341c-…` — do not recurse `__ExternalActors__`) |
| Verse devices placed | **0** |
| Data layers | **0** |
| Fab `.uref` | **0** |
| Compatibility | **43.00** |
| Matchmaking | **16** (`maxPlayers=16`, `teamSize=Dynamic`, `locomotionPreset=Custom`) |
| Streaming | Partitioned + actor folders + OFPA. **`LevelHasStreamingDisabled` omitted** (Social/Arcade camp) |

Empty `Developers/felixkupis/Collections/`. First-party scratch folder.

`experimental.*` (Python / toolsets / Scene Graph) in `.uefnproject` is **Unverified**. Content-side: **no** `.verse`, **no** `__ExternalEntities__`.

Largest file ~10 MB OFPA. Whole folder ~155 MB. Under the GHES **100 MB** per-file cap.

## Level organisation (**Observed** — live `get_folders` + folder counts)

```
01_Core/                 25 — Island Settings, HUD, HUD Messages, camera/controls, tuners
02_Lighting/             40 — PPVs, DSD, SkyLights, DirectionalLights, PointLight(_NoSpec*), SkySphere
  CharacterLighting/ Large/ MainLighting/
03_StarterArea/          485 — sewer lair (spawn pads, style Buttons + cines)
04_Traversal/            26
05_Environment/          8622 — the city kit (Generic/Hero/Mid/StreetStuff/Tall/Ground/windows)
Devices/                 0  leftover
06_Environment/          0  leftover (and 03_Traversal / 05_Traversal empty)
HLOD/                    17
```

No `Device_Billboard_V2`. Syllabus is **HUD Message** text (`WBP_Message`). Sampled devices `bIsSpatiallyLoaded=false`; `load_actors` not required.

## Device inventory (the design)

| Count | Class | Role |
|---|---|---|
| 5 | `Device_HUDMessage_V2` | tour copy + `WBP_Message` |
| 4 | `Device_Teleporter_V2` | Channel 6 / rooftops (`Group_A/B/C`) |
| 3 | `Device_Beacon_V2` | SewerExit / Pizza / Teleporter |
| 2 | `Device_HidingProp_Sewer_Tunnel` | lair↔streets Hidden Travel |
| 2 | `Device_MutatorVolume_V2` | Activate teleporters (not material swap) |
| 2 | `Device_Button_V2` | New Cartoon / New Comic Shader |
| 2 | `Device_CinematicSequence` | `LS_PP_CelShader33` / `LS_PP_ComicShader33` |
| 2 | `BP_Creative_Player_Spawner_Prop` | sewer lair |
| 2 | `PostProcessVolume` | Cartoon (weight 1) / Comic (weight 0) |
| 1 | `Device_CameraModes_Perspective` | rooftops Fixed Angle |
| 1 | `Device_CustomControls_ThirdPerson` | rooftops (not Sidescroller) |
| 1 | `Device_HUDController_V2` | `showHUD=HideElement`, messages Show |
| 1 | `BP_DaySequenceModifier_Device` | priority 2000; kit emissives off |
| 1 | `Device_Volume` | `Volume_ResetCam` — Remove camera/controls |
| 1 | `Device_ItemSpawner_V3` | pizza |
| 1 | `Device_Zipline` + `Device_AirVent_V2` | traversal toys |
| 1 | `Device_ExperienceSettings_V2_UEFN` | IslandSettings1 |
| 24 | `PlaysetRoot` | mounted city/sewer galleries |
| 0 | Billboard / TURTLE Character Spawner / Verse device | — |

## Authored kit (**Observed** — `assets.txt` + live)

```
Content/Materials/PP/   PI_OutlinesCartoon/Comic, PP_CelShaderNeutralEV, PP_ComicCelShader,
                        PP_DepthFog_Inst, PP_Posterize, PP_Frame, PP_DesaturateLast, leftovers
Content/Materials/      MI_/M_ProceduralSkybox, M_Windows_Interior/Mask/Channel6, neon/decal
Content/Meshes/         SM_SkySphere, SM_BGBuildingA/B_Window*
Content/Sequences/      LS_PP_CelShader33, LS_PP_ComicShader33 (live cines);
                        LS_PP_*Active + LS_BuildingsEmissive leftover/unused-as-device
Content/UI/             WBP_Message, T_Turtlecom, skewed MIs
```

SkySphere: `SM_SkySphere` + `MI_ProceduralSkybox` (**Observed** on `StaticMeshComponent0`).

**Tier 2 mounts:** `/City_Comp` 4730 + `/City_Assets` 2054 + `/Sewer_Comp` 603 + `/Sewer_Assets` 304 + `/MIron_Visuals` 3739. No Fab `.uref`.

## What this sample actually teaches

1. **Two island looks = two unbound PPVs, comic `blendWeight` 0 until a cine plays.** techniques/cartoon-comic-unbound-ppv.md.
2. **Fake occupied skyline = cubemap windows + mask + sheen; DSD kills stock emissives.** techniques/fake-interior-windows.md.
3. **Authored skybox over a mounted city kit.** Same `SM_SkySphere` + procedural skybox MI as NightEnvironment / alternate dimension; kit is `/City_*` + `/Sewer_*` + `/MIron_Visuals`, not this project's meshes. Night stack without DSD: night-atmosphere-stack.md. alternate dimension keeps DSD and **disables** its skybox component; City **enables** it.
4. **Sewer grate is Hidden Travel**, same device as Arcade, groups **1→1**, occupants **2**. sewer-tunnel-hidden-travel.md.
5. **Rooftops are Fixed Angle + Third Person controls, not a sidescroller.** Distance **850**, pitch **-3**, FOV **80**, Ease-In-Out **0.2**. `AddToPlayer` / `AddTo` on Teleporter2 + Mutator Zone; `RemoveFromPlayer` / `RemoveFrom` on Teleporter4 + `Volume_ResetCam`. Contrast Arcade: sidescroller-camera-controls.md.
6. **HUD is hidden; tour copy is a project `WBP_Message` on HUD Message devices.** Controller overrides are **null** (not ExtraAssets, not Arcade's `UW_HUD_PlayerInfoStack`). `showHUD=HideElement`, `showHUDMessages=ShowElement`.
7. **This is not Arcade.** No TURTLE Character Spawner, no tracker arenas, no Verse gather. Docs send encounters to Arcade. `allowBuilding=None`, `timeOfDayOverride=Hour12`.

HUD Message chain (**Observed** bindings):

| Message | Show | Hide |
|---|---|---|
| Pick Visual Style and exit Sewer | round start | sewer `OnStopHiding` |
| Get Pizza from the Rabbit Hutch! | sewer `OnStopHiding` | pizza `On Item Picked Up` |
| Find the Teleporter at Channel 6 News | pizza pickup | Teleporter2 `On Teleported` |
| Navigate the rooftops to get home | Teleporter2 | Teleporter4 `On Enter` |
| Pizza Time! | Teleporter5 `On Teleported` | — |

## Official tutorial vs shipped

**External:** TURTLE City Starter Visual Styles (headed Chrome 2026-09-14). Single page. Verbatim: `_raw/city_starter_template/tutorial.txt`.

### Lines up

| Tutorial | Shipped |
|---|---|
| theme Templates > City Starter; two styles cartoon + comic | picker; PPV labels |
| Cartoon default; comic via Blend Weight 1 | Cartoon `blendWeight=1`, Comic `=0` |
| `PostProcessVolume_Cartoon` / `_Comic`; array of custom PP materials | live `weightedBlendables` |
| Depth fog `PP_DepthFog_Inst`; outlines; cel; comic tone/frame; posterize | those assets on the volumes |
| SkySphere → `MI_ProceduralSkybox`; ToD manager off | authored sphere + MI; DSD instead of changing ToD |
| Windows: `M_Windows_Interior`, `M_WindowMask`, `M_Windows_Channel6` | in `Content/Materials/` |
| Point lights, specular disabled | `PointLight_NoSpec*` |
| `DirectionalLight_Index1` named in sunny exercise | actor exists |
| Not a gameplay tutorial; add Arcade encounters / no barriers | no TURTLE spawners, no Barrier devices |
| Sunny fog Curve Far **93511** / Near **14887** | exercise numbers — not live defaults |

### Does not line up

| Tutorial | Shipped |
|---|---|
| “switch devices on the wall” | **Buttons** (`New Cartoon Shader` / `New Comic Shader`) + Sequencer |
| Platformer theme | rooftops = Fixed Angle + **`Device_CustomControls_ThirdPerson`**, not Sidescroller |
| Search Outliner `"postproccess"` | docs typo; actors are `PostProcessVolume_*` |
| Cartoon cel as a generic “cel shader material” | live cartoon cel = **`PP_CelShaderNeutralEV`**, not unused `PP_CartoonCelShader` |
| Dedicated chapter per style (Arcade-style hub) | **one** page |

Training rule: live `weightedBlendables` + cine sequence paths win. Do not emit Switch devices, Arcade sidescroller, or `PP_CartoonCelShader` as the shipped cartoon stack.

## Negative findings

- **Name implies a city *game*; files show a look tour.** Pizza / Channel 6 / rooftops are HUD + teleporters, not combat.
- **`maxPlayers=16` vs two spawn pads vs sewer `numberOfOccupants=2`.** Occupants follow pads, not matchmaking.
- **`06_Environment` / `Devices` / `03_Traversal` / `05_Traversal` are empty leftovers** after the numbered move.
- **`PP_CartoonCelShader`, `LS_PP_CelShaderActive`, `LS_PP_ComicShaderActive`:** 0 referencers. Live cines use `LS_PP_*33`.
- **Mutator Zones are not material demos.** They Activate Teleporter3/4.
- **HUD Controller does not point at ExtraAssets or `UW_HUD_*`.** Overrides null; messages own `WBP_Message`.
- **`GetDeviceProperties` N/A** — no Verse devices.
- **No player-facing billboards.** theme prop `BP_MIron_City_Billboard_*` is set dressing.

## Capture backlog

| Item | Status |
|---|---|
| Sequencer tracks inside `LS_PP_*33` (which PPV weight they key) | **Waived** — keyframes unreachable; button→PlayFunction + live blendWeights are the lesson |
| Per-light Intensity / SpecularScale on components | **Waived** — actor-level light properties empty; `NoSpec` labels + docs |
| Exact PlaysetRoot gallery names (24) | **Waived** — `/City_*` `/Sewer_*` `/MIron_Visuals` mounts are enough |
| `PP_Halftone` / `PP_OutlinesClean` vs live stacks | **Waived** — live blendable arrays are ground truth |
| UMG graph inside `WBP_Message` | **Waived** — widget path on all five HUD Messages is the lesson |

## Live notes

- Mount `/2128341c-f5b4-4107-ba06-40bb501d7847`. Project Verse path `/epictemplates@fortnite.com/city_starter_template`.
- Did **not** `load_level` away from the default map.
- `get_properties` fails the **whole** call if one name is wrong (`ActorLabel` is not a field). Use `list_properties` names (spaces: `angle Pitch`, `hUD Widget`, `interaction Text`).
- Island Settings fields are on the **actor** (`maxPlayers` is toward the end of the list). `ToyOptionsComponent.playerOptionData` not used this pass.
- `get_referencers` on authored PP / sky / `WBP_Message` returns 0; confirm on placed PPV / `StaticMeshComponent0` / HUD Message `hUD Widget`.
- `get_actors_in_folder` descriptors often have `folderPath=None`; counts still match the numbered tree.
- Folder on disk / title is **`city_starter_template`** (matches picker “City Starter”).
