---
id: digest-005
kind: digest
---

# day_sequence_template — sample digest

> Offline `_raw/day_sequence_template/`. Live
> `/5b7be54a-fc53-462d-a89d-989b9918d6f0/day_sequence_template` (**Observed**).
> Docs: 5-chapter hub. Spreadsheet Keep=Yes. **No Verse.**

| | |
|---|---|
| Classification | **Environment lighting curriculum** — one Day Sequence per biome + PPV priority stack; Lumen vs non-Lumen |
| Content type | catalog wrap: **18 authored vs 287 placed** (stock Fortnite biomes) |

## Scale

| | |
|---|---|
| Authored | 18 (landscape layer infos, one MI instance, HLOD, leftover `CC_ToDM_POC*`) |
| OFPA | 287 |
| Verse | **0** |
| Day Sequence devices | **5** (`BP_DaySequenceModifier_Device`) |
| Post-process volumes | **4** |
| Lumen Exposure Manager | 1 |
| Billboards | 15 |
| Maps | **2** (`day_sequence_template` + leftover `CC_ToDM_POC`) |
| Matchmaking | 16 |
| Compatibility | 43.00 |

Empty `Developers/felixkupis/Collections/`. Twenty-first first-party hit.

Live folders (**Observed**): `Island_Center`, `Winter_North`, `Jungle_South`, `Desert_East`, `Beach_West`, `DaySeq_EnvLook/{North,South,East,West}`, `Billboards`, `X_Landascape and Water` (typo: Landscape).

## What it is *for*

Four biomes around a **neutral island center**. Center Day Sequence = Chapter 4 Fortnite default look, then peeled toward a shared “neutral” (exposure 0, CA 0.3, vignette 0.6, sat 0.8, grain 0.2). Each cardinal biome is a **second** Day Sequence with **Trigger Volume** On, higher **Priority** than 2000, plus a polish PPV (priority > 10 so it beats Lumen Exposure Manager).

Teleport keys (tutorial): **1** center, **2** north, **3** east, **4** south, **5** west.

No custom meshes. Lighting is the product.

## Device inventory (the design)

| Count | Class | Role |
|---|---|---|
| 5 | `BP_DaySequenceModifier_Device` | center + N/S/E/W |
| 4 | `PostProcessVolume` | per-biome polish (`North_PPV` … `West_PPV`) |
| 1 | `BP_LumenExposureManager` | Lumen vs Non-Lumen PPV pair |
| 15 | `Device_Billboard_V2` | captions |
| 2 | player spawners | 16P |
| 1 | campfire | set dressing |
| rest | lighthouse / rocks / trees | throwaway biomes |

## What this sample actually teaches

1. **Stack Day Sequences by Priority + Trigger Volume**, not by deleting the island default. Default Priority **2000**; biome device must be **higher** (e.g. 3000) or it *mixes*. Unchanged fields inherit from the lower device. See techniques/day-sequence-priority-stack.md.
2. **Lumen Exposure Manager is optional.** Tutorial: skip it unless you need low-end to match Lumen. Default LEM PPV priority **10**; extra PPVs must be **>10**.
3. **Time of day first** (sun angle/shadows), then intensity/color/fog/skylight, then lens (CA/vignette/bloom).
4. **Show Only Modified Properties** is how you read the island.
5. **Bold then middle** — tiny increments hide which knob did what.

## Official tutorial vs shipped

**External** hub: Day Sequence Starter Island Template

5 chapters (`day-sequence-N-…-in-unreal-editor-for-fortnite`). Unsuffixed `/day-sequence-1-device-information` **redirects to Fortnite docs home** — always use the `-in-unreal-editor-for-fortnite` suffix (same Speedway gotcha).

`docsUrl` empty. Spreadsheet `/en-us/uefn/day-sequence-starter-island-template-…`. Retrieved signed-in Chrome 2026-09-09.

### Lines up

| Tutorial | Shipped |
|---|---|
| Quadrants N/S/E/W + center | live folders match |
| 5 Day Sequence devices | 5 `BP_DaySequenceModifier_Device` |
| Per-biome PPV | 4 PPV |
| Lumen Exposure Manager at center | 1 `BP_LumenExposureManager` |
| No custom assets | 18 authored = landscape layers + leftover POC |
| Feature Examples in Project Browser | folder `day_sequence_template` |

### Does not line up

| Topic | Tutorial | Shipped |
|---|---|---|
| Outliner folder | `DaySeq_EnvLook` | present **and** biome folders (`Winter_North` …) |
| Landscape folder | — | `X_Landascape and Water` typo |
| Second umap | not mentioned | `CC_ToDM_POC` + `CC_ToDM_POC_sharedassets/` leftover |
| Numeric knobs (sun 2 vs 10, fog 1.2, …) | quoted | **Unverified live** — train as tutorial intent; re-read Details if shipping a lighting clone |

No Verse to drift. Train on tutorial prose + the priority/trigger technique.

## Negative findings

- `CC_ToDM_POC` is not the lesson.
- Scratch `Developers/` folder.
- 287 actors are stock biomes, not a kit.

## Capture backlog

1. Live Priority numbers vs docs 2000/3000/10/15. *Open, low — editor Details.*
2. Billboard 15 captions vs teleport keys 1–5. *Waive — tutorial is the syllabus.*

## Live notes

Level confirmed. Streaming folders named biomes. Streaming off (not verified here).
