---
id: digest-006
kind: digest
---

# domination_sample — sample digest

> Offline `_raw/domination_sample/`. Live
> `/ab315071-b210-44bc-8cbf-b6861f822a6b/domination_sample` (**Observed**).
> Docs: long island walkthrough (spawn Verse + device tables). Spreadsheet **Keep=No / Deprecate** (landscape actors out of date; lighting pass needed without Lumen). Extract anyway — teaching is still the spawn/VO/lighting pattern.

| | |
|---|---|
| Classification | **2-team domination game** — 3 capture points, weighted Verse spawn, VO queue, cine lobby |
| Content type | full island: **30 authored vs 5998 placed** |

## Scale

| | |
|---|---|
| Authored | 30 (`SEQ_*` jets/lobby, `CapLighting_*`, landscape, `Deserted_Door`) |
| OFPA | 5998 (2 unclassified) |
| Verse | **6** |
| Fab | **3** — corpus-unique. `References/Deserted{Audio,Props,VFX}_*.uref` → `/EpicGames@Fab.com/Deserted*` |
| `FortStaticMeshActor` | 4973 |
| Decals | 498 |
| Billboards | 175 |
| Audio players | 91 |
| Player spawners | 69 |
| Capture areas | **3** |
| Player counters | **3** (LargeZone A/B/C) |
| Verse devices | 5 |
| Compatibility | **42.30** |
| Matchmaking | **12 / 2 teams** |

Empty `Developers/felixkupis/Collections/`. Twenty-seventh first-party hit.

**Why 30 authored vs 5998 placed:** the look is not in `Content/`. Three Fab packs (`DesertedAudio` / `DesertedProps` / `DesertedVFX`) mount as **61 GUID folders**. See techniques/fab-uref-content-packs.md. No other packet in the corpus has a `References/` dir.

Live Outliner `get_folders` returned **[]** — World Partition, tiles not loaded. Disk OFPA + Verse is the inventory; do not treat empty Outliner as “no folders.”

## Verse (the design)

| File | Class | Role |
|---|---|---|
| `Deserted.verse` | `DominationSpawnTags` module | tag tree: `Tag_SpawnGroup_{Initial,A,B,C}` + `Tag_TeamOwnership_{Team1,Team2}` |
| `DominationSpawnManager.verse` | `domination_spawn_manager` | disable initial pads after 5s; enable A/B/C pads by **weight** |
| `capturepoint_lighting_control.verse` | `capturepoint_lighting_control` | play team cinematic on `ControlChangeEvent` (color + intensity) |
| `DominationVOController.verse` | `domination_vo_controller` | queued VO + HUD; skip “losing” VO until point has been captured once |
| `SequencerRandomizer.verse` | `sequencer_randomization_device` | random jet/crane loops, 120–240s, no back-to-back |
| `CineLobbyController.verse` | `cinematic_lobby_controller` | stop pregame cine; play cleanup (reset material params) |

**Spawn weight (the lesson):**  
`Weight = CaptureWeight (±1 by owner team) + Floor(PlayerWeight)` where each ally in `LargeZones[i]` adds **+0.5**, each enemy **−0.5**. Team1 pads enable iff `Weight > 0`; Team2 iff `Weight < 0`. Tags found at runtime via `FindCreativeObjectsWithTag`. See techniques/weighted-capture-spawns.md.

Fallback pads (tutorial: Priority 20, Always, team Any) are **not** Verse-tagged — they stay on so a wiped team still has a pad.

## Device inventory (high-signal)

| Count | Class | Role |
|---|---|---|
| 3 | `Device_CaptureArea` | A/B/C — periodic score, no neutralize, HUD badges |
| 3 | `Device_PlayerCounter` | LargeZone volumes for spawn weighting |
| 69 | player spawners | Initial (16) + Group A/B/C pairs + fallback + lobby |
| 5 | `VerseDevice` | the six scripts (one unused slot / combined) |
| 13 | `Device_CinematicSequence` | lobby + jets + cap lighting |
| 91 | `Device_CRD_AudioPlayer` | VO banks |
| 5 | `Device_Keylock` + 8 triggers | spawn-closet doors (open on contact, close +2s) |
| 7 | class designer / selector / evaluator | loadouts |
| 1 | `BP_LumenExposureManager` | same Day Sequence / Materials pattern |
| 4 | `LandscapeStreamingProxy` + 1 Landscape | desert terrain (`M_Deserted_Landscape`) |

## What this sample actually teaches

1. **Don't hard-ref 69 spawners.** Tag them; `FindCreativeObjectsWithTag` at `OnBegin`.
2. **Spawn near owned points *and* allies** — capture ownership is not enough; player-counter zones supply the 0.5 ally/enemy term.
3. **Disable the round-start pads after N seconds** (`Sleep(5)` then `Disable()`).
4. **Capture lighting is Sequencer, not the device's built-in light** — one cine per team per point.
5. **VO is a queue with a 2s cooldown** and a `HasBeenCaptured` gate so intro doesn't shout “losing control.”
6. **Pregame cine must be stopped + cleaned** (`SEQ_CineLobby` Stop, `SEQ_CineLobbyCleanup` Play) or material params stick.
7. **Spawn closets:** Lock + two Triggers (open on overlap, delayed close) — no Verse.
8. **Art lives in Fab, not the project.** `.uref` stubs enable packs; actors hard-ref `/<uuid>/SM_*`. Duplicate without the three stubs and 4973 meshes vanish.

## Official tutorial vs shipped

**External:** Deserted: Domination Template

Long walkthrough: Island Settings tables, paste `DominationSpawnManager` + tag module, pad clusters, Capture Area property table, Player Counter LargeZones, Lock closets, HUD Controller. Spreadsheet `/en-us/uefn/deserted-domination-template-in-unreal-editor-for-fortnite`. `docsUrl` empty. Retrieved signed-in Chrome 2026-09-11.

### Lines up

| Tutorial | Shipped |
|---|---|
| Max Players 12 / 2 teams / Time 15 / Score 500 | matchmaking 12 / 2 teams |
| 16 initial pads, then Verse-disabled | `Tag_SpawnGroup_Initial` + `Sleep(5)` |
| Group A/B/C tags + TeamOwnership | `Deserted.verse` module |
| 3 Capture Areas + 3 Player Counters | 3 + 3 |
| Fallback pads Priority 20 Always | described; not Verse-tagged |
| Lock + delayed close Trigger | 5 locks, 8 triggers |
| Verse device binds CapturePoints[0..2] + LargeZones[0..2] | `@editable` arrays |

### Does not line up

| Topic | Tutorial | Shipped |
|---|---|---|
| Recreate-from-scratch | pastes spawn manager + tags only | **5 other Verse files** (lighting, VO, sequencer randomizer, cine lobby) undocumented as scripts |
| “ProjectName.verse for tags” | says define tags in project file | tags live in `Deserted.verse` |
| Cap lighting / VO queue / jet randomizer | not in the spawn chapter | full devices + sequences |
| Live Outliner | — | empty until WP tiles load |
| Keep | — | spreadsheet **No / Deprecate** (landscape stale, non-Lumen lighting) |

Train on the **weight formula + tags**. Do not emit “paste only DominationSpawnManager” as the whole island. Do not treat Keep=No as “skip the pattern.”

## Negative findings

- Scratch `Developers/` folder.
- 4973 static meshes — island art, not a greybox kit.
- 2 unclassified OFPA.
- Live folders [].
- Spreadsheet: landscape actors out of date; needs lighting pass without Lumen.

## Capture backlog

1. Live WP load so Outliner folders exist. *Open, low — disk classes already counted.*
2. Which 5 `VerseDevice` slots map to which 6 scripts. *Open, low.*

## Live notes

Level path confirmed. `get_folders` empty. No file >80MB. Chrome stayed signed in from Greyboxing tab.
