---
id: digest-007
kind: digest
---

# detonation_template — sample digest

> Offline `_raw/detonation_template/`. Live level
> `/7a3e4a76-a564-4ddc-a0aa-32bcaad637ea/detonation_template` (**Observed**).
> Docs: **one page**, no chapter split.

| | |
|---|---|
| Classification | **Drop-in Verse SnD** — one device, `@editable` wiring |
| Content type | **catalog wrap**: 6 authored assets vs **724** placed (mole-factory / weapon-lab kitbash) |

## Scale

| | |
|---|---|
| Authored | **6** (3 materials + scaffolding) |
| OFPA | 724 |
| Verse | **1** (`detonation_device.verse`) |
| Fab | 0 |
| Teams in matchmaking | **2** (`maxTeamCount`: 2) |
| `LevelHasStreamingDisabled` | 1 |
| Compatibility | 43.00 |
| Live folders | Gameplay/{Billboards,Explosive}, Interior/*, Lighting |

Spreadsheet: *good easy-to-use Verse people can take and put in their game* / FAB candidate. That matches: **the Verse file is the product**; the map is a demo shell.

Empty `Developers/felixkupis/Collections/`. Ninth first-party hit.

## What it is *for*

Two bomb sites (A/B). Team 1 arms a `timed_objective_device`; Team 2 disarms. First arm **disables the other site**. Then `race` wait-for-complete vs wait-for-stop. Complete → explode `@editable []explosive_device` and `end_game_device.Activate`. Stop → End Game for the disarming player.

Beacons + map indicators flip with `bomb_state` enum.

## Device inventory (the design)

| Count | Class | Role |
|---|---|---|
| 8 | `Device_ExplosiveBarrel_V2` | 4 per site |
| 2 | `Device_TimedObjective_V2` | arm/disarm timers |
| 4 | `Device_MapIndicator_V2` | minimap per team × site |
| 4 | `Device_Beacon_V2` | arm/disarm look |
| 1 | `Device_GameEnd` | round end |
| 1 | `VerseDevice` | `detonation_device` |
| 2 | player spawners | Team 1 / 2 |
| 12 | billboards | captions |
| rest | factory kitbash | throwaway |

No HUD Controller, no NPC, no Scene Graph entities required.

## Verse

Snippet: `snippets/verse/detonation_device.verse`

```verse
bomb_state<public> := enum {AllUnarmed, BombAArmed, BombBArmed}

detonation_device := class(creative_device):
    @editable TimedObjectiveA/B: timed_objective_device
    @editable ExplosiveBarrelsA/B: []explosive_device
    @editable BombA/BMapIndicators: []map_indicator_device
    @editable BombA/BBeaconArm, BombA/BBeaconDisarm: beacon_device
    @editable EndGameDevice: end_game_device
    var BombState: bomb_state = bomb_state.AllUnarmed
```

`OnBegin`:

1. `race` two `block`s: `TimedObjectiveA.BeganEvent.Await()` vs B. Winner sets enum, **Disables the other Timed Objective**, disables the other site’s map indicators.
2. `UpdateBeacons()` — enable the disarm beacon on the armed site.
3. `race` four awaits: A/B `CompletedEvent` → `BombDetonated`; A/B `EndedEvent` → `EndGameDevice.Activate(DisarmingPlayer)`.

`BombDetonated`: `ExplodeBarrels` on the matching array (`Barrel.Explode(Agent)`), then End Game.

Logger: `log_detonation` channel (file comment still says “parkour experience” — leftover).

## What this sample actually teaches

1. **`race` of `Await`s is how Verse expresses “whichever bomb/event happens first.”** Nested races: first arm, then complete-vs-disarm. Losing branches cancel. See techniques/race-await-game-flow.md.
2. **Closed bomb state = `enum`.** Unlike Verse Commander’s open `command` classes, three states never grow after publish — enum is the right tool (Commander tutorial said this).
3. **A Verse device is a wiring panel:** every Creative actor is `@editable` (or `@editable []`). No tags, no find-by-name.
4. **Devices do the interactable work** (Timed Objective Arm/Disarm times, team filters). Verse only sequences and explodes barrels.
5. **Take the `.verse` file, not the factory map.** 6 authored vs 724 placed.

## Official tutorial vs shipped

**External** (single page, no chapters):
Verse Detonation Template

`docsUrl` empty. Spreadsheet `Link in Desc` is the hub (`/en-us/uefn/verse-detonation-template-in-unreal-editor-for-fortnite`).

### What lines up

| Tutorial | Shipped |
|---|---|
| Two-team SnD, arm/disarm bomb sites, 5 rounds / 2 teams in Island Settings | matchmaking `maxTeamCount` 2 |
| 8 explosives, 2 timed objectives, 4 map indicators, 1 end game, spawn pads, 1 Verse device | OFPA counts |
| `bomb_state` enum `{AllUnarmed, BombAArmed, BombBArmed}` | identical |
| `@editable` fields matching Details-panel names | same names |
| `race` then explode `Barrel.Explode(Agent)` | same |
| Device setup tables (30s timer, Team 1 start / Team 2 stop, Whole Map sound) | **External** — not re-read from live ToyOptions this pass |

### What does **not** line up

| Topic | Tutorial | Shipped |
|---|---|---|
| Verse class | `search_and_destroy := class(creative_device)` | **`detonation_device`** |
| Log channel | `log_search_and_destroy` | `log_detonation` (comment still says parkour) |
| First-arm event | `TimedObjectiveB.StartedEvent.Await()` | **`BeganEvent`** (A and B) |
| Beacons | tutorial Details list BeaconArm/Disarm; Verse snippet `UpdateBeacons` similar | 4 beacons **Observed** in OFPA; names match `@editable` |
| Tutorial “Map Indicators Text Explosive A/B vs Bomb A/B” wording | inconsistent in the same page | n/a |

**Training rule:** snippets win. Emitting `search_and_destroy` or `StartedEvent` on timed_objective is doc regurgitation. Use `detonation_device` + `BeganEvent` / `CompletedEvent` / `EndedEvent`.

No extra chapters — Pass B½ = this one page vs the file.

## Negative findings

- Map is not reusable; Verse is.
- Stale comment “parkour experience channel”.
- Scene Graph flag Unverified; unused.
- Streaming off; scratch `Developers/` folder.

## Capture backlog

1. Live Timed Objective `BeganEvent` vs docs `StartedEvent` already closed by reading the file. Confirm digest signatures in Verse digest if a trainer doubts names. *Waive.*
2. Island Settings vs tutorial User Options table (rounds=5, no env damage). *Open, low — Island Settings not in OFPA histogram.*
3. Billboard captions. *Waive — mechanic is the Verse, not the hall.*

## Live notes

Level confirmed. Folders `Gameplay/Explosive` + Interior kitbash. Streaming disabled.
