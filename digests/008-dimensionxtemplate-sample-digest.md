---
id: digest-008
kind: digest
---

# dimension_x_template — sample digest

> **Extractor provenance.** `tools/extract-template.py` (Pass A) + UnrealMCP on
> `localhost:8000` (Pass B). Default level
> `/1485696e-9fb4-490a-a936-030b8fdc0c63/dimension_x_template` (**Observed**).
> Docs: Working with TURTLE Islands
> (one-paragraph blurb, not a walkthrough). `docsUrl` empty; spreadsheet Documentation URL is the
> `uefn/working-with-TURTLE-islands-…` slug. Compat **43.00**. `featureSet`: **CustomGameplayFeature**.

| | |
|---|---|
| Classification | **theme environment starter** — alternate dimension landscape + SciFiFacility kitbash |
| Content type | **authored atmosphere + mounted alien kit** (95 authored vs 2019 OFPA) |
| Verse | **0 files** |
| Fab `.uref` | **0** |

## What it is

The picker name is **alternate dimension Starter**. The island is a **16-player sandbox landscape**, not a TURTLE fight: Melted Iron SciFiFacility playset, painted desert, ooze pools, custom nebula sky. Teaching that survives deletion is the ooze stack (default Water + sBRICK + Niagara + steam + local Day Sequence fog) and the split between authored atmosphere vs `/SCI_FI_*` mounts. Arcade’s spawner / sidescroller / Verse gather live in turtle_arcade_sample.

## Scale (**Observed** — `summary.txt` + live)

| | |
|---|---|
| Authored assets | **95** |
| OFPA | **2019** |
| Maps | `dimension_x_template.umap` (only) |
| Verse files | **0** (`VerseToolset.ListFiles` on `/1485696e-…` — do not recurse `__ExternalActors__`) |
| Verse devices placed | **0** |
| Data layers | **0** |
| Fab `.uref` | **0** |
| Compatibility | **43.00** |
| Matchmaking | **16** (`maxPlayers=16`, `teamSize=Dynamic`) |
| Streaming | Partitioned + actor folders + OFPA. **`LevelHasStreamingDisabled=1`** |

Empty `Developers/felixkupis/Collections/`. First-party scratch folder.

`experimental.*` (Python / toolsets / Scene Graph) in `.uefnproject` is **Unverified**. Content-side: **no** `.verse`, **no** `__ExternalEntities__`.

Largest file `Content/Textures/T_ENV_Goo_N.uasset` ~12 MB. Whole folder ~141 MB. Under the GHES **100 MB** per-file cap.

## Level organisation (**Observed** — live `get_folders`)

```
Environment/                    Landscape + 4 LandscapeRegion LocationVolumes
  Background/                   SkySphere, ocean waterbody + WaterZone
  Foliage/                      205 (InstancedFoliage + Frontier/Apollo trees)
  Lighting/                     Day Sequence Device_Main, SkyLight, SkyAtmosphere,
                                DirectionalLight, Global + SciFiFacility PPVs
    Lights/                     9 PointLight + 1 RectLight
  Ooze/                         sBRICK meshes + 21 NS OozeBubble Niagara
    OozePools/                  15 Water devices + Day Sequence Device_Pool1–4
    Steam/                      27 Engine planes + M_OozeSteam
  Rocks/ + Rocks/Vista Rocks/   Helios / modular rock kitbash
  SetDressing/                  Pretzel shards, crystals, vines
  SciFiFacility/                  948 BP_MIron_SCI_FI_* + PlaysetRoot “TURTLE SciFiFacility”
  VFX/                          18 VFX Spawners
World/                          IslandSettings0, minimap, LevelBounds
Enviroment/                     EMPTY typo tree (Background/Foliage/Mutagen/Rocks/…)
Lighting/ VFX/ SciFiFacility/ …   EMPTY leftover roots after the Environment/ move
```

No billboards. `bIsSpatiallyLoaded=false` on sampled devices; `load_actors` not required.

## Device inventory (the design)

| Count | Class | Role |
|---|---|---|
| 18 | `Device_VFXSpawner_V2` | LightFog / Ambient_Dust / WaterfallSplash, colorized |
| 15 | `Device_Water_V2` | ooze-pool volumes (`water Type=Default`) |
| 5 | `BP_DaySequenceModifier_Device` | Main (2000, global) + Pool1–4 (2020, volume) |
| 4 | `LocationVolume` | LandscapeRegion_0_0 … 1_1 |
| 2 | `PostProcessVolume` | Global unbound Prio 100; SciFiFacility bound Prio 200 |
| 2 | `VK_Waterbody_Ocean_Parent` + `WaterZone` | background ocean + pool zone |
| 1 | `Device_ExperienceSettings_V2_UEFN` | IslandSettings0 |
| 1 | `PlaysetRoot` | TURTLE SciFiFacility |
| 21 | `NiagaraActor` | `NS_GooBubble` |
| 0 | Billboard / HUD / TURTLE spawner / Player Spawner / Verse device | — |

## Authored kit (**Observed** — `assets.txt` + live referencers)

```
Content/Materials/   M_/MI_ProceduralSkybox, M_DimensionX_Landscape/Foliage,
                     M_/MI_Ooze_* + MPC_Ooze, leftover M_/MI_Mutagen_* + MPC_Mutagen
Content/Textures/    nebula/stars/planets, T_OozeWave_*, leftover T_Mutagen*
Content/Meshes/      SM_SkySphere, SM_ENV_SBRICKA (live), SM_ENV_SBRICKB/C unused
Content/VFX/NS/      NS_GooBubble (live on 21 NiagaraActors)
Content/Foliage/     SMF_PlantD (0 referencers)
Content/LandscapeLayers/  Base_Layer + Layer_1..6 LayerInfo
```

Live landscape material: `MI_DimensionX_Landscape_B` (`MI_DimensionX_Landscape_A` unused).
SkySphere: `SM_SkySphere` + `MI_ProceduralSkybox`. Registry `get_referencers` on those meshes
often returns 0 because OFPA does not count — read the placed component.

**Tier 2 mounts** (`reference-mounts.txt`): `/SCI_FI_Assets` 1932 + `/SCI_FI_Comp` 963 (SciFiFacility
`BP_MIron_SCI_FI_*`); Helios rocks `/ENV_PACK_*`; stock `/Game`. No Fab `.uref`.

## What this sample actually teaches

1. **Ooze is a stack, not a Water Type.** techniques/ooze-pool-stack.md. Live Water = **Default**; sBRICK mesh + `NS_GooBubble` + `M_OozeSteam` planes + Pool Day Sequences (fog **10**, green, priority **2020**).
2. **Local toxic air = Day Sequence volume above the island device.** Main 2000 / fog 0.2 / `use Volume=false`; pools 2020 / fog 10 / `use Volume=true`. Do not duplicate — day-sequence-priority-stack.md.
3. **Authored skybox can sit next to Day Sequence lighting.** SkySphere uses `MI_ProceduralSkybox`; Main DSD `overrideSky=true` but `enableSkyboxComponent=false` and `skydomeMaterial` is stock `MI_DaySequenceSkydome`. Contrast NightEnvironment, which has **no** Day Sequence: night-atmosphere-stack.md.
4. **SciFiFacility is a mounted playset, not this project’s kit.** 948 `BP_MIron_SCI_FI_*` under `PlaysetRoot` “TURTLE SciFiFacility” from `/SCI_FI_Assets` + `/SCI_FI_Comp`. Same “kit is a mount” split as NightEnvironment’s village.
5. **Ambient color is VFX Spawner presets**, not VFX Creator: 16× LightFog, 1× Ambient_Dust, 1× WaterfallSplash, all colorized, no bindings. vfx-surface-split.md.
6. **This is not Arcade.** No TURTLE Character Spawner, no sidescroller, no HUD, no Verse. `maxPlayers=16`, `allowBuilding=All`, `locomotionPreset=Current BR`.

## Official tutorial vs shipped

**External:** Working with TURTLE Islands (headed Chrome 2026-09-14). Spreadsheet Documentation points here; `.uefnproject` `docsUrl` empty. Arcade / City Starter have dedicated chapters; alternate dimension is **one paragraph**. Verbatim: `_raw/dimension_x_template/tutorial.txt`.

Guessed `TURTLE-dimension-x-starter-template-in-unreal-editor-for-fortnite` **redirects to the Fortnite docs hub**. Community search “alternate dimension” → **No results**. Hub ≠ this island.

### Lines up

| Tutorial | Shipped |
|---|---|
| Brands → alternate dimension Starter (one of three TURTLE templates) | Project Browser; `featureSet` CustomGameplayFeature |
| Landscape + pools of ooze + swirling outer-space sky | Landscape + OozePools + SkySphere nebula textures |
| SciFiFacility galleries | `PlaysetRoot` “TURTLE SciFiFacility” + `BP_MIron_SCI_FI_*` |
| Starting point for adventures (not a finished game) | No combat devices, building **All**, 16 players |

### Does not line up

| Tutorial | Shipped |
|---|---|
| “Sculpt the barren landscape” (user action) | Landscape already painted (`MI_DimensionX_Landscape_B`) |
| “Place pools of ooze” (user action) | 15 Water devices + sBRICK already placed |
| Dedicated walkthrough (Arcade / City Starter have one) | **None** — blurb only |
| Implies a alternate dimension docs page | Guessed slug → hub; search → no results |
| Water that *is* ooze | `water Type=Default`; look is meshes + DSD fog |

Training rule: live island wins. Do not emit a custom Water Type, a alternate dimension tutorial chapter, or Arcade devices on this map.

## Negative findings

- **Name implies a alternate dimension *game*; files show an empty sandbox.** No spawners, no HUD, no billboards, `find_actors` name=Player → debugger manager only.
- **`Enviroment/` (missing n) is an empty duplicate** of `Environment/`. Root `Lighting/`, `VFX/`, `SciFiFacility/`, `Mutagen/`, `Rocks/`, `Foliage/`, `Level/`, `Background/` are also empty leftovers after the move.
- **Mutagen material family is unused.** `M_Mutagen_*` / `MI_Mutagen_*` / `T_Mutagen*` / `MPC_Mutagen` have 0 referencers. Live folders are `Ooze/`.
- **`SM_ENV_SBRICKB` / `SBRICKC` / `SM_ENV_Growth_C` / `T_TexTest` / `SMF_PlantD` / `MI_DimensionX_Landscape_A`:** 0 referencers. Do not cite as worked examples. `SM_ENV_SBRICKA` *is* used (OFPA; registry can still say 0).
- **No player spawners.** Island Settings still has default `spawnLocation`.
- **`GetDeviceProperties` N/A** — no Verse devices.

## Capture backlog

| Item | Status |
|---|---|
| Niagara `NS_GooBubble` emitter rates | **Waived** — asset committed; live `Asset` + auto-activate captured |
| Per-pool Water zone sizes beyond sample Water 14×12×1 | **Waived** — class + Type Default is the lesson |
| Exact SciFiFacility playset gallery asset name | **Waived** — mount + `BP_MIron_SCI_FI_*` histogram is enough |
| Mutagen vs ooze material graph delta | **Waived** — mutagen has 0 referencers |
| Sequencer / cines | **Waived** — none in inventory |

## Live notes

- Mount `/1485696e-9fb4-490a-a936-030b8fdc0c63`. Project Verse path `/epictemplates@fortnite.com/dimension_x_template`.
- Did **not** `load_level` away from the default map.
- No billboards (unlike Arcade / Walker / VFX).
- Island Settings properties are on the **actor**, not `ToyOptionsComponent.playerOptionData` (that path 404s here).
- `get_referencers` on authored meshes often misses OFPA; confirm with `StaticMeshComponent0`.
