---
id: digest-009
kind: digest
---

# elimination_template — sample digest

> Offline `_raw/elimination_template/`. Live
> `/10d54626-1784-4998-96c5-0e8b3a7d1c60/elimination_template` (**Observed**).
> Docs: one page, no chapters.

| | |
|---|---|
| Classification | **Drop-in Verse FFA** — gun upgrades on elim |
| Content type | catalog wrap: **6 authored vs 700 placed** (same mole-factory kit as Detonation) |

## Scale

| | |
|---|---|
| Authored | 6 |
| OFPA | 700 |
| Verse | **1** (`elimination_game_device.verse`) |
| Item Granters | **4** |
| End Game | 1 |
| Spawners | 2 |
| Billboards | 8 |
| `LevelHasStreamingDisabled` | 1 |
| Matchmaking | 16 players / 16 teams (FFA-shaped) |
| Compatibility | 43.00 |

Spreadsheet twin of Detonation: *Verse people can take / FAB candidate.* Take the `.verse`, not the factory.

`Developers/felixkupis/Collections/` empty. Tenth first-party hit.

## What it is *for*

Each player starts with granter[0]. On elim, the **eliminator** gets the next weapon. Win = first player whose elim count == `WeaponItemGranters.Length` (scalable: add granters in Details, win condition follows). Order shuffled once at `OnBegin`.

Progress is a `[agent]int` map, not an enum (open-ended count vs Detonation’s 3 states).

## Device inventory

| Count | Class | Role |
|---|---|---|
| 4 | `Device_ItemGranter_V2` | one unique weapon each; Verse `Shuffle`s the array |
| 1 | `Device_GameEnd` | `Activate(winner)` |
| 1 | `VerseDevice` | `elimination_game_device` |
| 2 | player spawners | FFA spread |
| 8 | billboards | captions |
| rest | factory kitbash | throwaway |

No Timed Objective, no explosives — contrast Detonation.

## Verse

Snippet: `snippets/verse/elimination_game_device.verse`

```verse
var WeaponItemGranters:[]item_granter_device = array{}
var NumberOfEliminationsToWin:int = 0
var AgentMap:[agent]int = map{}

OnBegin:
    set NumberOfEliminationsToWin = WeaponItemGranters.Length
    set WeaponItemGranters = Shuffle(WeaponItemGranters)
    for Player in GetPlayers():
        FortCharacter.EliminatedEvent().Subscribe(OnPlayerEliminated)
        set AgentMap[Player] = 1
        WeaponItemGranters[0].GrantItem(Player)
```

`OnPlayerEliminated(Result:elimination_result)`: `Result.EliminatingCharacter?` → `GetAgent[]` → `GrantNextWeapon`.

`GrantNextWeapon`: if `IsVictoryConditionMet[CurrentItemNumber]` then EndGame else `WeaponItemGranters[CurrentItemNumber].GrantItem` then increment map.

Victory check is `<decides><transacts>` — call with `IsVictoryConditionMet[n]`, not `= true` (**Observed**; docs show `= true`).

**Bug-shaped off-by-one vs docs:** tutorial grants `WeaponItemGranters[CurrentItemNumber - 1]` after increment; shipped grants `WeaponItemGranters[CurrentItemNumber]` **then** increments. Map starts at 1, first extra gun is index 1. Train on the file.

`using { /Verse.org/Simulation/Tags }` is imported and unused.

## What this sample actually teaches

1. **Subscribe to `fort_character.EliminatedEvent()`**, not a Creative elimination device. Payload is `elimination_result` with optional EliminatingCharacter (self-elim / world = no grant). See techniques/elim-event-loadout.md.
2. **`[agent]int` map = per-player progress.** Weak identity of `agent` as key.
3. **Win condition scales with `@editable []` length** + `Shuffle` so loadout order isn’t authoring order.
4. **`<decides>` predicate** instead of returning `logic`.
5. Same shell lesson as Detonation: 6 vs 700 — migrate the device script.

## Official tutorial vs shipped

**External:** Verse Elimination Template

Single page. `docsUrl` empty.

### Lines up

| Tutorial | Shipped |
|---|---|
| Class `elimination_game_device` | same |
| Map APIs, arrays, `EliminatedEvent` | same |
| `@editable EndGameDevice`, `WeaponItemGranters` | same |
| Shuffle, grant on elim, EndGame.Activate | same |
| 4 item granters in the Details example | OFPA 4 |
| Island: keep items on elim, no drop, no env destroy | **External** (not live-read) |

### Does not line up

| Topic | Tutorial | Shipped |
|---|---|---|
| Victory call | `if (IsVictoryConditionMet(CurrentItemNumber) = true)` | `if (IsVictoryConditionMet[CurrentItemNumber])` — `<decides>`, failable `[]` |
| Victory function | `):logic=` return true/false | `<decides><transacts>:void=` with `EliminationNumber = NumberOfEliminationsToWin` |
| Grant index | `[CurrentItemNumber - 1]` after increment | `[CurrentItemNumber]` then increment; map starts at 1 |
| Logging | `Print("…")` | `Logger.Print` + `log_elimination` |
| `Simulation/Tags` import | present in snippet header | present, unused |

Train on snippet. Emitting `= true` on a `<decides>` function is doc regurgitation.

## Negative findings

- Factory map is Detonation’s twin; don’t double-count as unique environment teaching.
- Unused Tags import.
- Self-elim / storm: EliminatingCharacter empty → no grant, no crash (good).
- Streaming off; scratch Developers folder.

## Capture backlog

1. Confirm live granter count vs Details (OFPA already 4). *Waive.*
2. Billboard text. *Waive.*
3. Island Settings vs tutorial table. *Open, low.*

## Live notes

Level confirmed. `LevelHasStreamingDisabled` 1.
