---
id: digest-010
kind: digest
---

# equip_furniture_props — sample digest

> **Extractor provenance — read before citing this digest.**
> Offline evidence was produced by `tools/extract-template.py`, a reconstruction of the
> pipeline described in `00-EXTRACTION-PROCESS.md` (the original `.sh` does not exist in this
> repo). Live-editor facts were read via UnrealMCP on `localhost:8000` with the project open;
> `SceneTools.get_current_level` returned `/equip_furniture_props/equip_furniture_props`.

| | |
|---|---|
| Verdict recorded | *"just a catalog of items … impossible to keep in sync with all the new items being added into Fortnite, so not worth maintaining"* |
| Classification | **Reference / catalog** — not a playable or teaching template |
| Recommendation | **Retire** — the extraction supports the stated rationale directly |

## The finding, in one line

**14,116 placed actors drawn from 6,356 distinct classes, and exactly 2 authored assets in the
entire template — neither of which is a furniture item.** This is a hand-placed snapshot of
Epic's prop library, with no data layer, no metadata, and no mechanism to refresh itself.

## Scale (Observed — `summary.txt`)

| | |
|---|---|
| **Authored assets** | **2** |
| External actor packages (OFPA) | **14,116** |
| External object packages | 2 |
| Total `.uasset` | 14,120 |
| Maps | 2 — `equip_furniture_props.umap`, `Design_Palettes.umap` |
| Verse files | **0** |
| Fab references (`.uref`) | **0** |
| Distinct actor classes | **6,356** |
| Unclassified actors | 0 |

The two authored assets are `GameFeatureData.uasset` and
`equip_furniture_props_DefaultHLODLayer.uasset` (**Observed** — `assets.txt`). Both are project
plumbing that UEFN generates. **The template authors zero furniture, zero materials, zero
Blueprints.** Every catalogued item is a *reference* to content that ships with Fortnite.

Compare: character_environment_sample has 307 authored assets, intro_to_uefnfk1 has 16. this sample has 2.

## Why it cannot be kept in sync (Observed)

This is the user's stated rationale, and the evidence gives it a precise mechanism.

### 1. There is no data-driven layer to refresh

Searching all 81,630 referenced paths (**Observed** — `referenced-paths.txt`):

```
DataTable            0
CurveTable           0
StringTable          0
CompositeDataTable   0
.csv                 0
DataAsset           49   (engine-internal, not a catalog index)
```

**No table, no CSV, no registry.** The catalog is not generated from a manifest that could be
re-run against a newer Fortnite build — it is 14,116 actors positioned by hand in a `.umap`.
Adding this season's props means a human placing them, one at a time, forever.

### 2. There is no organisational structure to extend

`SceneTools.get_folders` on the live level returns exactly one folder (**Observed**):

```
["Background"]
```

**One actor folder for 14,116 actors.** No per-category grouping, no per-season grouping, no
World Partition data layers (`datalayers.txt`: 0 distinct, 0 instances). There is no slot to add
a new season's props *into*, and no way to see what is missing.

### 3. It is indexed by Fortnite art pack, not by furniture category

Actor classes cluster by Epic's internal season/location codenames, not by what the object *is*
(**Observed** — `actor-classes.txt`, prefix rollup):

| Actors | Prefix | What it is |
|---|---|---|
| 3,819 | `BP_` | generic Blueprint props |
| 2,576 | `Apollo_` | Fortnite season/biome art pack |
| 2,162 | `CP_` | Creative Prop library |
| 910 | `Athena_` | Battle Royale core art |
| 533 | `Asteria_` | season art pack |
| 463 | `ENV_PACK_` | Fortnite Ecosystem Content |
| 400 | `Prop_` | misc props |
| 319 | `Artemis_` | season art pack |
| 213 | `NeoTilted_` | POI art pack |

A furniture catalog organised by *shipping vehicle* rather than by *furniture type* inherits
Epic's release cadence as its maintenance burden. Every new art pack is a new prefix nobody has
placed.

### 4. The long tail makes bulk maintenance impossible

| Coverage | Share of the 14,116 actors |
|---|---|
| Top 10 classes | 5.0 % |
| Top 50 classes | 11.6 % |
| Top 100 classes | 16.9 % |
| Top 500 classes | 33.4 % |

**2,590 of the 6,356 classes (40.7 %) appear exactly once.** There is no meaningful head to
maintain — two thirds of the catalog lives outside the top 500 classes. Any update touches
thousands of unique, individually-placed items. This is the quantitative form of "not worth
maintaining."

### 5. It depends on 117 separate Fortnite content mounts

**Observed** — `reference-mounts.txt`, 81,630 references across **117 distinct mounts**:
`/Game` (35,474), `/equip_furniture_props` (28,246), `/TREE_PACK` (3,031), `/Engine` (2,282),
`/CreativeCoreDevices` (1,332), then a long tail of per-POI packs (`/CRP_NeonCity`,
`/LACity_Comp`, `/TTides_Comp`, `/CRP_SlappyShores`, …).

Every one of those is a moving target Epic revises independently. The template has **no `.uref`
files**, so none of these dependencies are declared — they are implicit, and nothing warns you
when one changes.

## Content of the template (Observed)

- **No custom logic.** 0 `.verse` files, 0 `__ExternalEntities__` directories — no Verse, no
  Scene Graph entities. Nothing to preserve on the code side.
- **Actor mix** is overwhelmingly static Fortnite geometry: `/Script/FortniteGame.BuildingProp`
  (12,711), `BuildingDeco` (595), `FortCreativeDeviceProp` (310), `BuildingFloor` (211)
  (**Observed** — `script-modules.txt`).
- **421 billboard-ish actors**, of which only **212 are actual `Device_Billboard_V2` signage** —
  the rest are decorative billboard *props* (`ENV_PACK_BP_CozyTea_*Billboard*`,
  `Apollo_Street_Billboard_*`). Signage device labels are mostly autogenerated
  (`Billboard56`, `Billboard165`, `Billboard190`), not curated section headers.
- **19 localisation languages** are wired up (`ar, de, en, es, es-419, fr, id, it, ja, ko, pl,
  pt-BR, ru, th, tr, vi, zh-Hans, zh-Hant` + source) for a single 264-byte `.ugcloc` — localisation
  scaffolding with essentially no strings behind it.
- **`Developers/felixkupis/`** is present but empty. Same housekeeping smell as character_environment
  (`Developers/jojitsuruga/`) and intro_to_uefnfk1 (`Developers/felixkupis/`) — **three for
  three** in first-party templates.

## Project configuration (⚠ flags Unverified)

> **The `experimental.*` block is not safe to grade as shipped state.** The project was open in
> UEFN during extraction and the operator has not confirmed whether Project Settings were
> touched; these flags are writable from that UI, are **off by default**, and UEFN rewrites the
> descriptor on disk the moment one is flipped, leaving nothing to distinguish shipped state
> from an operator toggle. This exact contamination invalidated a headline finding in
> intro_to_uefnfk1.md.
> Guard codified in 00-EXTRACTION-PROCESS.md Step 2.
>
> **Nothing in this digest's retirement case depends on these flags.**

```
compatibilityVersion               42.10          (Observed)
requiredRedirectorStartingVersion  42.10.57566230 (Observed)
apiAccess                          Private        (Observed)
matchmaking.maxPlayers             16             (Observed)
VersePath  /epictemplates@fortnite.com/equip_furniture_props  (Observed)
VerseVersion 1 / VerseAPI 1                       (Observed)

⚠ experimental.sceneGraph.bIsSceneGraphSystemAllowed  true   <- Unverified
⚠ experimental.sceneGraph.bEnableOneFilePerEntity     false  <- Unverified
⚠ experimental.pythonExperimental                     true   <- Unverified
⚠ experimental.toolsets                               true   <- Unverified
⚠ .uplugin EnableSceneGraph                           true   <- Unverified
⚠ .uplugin EnableVerseAssetReflection                 true   <- Unverified
```

Note the flag set is *narrower* than intro_to_uefnfk1's (no `sceneGraphAnimation`, no
`sceneGraphExperimentalFeatures`, no `verseControlRigEnabledSetting`, no `inventory`,
no `assetRedirectorReflectionSetting`) — consistent with a project that was never set up for
authoring, but not evidence either way given the contamination risk.

## What is worth keeping

**Almost nothing, and that is the point.** There is no Verse, no custom Blueprint, no authored
material, no data table. The template's entire value is *the curation itself* — the knowledge of
which props exist and what they look like next to each other — and that value is exactly what
decays with every Fortnite release.

If any part of it is worth rescuing, it is the **question the template was built to answer**
("what props are available to me?"), which is better served by:

- the in-editor Content Browser filtered to `/Game` prop paths, which is *always* current, or
- a generated index (script the asset registry, emit a browsable list) that can be **re-run**
  against each new Fortnite build.

Both replace a 14,116-actor hand-placed map with something self-updating. That is the structural
fix; no amount of maintenance effort on the current form gets there.

## Retirement assessment

| Criterion | Status |
|---|---|
| Digest written | **Met** |
| Verse preserved | **N/A** — zero Verse in template |
| Unique content captured | **Met** — nothing unique exists; every asset is Epic-owned and remains available |
| Capture backlog empty | **Met** — 0 open (see below) |
| Extractor provenance | **Qualified** — reconstructed extractor, same caveat as intro_to_uefnfk1 |

### Capture backlog

**0 open.** Unusually, there is nothing worth going back for. The three questions that would
normally be open — *what does it teach*, *what logic does it contain*, *what does it depend on
that would be lost* — are all answered negatively by the offline pass: it teaches nothing
(no signage curriculum, autogenerated billboard labels), contains no logic (0 Verse,
0 SG entities), and depends only on Fortnite-owned content that survives deletion.

The one item deliberately **not** pursued: reading the `Text` property off all 212
`Device_Billboard_V2` actors to check for curated category names. Sampled labels are
autogenerated (`Billboard56`, `Billboard120`, `Billboard171`), the level has a single
`Background` folder, and 212 signs cannot organise 14,116 items into a navigable taxonomy —
so the answer cannot change the verdict. Abandoned after the probe exceeded 60 s.

### Recommendation

**Retire.** The user's rationale is correct and the extraction quantifies it: this is a static
snapshot of a moving target, with no data-driven layer, no categorisation, and a 40.7 %
singleton class distribution that makes bulk maintenance impossible. It authors **2 assets**,
neither of which is furniture.

Nothing is at risk on deletion — every catalogued prop is Epic-owned and remains available in
the Content Browser. Replace the *function* with a generated, re-runnable asset index rather
than a hand-placed map.
