---
id: digest-011
kind: digest
---

# obstacle_course_starter — sample digest

> **Extractor provenance.** `tools/extract-template.py` (Pass A) + UnrealMCP on
> `localhost:8000` (Pass B). Default level
> `/b9c10c03-85e6-4cab-b6cf-6b98605c434e/obstacle_course_starter` (**Observed**).
> Hub: Building obstacle-course Islands in Fortnite Creative
> + dedicated guide Working with obstacle-course Islands
> + design companion obstacle-course — Designing an Obstacle Course
> + catalog reference obstacle-course Obstacle Course Assets.
> `docsUrl` empty in `.uefnproject`. Not on the legacy spreadsheet (discovered on the 2026-09-14 UEFN Project Browser theme snapshot, #26 in `theme.md`). Compat **43.00**. `featureSet`: **PlatformLocomotion**.

| | |
|---|---|
| Classification | **obstacle-course theme starter / obby greybox** — minimal linear obstacle course demonstrating Avatar physics, Season 7 modular geometry, intangible sBRICK pits, and an autonomous ping-pong platform |
| Content type | **catalog wrap** (2 authored vs 259 OFPA / 273 live actors; `/platformingLevelArt` S07 modular kit + `/platformingExperienceSettings`) |
| Verse | **0 files** |
| Fab `.uref` | **0** |

## What it is

The picker name is **obstacle-course Starter** (obstacle-course theme tile in the UEFN Project Browser). The on-disk folder is `obstacle_course_starter`. The template is the official reference sandbox for authoring obstacle-course obstacle courses in UEFN. It features a complete linear linear course suspended over an ocean backdrop, progressing from a starting platform through three hazard stages (a sBRICK pit with four bumper poles, a bounce drum pit crossing, and a moving hover platform gap) up to a celebratory Finish Line Arch and victory bouncer podium.

Teaching that survives deletion is the **PlatformLocomotion feature-set locomotion tuning** (`Device_ExperienceSettings_FallGuys`), the **S07 half-tile modular kit standard** (2x2 tile base, F/H/Q height tiers, corner pivots), the **intangible sBRICK hazard stack** (`bCollisionInGame: false` visual meshes over multi-tier Elimination Damage Volumes), and the **autonomous zero-wiring Prop Mover** (`pathCompleteAction=EFCLMMS_PingPong` with `propSelection=EFCLMPS_Overlap`).

## Scale (**Observed** — `summary.txt` + live)

| | |
|---|---|
| Authored assets | **2** (`GameFeatureData`, `obstacle_course_starter_DefaultHLODLayer`) |
| OFPA packages | **259** (273 actors loaded in live level) |
| Maps | `obstacle_course_starter.umap` (only) |
| Verse files | **0** (`_raw/obstacle_course_starter/verse/_NONE.txt`; `VerseToolset.ListFiles` returns folders only) |
| Verse devices placed | **0** |
| Data layers | **0** (`SceneTools.get_data_layers` returned `[]`) |
| Fab `.uref` | **0** |
| Compatibility | **43.00** (`requiredRedirectorStartingVersion: 43.00.57945210`) |
| Matchmaking | **16** max players / `teamSize=Dynamic` / `allowJoinInProgress=true` / `allowSquadFillOption=false` (`.uefnproject` + Island Settings) |
| Streaming | Partitioned + actor folders + OFPA. **`LevelHasStreamingDisabled=1`**. Live `get_folders` returned only `Background` (16 backdrop landscape actors); all 257 gameplay actors are unfoldered (`folderPath=None`). |

Empty `Developers/` and `Collections/`. No experimental Scene Graph entities (`__ExternalEntities__` absent; plugin `EnableSceneGraph: true` without graph actors).

Whole folder is compact (~**2.5 MB**), far below GHES 100 MB cap.

## Level organisation (**Observed** — live `get_folders` + finds)

```
Content/
  GameFeatureData.uasset
  obstacle_course_starter_DefaultHLODLayer.uasset
  obstacle_course_starter.umap
  __ExternalActors__/obstacle_course_starter/   (259 OFPA packages)
  __ExternalObjects__/obstacle_course_starter/  (1 package)
```

Outliner folder paths:
- `Background/` (16 actors): `S_BackDrop`, `M_BackDrop` mountain background instances.
- `(None)` (257 actors): Spawner, Island Settings, 4 Damage Volumes, Prop Mover, and all obstacle-course modular pieces.

The course extends linearly along the **Y axis** from Y = -2000 cm (Start) to Y = +9100 cm (Podium):

| Y Coordinate | Z (Elevation) | Component | Role |
|---|---|---|---|
| **Y = -2000** | Z = 1260 | `Player Spawner0` | Island start spawn pad; direct spawn (no skydive). |
| **Y = -516** | Z = 728 / 1252 | `Damage Volume0` + `FG01_Cube_SBRICK_4s_C0` | First hazard pit: intangible sBRICK over local elimination volume. |
| **Y = 264 - 776** | Z = 1567 | 4× `FG01_Bumper_Pole` | Diamond-pattern bumper obstacle corridor. |
| **Y = 3328 - 3352** | Z = 1096 / 1796 | `Damage Volume1` + `FG01_Cube_SBRICK_4s_C1` + `FG01_Drum_S_C0` | Second hazard: bounce drum leap across sBRICK pit. |
| **Y = 5888 - 5976** | Z = 1288 / 2192 | `Damage Volume2` + `FG01_Cube_SBRICK_4s_C2` + `Prop Mover0` + `FG01_HoverPlatform_M_C0` | Third hazard: translating hover platform (15m ping-pong) across sBRICK gap. |
| **Y = 7768 - 7916** | Z = 2831 / 3264 | `FG01_Finish_Arch_*` | Triple-piece finish line arch (`Side`, `Mid`, `Billboard`). |
| **Y = 9099** | Z = 2507 | `FG01_Drum_L_C0` | Victory celebration bouncer behind the finish line. |
| **Global (Y = 3572)** | Z = 728 | `Damage Volume3` (100×100 tiles) | Ocean catch-all elimination plane beneath entire suspended island. |

## Device inventory (the design)

The entire gameplay experience runs on just **7 devices** (0 Verse, 0 Event Bindings):

| Count | Class | Device Label | Role | Key Properties |
|---|---|---|---|---|
| 1 | `Device_ExperienceSettings_FallGuys` | `IslandSettings0` | Game Rules & Avatar Locomotion | `locomotionPreset=Custom`, `bAllowSprinting=false`, `bAllowSliding=false`, `bAllowVaulting=false`, `bAllowMantling=true`, `bAllowHurdleOverJumpableObjects=true`, `bFallDamage=false`, `bGliderRedeploy=true`, `roundTimeLimit=5`, `totalRounds=1` |
| 1 | `BP_Creative_Player_Spawner_Prop` | `Player Spawner0` | Starting Line Spawn Pad | `useAsIslandStart=true`, `visibleInGame=false`, `playAudio=No`, `enabledDuringPhase=Always` |
| 1 | `Device_PropMover_V2` | `Prop Mover0` | Moving Gap Platform Mover | `movementMode=EFCLMM_Translation`, `pathCompleteAction=EFCLMMS_PingPong`, `distanceInMeters=15`, `speedInMeters=5`, `shouldMoveFromStart=true`, `propSelection=EFCLMPS_Overlap` (attached to `FG01_HoverPlatform_M_C_0`) |
| 4 | `BP_Device_DamageVolume_V2` | `Damage Volume0..3` | Elimination Hazard Volumes | `damageType=Elimination`, `damage=10`, `damage Tick Rate=2 Seconds`, `enabledOnPhase=Gameplay Only`, `zoneVisibleDuringGame=false`. DV 0, 1, 2 = local pit volumes; DV 3 = 100×100 ocean catch-all |

## obstacle-course modular asset inventory

From `/platformingLevelArt/Environments/S07/`:
- **Shapes & Walls:** 44× `FG01_Shape_Rect_H_2s`, 40× `FG01_Shape_Rect_F_2s`, 32× `FG01_Wall_Str_H_1s`, 4× `FG01_Shape_Cylinder_1s`
- **Barriers & Pillars:** 36× `FG01_Barrier_Pillar_Flat_Q`, 26× `FG01_Barrier_Str_Q_2s`
- **Floors & Ramps:** 13× `FG01_Floor_Str_4x2s`, 6× `FG01_Ramp_Str_H_1x2s_B`, 2× `FG01_Floor_Str_4x3s`, 2× `FG01_Floor_Str_4x1s`
- **Hazards & Bouncers:** 4× `FG01_Bumper_Pole`, 3× `FG01_Cube_SBRICK_4s` (`bCollisionInGame=false`), 1× `FG01_Drum_S` (`bPlayBounce=true`), 1× `FG01_Drum_L` (`bPlayBounce=true`), 1× `FG01_HoverPlatform_M`
- **Finish Line & Accents:** 2× `FG01_Finish_Arch_Side`, 1× `FG01_Finish_Arch_Mid`, 1× `FG01_Finish_Arch_Billboard`, 4× `FG01_Banner_Floating_Base` / `Blades`, 2× `FG01_Banner_Standing`

## What this sample actually teaches

1. **How to activate the obstacle-course feature set:** Set `"featureSet": "PlatformLocomotion"` and `"DefaultPawn": {"bUseDefaultCharacter": true}` in `.uefnproject` to mount `/platformingLevelArt` and spawn players as customized obstacle-course Avatars.
2. **Locomotion restrictions that protect platforming design:** Disabling Fortnite tactical sprinting, sliding, and vaulting while keeping mantling enabled preserves intentional platform spacing and jump-gap mathematics (techniques/fall-guys-locomotion-tuning.md).
3. **S07 modular kit grid taxonomy:** Standard tile = 512×512×384 cm. obstacle-course assets use half-tile footprint units (1 = 256 cm, so standard tile floor is `2x2`, platform is `4x2`) and standardized height tiers (Full = 384 cm, Half = 192 cm, Quarter = 96 cm) with corner pivot points (techniques/fall-guys-modular-scale-conventions.md).
4. **Intangible visual hazards over elimination volumes:** Placing visual hazard meshes with `bCollisionInGame: false` (`FG01_Cube_SBRICK_4s`) directly above invisible Elimination Damage Volumes creates a seamless trap where players drop through into instant elimination and respawn (techniques/intangible-sBRICK-hazard-killzone-stack.md).
5. **Multi-tier hazard layering:** Pairing localized pit elimination volumes (DV 0, 1, 2) with a massive 100×100 tile global ocean catch-all plane (DV 3 at Z = 728) prevents ragdolled players from falling infinitely or swimming in ocean water.
6. **Zero-wiring autonomous moving platforms:** Using `Device_PropMover_V2` with `pathCompleteAction: "EFCLMMS_PingPong"` and `propSelection: "EFCLMPS_Overlap"` creates autonomous translating obstacle platforms without Verse scripts, helper Switches, or event wiring (techniques/autonomous-overlapping-prop-mover.md).

## Official tutorial vs shipped

The official documentation suite directly documents this starter template across multiple dedicated guides:
- Building obstacle-course Islands in Fortnite Creative
- Working with obstacle-course Islands
- obstacle-course — Designing an Obstacle Course
- obstacle-course Obstacle Course Assets

### Lines up

| Feature | Tutorial prose | Shipped project |
|---|---|---|
| Starter concept | Linear sample obstacle course for testing the obstacle-course toolset | Single linear course along Y axis (-2000 to +9100) |
| Modular assets | Banners, ramps, platforms, blocks, pillars, finish arch | Season 7 platforming modular kit (`/platformingLevelArt/Environments/S07/`) |
| Platform mover | Prop Mover paired with a platform to cross a void gap | `Device_PropMover_V2` overlapping `FG01_HoverPlatform_M` (15m ping-pong at 5m/s) |
| SBRICK hazard physics | SBRICK cubes default to `Collision In Game: Off`, dropping Avatars into respawn | `FG01_Cube_SBRICK_4s` instances have `bCollisionInGame=false` |
| Elimination volumes | Damage Volumes placed under pits and water to eliminate Avatars | `BP_Device_DamageVolume_V2` instances set to `damageType="Elimination"` |
| Locomotion rules | Avatars do not sprint, slide, vault, or take fall damage; mantling is core | Verified via live `Device_ExperienceSettings_FallGuys` |
| Direct spawning | Avatars spawn directly on pads (no skydiving) | `BP_Creative_Player_Spawner_Prop` with `useAsIslandStart=true` |

### Does not line up

| Feature | Tutorial claims | Shipped project |
|---|---|---|
| **Damage Volume count** | Tutorial specifies "3 Damage Volumes: Two Damage Volumes to go under the SBRICK pits, One Damage Volume to cover the water" | Shipped contains **4 Damage Volumes** (DV 0, 1, 2 under the three distinct hazard pits + DV 3 100×100 over the ocean) |
| **SBRICK pit count** | Tutorial specifies "SBRICK cubes to make two SBRICK pits" | Shipped contains **3 separate SBRICK cube actors** across three hazard zones (Y = -512, Y = 3328, Y = 5888) |
| **Hop Flower device** | Tutorial lists "A Hop Flower device (Fortnite device)" among the starter island assets | Shipped has **0 Hop Flower devices**; it uses native obstacle-course bounce drums (`FG01_Drum_S` at Y=3328 and `FG01_Drum_L` at Y=9099) with `bPlayBounce=true` |
| **Hurdling capability** | Tutorial states "Hurdle" is a movement type unavailable for Avatars | Shipped `Device_ExperienceSettings_FallGuys` explicitly sets `bAllowHurdleOverJumpableObjects: true` |
| **Building permissions** | Tutorial states "Avatars don't build" and "Player-Built Structures: Avatars don't build" | Shipped Island Settings device retains `allowBuilding: "All"` in its property table |

## Negative findings

- **No Finish Line Trigger or End Game Device:** Despite featuring a prominent Finish Line Arch (`FG01_Finish_Arch_*`) and victory podium, there is **no Trigger, Mutator Zone, Score Manager, or End Game Device** at the finish line. The template is purely a physical obstacle course sandbox; match end occurs solely via the 5-minute round timer (`roundTimeLimit: 5`).
- **No Checkpoint devices:** Shipped island has 0 checkpoints or Player Spawners along the course; falling in any pit eliminates the player and returns them to the initial starting pad at Y = -2000.
- **No Verse:** There are **zero** `.verse` files and zero Verse devices in the template. All motion and elimination logic is handled by standard Creative devices (`Device_PropMover_V2`, `BP_Device_DamageVolume_V2`).
- **No HUD Message or Billboard syllabus:** Unlike `obstacle_course_sample` (21 billboards) or `music_concert_sample` (25 billboards), `obstacle_course_starter` contains **0 instructional billboards** or HUD devices. The only actor with "Billboard" in its class name is `FG01_Finish_Arch_Billboard`, which is a static mesh decorative sign attached to the finish arch.

## Capture backlog

- None. All 7 devices, live transforms, property tables, and asset references were fully inspected and recorded via UnrealMCP.

## Live notes

- Default level is `/b9c10c03-85e6-4cab-b6cf-6b98605c434e/obstacle_course_starter`.
- `find_actors` requires `{"collision_channels": []}` in `SceneTools.find_actors` schema.
- All gameplay actors are unfoldered in the World Partition hierarchy (`folderPath="None"`).
- `LevelHasStreamingDisabled=1` confirmed via `AssetTools.get_asset_tags`.
