---
id: digest-012
kind: digest
---

# playset_sample — sample digest

> **Extractor provenance.** Offline evidence from `tools/extract-template.py` (reconstruction of
> the pipeline in `00-EXTRACTION-PROCESS.md`; the original `.sh` is absent from this repo).
> Live-editor facts read via UnrealMCP on `localhost:8000` with the project open;
> `SceneTools.get_current_level` returned `/playset_sample/playset_sample`.

| | |
|---|---|
| Classification | **Authoring scaffold** — a starting project for *building* a playset, not a playset |
| Recommendation | **KEEP** |

## Verdict first

**Keep it.** This is the structural inverse of `equip_furniture_props`, and the retirement
reasoning that applied there does not apply here.

| | equip_furniture_props | **playset_sample** |
|---|---|---|
| Authored assets | 2 | **130** |
| Placed actors | 14,116 | **18** |
| Custom Blueprints | 0 | **62** |
| Custom meshes | 0 | **63** |
| What it holds | references to Epic content | **its own content** |

The furniture catalog was retired because it authored nothing and could never keep pace with
Fortnite's asset churn. this sample authors **130 assets that exist nowhere else**, and its
subject matter — how to build a modular kit that plugs into Fortnite's building system — does
not go stale when Epic ships a new season.

## What this sample is *for*

It is a **fill-in-the-blank scaffold for authoring and shipping your own playset**. Three pieces
of evidence converge on that reading and nothing contradicts it:

**1. The map names are literal placeholders** (**Observed** — `summary.txt`):

```
Content\ENV_PACK_SetName_Gallery_NumAssets_Version.umap   (17,268 bytes)
Content\ENV_PACK_SetName_Prefab_NumAssets_Version.umap    (17,111 bytes)
Content\playset_sample.umap                    (55,474 bytes)
```

`SetName`, `NumAssets`, `Version` are **fields you are meant to replace**. `ENV_PACK` is Fortnite
Ecosystem Content — Epic's naming scheme for shipped content sets. The template hands you a
Gallery map and a Prefab map pre-named in the required convention.

**2. The main map is an empty build stage** (**Observed** — live editor, 25 actors, 0 outliner
folders). The entire authored kit is placed **nowhere**:

```
GridPlane1..4                      the build grid you work on
Player 1 Spawn Pad / Player 2       BP_Creative_Player_Spawner_Prop
IslandSettings0                     Device_ExperienceSettings_V2_UEFN
LevelBounds, WorldPartitionMiniMap, WorldDataLayers
```

Everything else is engine boilerplate (`FortWorldSettings`, `BuoyancyManager`, Mass visualizers,
nav data). **No kit geometry is placed in any map** — the 18 OFPA actors are grid, spawns and
world plumbing. You are given a clean grid and a parts bin. Do not copy modular-BrickAssembly-building-modules.md here — those are Atom-generated `SM_*` exhibits, not a `Parent_Building*` scaffold.

**3. There is dedicated thumbnail-capture rigging** (**Observed** — `actor-classes.txt`):

```
2  PlaysetIconSceneCamera     /Game/ArtTools/PlaysetIconSceneCamera
2  PlaysetIcon_Floor          /Game/Playgrounds/Items/IconRenders/PlaysetIcon_Floor
```

Plus two placeholder icon textures: `T-Icon-PS-CP_Devices_Example_WIP` and
`T_Icon_PS_CP_Devices_Example_WIP_SM` — note **`_WIP`** and **`Example`** in the shipped names.
Icon capture is part of *publishing* a playset, not playing one. this sample covers the
packaging step that most templates omit entirely.

## The kit itself

`SetupAssets/` is organised by function, with **Blueprints and Meshes paired 1:1 per category**
(**Observed** — `folders.txt`):

| Category | Blueprints | Meshes |
|---|---|---|
| Walls | 25 | 26 |
| Balconies | 8 | 8 |
| Stairs | 8 | 8 |
| Roofs | 5 | 5 |
| Props | 4 | 4 |
| Railings | 4 | 4 |
| Pillars | 3 | 3 |
| Trims | 3 | 3 |
| Floors | 2 | 2 |
| **Total** | **62** | **63** |

Plus one shared `TextureData/GrayBox` and two icon textures. Naming is strictly
`BP_GrayBox_<Category>_<Variant>` / `SM_GrayBox_<...>`, with suffix conventions for shape
(`_C` corner, `_D` diagonal, `_I` inner, `_O` outer, `_S` small, `_L` large, `_T`, `_W`,
`_Half`, `_Quarter`, `_Rail`). **This convention is itself part of the lesson** — it is the same
scheme Epic's own kits use, and the digest for
modular-kitbash-blueprints documents the pattern
(cross-template note — originated in character_environment_sample, extended by this sample).

## The Blueprints are not mesh wrappers — this is the real finding

The question that decides this sample's teaching value: does `BP_GrayBox_*` wrap a mesh, or
does it wire into Fortnite's building system? **It wires in.** (**Observed** — live editor,
`AssetTools.get_dependencies` + `ObjectTools.list_properties`.)

`BP_GrayBox_Solid_Wall` dependencies:

```
/Game/Building/ActorBlueprints/Wall/Parent_BuildingWall.Parent_BuildingWall_C   <- parent class
/Game/Building/EditModePatterns/Wall/EMP_Wall_Solid                             <- edit-mode pattern
/Game/Building/NavigationAreas/FortMetaNavWallAreaDef.FortMetaNavWallAreaDef_C  <- AI navigation
/Game/Balance/DataTables/ResourceRates                                          <- harvesting economy
/playset_sample/SetupAssets/Meshes/Walls/SM_GrayBox_Wall_01              <- the mesh
```

Four of the five dependencies are **Fortnite gameplay systems**. The mesh is the last item, not
the point. `Parent_BuildingWall_C` exposes **302 properties**, 101 of them building-related:
full door rigging (`bSwingingDoor`, `doorTargetRotationOffset`, `automaticSlidingDoorBoxExtent`,
`slamOpenDoorBoxExtent`, `doorSmartLinkComp`, `doorPropClass`), which is why the kit ships
`BP_GrayBox_Door_*`, `BP_GrayBox_DoubleDoor_Sliding_*` and `BP_GrayBox_Half_Wall_Door_*` as
distinct assets — each is a **different configuration of one inherited system**, not a different
mesh.

`BP_Graybox_Prop_Large` is deliberately simpler — only `ResourceRates` + its mesh — but its CDO
still exposes **273 properties**, 73 building-related:
`bIsBuildingVolume`, `resourceType`, `buildingAttachmentSlot`, `buildingAttachmentType`,
`buildingPlacementType` (trap/deco placement), `bAutoAssignNavProperties`, `bUseDamageSet`,
`bShowDamageCracks`, `savedHealthPct`, `currentBuildingLevel`, `onRepairBuildingStarted`.

**So the template teaches, by worked example:** pick the right `Parent_Building*` class per
element type (`BuildingWall`, `BuildingFloor`, `BuildingStairs`, `BuildingRoof`, `BuildingPillar`,
`BuildingProp`, `BuildingPropWall`, `BuildingCorner`, `BuildingPropCorner` — **all nine appear**
in `script-modules.txt`), attach an edit-mode pattern so players can edit the piece, set the nav
area so AI paths through it, and declare the resource type so it harvests correctly. That is
non-obvious, load-bearing knowledge and it is not written down anywhere else in this KB.

## No Verse, no Scene Graph — and that is fine here

**Observed:** 0 `.verse` files, 0 `__ExternalEntities__` directories, 0 data layers, 0 `.uref`
Fab dependencies. Only 5 reference mounts (`/Game` 286, `/playset_sample` 254,
`/Engine` 72, `/CRD_PlayerSpawn` 6, `/CreativeCoreDevices` 2) — compare the furniture catalog's
117. This is a **tight, self-contained project**.

Unlike `intro_to_uefnfk1`, the absence of Verse is not a staleness signal. A graybox construction
kit has no runtime logic to express; the content *is* the deliverable. Nothing here needs Verse
to do its job.

## Project configuration (⚠ flags Unverified)

> The project was open in UEFN during extraction and the operator has not confirmed whether
> Project Settings were touched. `experimental.*` flags are writable from that UI, are off by
> default, and UEFN rewrites the descriptor on disk the moment one is flipped. This exact
> contamination invalidated a headline finding in
> intro_to_uefnfk1.md.
> Guard: 00-EXTRACTION-PROCESS.md Step 2.
> **The KEEP verdict does not depend on these flags.**

```
compatibilityVersion               42.10           (Observed)
requiredRedirectorStartingVersion  42.10.57566230  (Observed)
apiAccess                          Private         (Observed)
editorPermissions                  VKCreateUGCPartner   (Observed — note: partner-authoring permission)
matchmaking  maxPlayers 16 / maxTeamCount 1 / maxTeamSize 16   (Observed)
VersePath  /epictemplates@fortnite.com/playset_sample   (Observed)
VerseVersion 1 / VerseAPI 1                        (Observed)
ValkyrieSettings.libraryFolderMode  "Absent"       (Observed)

⚠ experimental.sceneGraph.bIsSceneGraphSystemAllowed  true   <- Unverified
⚠ experimental.sceneGraph.bEnableOneFilePerEntity     false  <- Unverified
⚠ .uplugin EnableSceneGraph                           true   <- Unverified
⚠ .uplugin EnableVerseAssetReflection                 true   <- Unverified
```

Two details *are* safe to read and both support the scaffold reading: **`maxTeamCount: 1`**
(no competitive setup — you are building, not playing) and **`editorPermissions:
VKCreateUGCPartner`**, a partner content-authoring permission.

## Pre-publish cleanup

- **`Developers/felixkupis/` ships again, empty** — plus a nested empty `Collections/`.
  **Four for four** across first-party templates now (character_environment `jojitsuruga`, intro_to_uefnfk1
  and equip_furniture_props `felixkupis`, this one `felixkupis`). At four occurrences this is a
  missing pipeline step, not carelessness.
- **`_WIP` and `Example` in shipped texture names** (`T-Icon-PS-CP_Devices_Example_WIP`) —
  arguably intentional as placeholders you replace, but worth a deliberate call.
- **`SM_GrayBox_Door_01` / `_1` / `_2` / `_L` / `_R` / `_S`** — six door meshes with an
  inconsistent numbering scheme (`_01` vs `_1`) against an otherwise strict convention. Minor,
  but it is the one place the naming discipline slips.

## Assessment

| Criterion | Status |
|---|---|
| Digest written | **Met** |
| Verse preserved | **N/A** — zero Verse |
| Unique content captured | **At risk if deleted** — 130 authored assets exist only here |
| Capture backlog | **1 open** (below) |
| Extractor provenance | **Qualified** — reconstructed extractor |

### Capture backlog

1. **What is actually inside the two `ENV_PACK_SetName_*` maps?** Each is ~17 KB with OFPA actors
   under `__ExternalActors__/ENV_PACK_SetName_Gallery_*` and `__ExternalActors__/ENV_PACK_SetName_Prefab_*`.
   Whether they are pre-wired gallery/prefab containers or empty shells changes how much of the
   packaging workflow the template hands you. Requires opening each map in the editor —
   `find_actors` only reports the currently-loaded level. Cheap next session; does not affect
   the KEEP verdict.

### Recommendation

**Keep.** It is small (148 `.uasset`, 5 mounts, no Fab dependencies), self-contained, and it is
the only artifact in this KB that documents two genuinely non-obvious workflows: **wiring custom
geometry into Fortnite's building system** (parent class + edit-mode pattern + nav area +
resource type) and **packaging a playset for publication** (gallery/prefab map naming, icon
capture rig).

Its 130 authored assets are irreplaceable — unlike the furniture catalog, deleting this loses
real content. If anything, the KB should **promote** what it teaches into
character_environment_sample/techniques/modular-kitbash-blueprints.md
— it lives there because character_environment originated the note — since
the `Parent_Building*` integration pattern is reusable knowledge that outlives the template.
