---
id: digest-013
kind: digest
---

# greyboxing101_template — sample digest

> Offline `_raw/greyboxing101_template/`. Live
> `/4e5b0414-c430-4b3a-bdb5-dfa52f799e68/greyboxing101_template` (**Observed**).
> Docs: Greyboxing landing (short). Spreadsheet Keep=Yes. **No Verse.**

| | |
|---|---|
| Classification | **Walkable greybox curriculum** — CubeGrid + Shape tools + color-coded playtest kit |
| Content type | kit: **79 authored vs 190 placed** |

## Scale

| | |
|---|---|
| Authored | 79 (CubeGrid meshes+MIs, ShapeTools primitives, 7 `LS_*` sequencers, greybox grids) |
| OFPA | 190 |
| Verse | **0** |
| `FortStaticMeshActor` | 113 |
| `BP_DemoDisplay` | 41 |
| Teleporters | 13 |
| Cinematic Sequence | **7** — CubeGrid tool demos |
| Demo rooms | 5 |
| Compatibility | **42.30** |
| Matchmaking | 16 |

Empty `Developers/felixkupis/Collections/`. Nested `Content/Content/{CubeGrid,GreyboxExample,ShapeTools}`. Twenty-sixth first-party hit. **Wave 2 complete.**

## What it is *for*

Four numbered halls (Modeling/Materials convention):

| Hall (live Outliner) | Teaches |
|---|---|
| `1 - Greyboxing 101` | usage / testing / considerations / tools — **typo ships:** `1.0 - Greboxing` |
| `2 - Greybox example` | finished blockout with color grids (Building / Danger / Goal / Navable) |
| `3 - CubeGrid` | Selection → Movement → Pull → Push → Move_Rotate → GridSize → Boolean → Corner Mode |
| `4 - Shapetools` | Arrow / Box / Cone / Cylinder / Disc / Rectangle / Sphere / Stairs / Torus |

CubeGrid exhibits are **played as Level Sequences** (`LS_Boolean_4_0`, `LS_Pulling_2_2`, `LS_Push_2_3`, plus `_Rev`) via 7 `Device_CinematicSequence` + 3 switches — not Verse.

## Device inventory (the design)

| Count | Class | Role |
|---|---|---|
| 113 | `FortStaticMeshActor` | cubes / shapes |
| 41 | `BP_DemoDisplay` | captions |
| 13 | teleporter | hall hop |
| 7 | `Device_CinematicSequence` | CubeGrid tool GIFs in-world |
| 5 | `BP_DemoRoom` | halls + lobby |
| 3 | `Device_Switch_V2` | start sequences |
| 1 | button | |
| rest | HUD / experience / volumes | boilerplate |

Authored kit: `CubeGrid/` (grid MIs + hashed CubeGridToolOutput meshes), `GreyboxExample/` (`MI_Greybox_Grid_{Building,Danger,Goal,Navable}`), `ShapeTools/` primitives, `LS_*` sequencers.

## What this sample actually teaches

1. **Greybox before art.** Scale + layout first; color grids (Goal/Danger/Navable/Building) so playtesters read intent. See techniques/cubegrid-then-color-grids.md.
2. **CubeGrid is the authoring tool** (select/pull/push/boolean/corner); Shape tools are the primitive kit. Merge after layout (docs Create → Merge).
3. **Demos are Sequencer, not Verse.** Same family as Animation 101’s cinematic, but here the shot *is* the lesson.
4. **Nested `Content/Content/`** same as Modeling — don’t “fix” the path when citing.

## Official tutorial vs shipped

**External:** Greyboxing

Single short page (typo in title: “UNreal”). Sections: Using Greyboxing, Shape Tools, Create Tools (Merge), Attributes (Edit Materials / color), Playtesting. Mentions a **Marketplace greybox kit** — not this island.

`docsUrl` empty. Spreadsheet `/en-us/uefn/greyboxing-in-unreal-editor-for-fortnite`. Retrieved signed-in Chrome 2026-09-11.

### Lines up

| Tutorial | Shipped |
|---|---|
| Shape tools → place/scale primitives | hall 4 exhibits |
| Merge under Create | CubeGrid combined exhibit `3.3 -Combined` |
| Color blocks to differentiate | `MI_Greybox_Grid_{Building,Danger,Goal,Navable}` |
| Playtest scale / vehicle fit / reach | hall `1.1.1 - Testing` |
| Modeling Mode parent crumb | under Modeling docs |

### Does not line up

| Topic | Tutorial | Shipped |
|---|---|---|
| Structure | 5 short headings | **4 numbered halls** (the real syllabus) |
| CubeGrid | **not named** | entire hall 3 + 7 sequences |
| Marketplace kit | pointed at | **this island is the kit** |
| Sequencer demos | — | 7 cinematic devices |
| Folder typo | — | `1.0 - Greboxing` |

Train on halls + color-grid MIs + CubeGrid sequences. Do not emit “buy the Marketplace kit” as this sample. Do not invent Verse.

## Negative findings

- Scratch `Developers/` folder.
- Compatibility 42.30.
- CubeGridToolOutput hashed mesh names — editor scratch, not a naming lesson.

## Capture backlog

1. Which `LS_*` sequence is wired to which CubeGrid exhibit. *Open, low — names already map (Boolean/Pull/Push).*
2. Demo Display captions. *Waive — Outliner is the syllabus.*

## Live notes

Level confirmed. Halls 1–4 + Lobby. Streaming off (not re-verified).
