---
id: digest-014
kind: digest
---

# gunfight_template — sample digest

> Offline `_raw/gunfight_template/`. Live
> `/89c85be5-fbb3-43db-b6ee-b157ea24186c/gunfight_template` (**Observed**).
> **Official docs: none found** as a dedicated page. Spreadsheet Docs=No. Starter-templates hub has a 4-bullet blurb only.

| | |
|---|---|
| Classification | **2v2 arena rotation** — persistable island pick + sequential weapon grant |
| Content type | catalog wrap: **10 authored vs 4043 placed** |

## Scale

| | |
|---|---|
| Authored | 10 (materials + 5 Verse) |
| OFPA | **4043** |
| Maps | **2** (`gunfight_template.umap` playable, `gunfight_template_template.umap` blank essentials) |
| Verse | **5** (`Scripts/VerseDevices/` + `Helpers/`) |
| VerseDevices | 6 |
| Teleporters | 64 |
| Item granters | 44 |
| Round settings | 22 |
| Mutator volumes | 20 |
| Matchmaking | **4** players, **2** teams of 4 (play is 2v2; balancer target 2) |
| Compatibility | 43.00 |

Spreadsheet: *Requires a fix to run with only 1 player and spawn an enemy automatically in UEFN for testing. Probably a good candidate for future FAB integration of the verse code.*

Empty `Developers/felixkupis/Collections/`. Eighteenth first-party hit.

## What it is *for*

Many tiny FPS maps in one island. Each round: pick an arena (teleporter set), optionally invert sides, grant the same loadout to everyone (new pool on odd rounds, previous on even), 2v2. Persist last island + exclusion list and last loadout across rounds via `weak_map(player, …)`. Late join: rebalance teams, grant current weapons.

Blank `*_template.umap` = devices + Verse without the 4k prop dump.

## Device inventory (the design)

| Count | Class | Role |
|---|---|---|
| 64 | teleporter | one arena = N pads + inverse (side swap) |
| 44 | item granter | weapon pools |
| 22 | round settings | which rounds pick island / grant new / load previous |
| 20 | mutator zone | bulk teleport (`TeleportZone.Enable()`) |
| 20 | player spawners | 4P |
| 12 | triggers | grant / island-complete |
| 6 | VerseDevice | selector, balancer, granter |
| 2 | elimination manager | round scoring |
| rest | blockout/furniture | throwaway |

## Verse (5 files)

Copied to `snippets/verse/`.

| File | Job |
|---|---|
| `island_selector_device.verse` | persistable island index + exclusion; invert teleporters; mutator-zone dump |
| `weapon_granting_device.verse` | pools → sequential `GrantItem` + `ItemGrantedEvent.Await`; late join |
| `team_balancer_device.verse` | `TargetTeamSize=2`; `AddToTeam` overflow |
| `array_helper.verse` | `FindMostFrequentIndex` (consensus across players) |
| `log_helper.verse` | `PrintDebug` gated by `EnableLogs` |

Consensus pattern: persist maps can disagree (JIP). Vote the mode of `LastIslandIndex` / first weapon index, treat that player as source of truth.

## What this sample actually teaches

1. **Many arenas, one island:** `[]island_data` of teleporter arrays + `InverseIslandsTeleporters`. Mutator zone Enable = “go there.” See techniques/persistable-arena-rotation.md.
2. **`class<persistable>` + `weak_map(player, …)`** for last island and exclusion list; constructor `MakeIslandData` copies previous. Speedway taught persistable stats — this is persistable *routing*.
3. **Consensus `FindMostFrequentIndex`** when per-player persist disagrees.
4. **Sequential grant:** wait `ItemGrantedEvent` between granters so slot order is gun then nade, not a race. Cache `DelayedGranters` for JIP. See techniques/sequential-weapon-grant.md.
5. **Odd/even rounds via two Round Settings arrays** (`RoundsToGrant` vs `LoadPreviousOnRounds`) — not a Verse counter.
6. **2v2 late-join balancer** — overflow `AddToTeam` the other side.

## Official tutorial vs shipped

**Official docs: none found.** Hub blurb only: UEFN Starter Templates § Gunfight

| Blurb | Shipped |
|---|---|
| 2v2 FPS, many small levels | 4P / 2 teams; 64 teles |
| Blank map + full map | two `.umap`s |
| Outliner naming | **Unverified** (live folders not dumped) |
| Join / JIP / Rejoin Verse | `PlayerAddedEvent`, persist maps, `UseLocalForLateJoiners` |

`docsUrl` empty. Do not invent a 10-chapter Gunfight tutorial.

## Negative findings

- 4043 OFPA is not a kit.
- 1P PIE does not spawn a dummy enemy (spreadsheet). Packet doesn’t fix that.
- Scratch `Developers/` folder.

## Capture backlog

1. Live round-settings wiring (which rounds grant vs load). *Waive — arrays Observed in Verse.*
2. Live `island_data` count. *Open, low.*

## Live notes

Level confirmed (`gunfight_template`). Streaming off.
