---
id: digest-015
kind: digest
---

# night_environment_sample — sample digest

> **Extractor provenance.** `tools/extract-template.py` (Pass A) + UnrealMCP on
> `localhost:8000` (Pass B); `SceneTools.get_current_level` returned
> `/d30e5a52-4492-48c2-ab77-b484bb4c8b62/NightEnvironment_StarterIsland`.
> Picker-only template (Game Examples), **not on the spreadsheet** — no Keep column.

| | |
|---|---|
| Classification | **Atmosphere layer over a mounted kit** — spooky fishing village at night |
| Verse | **0 files** |
| Fab `.uref` | **0** |

## Verdict first

Environment starter island. The transferable content is **not** the village — the village
buildings (`ENV_PACK_BP_NightEnvironment_*`) are placed from external `/NightEnvironment_Assets` +
`/NightEnvironment_Comp` mounts (a pre-built ENV_PACK modular house kit this project does not
author). What the template *does* author — and what survives deletion — is the **night
atmosphere stack**: custom skybox, fog cards, 19 ambient audio loops, landscape layers,
foliage types. That stack is the packet's teaching value (see `techniques/`).

## Scale (**Observed** — `summary.txt`)

| | |
|---|---|
| Authored assets | **49** — sky/fog/landscape/foliage only (see below) |
| Placed actors | **2,273** OFPA packages, 466 distinct classes |
| Verse files | **0** |
| Data layers | **0** (confirmed live, `get_data_layers` → `[]`) |
| Feature flags | SceneGraph allowed (unused), Python, toolsets; `editorPermissions` empty (public) |
| Compatibility | 42.30 |

## The two content tiers (**Observed** — `reference-mounts.txt`)

**Tier 1 — authored here (49 assets), the atmosphere layer:**

```
Content/
  Materials/   M_ProceduralSkybox, M_LightClouds, M/MI_Harrowvile_Landscape_A,
               MI_Harrowvile_Ocean, MF_{ParticleRandomiser, PlanetImage, SphericalUVsCap}
  Textures/    T_Moon, T_MoonBlur, T_Stars_A/B, T_SkyNebula_A, T_Planet_A_N, T_Noise_CloudsB
  Meshes/      SM_SkySphere                       (the only authored mesh)
  VFX/Fog/     NS_FogCards, M_FogCard(+_Masked), T_FogCard, T_Smoke_Puff_Tile_Up
  Foliage/     11 FoliageTypes (ENV_PACK_Grass_Helios_* ×8, TREE_PACK grass/leaves ×3)
  NightEnvironment_StarterIsland_sharedassets/  Layer_1..6_LayerInfo (landscape paint layers)
```

**Tier 2 — placed from mounts (2,273 actors):** `/NightEnvironment_Assets` (1,231 refs) +
`/NightEnvironment_Comp` (1,228) — the ENV_PACK_BP_NightEnvironment_* house kit (RoofWall / Pillar_Corner /
Roof_Trim / Doorway / Window / Porch …); then stock: `/Game` (3,606), `/Hera` (722 cliffs/rocks),
`/TREE_PACK` (331 dead trees), `/ENV_PACK_*`, `/AmbientProps`, `/RRoost_*`, etc. **78 distinct mounts
total** — the widest mount spread in the corpus.

## Level organisation (**Observed** — live `get_folders`)

```
Audio/AmbientLoop          19 Device_CRD_AudioPlayer
Environment/               Background, Cliffs&Rocks, FisherHouse, Graveyard,
                           HauntedHouse, Overlook, SetDressing, ToolShed, Trees
LevelData/                 Landscape, WaterZone, VK_Waterbody_Ocean, WaterBrushManager,
                           Item Spawner, Player Spawner, IslandSettings, minimap, LevelBounds
Lighting/                  DirectionalLight, SkyLight, SkyAtmosphere, ExponentialHeightFog,
                           GlobalPP + InteriorPP volumes, SkySphere, house spot/point lights
VFX/                       NS FogCards ×~20, Biome Steam, Seagull orbits, bat swarm
```

POI-named environment folders (FisherHouse / Graveyard / HauntedHouse / Overlook / ToolShed) —
composition is organized by **place**, not by asset type. Contrast: intro_to_uefnfk1 organizes
by lesson number; NightEnvironment by location.

## Device inventory (tiny — this is an environment, not a game)

19 Audio Players, 1 Item Spawner (V3), 1 Player Spawner prop, 1 Island Settings
(ExperienceSettings V2), 1 inspector camera. **No Verse devices, no triggers, no timers.**

## What this sample actually teaches

1. **Night-atmosphere stack** (`techniques/night-atmosphere-stack.md`) — custom procedural
   skybox (authored mesh + material + moon/star textures), Niagara fog cards, ambient audio
   loop grid. This is the packet's core.
2. **A village is a mount, not a project** — the house kit ships as `/NightEnvironment_Comp`
   components and is only *placed* here. Same lesson as equip_furniture (project ≠ owner of
   its content), but NightEnvironment answers "what DOES belong in the project": the atmosphere.

## Official tutorial vs shipped

**Official docs: none found** (picker-only template, not on spreadsheet; `.uefnproject`
`docsUrl` empty). Guessed slug
`…/uefn/NightEnvironment-environment-template-in-unreal-editor-for-fortnite` returns an empty
hub page. Only public mention: **42.10 ecosystem release notes** ("the frightening new
template NightEnvironment: Environment") (**External** — DuckDuckGo result).

## Negative findings

- **No Verse, no data layers, no Fab packs, no Sequencer.** SceneGraph flag on, unused.
- The village architecture is **not** authored here — `ENV_PACK_BP_NightEnvironment_*` mounts in.
  Don't cite this project for modular-kit authoring (that's playset_sample's lesson).
- Streaming: `LevelIsPartitioned=1` and **no `LevelHasStreamingDisabled` tag** (**Observed** —
  `get_asset_tags` on the umap) → this map **streams**, unlike character_environment/intro/lighting/
  scene_graph which all disable runtime streaming. At 2,273 actors it has to.
- `PlaysetRoot` present (1 instance) — the village was likely placed as a playset;
  the mount split above is why.

## Capture backlog

1. Exact fog-card Niagara emitter settings (spawn rate, card size) — reachable via
   NiagaraToolset; waived (asset `NS_FogCards` is authored + committed, values inspectable).
2. InteriorPP vs GlobalPP property delta — `ActorTools.diff` candidate; waived (two PPVs,
   pattern already documented in digest).
3. Landscape layer → LayerInfo mapping — 6 LayerInfos committed; waived.

## Live notes

- Level: `/d30e5a52-4492-48c2-ab77-b484bb4c8b62/NightEnvironment_StarterIsland`.
- `find_actors` by name does not see spatially-unloaded WP actors (ItemSpawner found via
  `get_actors_in_folder "LevelData"`, not by name search).
- Audio player options read via `ObjectTools.get_properties` on
  `<actor>.ToyOptionsComponent` → `playerOptionData` (propertyOverrides array of
  name/data pairs).
