---
id: digest-016
kind: digest
---

# ice_base_sample — sample digest

> **Extractor provenance.** `tools/extract-template.py` (Pass A) + UnrealMCP on
> `localhost:8000` (Pass B). Default level
> `/37632797-426b-47be-b7f0-69a145343211/ice_base_sample` (**Observed**).
> Docs: Working With space-opera™ Islands
> (hub; IceBase is one environment-starter bullet). `docsUrl` empty. Guessed
> `working-with-star-wars-islands-in-unreal-editor-for-fortnite` **hub-redirects**.
> Compat **43.00**. `featureSet`: **CustomGameplayFeature**.

| | |
|---|---|
| Classification | **theme environment starter** — IceBase ice-base sandbox, not EnergyBlades/Narrative/Team Conquest |
| Content type | **catalog wrap** (35 authored vs 3018 OFPA; `/SPACE_THEME_E_*` ice-base + `/ENV_PACK_Biome_*` snow) |
| Verse | **0 files** |
| Fab `.uref` | **0** |

## What it is

The picker name is **IceBase Inspired Starter Island**. The island is a **16-player snow base**: hangars, ice tunnels, yard NPCs, ridable Mounts, four turret turrets. Teaching that survives deletion is the **Guard Character Definition + patrol path + doorway volume swap**. EnergyBlades/Telekinetic, Conversation, Capture Area, and Hero Device are other space-opera tiles. Custom skyboxes stay on NightEnvironment-night-atmosphere. Day Sequence volumes stay on day-sequence-priority-stack.md. ExtraAssets HUD stays on Social/Walker — this HUD is **in-project**, like turtle_arcade_sample.

## Scale (**Observed** — `summary.txt` + live)

| | |
|---|---|
| Authored assets | **35** |
| OFPA | **3018** |
| Maps | `ice_base_sample.umap` (only) |
| Verse files | **0** (`VerseToolset.ListFiles` on `/37632797-…` — do not recurse `__ExternalActors__`) |
| Verse devices placed | **0** |
| Data layers | **0** |
| Fab `.uref` | **0** |
| Compatibility | **43.00** |
| Matchmaking | **16** (`maxPlayers=16`, `teamSize=Dynamic`, `teams=FreeForAll`, `locomotionPreset=Current BR`) |
| Streaming | Partitioned + actor folders + OFPA. **`LevelHasStreamingDisabled` omitted** (Social Deduction / Arcade / City / Mastery camp). Sampled `bIsSpatiallyLoaded=false`; `load_actors` not required. `find_actors` returns **0** — use `get_actors_in_folder` + `actorPath`. |

Empty `Developers/felixkupis/Collections/`. First-party scratch folder.

`experimental.*` (Python / toolsets / Scene Graph) in `.uefnproject` is **Unverified**. Content-side: **no** `.verse`, **no** `__ExternalEntities__`.

Largest OFPA ~**68.2 MB**. Whole folder ~**917 MB**. Under the GHES **100 MB** per-file cap.

## Level organisation (**Observed** — live `get_folders` + folder counts)

```
Devices/                              1 — HUD Controller
Environment/                       2712
  IceBase/                         1476 — Hangar A 346 / Hangar B 364 / tunnels 516 / 250 direct
  IceCave/                          223
  Snow/                             552
  Rocks & Cliffs/                   239
  Landscape/                         74 (Foliage/ 9)
  ScatterProps/                     101
  Vista/                             19
  WalkerVehicles/                            15
  Background/                        13 — authored SM_landscape_01–12
LevelData/                            8 — IslandSettings0, 1 player pad (useAsIslandStart), bounds, minimap
Lighting/                             3 — DSD + IceBasePostProcess + IceCavePostProcess
NPCs/                               134 — 48 spawners, 82 patrol, 4 volumes
Turrets/                              4 — turret CopperPuck
VFX/                                 25 — Niagara snow
HLOD/                               131
Hangar_A_SetDressing/ / IceBase/ / IceBase_Architecture/   0 leftover roots
```

No `Device_Billboard_V2`. No syllabus boards.

## Device inventory (the design)

| Count | Class | Role |
|---|---|---|
| 82 | `AIPatrolPath_Device` | BackAndForth groups (sample 6 / storm 20) |
| 48 | `Device_CharacterSpawner` | rebels / Guards / Mounts |
| 4 | `Device_Volume` | doorway Enable/Disable swap |
| 4 | `CRD_SpawnerDevice_CopperPuck` | turret turrets, health **800**, phase Always |
| 1 | `Device_HUDController_V2` | in-project `UW_HUD_*` overrides |
| 1 | `BP_DaySequenceModifier_Device` | dusk 18:00, priority **2000**, `use Volume=false` |
| 2 | `PostProcessVolume` | IceBase / IceCave, bound, sampled overrides **false** |
| 1 | `BP_Creative_Player_Spawner_Prop` | island start |
| 1 | `PlaysetRoot` | `WalkerVehicle4` |
| 3 | `FortSkeletalMeshActor` | WalkerVehicle copies, `SKM_WALKER_VEHICLE`, `animToPlay` **None** |
| 25 | `NiagaraActor` | snow VFX |
| 0 | Billboard / Verse device / Hero / EnergyBlade / Conversation | other space-opera tiles |

## Authored kit (**Observed** — `assets.txt` + live)

```
Content/NPCs/CD_RebelNPC_GuardBehavior
Content/UI/PlayerHUD/   UW_HUD_PlayerInfo, Equipped, Quickbar, PlayerVitals, TeamVitals
Content/Landscape/      ENV_PACK_MI_Biome_Snow_Landscape_A_Inst + SM_landscape_01–12 + distant meshes
Content/IceBase_Starter_sharedassets/  Layer_1–6_LayerInfo  (0 asset-registry refs; Landscape.layers[] empty)
Content/Foliage/        SM_TREE_PACK_* FoliageTypes (0 refs)
Content/WALKER_VEHICLE_C_Take1   LevelSequence, 0 referencers
```

**Tier 2 mounts:** `/SPACE_THEME_E_Assets` 2802 + `/SPACE_THEME_E_Comp` 1337 (ice-base gallery), `/ENV_PACK_Biome_Assets` 2203 + `/ENV_PACK_Biome_Comp` 752 (snow), `/IceBase_Starter` 1765, `/THEME_PACK` 1141 (FlightCrew dressing), `/ENV_SPACE_THEME_UI_Assets` 40 (HUD materials), `/FNEPLATFORM_PACK` 36 (Mount), `/ENV_SWNiceCrown` 4 (WalkerVehicle), `/ENV_SWVehicleClasses` 8 + `/ENV_SWVehicleSpawners` 4 (turret). No Fab `.uref`.

## What this sample actually teaches

1. **Guard NPCs + patrol + doorway swap, no Verse.** Type=`Guard`, not Walker. techniques/guard-patrol-volume-swap.md. Walker contrast: walker-character-definition.md.
2. **The ice base is a mounted `/SPACE_THEME_E_*` + `/ENV_PACK_Biome_*` catalog, not 35 authored meshes.** Same catalog-vs-kit split as prison_sample (`/ENV_Meadow_*`) and dimension_x_template (`/SCI_FI_*`). One playset root (`WalkerVehicle4`); hangars are galleries, not `PlaysetRoot` rooms.
3. **HUD is authored UMG in `Content/UI/PlayerHUD`, not ExtraAssets.** Live overrides: PlayerInfo, Equipped, Quickbar. HideElement minimap + build menu; health/shield **DoNotOverride**. Materials `/ENV_SPACE_THEME_UI_Assets`. Contrast: the Custom UI chapter’s cantina / The Warren / `UI/Widgets` example lives on narrative_roleplaying_template (HUD in Outliner `UI/`, Equipped+Quickbar **null**). ExtraAssets HUD coincidence stays **three** (Social Social, Mastery, Walker). Deeper UMG: umg-viewmodel-device-binding.md.
4. **Dusk is one Day Sequence at 2000, not a volume stack.** FixedTime **18:00**, sun **10**, fog **0.02** `overrideFog`, stock `MI_DaySequenceSkydome`. **Not** a fourth 2000-vs-higher source (day-sequence-priority-stack.md). **Not** NightEnvironment’s custom `SM_SkySphere`. Bound PPVs do not sample-override. Environment lighting rig: environment-lighting-rig.md — this island does not ship that Blueprint.
5. **Landscape layers are leftover LayerInfos, not a weight-blend demo.** Live `layers[]` **empty**; snow MI parent `/ENV_PACK_Biome_Assets/…`. Distant meshes `SM_landscape_01–12` in Background. Do not train as landscape-weight-blend-layers.md.
6. **Sandbox flags, not a round.** `allowBuilding=All`, `bInfiniteResources=true`, `bFallDamage=false`, `timeOfDayOverride=NoOverride` (DSD owns 18:00). One spawn pad, `spawnLocation=SpawnPads`.
7. **WalkerVehicle is a T-pose mesh + unused sequence, not the Control Rig lesson.** Docs Vistas list “Control Rig WalkerVehicle”; live `animToPlay=None`, `WALKER_VEHICLE_C_Take1` **0** refs. Cinematic Vehicles chapter is theme reference. `Environment/WalkerVehicles/` also ships leftover `ENV_PACK_BP_StrongArm_FryingPan_*` pieces.

## Official tutorial vs shipped

**External hub + Vistas + Custom UI + Cinematic Vehicles** (headed Chrome 2026-09-15). Verbatim: `_raw/ice_base_sample/tutorial.txt`. Spreadsheet `docsUrl` empty. No dedicated IceBase chapter.

### Lines up

| Tutorial | Shipped |
|---|---|
| theme Templates → space-opera → IceBase Inspired Starter Island | picker; `featureSet` CustomGameplayFeature |
| “Snow covered terrain and icy tunnels” | `Environment/Snow` 552, `IceCave` 223, `IceBase` 1476 |
| IceBase Base prefab/gallery; ice caverns vistas; snow props | `/SPACE_THEME_E_*`, `/ENV_PACK_Biome_*`, `Vista/` 19 |
| Roaming NPCs + NPC Spawners + AI Patrol Path | 48 spawners, 82 patrol devices |
| Ridable Mount; predefined CD in Content > AI | `/FNEPLATFORM_PACK/…/Mount_CharacterDefinition`; Taming modifier |
| Snow VFX | 25 Niagara |
| turret turret among space-opera turrets | 4 CopperPuck, health 800 |
| Custom UI widgets `UW_HUD_Equipped` / `PlayerInfo` / `Quickbar` | same names in `Content/UI/PlayerHUD`; HUD Controller overrides |
| Hide minimap | `showMinimap=HideElement` |

### Does not line up

| Tutorial | Shipped |
|---|---|
| Control Rig WalkerVehicle | Skeletal mesh T-pose; sequence **0** refs; `animToPlay` None |
| Landscape snow material **6 paint layers** | Authored LayerInfo×6; live `Landscape.layers[]` **[]** |
| Custom HUD “preconfigured on Flat Grid / XL Flat Grid” | **This** island also ships a HUD Controller (Devices/, not UI/) |
| Editable widgets under `UI > Widgets > PlayerHUD` | Shipped `Content/UI/PlayerHUD` (no `Widgets/` folder). Widgets-folder layout is Narrative |
| Player Info override = `UW_HUD_Vitals` | Live override `UW_HUD_PlayerInfo` (same name drift as Narrative) |
| HUD Controller in Outliner **UI** folder; cantina / The Warren; no building; nothing to equip | That chapter’s example is Narrative. This island: folder **Devices**; `allowBuilding=All`; Equipped + Quickbar overrides on |
| Dedicated IceBase walkthrough | Hub bullet + Vistas list; guessed UEFN-suffix slug hub-redirects |
| Vehicle Spawner dedicated page | Guessed slug hub-redirects; turret still on hub + Cinematic Vehicles list |

Training rule: live Guard CD + volume bindings + in-project HUD paths win. Do not emit “IceBase ships a Control Rig WalkerVehicle / 6 painted snow layers / HUD only on Grid islands.”

## Negative findings

- **Name implies a IceBase *game*; files show a kitbash sandbox.** No round manager, no Verse, no EnergyBlades, no Hero device.
- **`hUDInfoType=AI Enemy Elimination` with Neutral unkillable guards** (`isVulnerable=false`). Leftover default, same family as Prison/Starter.
- **Root `IceBase/` / `IceBase_Architecture/` / `Hangar_A_SetDressing/` are empty leftovers** after the nested move under `Environment/`.
- **Volume2 does not Enable interiors.** Second doorway only swaps exterior+Mount spawners.
- **`get_referencers` on HUD widgets / Guard CD returns []** despite live overrides — confirm on the placed HUD Controller / spawners, same trap as Prison landscape MI.
- **`GetDeviceProperties` N/A** — no Verse `creative_device`s. CopperPuck `vehicleClass` unread via ObjectTools (health/phase readable).
- **FryingPan StrongArm pieces in `WalkerVehicles/`.** Leftover kit, not an exhibit.

## Capture backlog

| Item | Status |
|---|---|
| Sequencer keys on `WALKER_VEHICLE_C_Take1` | **Waived** — 0 referencers; not the lesson |
| Per-hangar child lists (1476 IceBase) | **Waived** — folder counts + `/SPACE_THEME_E_*` mounts are the lesson |
| LayerInfo ↔ material slot wiring | **Waived** — empty `layers[]`; don’t train as weight-blend |
| PPV blendable graphs | **Waived** — sampled `bOverride_*=false` |
| PIE volume swap / Mount taming | **Waived** — bindings + CD modifiers sufficient |
| Vehicle Spawner docs page | **Waived** — slug hub-redirects; turret Observed on the turret device |

## Live notes

- Mount `/37632797-426b-47be-b7f0-69a145343211`. Project Verse path `/epictemplates@fortnite.com/ice_base_sample`. Plugin filename is that GUID.
- Did **not** `load_level` away from the default map.
- Billboard `text` N/A. `GetDeviceProperties` N/A.
- `find_actors` by name/class returns **[]** while folders are full — World Partition streaming-on + use `get_actors_in_folder`.
- Hangar folder names use a **space**: `Environment/IceBase/Hangar A` (underscore paths 404).
- Character Definition modifiers: `load_asset` then `get_properties` on `:CharacterModifier_*_C_0` suffixes, not `.modifiers[0]`.
- Folder on disk / title is **`ice_base_sample`** (picker “IceBase Inspired Starter Island”).
