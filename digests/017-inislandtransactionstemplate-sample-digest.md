---
id: digest-017
kind: digest
---

# in_island_transactions_template — sample digest

> Offline `_raw/in_island_transactions_template/`. Live
> `/0629ad14-0459-47e2-bb96-02ffe20f5f5e/in_island_transactions_templateTools` (**Observed** — umap name has `Tools` suffix).
> Docs: hub + 6 chapters. Device-template chapter is a **potion/sword starter**, not this farm.

| | |
|---|---|
| Classification | **Marketplace kit** — Verse-defined entitlements/offers + booth examples |
| Content type | kit (155 authored, 190 placed, **21 Verse**, 20 VerseDevices) |

## Scale

| | |
|---|---|
| Authored | 155 (`Art/`, widgets, meshes) |
| OFPA | 190 |
| Verse | **21** |
| VerseDevices | **20** |
| Buttons | 21 |
| Compatibility | 43.00 |
| Custom inventory / Aurora / SG experimental | true / true / true |
| Auto-localization | 18 cultures |

Empty `Developers/felixkupis/Collections/`. Fourteenth first-party hit.

## What it is *for*

Garden economy in V-Bucks: consumable seeds/packs, durable shovel/can, cosmetic gnome (`ConsequentialToGameplay=false`), VIP door (`PaidArea=true`), Flowberry capacity. Offers (50–5000, multiples of 50), stacked + nested bundles. Buttons call `BuyOffer` / `GrantEntitlement` / `ConsumeEntitlement`. Scene Graph items (`item_interactable_component`) for plant/water.

## Device inventory

| Count | Class | Role |
|---|---|---|
| 20 | `VerseDevice` | one booth each |
| 21 | `Device_Button_V2` | buy/grant/consume |
| 16 | teleporters | hall nav |
| 35 | `BP_DemoDisplay` | captions |
| 2 | item granter/remover | inventory demos |
| 2 | popup dialog | offer UI |

## Verse (21 files)

Copied to `snippets/verse/`. Filename typo ships: **`buy_offer_buton_device.verse`**.

| File | Job |
|---|---|
| `entitlement_definitions.verse` | `feature_example_entitlement` + corn/shovel/gnome/VIP/berries/pumpkin/can |
| `offer_definitions.verse` | `OfferInfo` copy vs entitlements; `ExampleOffers` entitlement_offer + bundle |
| `entitlement_enum.verse` | `@editable` enum → entitlement class |
| `buy_offer_buton_device.verse` | `BuyOffer(Player, Offer)` — **do not** branch on Result for grants |
| `grant_entitlement_device.verse` | `GrantEntitlement` (promo/restore, not a sale) |
| `consume_example_device.verse` | consume packs |
| `door_unlock_example_device.verse` | VIP `PaidArea` |
| `helpers.verse` | Scene Graph ancestor → player; entitlement → prefab |

## What this sample actually teaches

1. **Entitlement ≠ offer.** Item fields (`Consumable`, `MaxCount`, `PaidArea`, `PaidRandomItem`, `ConsequentialToGameplay`) vs shop copy + `MakePriceVBucks`. Same corn can have two offers (full vs half-price). See techniques/entitlement-vs-offer.md.
2. **`BuyOffer` success is not your grant.** Subscribe `GetEntitlementsChangedEvent`. Validate on join with `GetPurchasedEntitlements`.
3. **Bundles hold `[]tuple(offer, int)`**, not entitlements. Nest ≤ 5. Starter = 10 packs + shovel.
4. **Durable: `Consumable=false`, `MaxCount` must be 1.** Consume on durable fails.
5. **Flags are moderation:** paid random needs odds in description; `RestrictPaidRandomItems` / `RestrictDirectPromptsToPurchase`; consequential-to-gameplay unless a free same-time alternative exists. Gnome is cosmetic → false; shovel/seeds → true.
6. **Price:** 50–5000, step 50.

## Official tutorial vs shipped

**External** hub: In-Island Transactions

Chapters (unsuffixed `/documentation/fortnite/<slug>-in-fortnite?lang=en-US` unless already `-in-fortnite`):

1. Overview
2. Creating Items and Offers — **this farm**
3. Best Practices
4. Device Template — **potion/sword File → New Device**, not this project
5. Monetization guidelines
6. Custom UI — points at User Interfaces template

### Lines up (ch. Creating Items)

| Tutorial | Shipped |
|---|---|
| Consumable vs durable; Name/Desc/Short/Icon/MaxCount/flags | `entitlement_definitions.verse` |
| Offer copy independent; `MakePriceVBucks`; 50–5000 ×50 | `offer_definitions.verse` |
| Stacked + nested bundles; depth 5; ≥1 entitlement | `corn_seed_pack_bundle`, `starter_bundle` |
| `BuyOffer` / `ShowOffersDialog` / `GrantEntitlement` / `ConsumeEntitlement` | booth devices |
| Don’t drive grants from BuyOffer Result — `GetEntitlementsChangedEvent` | comments in `buy_offer_buton_device` |
| Corn pack / shovel / starter bundle examples | same names |

### Does not line up

| Topic | Tutorial | Shipped |
|---|---|---|
| Base class (Creating Items) | `my_island_entitlement` | `feature_example_entitlement<abstract><castable><unique>` |
| Identifier style | `cornseedpacket` in some snippets | `corn_seed_pack` |
| Device Template chapter | potion/sword/`basic_sword` complete code | **different sample** — File→New Device |
| Typo | — | `buy_offer_buton_device` (missing **t**) |
| Gnome ShortDescription | — | **“decrative”** in OfferInfo |
| Custom UI chapter | User Interfaces template | this island is booths + default dialog |

Train on snippets. Emitting potion/sword from Device Template as *this* island is the wrong file.

## Negative findings

- Nested umap `*Tools.umap`.
- Experimental Aurora + custom inventory — don’t treat as default island flags.
- Empty `art_declarations.verse`.

## Capture backlog

1. Live `BuyOffer` event bindings. *Waive — API is in Verse.*
2. Best-practices + restrictions chapters (odds, IARC). *Open, low — policy, not APIs.*

## Live notes

Level confirmed (`…templateTools`). Streaming off.
