---
id: digest-018
kind: digest
---

# Intro to UEFN (`intro_to_uefnfk1`)

Epic first-party onboarding template, found under **Project Browser → Feature Examples →
Getting Started**. A walk-through "content hall": a built environment whose rooms are lined
with **55 explanatory display boards**, each captioned with text, introducing UEFN concepts to
a first-time user. The island is the lesson; there is no game.

**Sources:** template extraction in `_raw/intro_to_uefnfk1/`, plus the deprecation tracking
sheet (Templates tab, row "Intro to UEFN ").

> **A live-editor pass was run.** The project was open in UEFN on **2026-09-03** and audited
> through UnrealMCP (`localhost:8000`). `SceneTools.get_current_level` returned
> `/intro_to_uefnfk1/intro_to_uefnfk1`, confirming the right project. **All four backlog items
> are closed**, one of them as a negative result, and **two claims in the first draft of this
> digest were corrected** — see Corrections from the live-editor pass.

> **Extractor provenance — read before citing this digest.** `_raw/intro_to_uefnfk1/` was
> produced by **`tools/extract-template.py`, a reconstruction**; the original
> `tools/extract-template.sh` referenced by `00-EXTRACTION-PROCESS.md` is **not present in this
> repository** and could not be recovered. The reconstruction emits the same file set by the
> same method (uncompressed FName-table scan), but it is **not bit-identical to the pipeline
> that measured character_environment_sample**, and character_environment's own `_raw/` is absent, so the two cannot be
> diffed. Counts here are reproducible against this tree; they are not cross-validated against
> the original tool. Every claim below is tagged **Observed** only where it is read directly
> from the extraction or a plain-text manifest.

## Sheet status (External — tracking spreadsheet)

| Field | Value |
|---|---|
| Status | **should remove** |
| Priority | High |
| Type / Category | UEFN / Feature Examples / Getting Started |
| Dev Owner | Ashley Stancill (ashley.stancill@xa.epicgames.com) |
| Open Errors | None |
| Playable | Yes |
| Last Check | 6/1/2026 |
| Documentation | **No** — links to documentation that does not exist |
| Current / Keep | No / No |
| Verdict recorded | *"can delete because my first island does a better job as an intro to UEFN"* |

The sheet's stated reason for removal is that the template *"needs an update to show how to use
scene graph and verse, will require a major overhaul of the whole template"*. The extraction
below **corroborates that**: there is no Verse in this sample at all, and no Scene Graph
entities. (An earlier draft added "Scene Graph is enabled in config but unused" — **retracted**;
those flags were switched on by the operator before extraction, not shipped. See the warning
under Project configuration.)

## Scale (Observed)

| | |
|---|---|
| Authored assets | **16** |
| External actor packages | 314 |
| External object packages | 42 |
| Total `.uasset` | 372 |
| Maps | 1 (`intro_to_uefnfk1.umap`, 15,879 bytes) |
| Verse files | **0** |
| Fab references (`.uref`) | **0** |

Sixteen authored assets against 314 placed actors. **this sample authors almost nothing** —
it is a composition of content that already ships with Fortnite. Contrast character_environment_sample:
307 authored assets, 2474 actors.

## Project configuration (⚠ partly contaminated — see warning)

> ### ⚠ The experimental flags below are NOT shipped template state
>
> The person running this extraction **manually enabled Scene Graph and the other experimental
> toggles in Project Settings before the extraction ran** — they are *not* on by default. UEFN
> rewrote `.uefnproject` and `.uplugin` on disk when they were flipped, so `config.txt` captured
> **the working copy, not the template as it ships**.
>
> This is not recoverable from this copy: every file in the project (all 373 content files plus
> all three descriptors) carries a mtime of **2026-09-03 10:47–10:53**, because the project was
> copied fresh. There is no mtime gap that separates shipped values from the toggles.
>
> **Treat every `experimental.*` value below, and `EnableSceneGraph` / `EnableSceneGraphAnimation`
> / `EnableVerseAssetRedirectorReflection` in the `.uplugin`, as UNVERIFIED.** To re-baseline,
> read the descriptors from a pristine copy of the template (fresh UEFN "new project from
> template", or the template source in Perforce) — do not re-read this working copy.
>
> Fields *not* exposed in the Project Settings experimental UI — `compatibilityVersion`,
> `requiredRedirectorStartingVersion`, `apiAccess`, `matchmaking.*`, `VersePath`, `VerseVersion`,
> `VerseAPI`, `ValkyrieSettings` — are unaffected and remain **Observed**.

```
compatibilityVersion            42.10
requiredRedirectorStartingVersion  42.10.57566230
apiAccess                       Private
matchmaking.maxPlayers          16       (maxTeamCount 16, maxTeamSize 16, minPlayers 1)
allowJoinInProgress             true
⚠ experimental.sceneGraph         bIsSceneGraphSystemAllowed:  true    <- OPERATOR-ENABLED
⚠                                 bEnableOneFilePerEntity:     false   <- unverified
⚠ experimental.sceneGraphAnimation             true    <- OPERATOR-ENABLED
⚠ experimental.sceneGraphExperimentalFeatures  true    <- OPERATOR-ENABLED
⚠ experimental.pythonExperimental              true    <- OPERATOR-ENABLED
⚠ experimental.verseControlRigEnabledSetting   true    <- OPERATOR-ENABLED
⚠ experimental.toolsets                        true    <- OPERATOR-ENABLED
⚠ experimental.inventory                       true    <- OPERATOR-ENABLED
⚠ experimental.assetRedirectorReflectionSetting true   <- OPERATOR-ENABLED
```

`.uplugin`: `VerseVersion 1`, `VerseAPI 1`, `CanContainVerse: true`,
`ValkyrieSettings.libraryFolderMode: "Absent"` (**Observed**);
⚠ `EnableSceneGraph: true`, `EnableSceneGraphAnimation: true`,
`EnableVerseAssetReflection: true`, `EnableVerseAssetRedirectorReflection: true`
(**unverified** — same contamination).

`VersePath` is `/epictemplates@fortnite.com/intro_to_uefnfk1` — a real creator domain, unlike
character_environment's `/invaliddomain/` placeholder.

**Retracted:** an earlier draft of this digest read the flags above as shipped state and
concluded *"the config advertises capability the content does not use."* **That inference is
withdrawn** — the flags were switched on by the operator, so they say nothing about the
template.

The correction makes the underlying criticism **stronger, not weaker**. The content-side facts
are untouched and independently sourced: **zero `.verse` files** (`script-modules.txt`,
`verse/_NONE.txt`) and **no Scene Graph entities** — and `bEnableOneFilePerEntity: false` with
zero `__ExternalEntities__` directories corroborates the latter regardless of who set the flag.
On the most likely reading (the toggles were *off* as shipped, which is the UEFN default), the
template does not merely fail to use modern UEFN — **it ships with modern UEFN switched off**.
A first-timer opening it to "learn UEFN" would have to find and enable Scene Graph themselves
before they could even begin. That is a sharper version of the gap the sheet flags.

Note also `compatibilityVersion` **42.10** — the project descriptor is current, which makes the
package staleness below more conspicuous, not less.

## External dependencies (Observed — `reference-mounts.txt`)

**No `.uref` files.** this sample declares no Fab pack dependencies; everything it references
is either its own content or content mounted from the Fortnite build.

| Mount | Refs | What it is |
|---|---|---|
| `/intro_to_uefnfk1` | 949 | the template's own content |
| `/Game` | 759 | Fortnite's own content |
| `/CR_Legacy` | 182 | legacy Creative props |
| **`/ContentHall`** | **181** | **the tutorial display/room Blueprints — see below** |
| `/CreativeCoreDevices` | 112 | Creative device definitions |
| `/Engine` | 93 | engine defaults |
| `/CRD_PlayerSpawn`, `/CRD_VolumetricRegion`, `/CRD_HUDController`, `/CRD_PointLight`, `/CRD_Volume` | 12 / 10 / 4 / 3 / 2 | per-device Creative mounts |
| `/PrimalGameplay`, `/FlipperGameplay`, `/FeralCorgiGameplay` | 6 / 1 / 1 | gameplay feature plugins |
| `/CRP_7Village`, `/CRP_HuntersHaven`, `/CRP_Primal`, `/Creative_TemporaryTeams`, `/CreativeCompanion`, `/FortUILibrary`, `/Itemization` | 1 each | prop packs / misc |

21 distinct mounts, 2322 path references. **`/ContentHall` is the load-bearing one** — it is
not a Fab pack and not template content, and the template's central teaching device lives
there (below).

## Device inventory (Observed — `actor-classes.txt`)

314 one-file-per-actor packages, **71 distinct classes, 0 unclassified**.

### The tutorial layer

| Count | Class | Mount |
|---|---|---|
| **55** | `BP_DemoDisplay` | `/ContentHall/Blueprints/` |
| 5 | `BP_DemoRoom` | `/ContentHall/Blueprints/` |

These 60 are the template's actual content. Each `BP_DemoDisplay` is a signage board carrying
`Title Text` / `Description Text` / `DescriptionFinal` text properties, a
`TextRenderComponent`, and a set of instanced-static-mesh components forming the board frame
(`ISM_Main`, `ISM_Side`, `ISM_Corner`, `ISM_Curve`, `ISM_CurveEdge`, `ISM_ShadowStraight`,
`ISM_ShadowRound`, `ISM_ShadowStraightTop`, `ISM_TitleBarEnd`), plus `/ContentHall/Materials/M_Display`.
`BP_DemoRoom` references `/ContentHall/Misc/S_RoomSettings` and `/ContentHall/Misc/E_EnclosureLevel`.

**The teaching content is 55 instances of one Blueprint from a mount the template does not
own.** The per-board text is instance data inside `__ExternalActors__` packages.

### Creative devices

| Count | Class |
|---|---|
| 13 | `Device_Teleporter_V2` |
| 13 | `Device_ItemSpawner_V3` |
| 4 | `BP_Creative_Player_Spawner_Prop` |
| 1 each | `Device_Billboard_V2`, `Device_Button_V2`, `Device_ConditionalButton_V2`, `Device_HUDController_V2`, `Device_PointLight_V2`, `Device_MutatorVolume_V2`, `Device_Barrier_V2`, `BP_Device_DamageVolume_V2`, `Device_ExperienceSettings_V2_UEFN` |

~38 device instances. The 13 teleporters + 13 item spawners are the navigation and
demonstration spine of the hall.

### Native (engine/Fortnite) classes — not authored content

`FortStaticMeshActor` (23), `BlockingVolume` (2), `PlaysetRoot` (2), `WorldPartitionMiniMap`
(1), `WorldDataLayers` (1), `FortMinigameSettingsBuilding` (1). These are placed instances of
engine/Fortnite classes, **not template-authored Blueprints** — do not read them as inventory
this sample created.

### Environment art — all stock Fortnite props

`Athena_Prop_Storage_Barrel_02` (26), `BP_WV_Tudor_Wall_05` (20),
`SM_FORT_Floors_Generic_BasicTile` (17), `BP_WV_Tudor_Beams_01` (17), plus a long tail of
`PirateShip_*`, `IceCastle_*`, `BP_Cove_*`, `Pueblo_*`, `Boardwalk_*`, `CP_Mansion_*`,
`Apollo_*` props. **These are Fortnite Creative props placed from the build, not assets the
template ships** — which is why 314 actors need only 16 authored assets.

The kit mixing is worth noting on its own: Tudor walls, a pirate ship, an ice castle, a pueblo,
a boardwalk and a mansion in one map. It is a sampler, not a coherent environment.

## Content organisation (Observed — `folders.txt`, `naming-prefixes.txt`, `assets.txt`)

```
Content/
  Collections/            (empty)
  Content/                <- doubled folder name
    Assets/
      Materials/          M_Locker, M_Wall
      Meshes/             SM_Locker_Uncooked
      Textures/           T_Locker_D, T_Locker_N, T_Locker_ORM
    UEFNWelcome101/
      Materials/          M_Atlas_UV, MI_Atlas_Launch, MI_Atlas_LaunchOptions, MI_Atlas_PushChanges
      Textures/           T_CEAtlas_A
  Developers/
    felixkupis/Collections/   (empty — local user scratch shipped in the template)
  Meshes/                 CubeGridToolOutput_328081B2, CubeGridToolOutput_DABB3107, SM_CornerInn_100
```

All 16 authored assets. Prefixes: `T_` 4, `MI_` 3, `M_` 3, `SM_` 2, no prefix 4.

Three housekeeping problems, all **Observed**:

1. **`Content/Content/`** — a folder named `Content` inside `Content`. Every asset path in this
   template reads `Content/Content/Assets/...`.
2. **`Content/Developers/felixkupis/`** — a developer scratch folder shipped in a first-party
   template. character_environment shipped `Developers/jojitsuruga/`; **the same mistake in two of two
   templates examined.**
3. **`CubeGridToolOutput_328081B2` / `CubeGridToolOutput_DABB3107`** — unrenamed modeling-tool
   output shipped as content, in a template whose job is to model good practice.

The one asset group with a clear teaching purpose is `UEFNWelcome101/` — an atlas texture plus
three material instances named `MI_Atlas_Launch`, `MI_Atlas_LaunchOptions`,
`MI_Atlas_PushChanges`, i.e. UI imagery for the "launch session / push changes" workflow.

## Verse (Observed)

**None.** Zero `.verse` files. `_raw/intro_to_uefnfk1/verse/` records the absence.

Nothing is retained in this sample's `snippets/` folder, because there is nothing to retain.

## Data layers (Observed — `datalayers.txt`)

**None referenced.** Zero data layer assets, against character_environment's 39. A `WorldDataLayers` actor
and an `intro_to_uefnfk1_DefaultHLODLayer` asset exist (engine defaults for a partitioned
world), but no `DL_*` layer is authored or assigned.

**Streaming is disabled** (**Observed** — `AssetTools.get_asset_tags` on the `.umap`, live
session):

```
LevelIsPartitioned          1
LevelIsUsingActorFolders    1
LevelIsUsingExternalActors  1
LevelHasStreamingDisabled   1
```

`SceneTools.get_data_layers` returns `[]` — **zero data layer instances in the world**,
confirming the extraction.

So this sample matches character_environment on the finding that mattered most there: **World Partition is
on, runtime streaming is off.** Two for two. The OFPA machinery (`LevelIsUsingExternalActors`)
is what produces the 314 one-file-per-actor packages; it is independent of streaming.

## Engine subsystems pulled in (Observed — `script-modules.txt`)

`/Script/CoreUObject` (373), `/Script/Engine` (372), `/Script/FortniteGame` (252),
`NavigationSystem` (28), `CascadeRuntime` (15), `DataDrivenGameplayEventRouter` (14),
`GameplayTags` (9), `FortniteAI` (6), `PhysicsCore` (4), `InterchangeEngine` (4),
`ItemizationCoreRuntime` (2), `WorldPartitionHLODUtilities` (2), `Slate` (1), `UMG` (1),
`McpProfileSys` (1), `ExperienceSettings` (1).

Actor-shaped classes: `FortniteGame.BuildingProp` (79), `BuildingWall` (56),
`Engine.Actor` (60), `Engine.ActorFolder` (42), `FortCreativeDeviceProp` (26),
`BuildingFloor` (25), `FortStaticMeshActor` (23), `FortCreativeTeleporter` (13),
`BuildingPropWall` (11), `BuildingRoof` (9).

**No `VerseDevices` module** — corroborating the zero-Verse finding from a second, independent
signal. **`CascadeRuntime`** (15) is notable: Cascade is the *legacy* particle system,
superseded by Niagara. character_environment pulled in `Niagara`; this sample pulls in Cascade.

## Package staleness (Observed)

Engine version stamps across all 372 packages + map, from a bounded `Select-String`:

| Packages | Stamp |
|---|---|
| 546 | `++Fortnite+Release-29.30` |
| 108 | `++Fortnite+Release-29.40` |
| 56 | `++Fortnite+Release-30.30` |
| 6 | `++Fortnite+Release-30.00` |
| 4 | `++Fortnite+Release-30.40` |
| 4 | `++Fortnite+Release-35.10` |
| 2 | `++Fortnite+Release-42.10` |

(Counts exceed the package total because a package may carry more than one stamp.)

**The overwhelming majority of this sample's content was last saved at Fortnite 29.30/29.40,
while its `.uefnproject` claims compatibility 42.10.** Only two packages have been touched at
42.10. That is roughly thirteen releases of drift between the descriptor and the content — a
concrete, measurable form of the "needs a major overhaul" judgement in the sheet.

## The board curriculum (Observed — live editor)

All 55 `BP_DemoDisplay` actors were read via `ObjectTools.get_properties`
(`title Text`, `description Text`, `descriptionFinal`). Full text is preserved verbatim in
`_raw/intro_to_uefnfk1/demo-board-text.json`; the reading-ordered version is
`_raw/intro_to_uefnfk1/syllabus.md`.

**55 boards ≠ 55 lessons.** Classified: **35 teaching boards** (numbered *and* carrying body
text) and **20 wayfinding boards** (chapter title cards, directional arrows, prop labels).

| Chapter | Teaching boards | Covers |
|---|---|---|
| **1 — Introduction** | 5 | Launch/Edit Session, Launch Session Options, Push Changes, and a "drag the barrel and watch it sync" exercise |
| **2 — Devices** | 8 | Player Spawn Pad, device visuals & team index, Island Settings, Zones (barrier + mutator), Item Spawners, Button, Billboards |
| **3 — Consumables** | 13 | Player vs shield consumables, a damage zone for testing, per-item stats (Corn/Pepper/Apple/Meat, four shield items), device consumables via Conditional Button |
| **4 — Assets** | 9 | Cooked vs uncooked content, why cooked assets are read-only, galleries in Live Edit vs editor, prefabs |

Ordering is by the section number in `title Text` (1, 1.1 … 4.4), **not** by the
`DemoDisplay C<N>` label — the C-number is placement order, not reading order. Ties are broken
by world X/Y, since the hall is spatially sequenced.

### Source-data anomalies (Observed — all real on the island, none are extraction errors)

- **Chapter title cards are duplicated as empty boards.** `1 - Introduction` (`C`),
  `2 - Devices` (`C46`), `3 - Consumables` (`C47`), `4 - Assets` (`C48`) each carry a title with
  **no body**, alongside a fully-bodied board with the same or similar title (`C7`, `C22`,
  `C38`). The bodied board is the canonical section text; the empty one is a section marker.
  **Both are retained** in the syllabus — the empty ones in the wayfinding table.
- **Section number `3.1` is used twice for different topics:** `3.1 - Player Consumables`
  (`C25`) and `3.1 - Food (Produce)` (`C30`). Both preserved; this is an authoring slip in the
  template, not a duplicate.
- **`4.1.1` / `4.1.2` / `4.1.3` all read "Cooked Content"** — a deliberate three-part
  explanation, not a duplicate.
- **Wayfinding boards** carry `<` / `>` arrows (`C50`, `C51`, `C52`, `C53`) or bare prop labels
  (`Animal Bone` `C37`, `Mech Part` `C27`, `Pizza Spawner` `C14`, `Weapon Spawner` `C17`,
  `Barrier to Block` `C10`, `Lower Gravity Zone` `C11`, `Player 2 Spawn` `C21`,
  `Island Settings` `C9`, `Team Index '02'/'03'` `C18`/`C19`, `Device Consumables` `C36`).
  `C49` has neither title nor body.

**This curriculum is the only thing in the template that exists nowhere else.** The sheet
records the linked documentation as non-existent, so these 35 boards *are* the lesson. They are
now captured, which is what unblocked retirement.

## Device wiring (Observed — live editor)

`DeviceToolset.ListEventBindings` across every device actor: **14 distinct event bindings**, and
they are strikingly uniform.

| Source | Event | Target | Function |
|---|---|---|---|
| `BP_Creative_Player_Spawner_Prop` | `On Player Spawned` | **12 ×** `Device_ItemSpawner_V3` | `Enable` |
| `Device_Button_V2` | `On Interact` | `Device_PointLight_V2` | `Toggle` |
| `Device_ConditionalButton_V2` | `On Activated` | `Device_ItemSpawner_V3` | `Spawn Item` |

Twelve of the fourteen bindings are one spawn event fanning out to enable every item spawner —
i.e. **the item spawners are dormant until a player spawns**, which matches board `2.5`'s note
that "by default, an Item Spawner has the status of Inactive". The remaining two are the two
interactions the boards actually teach (button→light in `2.6`, conditional-button→weapon in
`3.3`).

**The 13 teleporters have zero event bindings** and every one reads
`Change Teleporter Target: Never`, `ChangeTeleporterTargetInterval: 5.0`,
`Face Player In Teleporter Direction: No`. They are **statically paired by teleporter group**,
not event-driven.

## What this sample actually teaches

Ranked by transferable value. Points 1–3 are the honest answer to "what would we lose".

1. **A tutorial island can be built almost entirely from stock content.** 314 placed actors,
   16 authored assets, zero Fab dependencies. The technique — compose from
   `/Game` + `/CR_Legacy` + Creative device mounts, author only the few materials you cannot
   borrow — is the most reusable thing here and applies to any explanatory or greybox island.
2. **A procedural signage Blueprint is the right shape for onboarding content.**
   `BP_DemoDisplay` is **not** a static component assembly — its components are named
   `NODE_AddInstancedStaticMeshComponent-0…9` and `NODE_AddTextRenderComponent-0…2`, the
   signature of a **construction script that builds the board at edit time** (**Observed** —
   `ActorTools.get_components` on a placed instance). Exposed knobs include `width`, `depth`,
   `height`, `title Text`, `description Text`, `text Alignment`, `title/description Scale`,
   `title/description Colour`, `seperate Title Panel`, `spotlight`, `cast Shadows`,
   `shadowCover`, `number of lines between paragraphs`, `text Padding`, and a
   `DataTableRowHandle` colour lookup. One Blueprint, 55 differently-sized boards, zero manual
   mesh placement.
   **This is the direct counter-example to character_environment**, whose 64 kitbash Blueprints had *empty*
   event graphs and fixed component lists. Do not generalise "UEFN Blueprints contain no logic"
   from character_environment alone — see the correction to
   project-config-and-dependencies.md
(cross-template note — originated in character_environment_sample, corrected by this sample).
   Caveat: `BP_DemoDisplay` lives in `/ContentHall`, so the *pattern* is portable but the
   Blueprint is not template content and survives deletion regardless.
3. **Gate demonstration devices on player spawn.** All 12 item spawners are enabled by a single
   `On Player Spawned` → `Enable` fan-out from the player spawner, rather than shipping active.
   A one-binding idiom for "nothing runs until someone is actually here".
4. **Teleporters as chapter navigation — statically paired, not event-wired.** 13
   `Device_Teleporter_V2` with **no event bindings** and `Change Teleporter Target: Never`; the
   tour is wired purely through teleporter *groups*. Cheaper than event plumbing, and it means
   the navigation graph lives in device properties rather than in a binding web.
5. **Counter-example — how an extraction can contaminate its own evidence.** The experimental
   flags in `.uefnproject` were read as shipped state when they had in fact been toggled on by
   the operator minutes earlier; UEFN rewrites the descriptor on disk when you flip them, so the
   contamination is invisible in the file itself. **Descriptor fields that a UI can write are
   not safe to grade "Observed" from a working copy** — only from a pristine template source.
   Fixed in the process doc's Step 2. The original counter-example, now withdrawn, read:
   ~~config promising what content does not deliver — Scene Graph, SG
   animation, Verse control rigs and Verse asset reflection are all enabled; there is no Verse
   and no Scene Graph usage.~~ A learner opening this to see "how UEFN works" sees the 2024 way
   of working in a 42.10 wrapper.
6. **Counter-example — housekeeping.** `Content/Content/`, `Developers/felixkupis/`,
   two `CubeGridToolOutput_*` assets, an empty `Collections/`. character_environment shipped
   `Developers/jojitsuruga/`; that is **two for two on leaked developer folders in first-party
   templates**, which promotes it from anecdote to a pre-publish checklist item.
7. **Counter-example — package drift is measurable without opening the editor.** The
   `++Fortnite+Release-*` stamp in the name table gives a staleness histogram for any template
   in seconds. Useful triage for the rest of the deprecation list.

## Corrections from the live-editor pass

The session on **2026-09-03** falsified two claims made from the extraction alone. Both were
**Inferred**, which is the failure mode this pipeline exists to catch.

| Was | Actually |
|---|---|
| "`LevelHasStreamingDisabled` does not appear in the `.umap` name table … absence is not proof streaming is enabled" — left open | **`LevelHasStreamingDisabled = 1`.** Streaming *is* disabled. The caution was right; the tag simply is not recoverable from a name-table scan. |
| "13 teleporters … let a linear tutorial be entered at any point" (implying wiring) | Teleporters have **zero event bindings** and `Change Teleporter Target: Never`. Navigation is static teleporter-group pairing. |

A third claim was not wrong but was materially incomplete: `BP_DemoDisplay` was described as a
component assembly with "instanced-static-mesh components forming the board frame". It is
actually **construction-script generated** (`NODE_Add*` components) with ~20 exposed parameters
— a materially more interesting finding, and one that contradicts a cross-template pattern
drafted from character_environment.

## Capture backlog

Ranked by teaching value, per `00-EXTRACTION-PROCESS.md` Step 5. this sample's teaching value
is low and its retirement is already agreed, so the bar for "worth capturing" is set high and
most candidate items are waived deliberately.

**0 open, 4 closed (one as a negative result), 5 waived.**

### Closed by the live-editor pass (2026-09-03)

| # | Question | Answer |
|---|---|---|
| 1 | Is World Partition **streaming** enabled? | **No — `LevelHasStreamingDisabled = 1`**, and `get_data_layers` returns `[]`. Same as character_environment. Recorded under Data layers. |
| 2 | What do the 55 boards actually *say*? | **Fully captured** — all 55 read out to `_raw/intro_to_uefnfk1/demo-board-text.json`, ordered into `_raw/intro_to_uefnfk1/syllabus.md`. **35 teaching boards across 4 chapters**, 20 wayfinding. See The board curriculum. |
| 3 | Does `BP_DemoDisplay` contain any logic? | **Yes — it is construction-script driven.** `ActorTools.get_components` returns `NODE_AddInstancedStaticMeshComponent-0…9` + `NODE_AddTextRenderComponent-0…2`, i.e. components spawned by a graph, with ~20 exposed parameters. **Partially unanswerable:** the class `/ContentHall/Blueprints/BP_DemoDisplay` is **not in the asset registry** — `get_asset_class` and `get_dependencies` both return *"Asset does not exist"*, and `get_actor_asset_path` resolves only to the OFPA package `…/__ExternalActors__/…/GXMDX5S0IG4LYEBKTEYT4R`, which the registry also rejects. So the board's *dependency list* (fonts, meshes, the colour DataTable) is **not reachable** from this session; the construction-script finding stands on the component evidence alone. |
| 4 | Are the 13 teleporters wired into a sequence? | **No.** Zero event bindings; `Change Teleporter Target: Never`. Navigation is static teleporter-group pairing. The level's **14 total bindings** are tabled under Device wiring. |

### Negative results worth recording (Step 6b)

- **`/ContentHall` is not a mounted, registry-visible content path** in this session, despite
  181 references to it and every board being an instance of a Blueprint that lives there. It is
  neither template content nor a `.uref` pack. Any future attempt to inspect `BP_DemoDisplay`
  *as an asset* will hit the same wall — read placed instances instead.
- **`AssetTools.get_dependencies` rejects OFPA external-actor packages.** The map's own
  dependency list resolves fine and includes the board packages, so use the `.umap` as the entry
  point when you need actor dependencies.

### Waived — not worth carrying

- **Per-device `@editable` / `playerOptionData` for all ~38 devices.** Stock Creative devices at
  default-ish settings; the device *list* is the finding, and Epic's device docs beat a 29.30-era
  template.
- **Material graphs for `M_Locker` / `M_Wall` / `M_Atlas_UV`.** Four-asset locker set and a UI
  atlas. this sample's `techniques/` has nothing to gain.
- **The 42 `__ExternalObjects__` packages.** Actor-folder and misc editor objects; no teaching content.
- **Transforms / layout of the 314 placed actors.** The kit-mixing observation is the lesson;
  the exhaustive placement is Epic's level.
- **HLOD settings** (`intro_to_uefnfk1_DefaultHLODLayer`). Engine default on a map with no
  authored data layers; character_environment's note already covers the settings that matter.

## Retirement review (`00-EXTRACTION-PROCESS.md` Step 7)

| # | Criterion | Status |
|---|---|---|
| 1 | Digest written; techniques promoted to `KB/templates/intro_to_uefnfk1/techniques/` | **Partial** — digest written and the curriculum captured. One correction pushed to `project-config-and-dependencies.md` (Blueprint-logic claim) — a **character_environment_sample** note this sample corrected rather than one of its own. A signage/onboarding technique note is **not** written, and is arguably not warranted — see the assessment below. |
| 2 | All Verse source retained verbatim in `KB/templates/intro_to_uefnfk1/snippets/` | **Met, vacuously** — the template contains no Verse. |
| 3 | External dependencies recorded | **Met** — no `.uref`; the 21 referenced mounts are tabled above. |
| 4 | Capture backlog empty | **Met** — 0 open, 4 closed (one partially unanswerable, with the reason recorded), 5 waived. |
| 5 | `_raw/` committed so claims stay auditable | **Qualified** — `_raw/intro_to_uefnfk1/` exists and is committed, but was produced by a **reconstructed** extractor, not the original `extract-template.sh` (see the provenance note at the top). Claims are reproducible against this tree; they are not cross-validated against the tool that measured character_environment. |

**Retirable.** The blocking item — the board curriculum — is captured verbatim in
`_raw/intro_to_uefnfk1/demo-board-text.json` and `syllabus.md`. The only remaining qualification
is criterion 5's extractor provenance, which is a repository-wide caveat rather than anything
specific to this sample.

### Content classification (Step 7, "reference or reusable art")

**Reference — with the strongest case of any template examined so far.** The art is not at
risk, and this is verifiable rather than assumed:

- **Zero `.uref`s and 16 authored assets.** There is no Fab pack to preserve and almost nothing
  authored to lose.
- **The environment is stock Fortnite content** placed from `/Game`, `/CR_Legacy` and the
  Creative device mounts — bytes already in every user's build. Deleting the template deletes
  no art.
- **The teaching Blueprints are not template content.** `BP_DemoDisplay` and `BP_DemoRoom` live
  in `/ContentHall`, outside the project. They survive retirement regardless.
- **What is genuinely template-owned:** the locker material/mesh/texture set (6 assets), the
  `UEFNWelcome101` atlas + 3 material instances (5 assets), two unrenamed CubeGrid outputs and
  `SM_CornerInn_100`. None of it is worth a Fab pack.

The only preservation-worthy artifact is **prose**: the board curriculum, and the composition
pattern in point 1. Both fit in this KB.

### Recommendation

**Retire.** The sheet's replacement rationale is sound and the extraction supports it:
**My First Island** should carry intro duties, and a rebuilt intro template should start from
Scene Graph and Verse rather than being patched onto content last saved at Fortnite 29.30.

Two things to carry forward into whatever replaces it:

1. **The 35-board curriculum is still pedagogically sound** — chapters on Launch Session /
   Devices / Consumables / Assets are exactly what a first-timer needs, and the copy is
   well-written. What is stale is the *delivery* (static boards, no Verse, no Scene Graph) and
   the content beneath it, not the syllabus. A rewrite should start from `syllabus.md`, not
   from a blank page.
2. **The signage pattern itself is worth reusing** — a construction-script board Blueprint with
   text + size parameters is a better authoring surface than hand-placed meshes, and
   `BP_DemoDisplay` already lives outside the template in `/ContentHall`, so it survives
   deletion.

No art is at risk, and nothing else in the template needs to be preserved.
