---
id: digest-019
kind: digest
---

# music_fantasy_sample — sample digest

> **Extractor provenance.** `tools/extract-template.py` (Pass A) + UnrealMCP on
> `localhost:8000` (Pass B). Default level
> `/7e17cc18-1d20-4b7f-97ac-685b91f3abe4/music_fantasy_sample` (**Observed**).
> Hub: music-fantasy Islands
> + island tutorial Working with MUSIC_THEME Islands
> + theme Rules (publishing, not a device syllabus).
> `docsUrl` empty. Not on the spreadsheet. Compat **43.00**. `featureSet`: **DemonFactionGFS**.

| | |
|---|---|
| Classification | **theme starter island** — alley hire + dual-wave demons + Portal + plaza jam, not a full concert match |
| Content type | **catalog wrap** (82 authored vs 5313 OFPA; `/TREE_PACK` + NeonCity/Heist/Utopia/TTides kitbash) |
| Verse | **0 files** |
| Fab `.uref` | **0** |

## What it is

The picker name is **music-fantasy streets Starter Island**. Docs and the `.uefnproject` title say **music-fantasy Starter Island**. Disk folder is `music_fantasy_sample`. The island is a **linear theme sampler**: hire hero squad in the alley, pick up HeroA’s sword to wake street demons, walk the bathhouse for a second wave + Portal gold, then plaza cine + soda-can jam. Teaching that survives deletion is the **DemonFaction device redesigns** (Character / Creature), **Portal dual-plane cine swap**, and **Mutator jam → Crowd + aura**. Do not duplicate: volume-checkpoint-gated-horde.md, guard-patrol-volume-swap.md, trigger-cinematic-quake.md. ExtraAssets HUD: add this island as a **fourth** source on social_deduction_sample — do not rewrite it. Authored `Content/UI/Widgets/UW_*` are **not** the live HUD (contrast Narrative Widgets-folder coincidence).

## Scale (**Observed** — `summary.txt` + live)

| | |
|---|---|
| Authored assets | **82** |
| OFPA | **5313** |
| Maps | `music_fantasy_sample.umap` (only) |
| Verse files | **0** (`VerseToolset.ListFiles` on `/7e17cc18-…` — folders only; do not recurse `__ExternalActors__`) |
| Verse devices placed | **0** |
| Data layers | **0** |
| Fab `.uref` | **0** |
| Compatibility | **43.00** |
| Matchmaking | **16** / `teamSize=Dynamic` / `teamFillOption=MustFill` (`.uefnproject` + live Island Settings) |
| Streaming | Partitioned + actor folders + OFPA. **`LevelHasStreamingDisabled=1`** (Walker/VFX camp). Sampled descriptors mixed (`HUD`/`Island`/`DSD` `bIsSpatiallyLoaded=false`; many lights true). `load_actors` required for device values. `find_actors` by `"Device_"` / `"Portal"` returned **0** — use `get_folders` + `get_actors_in_folder`. `folderPath` often **None**. |

Empty `Developers/felixkupis/Collections/`. First-party scratch folder.

`experimental.*` (Python / toolsets / Scene Graph) in `.uefnproject` is **Unverified**. Content-side: **no** `.verse`, **no** `__ExternalEntities__`. Plugin `EnableSceneGraph: true` with no Scene Graph entities.

Largest file ~**1 MB**. Whole folder ~**43 MB** (under GHES 100 MB).

## Level organisation (**Observed** — live `get_folders` + folder counts)

```
01_Lighting/                      27 + Bathhouse BackLights + PP_Managers 2; PostProcess 0 leftover
Environment/Alleyway/…            Casino 193, TVStudioBack 159, ParkingGarage 90, …
Environment/BathHouse/            Interior 256 + Front 120 + Locker 34 + steam/water FX
Environment/Buildings/            ParkHotel 299, TallStores 245, CornerStore 113, …
Environment/Plaza/PlazaBuildings/ DepartmentStore 224, NSquare 151, BigBillboard 133, …
Environment/SaunaCorridor         359
Environment/Street                409
Environment/UpperApartments       310
Environment/Sidewalk              240
Gameplay/                         6 (HUD×2, IslandSettings, 2 spawn pads)
Gameplay/Bathhouse               22
Gameplay/Plaza                   26
Gameplay/Streets                 18
PortalBarriers_Demonic          10
PortalBarriers_Golden           10
SodaPopStage                     47
```

Empty leftover folders: `BathA`, `BathFX`, `BathhouseGlass`, `HotelD`, `UpperAppartments` (typo vs `UpperApartments`), `UpperTallBuildings`, `Environment/PlazaBuildings` (duplicate of `Plaza/PlazaBuildings`, all 0), `01_Lighting/PostProcess`.

No `Device_Billboard_V2`. Syllabus is the public docs page, not boards.

## Device inventory (the design)

| Count | Class | Role |
|---|---|---|
| 14 | `Device_Chair` | Plaza seating |
| 13 | `BP_CrowdVolume_Device` | Jam audience; start `enabledDuringPhase=None` |
| 10 | `Device_ItemSpawner_V3` | HeroA sword, HeroB ramyeon, HeroC bubble, Tiger mask + consumables |
| 8 | `Device_DemonFaction_CreatureSpawner` | Street 3 + bathhouse 5; `enabled at Game Start=false` |
| 6 | `Device_DemonFaction_CreatureManager` | Tutorial stats (health **1**, player damage **0**) |
| 6 | `NiagaraActor` | Steam / stage smoke |
| 5 | `Device_Volume` | Street / bathhouse Enable-Disable + EmoteVolume |
| 5 | `Device_CinematicSequence` | Portal Demonic autoplay + Golden; CombatOn/Off; PlazaDance |
| 3 | `Device_Character_V2` | Plaza emote mannequins |
| 3 | `Device_DemonFaction_CharacterSpawner` | Hireable HeroA / HeroB / HeroC |
| 3 | `Device_VFXSpawner_V2` | Stage / plaza FX |
| 3 | `BP_LumenExposureManager` | Global / BathHouse / Saunas; `useNonLumenPost=true` |
| 2 | `Device_Trigger_V2` | Bathhouse cine bus; plaza cine + HUD hide |
| 2 | `Device_Powerup_VisualEffect_V2` | Demon aura default; gold aura on jam |
| 2 | `Device_HUDController_V2` | ExtraAssets HUD + cutscene HideElement |
| 2 | `Device_CRD_AudioPlayer` | Volume Play/Stop |
| 2 | `Device_Keylock_V2` | Bathhouse locks |
| 2 | `BP_Creative_Player_Spawner_Prop` | Team 1; Player 1 `useAsIslandStart=true` |
| 1 | `Device_Timer_V2` | **3s** bathhouse delay; not HUD |
| 1 | `Device_ScoreManager_V2` | `maximum Score=12` → CombatOff |
| 1 | `Device_MutatorVolume_V2` | Stage jam bus |
| 1 | `Device_Prop_PlayerCheckpoint_V2` | One pad |
| 1 | `Device_ItemGranter_V2` | Consumable (2 labels, 1 class count) |
| 1 | `BP_DaySequenceModifier_Device` | Priority **200**, `dayNightCycleTime=23`, stock skydome |
| 0 | Storm / Demon Placer / VerseDevice / Billboard_V2 / PlaysetRoot | docs feature set or leftover |

## Verse (0 files)

No `snippets/verse/` folder. Valid.

## Authored kit (**Observed** — `assets.txt` + live)

```
Content/BP/                BP_SodaStage, BP_Stage
Content/Materials/         PP, bathhouse, signs, MI_Portal_Blue / Golden
Content/Meshes/            SM_SodaPop_Stage, SM_KDH_Sign_*, modeling leftovers (Boolean_*, Box_*)
Content/Sequences/         LS_PortalChange_{Demonic,Golden}, LS_CombatOn/Off, LS_PlazaDance
Content/UI/Widgets/        UW_PlayerInfo, EquippedItem, Quickbar, AllPlayers, TeamInfo
Content/VFX/               NS_Steam, NS_Smoke, NS_Smoke_Stage
Content/LS_PortalChange   leftover (plus Sequences/LS_PortalChange)
Content/NewNiagaraSystem   leftover
```

**Tier 2 mounts:** `/TREE_PACK` 2940, `/CRP_NeonCity` 789, `/CRP_Heist` 655, `/CR_Legacy` 618, `/Utopia_Comp` 616, `/TTides_Comp` 490, `/DemonFactionExtraAssets` 46, `/CRD_DemonFaction_FiendOverrides` 54, `/DemonFaction_CharacterSpawner` 11, `/DemonFaction_Assets` 17, `/DemonFaction_Comp` 9. Leftover GUID `/e455d126-4a95-8ed2-feec-6ea4d08ba827` (old sequence mount). No Fab `.uref`.

## What this sample actually teaches

1. **Portal is two plane folders + cine swap, not Verse.** Autoplay `LS_PortalChange_Demonic`; bathhouse-exit volume plays `_Golden`. See techniques/Portal-dual-plane-sequence.md.
2. **Street demons wake on item pickup; bathhouse demons wake on Volume → 3s Timer.** DemonFaction Creature Spawner, not Walker CDs. techniques/item-volume-gated-demon-waves.md.
3. **hero squad is a Guard Spawner redesign with `DemonFaction Character` + `bCanBeHired`.** techniques/hireable-theme-character-spawner.md.
4. **Jam is Mutator `On Player Starts Jamming` → Crowd Enable + gold Pickup / demon Clear.** techniques/jam-mutator-crowd-aura.md.
5. **HUD Controller widget overrides live in ExtraAssets, not the authored Widgets folder.** Live: `/DemonFactionExtraAssets/UI/Widgets/UW_HUD_PlayersInfoStack` + `UW_HUD_EquippedInfoBlock` + `UW_HUD_QuickbarSlot`. `showMinimap` / `showTeamInfo` / `showBuildMenu` / resource / sprint = **HideElement**; health/shields **DoNotOverride**. Cutscene controller is a second device (`priority=VeryHigh`). Fourth ExtraAssets HUD source — originating write-up stays social_deduction_sample.
6. **Nightlife DSD is priority 200, not 2000.** `use Volume=false`, `dayNightCycleTime=23`, `sunlightIntensity=3`, `overrideSky/Sunlight/Fog/Clouds=true`, `enableFogComponent=false`, stock `MI_DaySequenceSkydome`, cached `DSA_UEFN_Ch5S1_C_0`. Three LumenExposureManagers (Global / BathHouse / Saunas). Not a Day Sequence volume-stack source.
7. **The city is TREE_PACK / NeonCity / Heist / Utopia / TTides catalog, not 82 authored rooms.** Docs name Mega City / Ch4 Jungle / Paradise. Cell Snap off is a **viewport** claim (**Unverified** in files).

## Official tutorial vs shipped

**External** hub + Working with MUSIC_THEME Islands + theme Rules (headed Chrome 2026-09-15). Verbatim: `_raw/music_fantasy_sample/tutorial.txt`. Spreadsheet: no row. `docsUrl` empty.

### Lines up

| Tutorial | Shipped |
|---|---|
| theme Templates → music-fantasy Starter Island; 16 players | picker **music-fantasy streets Starter Island**; live `maxPlayers=16` |
| Alley hire + HeroA sword → demons → bathhouse wave → plaza jam | Streets Character Spawners + sword Enable; Timer 3s bathhouse; Mutator jam |
| Volume enables Timer + Trigger; Trigger plays Cinematic | `Vol_EnableBathHouseSpawners` → Timer `Start` + Trigger2 → `LS_CombatOn` |
| Plaza cine + HUD Controller hides HUD | Device4 `LS_PlazaDance`; cutscene HUD `showHUD=HideElement` VeryHigh |
| Mutator enables Crowd Volume; gold aura on jam | Crowd `On Player Starts Jamming` Enable; Powerup2 GoldAura Pickup |
| Character Spawner = Guard redesign; HeroA/HeroB/HeroC hireable | `DemonFaction Character` + `bCanBeHired=true` |
| Demon Spawner/Manager = Creature*; 11 types in the set | 8 spawners + 6 managers; live samples Small Demon / Demon |
| HUD widgets PlayersInfoStack / Equipped / Quickbar via HUD Controller | ExtraAssets overrides **on** (all three) |
| Portal blue + gold planes; sequence after exiting bathhouse | folders ×10; Device5 autoplay Demonic; exit volume → Golden |
| Gold + demon player auras via VFX Powerup Custom Effect | ExtraAssets `NS_DemonFaction_{Gold,Demon}Aura` |
| Nightlife lighting | DSD time **23**, cool sunlight color |

### Does not line up

| Tutorial | Shipped |
|---|---|
| Kitbash galleries **Mega City / Chapter 4 Jungle / Paradise** | Live mounts **TREE_PACK / CRP_NeonCity / Heist / Utopia_Comp / TTides_Comp** |
| **Portal Storm** on storm devices | **0** storm devices |
| Demon **Placer** in the device table | **0** placed |
| HUD widget `UW_HUD_EquippedInforBlock` (typo) | live `UW_HUD_EquippedInfoBlock` |
| “Editable widgets in the project files” as the HUD | Authored `UW_PlayerInfo` etc.; **live overrides are ExtraAssets**. `get_referencers` on ExtraAssets HUD: path “does not exist” (mount-at-runtime trap) |
| Cell Snap disabled | Viewport setting **Unverified** (not in `.uefnproject`) |
| Non-destructible island workflow | live `allowBuilding=All` |
| Hub tag “music-fantasy demon hunsters” | typo on the hub, not in the island |

Training rule: shipped bindings win. Tutorial *prose* is the syllabus (no billboards). Do not emit Mega City kitbash / Portal Storm / Demon Placer / in-project HUD as this island’s live setup.

## Negative findings

- Picker **music-fantasy streets** / docs **music-fantasy Starter Island** / folder `music_fantasy_sample` are the same tile.
- Name does **not** mean Verse concert scripting or a Storm circle.
- Authored `Content/UI/Widgets` is **not** Widgets-folder HUD coincidence — ExtraAssets wins.
- Root `LS_PortalChange` + `NewNiagaraSystem` + empty `Environment/PlazaBuildings` + `UpperAppartments` are leftovers.
- `creature Type=Cube Random` on Demon Spawners is leftover Creature DNA; identity is `bright Coil Creature Type`.
- Island `eliminationsToEnd=1` is not the Score Manager bus.

## Capture backlog

Waived: Sequencer keyframes (Portal / Combat / PlazaDance); Blueprint graphs on `BP_SodaStage`; Cell Snap viewport; item spawner `item` property names (GetObjectProperties missing); exact Portal plane mesh class; Gwi-Ma vista placement (feature-set asset, not required on this umap). `experimental.*` Unverified.

## Live notes

- Cursor `user-unreal-mcp` was down; `tools/mcp.py` on `:8000` worked.
- Mount `/7e17cc18-1d20-4b7f-97ac-685b91f3abe4` from `ListFiles("")`.
- `ToyOptionsComponent` has **no** `playerOptionData` — read HUD / Island properties on the **actor**.
- `get_properties` fails the whole call if any requested name is missing; list first, then request only existing names.
- Wrong `refPath` `/mount/map.ActorName` is invalid; use `/{mount}/{map}.{map}:PersistentLevel.{name}`.
- Did not flip experimental flags. Did not switch maps.
