---
id: digest-020
kind: digest
---

# landscape_template — sample digest

> Offline `_raw/landscape_template/`. Live
> `/461f5c2a-d4ce-449a-b6af-d5819b2e4a2e/landscape_template` (**Observed**).
> Docs: Landscape Mode (Manage / Sculpt / Paint). Spreadsheet Keep=Yes. **No Verse.**

| | |
|---|---|
| Classification | **Landscape Mode lab** — heightmap + weight-blend layers + foliage + splines |
| Content type | small kit: **27 authored vs 67 placed** |

## Scale

| | |
|---|---|
| Authored | 27 (`LI_*` layer infos, `M_Landscape*`, foliage types, demo meshes) |
| OFPA | 67 (1 unclassified) |
| Verse | **0** |
| Landscapes | 2 + 7 `LandscapeStreamingProxy` |
| `InstancedFoliageActor` | 15 |
| `LandscapeSplineActor` | 1 |
| Demo displays | 13 |
| Compatibility | **42.30** |
| Matchmaking | 16 |

Empty `Developers/felixkupis/Collections/`. Twenty-fifth first-party hit.

Live folders (**Observed**): `Demo Display`, `HLOD`, `Landscape Assets`, `WorldSettings`. **Not** numbered halls like Modeling/Materials.

## What it is *for*

Walkable demos of Landscape Mode output: weight-blended Grass/Path/Sand/Stone, foliage instances (grass/shrub/tree), a spline (road/path), a tunnel hole, HLOD proxies. The **tutorial is the editor mode** (Manage → Sculpt → Paint); the island is the finished result, not a 1:1 tool exhibit.

## Device inventory (the design)

| Count | Class | Role |
|---|---|---|
| 15 | `InstancedFoliageActor` | grass / shrub / tree instances |
| 13 | `BP_DemoDisplay` | captions |
| 9 | `WorldPartitionHLOD` | landscape HLOD |
| 7 | `LandscapeStreamingProxy` | WP tiles |
| 2 | `Landscape` | the terrains |
| 1 | `LandscapeSplineActor` | path/road following terrain |
| 5 | `BlockingVolume` | bounds |
| rest | HUD / experience / spawner | boilerplate |

Authored kit: `Materials/Landscape_WeightBlends/LI_{Grass,Path,Sand,Stone}`, `MI_Landscape_WeightDemo_A`, `M_LandscapeGrass`, `M_Landscape_WeightDemo`, foliage types `SM_{Shrub,Tree}Demo_*_FoliageType`, `LGT_GrassDemo`, `SM_Tunnel`, `T_LandscapeLayerExample`.

## What this sample actually teaches

1. **Weight-blend layers, not one material.** Four `LI_*` + `M_Landscape_WeightDemo`. Paint weightmaps; layer order hides lower layers. See techniques/landscape-weight-blend-layers.md.
2. **Foliage is `InstancedFoliageActor` + FoliageType assets**, not StaticMesh spam.
3. **Splines for roads** (`Ctrl+Click` in Manage) — Ramp sculpt is one-shot; Splines stay editable.
4. **WP streaming proxies + HLOD** come with any real Landscape — 7 proxies, 9 HLOD actors on a tiny island.
5. **Visibility-brush caves are a gameplay hazard** (tutorial: players despawn, vehicles teleport). Island uses `SM_Tunnel` mesh instead.

## Official tutorial vs shipped

**External:** Changing Terrain with Landscape Mode

Single long page: How It Works (Manage / Sculpt / Paint) + every sculpt/paint tool. Spreadsheet `/en-us/uefn/landscape-mode-in-unreal-editor-for-fortnite`. `docsUrl` empty. Retrieved signed-in Chrome 2026-09-11.

This is the **mode manual**, not “walk this island.” Same class as Modeling Mode / Material Library hubs.

### Lines up

| Tutorial | Shipped |
|---|---|
| Manage / Sculpt / Paint stages | finished terrain showing all three |
| Paint weight layers (Sand example) | `LI_Grass / Path / Sand / Stone` |
| Splines for roads vs one-shot Ramp | 1 `LandscapeSplineActor` |
| WP grid / streaming proxies | 7 `LandscapeStreamingProxy` |
| Material assigned at Create New | `M_Landscape_WeightDemo` + MI |
| 2048×2048 vertex cap | not measurable from disk; train as authored limit |

### Does not line up

| Topic | Tutorial | Shipped |
|---|---|---|
| Numbered tool halls | — | **none** — Demo Display captions only |
| Visibility brush caves | warned against | `SM_Tunnel` mesh hole, not Visibility |
| Foliage | not on this page | 15 `InstancedFoliageActor` — extra, not contradicted |
| Edit Layers UI | documented | **Unverified live** — re-read Landscape panel |

Train on: weight-blend `LI_*` + foliage types + spline. Do not emit Visibility-brush caves. Do not treat this page as a Verse or device tutorial.

## Negative findings

- Scratch `Developers/` folder.
- Compatibility 42.30.
- 1 unclassified OFPA actor.

## Capture backlog

1. Live Landscape material layer names vs `LI_*`. *Open, low — names already match.*
2. Demo Display 13 captions. *Waive — tutorial is the syllabus.*

## Live notes

Level confirmed. Folders Demo Display / HLOD / Landscape Assets. Streaming off (not re-verified).
