---
id: digest-021
kind: digest
---

# brick_editor_sample — sample digest

> **Extractor provenance.** `tools/extract-template.py` (Pass A) + UnrealMCP on
> `localhost:8000` (Pass B). Default level
> `/a4e7c914-a242-4b92-8083-3cfd3231cc2f/brick_editor_sample` (**Observed**).
> Hub: Building BRICK Islands
> + discipline hub BRICK Brick Editor
> + island tutorial BRICK Brick Editor Template
> + Working with the BRICK Brick Editor
> + BRICK Brick Index (catalog, not this walkthrough)
> + Tips and Tricks of Building in 3D Space
> + companion Working with BRICK Islands
> + theme Rules (publishing, not a device syllabus).
> BRICK Templates index (prior same-day capture) **omits** this island — the tutorial lives under the Brick Editor hub.
> `docsUrl` empty. Not on the spreadsheet. Compat **43.00**. `featureSet`: **BrickGameplay**.

| | |
|---|---|
| Classification | **BRICK learning island** — viewport BrickAssembly lab (gallery bricks → Atom static meshes + modular building modules + popup theme cine) |
| Content type | **kit** (45 authored vs 178 OFPA; `/BRICKBasicBrickGallery_Assets` + `/BRICKAtomAssets` + 40 `Examples/SM_*`) |
| Verse | **0 files** |
| Fab `.uref` | **0** |

## What it is

The picker name is **BRICK Brick Editor Template** (BRICK). Docs H1 is **BRICK Brick Editor Template**. Disk folder is `brick_editor_sample`. The island is **four numbered teaching pods plus a playground and a PLAY cinematic podium**: snap gallery bricks in Brick Editor mode (`Shift+6`), BrickAssembly them into reusable Atom meshes, size wall/pillar/doorway/roof modules to the Fortnite building grid, then drop a build on the podium and pick Party vs Scary. Teaching that survives deletion is **BrickAssembly → `AtomGeneratedAsset` SM**, **modular BrickAssembly building modules**, and a **popup two-button theme cine** (no Verse). Do not duplicate: phone-copy-paste-prefab-quickbar.md / BRICK-fortnite-mix-no-scale.md (Phone on catalog `BrickSimple_*` / `SM_FORT_*`; **0** Phone here), walkable-modeling-curriculum.md (Modeling Mode halls), cubegrid-then-color-grids.md (**0** CubeGrid), playset_sample / modular-kitbash-blueprints.md (`Parent_Building*` kit; these modules stay `StaticMeshActor`), cartoon-comic-unbound-ppv.md (one Button per look cine), trigger-cinematic-quake.md (Trigger → one sequence), beacon-volume-popup-trail.md (visit trail). Assembly stays on action_adventure_sample (**29** devices) — Brick Editor meshes **cannot** feed Assembly. ExtraAssets HUD stays **four** — this HUD Controller has **null** overrides.

Docs: most of the experience is the **viewport**, not the launched island. PLAY is the podium.

## Scale (**Observed** — `summary.txt` + live)

| | |
|---|---|
| Authored assets | **45** (`Examples/` 40 `SM_*` + `M_Invisible` + 2 sequences + GameFeatureData / HLOD) |
| OFPA | **178** (151 classified + **27 unclassified** gallery bricks) |
| Maps | `brick_editor_sample.umap` (only) |
| Verse files | **0** (`VerseToolset.ListFiles` on `/a4e7c914-…` — folders only; do not recurse `__ExternalActors__`) |
| Verse devices placed | **0** |
| Data layers | **0** |
| Fab `.uref` | **0** |
| Compatibility | **43.00** |
| Matchmaking | **16** / live `teamSize=Dynamic` / `teamFillOption=MustFill` |
| Streaming | Partitioned + actor folders + OFPA. **`LevelHasStreamingDisabled=1`** (Walker/VFX/music-fantasy/Music/Obstacle/Home Builder camp). Sampled `bIsSpatiallyLoaded=false`. `load_actors` required. `find_actors` by `"SM_"` / `"StaticMesh"` / `"Popup"` / `"PostProcess"` returned **0** — labels are **PLATE 2X4** / **Pop-up Dialog Device** / **PP_***; cine needle `"Cinematic"` hits podium + camera, not `Device_CinematicSequence` (use `"Sequence"`). `folderPath` often **None**. Live `get_folders` returned the **full** numbered Outliner (unlike Music/Obstacle/Home Builder Background-only). |

Empty `Developers/felixkupis/Collections/`. First-party scratch folder.

`experimental.*` (Python / toolsets / Scene Graph) in `.uefnproject` is **Unverified**. Content-side: **no** `.verse`, **no** `__ExternalEntities__`. Plugin `EnableSceneGraph: true` with no Scene Graph entities.

Largest authored mesh well under GHES 100 MB (`SM_AsPeakRoof` ~9.7 MB). Whole folder ~**121 MB**.

## Level organisation (**Observed** — live `get_folders` + `get_actors_in_folder`)

```
1. Ducks/                         26  un-BrickAssembled gallery bricks + example duck
2. Hearts/                         2
3. Modular Fence/                  8  BrickAssembly lesson
4. Building Systems/              16  Den_* floor / wall / pillar / doorway
4. Building Systems/Castle/       34  CastlePillar / Window / Doorway / SpireRoof
4. Building Systems/Temple/       18  AsWall / AsPillar / AsDoorway / AsPeakRoof / AsRoofMid
5. Playground/                    24  sketches, miniatures, road cones
Cinematics/                       12  podium / button / popup / HUD
Cinematics/PartyCinematic/        25  stages, balloons, party animals, Creator VFX, PP
Cinematics/ScaryCinematic/         9  Spawner VFX, PP_Vignette, Scary DSD, skull
```

**2** `Device_Billboard_V2` (podium only). Syllabus is the **public Brick Editor Template page + 9 `BP_DemoDisplay` captions**, not a street of boards (`_raw/brick_editor_sample/billboard-texts.md`). Demo `title Text` / `description Text` are on the actor. Event graph: `_raw/brick_editor_sample/event-graph.md`.

## Device inventory (the design)

| Count | Class | Role |
|---|---|---|
| 89 | `StaticMeshActor` | BrickAssembled `SM_*` exhibits |
| 27 | native `Actor` (unclassified OFPA) | Gallery bricks (`PLATE 2X4`, `BRICK 2X10`, …) + `AtomCollisionComponent` / `AtomPrimitiveStaticMeshComponent` |
| 11 | `FortStaticMeshActor` | Example duck/heart/fence/podium (`SM_CinematicPodium`) |
| 9 | `BP_DemoDisplay` | Teaching-pod captions (construction-script signage) |
| 6 | `Device_VFXSpawner_V2` | Scary cine presets (SpookyGhosts / AreaOfFog / Bats; **Fireworks** live `Bubbles`) |
| 4 | `Artemis_Concert_StageBase_03_Celebration` | Party stage |
| 3 | `Device_VFX_Creator_V2` | Party sprites (Bubble / Heart / Triangle-as-Confetti) |
| 3 | `Device_PostProcess` | `PP_ActionLines` / `PP_HappyPlace` / `PP_Vignette`; spawn DisableForAll |
| 2 | `Device_Billboard_V2` | “YOUR BUILD HERE!” / launch arrow |
| 2 | `Device_CinematicSequence` | PartySequence / ScarySequence |
| 2 | `BP_Sign2` | Podium arrows |
| 2 | `Athena_PartyBalloons` | Party dressing |
| 1 | `Device_ExperienceSettings_V2` | **Not** BRICK. `allowBuilding=All`, `enableJam=Default` |
| 1 | `Device_HUDController_V2` | HideElement HUD/minimap/build/inventory; overrides **null**; Enable/Disable around cine |
| 1 | `BP_DaySequenceModifier_Device` | Scary only: priority **2000**, FixedTime **21.5**, `sunlightIntensity=11.76`, `use Volume=false`, `enableDuringPhase=None` |
| 1 | `BP_BRICK_Player_Spawner_Prop` | `useAsIslandStart=true` |
| 1 | `Device_Button_V2` | `Play Cinematic` → popup Show |
| 1 | `Device_PopUpDialog_V2` | Choose Theme: Party / Scary / Cancel |
| 1 | `Device_Barrier_V2_Placed` | Podium rail |
| 1 | `CineCameraActor` | Sequencer camera |
| 0 | BRICK Experience Settings / Assembly / Collectible / Tracker / Phone / CubeGrid / Playset / VerseDevice / Conversation | Action Adventure / Home Builder / Modelling / fort_playset |

Extractor histogram named spawners `BP_Creative_Player_Spawner_Prop`; live class is **BRICK**. Use live. Island Settings live class is **`Device_ExperienceSettings_V2`**, not `Device_ExperienceSettings_BRICK`.

Dressing: `/BRICKBasicBrickGallery_Assets` 366, `/BRICKAtomAssets` 120, `/Musical_Walk_Assets` 8 (Mascot/Frog/Snail party animals), `/ENV_PACK_UnplayableTestMap` 4 (skull box), `/BRICK_Cane` 3.

## Verse (0 files)

No `snippets/verse/` folder. Valid. Playground DemoDisplay C6 invites Verse / sequencer; this island ships the **device + sequence** half only.

## Authored kit (**Observed** — `assets.txt` + live)

```
Content/Examples/SM_{Duck,Heart,Fence,GroundPlate,RoadCone,CinematicPodium}.uasset
Content/Examples/SM_Den*            floor / wall / pillar / doorway / gable / roof
Content/Examples/SM_Castle*         pillar / window / doorway / spire roof
Content/Examples/SM_As*             temple wall / pillar / doorway / peak + mid roof
Content/Examples/SM_Med*            ornamental roof / pillar
Content/Examples/SM_MiniBuilding*   playground miniatures
Content/Examples/SM_Sketch{1,2,3}   brick sketches
Content/Materials/M_Invisible.uasset
Content/Sequences/{Party,Scary}Sequence.uasset
```

Sampled `SM_*` (`get_asset_tags`): `AtomGeneratedAsset=True`, three `/BRICKAtomAssets` LUT materials (`MI_Primitive_Standard_VC_LUT_FakeBevel` + `_Underside` + `_CP`), `CollisionComplexity=CTF_UseSimpleAsComplex`, Nanite + 6 LODs, 2 UV channels.

**Tier 2 mounts:** `/a4e7c914-…` 549, `/BRICKBasicBrickGallery_Assets` 366, `/Game` 213, `/BRICKAtomAssets` 120. No Fab `.uref`.

## What this sample actually teaches

1. **Gallery bricks BrickAssembly into an Atom-generated static mesh — not Phone, not Modeling Mode.** Native `Actor` + Atom collision/mesh → `Examples/SM_*` with `AtomGeneratedAsset=True`. Reuse the SM; do not scale; recolor = Separate → re-BrickAssembly. techniques/BrickAssembly-atom-static-mesh.md.
2. **Build wall / pillar / doorway / roof modules sized to the Fortnite building grid, then BrickAssembly copies.** Pod 4 + Castle / Temple are the worked examples. Not a `Parent_Building*` harvest kit. techniques/modular-BrickAssembly-building-modules.md.
3. **One PLAY button opens a popup; Party / Scary each Play a Cinematic Sequence.** Cancel is unbound. HUD Enable on pick, Disable on `OnStopped`. techniques/popup-theme-cinematic.md.
4. **`allowBuilding=All` + `enableJam=Default` on `ExperienceSettings_V2`.** Other extracted BrickGameplay learning islands use BRICK settings + `allowBuilding=None` / `enableJam=Off`. This is the viewport-build exception.
5. **HUD Controller is HideElement + null overrides**, same BRICK-null family as Music Concert / Obstacle Course, plus cine Enable/Disable. ExtraAssets stays **four**. Action Adventure in-project `WBP_BRICK_PlayerInfoStack` is **not** this.
6. **Scary DSD is 2000-only FixedTime 21.5 and `enableDuringPhase=None`** — cine dressing, not an always-on island clock. `timeOfDayOverride=NoOverride`. Not a 2000-vs-higher volume stack.
7. **9 DemoDisplay boards + 2 podium billboards are the in-island editor syllabus.** Tutorial prose is the four challenges + playground + podium; Tips page owns rotation hotkeys (Template page has none).
8. **BrickAssembled meshes cannot feed an Assembly device** (static mesh ≠ geometry collection). Test Assembly on action_adventure_sample.

## Official tutorial vs shipped

**External** hub + Brick Editor discipline hub + island tutorial + Working with the Brick Editor + Brick Index + Tips + Working with BRICK Islands + theme Rules (headed Chrome 2026-09-15). Verbatim: `_raw/brick_editor_sample/tutorial.txt`. Spreadsheet: no row. `docsUrl` empty. Live slug `BRICK-brick-editor-template-in-fortnite` (H1 **BRICK Brick Editor Template**). BRICK Templates index (same-day prior capture) listed Home Builder / Action Adventure / Music / Obstacle / PvP Extraction / GiftGiver's / Scary Space — **not** this tile.

### Lines up

| Tutorial | Shipped |
|---|---|
| Picker / hub **BRICK Brick Editor Template**; viewport more than PLAY | picker **BRICK Brick Editor Template**; 4 numbered pods |
| Content Browser `Examples/`; BrickAssembly writes there | 40 `SM_*` |
| Challenge 1 Duck / 2 Hearts / 3 Modular Fence / 4 Building Modules | DemoDisplay titles match; folders `1. Ducks` … `4. Building Systems` |
| Fence: Separate → one bay → BrickAssembly → duplicate | DemoDisplay C copy; live `SM_Fence` `AtomGeneratedAsset=True` |
| Modules like UEFN building actors; temple + castle from modules | Castle/Temple folders; DemoDisplay C3 |
| Playground miniatures / sketches; involve devices / sequences / Verse | `SM_MiniBuilding*` / `SM_Sketch*`; cine shipped, **0** Verse |
| Cinematic podium: place build, launch, Party or Scary | 2 boards + popup Party/Scary/Cancel |
| `Shift+6`; Snap / BrickAssembly / Separate; no custom bricks / no scale | Working-with page; live gallery `Actor`s |
| Assembly Device incompatible | **0** Assembly |
| BrickAssembled SM can be Building Prop / Building Static Mesh | FAQ; this island leaves them as mesh actors |

### Does not line up

| Tutorial | Shipped |
|---|---|
| BRICK Templates index lists this island | index **omits** Brick Editor (lives under Brick Editor hub) |
| “For rotation see the BRICK Brick Editor Template document” (Tips) | Template page has **no** hotkey table — hotkeys live on Tips |
| Playground “even Verse scripts” as shipped | **0** `.verse` |
| Fireworks VFX | live `visual_Effect=Bubbles` (label drift) |
| Extractor class `BP_Creative_Player_Spawner_Prop` / BRICK Experience Settings | live BRICK spawner + **`ExperienceSettings_V2`** |
| `find_actors` `"Popup"` / `"PostProcess"` / `"Cinematic"` | labels **Pop-up Dialog Device** / **PP_***; cine class needs `"Sequence"` |
| Always-on BRICK DSD (other BRICK packets) | only Scary DSD, phase **None** |

Training rule: shipped bindings win. Tutorial *prose* is the syllabus (plus DemoDisplay). Do not emit Home Builder Phone/mix, Modeling Mode halls, CubeGrid, fort_playset `Parent_Building*`, Assembly, ExtraAssets HUD, City one-Button-per-look, or Verse as this island’s setup. Brick Index is a **catalog**, not this walkthrough. Modeling / Greyboxing / fort_playset hubs are other packets.

## Negative findings

- Picker **BRICK Brick Editor Template** / docs **BRICK Brick Editor Template** / folder `brick_editor_sample` are the same tile — not Home Builder, not Action Adventure, not MartialArts, not Brick **Grid** (skipped blank).
- Name does **not** mean player-build Phone, Modeling Mode brick meshes, or Assembly puzzles.
- **0** Assembly / Collectible / Tracker / Prop Mover / Patchwork / Conversation / VerseDevice.
- Gallery prefabs cannot be un-BrickAssembled in Brick Editor mode (Tips) — teaching pods 1–2 are loose gallery `Actor`s on purpose.
- `get_properties` fails the whole call if any name is missing (`useVolume` is **`use Volume`**; popup `button 1 Text` is **`button1Text`**; HUD `equippedItemWidgetOverride` is **`equippedItemInfoWidgetOverride`**).
- `ListEventBindings` on the Button is **[]** — Show lives on the popup; HUD Enable lives on the controller.
- Sequencer tracks that Enable PP / VFX are **unreachable**. Bindings + phase flags are Observed.
- Do not flip `experimental.*`. Do not treat Musical Walk party animals as Patchwork (`enableJam=Default` is not a Jam Stage).

## Capture backlog

Waived: Sequencer keyframe times / which tracks Enable VFX and PP; per-brick gallery path beyond sampled `PLATE 2X4`; which FortStaticMeshActor is the example duck vs heart; Building Prop actor class on a BrickAssembled SM (FAQ only — not placed). `experimental.*` Unverified.

## Live notes

- Cursor `user-unreal-mcp` was down; `tools/mcp.py` on `:8000` worked.
- Mount `/a4e7c914-a242-4b92-8083-3cfd3231cc2f` from `ListFiles("")`. Do not recurse `ListFiles("")`.
- Billboard / DemoDisplay text is on the **actor**.
- Wrong `refPath` `/mount/map.ActorName` is invalid; use `actorPath`.
- Did not flip experimental flags. Did not switch maps.
