---
id: digest-022
kind: digest
---

# brick_extraction_sample — sample digest

> **Extractor provenance.** `tools/extract-template.py` (Pass A) + UnrealMCP on
> `localhost:8000` (Pass B). Default level
> `/1e36a9cd-e8ad-4c56-9deb-c5b476976b43/Extraction` (**Observed**).
> Hub: Building BRICK Islands
> + index BRICK Templates
> + island tutorial BRICK PvP Extraction Template
> + companion Working with BRICK Islands
> + theme Rules (publishing, not a device syllabus)
> + BRICK Brick Editor (how the modules were made).
> `docsUrl` empty. Not on the spreadsheet. Compat **43.00**. `featureSet`: **BrickGameplay**.

| | |
|---|---|
| Classification | **BRICK learning island** — 8-player FFA extraction loop (lobby bank + level-gated rooms + elevator extract) |
| Content type | **kit + wrap** (331 authored vs 2830 OFPA; BrickAssembly greybox/themed modules + extraction props) |
| Verse | **15 files** (7 devices placed) |
| Fab `.uref` | **0** |

## What it is

The picker name is **BRICK PvP Extraction Template** (BRICK). Docs H1 is **BRICK PvP Extraction Template**. Disk folder is `brick_extraction_sample`. The live map is **`Extraction`**. Teaching that survives deletion is **persist banked studs + lobby 100/10/1 grants**, **Barrier ignore-list by Stat level**, **XP via `<concrete>` device interfaces**, and a **Button→Timer→Cinematic extract elevator**. Do not duplicate: stud-itemspawner-teleport-pool.md / destructible-prop-validity-poll.md / tagged-collectible-stud-granter.md (pool + destructible copies; **0** BRICK Collectibles; granter file unused), persistable-quest-line-verse-dialog.md (UUID fortplayer persist is not a quest log), sequential-weapon-grant.md / persistable-arena-rotation.md (lump-sum lobby granters, **no** Await; Close/Long/Mid/Town arenas are meshes), sequential-capture-unlock.md / weighted-capture-spawns.md (Barrier ignore-list, not capture areas). ExtraAssets HUD stays **four** — HUD Controller widget overrides are **null**; Stat Creators own `WBP_PlayerStuds` / `WBP_BankedStuds` / `WBP_PlayerLevelStat`.

## Scale (**Observed** — `summary.txt` + live)

| | |
|---|---|
| Authored assets | **331** (`BRICKModels/` 257, `Cinematics/` 31, `Verse/` 15, `UI/` 11, `VFX/` 11, `NPCSpawners/` 10) |
| OFPA | **2830** |
| Maps | `Content/Extraction.umap` (only; 14,591 bytes) |
| Verse files | **15** |
| Verse devices placed | **7** (folder hist 6 + unfoldered goldstuds) |
| Data layers | **0** |
| Fab `.uref` | **0** |
| Compatibility | **43.00** |
| Matchmaking | **8** / `maxTeamSize=1` / `maxTeamCount=8` / JIP **true**. Live BRICK: `maxPlayers=8`, `teamSize=Solos`, `teams=TeamIndex 8`, `teamFillOption=MustFill` |
| Streaming | Partitioned + actor folders + OFPA. Umap **omits** `LevelHasStreamingDisabled` (IceBase/Social/Action Adventure camp). Sampled `bIsSpatiallyLoaded=false` (2812). `load_actors` required. `find_actors` without `collision_channels` errors; with it, name `"Gold"` finds goldstuds. `folderPath` often **None**. `get_folders` returned the **full** Outliner (Lobby / Map arenas / Demo Area / Emergence / Extract). |

`Developers/felixkupis/` exists (empty Collections). First-party scratch folder.

`experimental.*` (Python / toolsets / Scene Graph) in `.uefnproject` is **Unverified**. Content-side: 15 `.verse`, no `__ExternalEntities__`.

Largest authored file ~**6.3 MB** (`BRICK_ExtractTunnel_Entrance_01.uasset`). Whole folder ~**250 MB**. Under the GHES **100 MB** per-file cap.

## Level organisation (**Observed** — live `get_folders` + folder counts)

```
Lobby/
  Intro                         7
  Main Lobby Room/Functional    6  (+ StarterWeapons 1, UnlockZones 1, teles/spawners 17)
  Vending Machines              5
  Visual                      266  (+ Shop 16)
Map/
  Arenas/Close Range Arena    233
  Arenas/Long Range Arena     397
  Arenas/Mid Range Arena      499
  Arenas/Town Arena           419
  Destructible Props           32   archery targets
  NPC Spawners                  9
  Terrain                       3  (+ Background 16)
Emergence Point               265
Extract Point                 158
Demo Area/
  Functionality               137
  Tutorial                     30
  Brick Editor Assets/Brickbuilt Modules 123
  Greybox Sample House         63
  Themed Sample House          93
Settings                        8   DSD + HUD (Island Settings actor is unfoldered)
Island Settings                 0   empty folder
```

Bang-prefix `! Demo Area` / `! Map` / `! Settings` folders exist and count **0** (editor leftover). 10 billboards + 11 DemoDisplays are the syllabus (`_raw/brick_extraction_sample/billboard-texts.md`). Event graph: `_raw/brick_extraction_sample/event-graph.md`.

## Device inventory (the design)

| Count | Class | Role |
|---|---|---|
| 1572 | `FortStaticMeshActor` | BrickAssembly modules / extraction props |
| 851 | `StaticMeshActor` | Extra meshes |
| 40 | `Device_BRICKItemSpawner` | Stud toss/gravity pool (`Item Spawner_BRICKStud_*`) |
| 36 | `BP_Sub_ArcheryTarget` | Destructible props |
| 32 | `Device_CinematicSequence` | Tunnel open/close, vault, emergence |
| 23 | `Device_CRD_AudioPlayer` | Extract siren / elevator / deposit |
| 22 | `Device_Timer` | GoingUp **1s**, AutoClose **30s**, GoingDown **5s** (+ copies) |
| 21 | `Device_Teleporter` | Lobby stack + extract landing |
| 17 | `Device_CharacterSpawner` | NPC Spawner 1–8 + Environmental_NPC_* |
| 16 | `Device_Trigger` | Lobby enter/exit, deposit, XP, tutorial |
| 14 | `Device_CameraModes_FixedPoint` | Emergence / extract / lobby cams |
| 14 | `Device_MutatorVolume` | Extract tunnel + emergence + lobby tools |
| 12 | `Device_Barrier` | Lobby UnlockZone_2 + emergence shafts |
| 11 | `BP_DemoDisplay` | Demo pods (`title Text` / `description Text[N]`) |
| 11 | `Device_Button` | Extract call/descend + deposit + tutorial |
| 10 | `Device_Billboard` | START HERE / extract / level 2 / teles |
| 8 | `BP_BRICK_Player_Spawner_Prop` | All `useAsIslandStart=true`, Always |
| **7** | `VerseDevice_C` | goldstuds (unfoldered) + unlock + LevelManager + fortplayer + stud pool + stud spawner + destructible |
| 4 | `Device_StatCreator` | GoldStuds / Banked / Banked VisualsOnly / PlayerLevel |
| 4 | `Device_VendingMachine` | Lobby shop |
| 3 | `Device_Powerup_Stat` | XP Small / Medium / Large |
| 2 | `Device_Tracker` | AllChests → CycleToRandomItem; EliminateBotPlayers |
| 2 | `BRICK_Creative_Tiered_Chest` | Loot (tracker, not chest inventory) |
| 1 | `Device_ExperienceSettings_BRICK` | unfoldered `IslandSettings` |
| 1 | `Device_HUDController` | HideElement minimap/team/build; widgets **null** |
| 1 | `BP_DaySequenceModifier_Device` | Priority **1**, FixedTime **24**, sun **0.96**, fog **0.01** |
| 0 | BRICK Collectible / Assembly / Phone / Patchwork / Conversation / CubeGrid / Playset / End Game | Action Adventure / Home Builder / Music / Obstacle |

Folder hist `VerseDevice_C: 6` missed goldstuds. OFPA extract said 7. Island Settings not in the Settings folder.

## Verse (15 files)

Copied to `snippets/verse/` (flat). Modules: `BRICKExtraction` (Core / Debug / Lobby), `BRICKStudRewards`, `BRICKUtilities`.

| File | Role |
|---|---|
| `modules.verse` | module tree |
| `BRICK_goldstuds_manager.verse` | bank / HUD swap / 100-10-1 grant |
| `BRICK_lobby_room_unlock_manager.verse` | Barrier ignore-list |
| `BRICK_xp_level_manager.verse` | concrete XP interfaces |
| `BRICK_fortplayer_extraction.verse` | `InLobby` flag |
| `BRICK_fortplayer_manager_extraction.verse` | extraction fortplayer factory |
| `BRICK_fortplayer.verse` / `_manager.verse` / `_persistence_data.verse` | UUID persist wrapper |
| `BRICK_utilities.verse` | helpers |
| `BRICK_stud_spawner.verse` / `_pool.verse` | Action Adventure copy |
| `BRICK_destructible_prop_manager.verse` / `_definition.verse` | Action Adventure copy |
| `BRICK_collectibe_stud_granter.verse` | filename typo collectibe; **unplaced** |

Placed labels: Lobby unlock zone manager; BRICKModule StudRewards-BRICK destructible manager; LevelManager; BRICKModule StudRewards-BRICK stud pool; Fortplayer manager; BRICKModule StudRewards-BRICK stud spawner device; **BRICKExtraction-goldstuds management device**.

## What this sample actually teaches

1. **Bank carried studs into a persist Stat, then grant lobby lumps 100/10/1.** ConditionalButton count + ItemRemover + VisualsOnly HUD. techniques/banked-studs-extract-loop.md.
2. **Gate lobby rooms with Barrier `AddToIgnoreList` from Stat level**, not Disable. techniques/lobby-barrier-ignorelist-by-level.md.
3. **Grant XP through `<concrete>` device interfaces**, then a Powerup Stat Pickup. Extract XP only if studs > 0. techniques/xp-via-concrete-device-interfaces.md.
4. **Extract elevator is Button → Timer → Cinematic Sequence.** GoingUp **1s**, AutoClose **30s**, GoingDown **5s**. techniques/elevator-extract-timer-cinematic.md.
5. **Reuse Action Adventure stud pool + destructible poll** (2nd source). Link those notes; do not copy the files as new techniques.
6. **`allowBuilding=None` + `enableJam=Off` on BRICK Experience Settings.** HUD Controller HideElement minimap/team/build with **null** widgets (4th BRICK HideElement source). Stats carry the custom WBPs.
7. **DSD is priority 1, FixedTime 24** (`use Volume=false`) — not a 2000-vs-higher stack, not MartialArts’s 2500.
8. **Chests are empty.** Tracker_AllChests When Complete → Item Granter Weapons CycleToRandomItem → Reset Progress.

## Official tutorial vs shipped

**External** dedicated chapter (headed Chrome 2026-09-16). Verbatim notes: `_raw/brick_extraction_sample/tutorial.txt`. Spreadsheet: no row. `docsUrl` empty. Guessed slug `BRICK-pvp-extraction-template-in-fortnite` is wrong; live is `…-in-unreal-editor-for-fortnite`.

### Lines up

| Tutorial | Shipped |
|---|---|
| Picker / H1 PvP Extraction; 8-player FFA | picker **BRICK PvP Extraction Template**; live `maxPlayers=8`, `teamSize=Solos`, 8 teams of 1 |
| Lobby unlock manager + Barrier ignore-list | `BRICK_lobby_room_unlock_manager`; UnlockZone_2 `invisibleToIgnoredPlayers=true` |
| `BRICK_xp_level_manager` + Stat_PlayerLevel | LevelManager + persist Stat, `maxLevel=100` |
| `BRICK_goldstuds_manager` + persist banked studs | placed (unfoldered); Stat_BankedGoldStuds `usePersistence=true`, HUD No |
| GoingUp delay → CS_Extraction_TunnelOpen; AutoClose; Health Powerup on lobby teleporter | Timer 1s → CS open; AutoClose **30s**; Trigger_OnEnterLobby → Health Powerup Pickup |
| Same destructible Verse as Action Adventure | 1 manager + 32 archery targets |
| Deposit demo **not used** as the live loop | extract banks; Button_Deposit still placed/wired |
| 8 spawners; On Player Spawned grant | 8 BRICK spawners all `useAsIslandStart=true`; ItemGranter_StarterWeapon_OnGameStart |
| Chests via Tracker + Item Granter | Tracker_AllChests → CycleToRandomItem |
| Brickbuilt modules / demo pods | Demo Area Brick Editor Assets + 11 DemoDisplays |

### Does not

| Tutorial | Shipped |
|---|---|
| StatBankedGoldStuds / StatBankedGoldStuds_VisualsOnly | `Stat_BankedGoldStuds` / `Stat_BankedGoldStuds_VisualsOnly` (underscore after Stat) |
| `BRICK_destuctible_prop_manager` | `BRICK_destructible_prop_manager` |
| `Button_Extract_Elevator` | `Button_ExtractPoint_Elevator_GoingUp` / `_GoingDown` |
| Team Size: 1 / Total Rounds 1 as Island Settings keys | live `teamSize=Solos`; Total Rounds not in the BRICK keys we could read |
| Goldstuds manager listed among the obvious Outliner Verse devices | unfoldered; folder hist 6 until `find_actors` name Gold |
| VisualsOnly “shown when non-zero” | live HUD **Yes** (not Non-Zero); GoldStuds is the Non-Zero one |
| Bloom Tycoon as the fortplayer origin story | extraction subclass adds `InLobby` only |

Training rule: snippets win.

## Negative findings

- Name “PvP Extraction” is not Gunfight. Four arenas are BrickAssembly set-dress; no persistable rotation, no sequential Await loadout.
- `BRICK_collectibe_stud_granter.verse` ships; **0** BRICK Collectibles placed.
- UnlockZone_2 Details **On Level Up → Disable** fights the per-player ignore-list. Disable is global.
- GoingDown delay timer is also named “Extract Point Closing” (same as the 30s AutoClose).
- DemoDisplay_DespositGoldStuds — shipped typo Desposit.
- Island Settings folder is empty; the BRICK actor is unfoldered. Settings folder is DSD+HUD.
- Folder-walk Verse count 6 is incomplete; goldstuds `folderPath` None.
- DSD priority **1** (not 2000). Do not copy other BRICK “always 2000” assumptions.
- `experimental.*` Unverified. Do not flip flags.

## Capture backlog

| Item | Status |
|---|---|
| Default level `/…/Extraction` | **Observed** |
| Goldstuds unfoldered VerseDevice | **Observed** (`find_actors` name Gold) |
| UnlockZones array UnlockLevel values | **Waived** — GetDeviceProperties subobject trap |
| BRICK Island Settings Total Rounds | **Waived** — key not in the readable set |
| experimental flags | **Unverified** — do not toggle |
| Which fortplayer class the placed device compiles to | **Inferred** extraction subclass (`InLobby` / GetExtractionPlayer) |

## Live notes

- Cursor `user-unreal-mcp` namespace was down; `python tools/mcp.py` on `:8000` worked. SceneTools `get_current_level`; DeviceToolset `ListDeviceProperties`. `GetDeviceEventBindings` is DeviceToolset (not a `verse` toolset).
- `find_actors` requires `collision_channels` (schema drift). Pass `[]`.
- Billboard `text` on actor after `load_actors`. DemoDisplay `title Text` / `description Text[N]`.
- Verse live class is `VerseDevice_C` at `/CRD_VerseDevices/VerseDevice.VerseDevice_C` (not ScriptDevice).
- Mount is `/1e36a9cd-e8ad-4c56-9deb-c5b476976b43`, **not** `/Game`. Plugin GUID matches `.uefnproject` `projectId` `71ba1211-…`.
- HUD first-pass keys `hideMinimap` failed; live names `showMinimap` / `showHealth`. Stat keys `statBarShownOnHUD` / `enabled Phase` / `usePersistence`.
- Default map was not switched.
