---
id: digest-023
kind: digest
---

# home_builder_sample — sample digest

> **Extractor provenance.** `tools/extract-template.py` (Pass A) + UnrealMCP on
> `localhost:8000` (Pass B). Default level
> `/9dbbf9de-6efd-47dc-9191-ccdf575cc87f/home_builder_sample` (**Observed**).
> Hub: Building BRICK Islands
> + index BRICK Templates
> + island tutorial Home Builder
> + companion Working with BRICK Islands
> + theme Rules (publishing, not a device syllabus).
> Discipline hub (not this island): Using Prefabs and Galleries. Build Your First Island / Phone videos are Creative onboarding.
> `docsUrl` empty. Not on the spreadsheet. Compat **43.00**. `featureSet`: **BrickGameplay**.

| | |
|---|---|
| Classification | **BRICK learning island** — Create-mode Phone / prefab sandbox neighborhood, not a minigame |
| Content type | **catalog wrap** (2 authored vs 1647 OFPA; `/NEIGHBORHOOD_Assets` + `BrickSimple_*` + `SM_FORT_*`) |
| Verse | **0 files** |
| Fab `.uref` | **0** |

## What it is

The picker name is **Home Builder** (BRICK). Docs H1 is **Home Builder**. Disk folder is `home_builder_sample`. The island is a **Minifigure-scale NEIGHBORHOOD street** of mixed Fortnite houses and BRICK brick walls, with **15 billboards** teaching the Phone tool, Content tab, Prefab/Gallery, and an empty lot. Teaching that survives deletion is **Phone copy/paste/delete + prefab Quick Bar** and **BRICK+Fortnite mix with no-scale / no-overlap**. Do not duplicate: homestead-claim-spawn.md (**0** Verse; all 8 BRICK spawners `useAsIslandStart=true`), playset_sample (**0** PlaysetRoot; sampled `bIsPlayerBuildable=false`), cubegrid-then-color-grids.md (**0** CubeGrid), beacon-volume-popup-trail.md / patchwork-switch-stems.md (**0** Beacon / Patchwork), collectible-tracker-endgame.md / prop-mover-device-wiring.md (**0** Tracker / Prop Mover). ExtraAssets HUD stays **four** — this island has **0** HUD Controller.

## Scale (**Observed** — `summary.txt` + live)

| | |
|---|---|
| Authored assets | **2** (`GameFeatureData`, `home_builder_sample_DefaultHLODLayer`) |
| OFPA | **1647** |
| Maps | `home_builder_sample.umap` (only) |
| Verse files | **0** (`VerseToolset.ListFiles` on `/9dbbf9de-…` — folders only; do not recurse `__ExternalActors__`) |
| Verse devices placed | **0** |
| Data layers | **0** |
| Fab `.uref` | **0** |
| Compatibility | **43.00** |
| Matchmaking | **16** / live `teamSize=Dynamic` / `teamFillOption=MustFill` (`.uefnproject` + Island Settings) |
| Streaming | Partitioned + actor folders + OFPA. **`LevelHasStreamingDisabled=1`** (Walker/VFX/music-fantasy/Music Concert/Obstacle camp). Sampled descriptors `bIsSpatiallyLoaded=false`. `load_actors` required for billboard `text` and device values. `find_actors` by `"Device_"` / `"Point Light"` returned **0** — lights are labeled **Customizable Light**; use `"Light"` + class `Device_PointLight_V2_C`. `folderPath` often **None**. Live `get_folders` returned only `Background` (16 landscape actors). |

Empty `Developers/felixkupis/Collections/`. First-party scratch folder.

`experimental.*` (Python / toolsets / Scene Graph) in `.uefnproject` is **Unverified**. Content-side: **no** `.verse`, **no** `__ExternalEntities__`. Plugin `EnableSceneGraph: true` with no Scene Graph entities.

Largest file well under GHES 100 MB. Whole folder ~**14.3 MB**.

## Level organisation (**Observed** — live `get_folders` + finds)

```
Background/     16 (landscape — only folder path the editor returned)
(unfoldered)    devices + houses; folderPath often None
```

**15** `Device_Billboard_V2`. Syllabus is both the public Home Builder page **and** the boards (`_raw/home_builder_sample/billboard-texts.md`). `text` is on the **actor**. Event graph: `_raw/home_builder_sample/event-graph.md` (empty).

## Device inventory (the design)

| Count | Class | Role |
|---|---|---|
| 33 | `Device_PointLight_V2` | Interior fill; labels **Customizable Light**; 32 Spot / 1 Point (`Customizable Light3`, intensity **10**, shadows **on**); most intensity **2–4%**, shadows off, Always |
| 15 | `Device_Billboard_V2` | Create-only street syllabus (Phone, Content, Prefab, empty lot, Goo/Snowball) |
| 8 | `BP_BRICK_Player_Spawner_Prop` | All `useAsIslandStart=true`, `enabledDuringPhase=Always` |
| 2 | `Device_BRICKItemSpawner` | PLAY weapons; `items Respawn=true`, 5s first / 5s between; OFPA `WID_BRICK_GooGun` + `WID_Launcher_Snowball_BRICK` |
| 1 | `Device_ExperienceSettings_BRICK` | `allowBuilding=None`, `enableJam=Off`, infinite resources, `environmentDamagePreset=All` |
| 1 | `BP_DaySequenceModifier_Device` | Priority **2000**, FixedTime **7**, `sunlightIntensity=8`, `fogDensity=0.05`, `use Volume=false` |
| 3 | `Prop_Athena_Doorbell_Interactable` | Catalog doorbells, not devices |
| 0 | HUD / Tracker / Collectible / Prop Mover / Beacon / Fabric / CubeGrid / Primitive / Playset / VerseDevice | Obstacle obby, Music Concert Patchwork, Village homestead, fort_playset kit |

Extractor histogram named spawners `BP_Creative_Player_Spawner_Prop`; live class is **BRICK**. Use live.

Dressing: `/NEIGHBORHOOD_Assets`, `/CR_Legacy`, `/Game` Fortnite houses, `BrickSimple_*`, `SM_FORT_*` tiles.

## Verse (0 files)

No `snippets/verse/` folder. Valid.

## Authored kit (**Observed** — `assets.txt` + live)

```
Content/GameFeatureData.uasset
Content/home_builder_sample_DefaultHLODLayer.uasset
```

**Tier 2 mounts:** `/Game` 3395, `/9dbbf9de-…` 3303, `/NEIGHBORHOOD_Assets` 998, `/CR_Legacy` 908, `/CreativeCoreDevices` 85, `/CRD_PointLight` 67. No Fab `.uref`.

## What this sample actually teaches

1. **Create-mode building is Phone copy/paste/delete + Content tab prefab → Quick Bar on an empty lot.** Not player-build, not CubeGrid. techniques/phone-copy-paste-prefab-quickbar.md.
2. **Mix BRICK brick walls with Fortnite house tiles at 0.5 UU; do not scale BRICK or Fortnite walls/floors/roofs/stairs.** Sampled pieces `bIsPlayerBuildable=false`. techniques/BRICK-fortnite-mix-no-scale.md.
3. **`allowBuilding=None` still allows Phone placement.** Infinite resources on; PLAY player-build off. Docs: not a mini-game.
4. **PLAY destruction sandbox is two BRICK Item Spawners (Goo Gun + Snowball) with `items Respawn`.** Boards 13/14; live `spawn Item` struct did not expose the class — OFPA strings + boards. `environmentDamagePreset=All`, `weaponDestructionPercentage=100`. Tutorial never names the weapons.
5. **Interior lighting is 33 Customizable Lights (Creative Point Light device), mostly Spot 2–4%, no shadows.** One Point with shadows (`Customizable Light3`). Not environment-lighting-rig.md.
6. **DSD is 2000-only FixedTime 7, not a volume stack.** `overrideSunlight=true`, `sunlightIntensity=8`, `overrideFog=true`, `fogDensity=0.05`.
7. **15 billboards are the in-island editor syllabus** (Phone, DELETE, rotate, Content/Prefab/Gallery, mix categories, empty lot, Goo/Snowball). Tutorial prose is the sandbox walkthrough; boards name the hotkeys.
8. **Eight BRICK spawners are all island starts** — not Village homestead claim.

## Official tutorial vs shipped

**External** hub + BRICK Templates index + Home Builder + Working with BRICK Islands + theme Rules (headed Chrome 2026-09-15). Verbatim: `_raw/home_builder_sample/tutorial.txt`. Spreadsheet: no row. `docsUrl` empty. Live slug `build-BRICK-home-builder-in-fortnite` (H1 **Home Builder**; not a hub redirect). Guessed `build-your-own-BRICK-home-in-fortnite-creative` was not the live URL.

### Lines up

| Tutorial | Shipped |
|---|---|
| Picker / templates index **Home Builder**; lodging / sandbox, not a mini-game | picker **Home Builder**; **0** Tracker / End Game / Collectible |
| Minifigure-scale Fortnite homes mixed with BRICK props | `BrickSimple_*` + `SM_FORT_*` + `/NEIGHBORHOOD_Assets` + Athena furniture |
| Billboards teach Phone copy / paste / delete / rotate; Content tab; Prefab/Gallery; empty lot; Quick Bar | 15 boards; texts match |
| Mix BRICK + Fortnite or BRICK-only; no BRICK overlap/scale; Fortnite walls/floors/roofs/stairs no scale | Billboard6/11; sampled tiles not player-buildable |
| Some Creative devices unavailable; Drop hidden | `featureSet=BrickGameplay`; Experience Settings BRICK |
| Island Settings / time of day | DSD FixedTime **7** + `timeOfDayOverride=Hour8` |
| Next learning (Obstacle Course page pointed here) | theme queue after Obstacle Course |

### Does not line up

| Tutorial | Shipped |
|---|---|
| No Goo Gun / Snowball / Item Spawner | live **2** BRICK Item Spawners + boards 13/14; OFPA `WID_BRICK_GooGun` / `WID_Launcher_Snowball_BRICK` |
| “Play” the neighborhood / enter homes | `allowBuilding=None`; Phone is CREATE; PLAY is destruction/respawn per boards |
| Widget / HUD / lighting-device walkthrough | **0** HUD Controller; lights are Customizable Light, not in the article |
| Reset Island / Create Backup as shipped devices | editor Overview commands, not placed devices |
| Extractor class `BP_Creative_Player_Spawner_Prop` | live `BP_BRICK_Player_Spawner_Prop` |

Training rule: shipped bindings win. Tutorial *prose* is the syllabus (plus boards). Do not emit Village homestead Verse, fort_playset kit authoring, CubeGrid halls, Patchwork/Jam, Collectible→Tracker, or player-build `allowBuilding` On as this island’s setup. Using Prefabs and Galleries is a **discipline hub**, not this walkthrough.

## Negative findings

- Picker **Home Builder** / docs **Home Builder** / folder `home_builder_sample` are the same tile — not Obstacle Course, not Starter Island, not Brick Editor (Phone vs BrickAssembly).
- Name does **not** mean Verse homestead pads or an authored playset kit.
- **0** Beacon / Fabric / Mutator / Crowd / Jam / HUD — do not copy Music Concert or ExtraAssets HUD here.
- Assembly Devices companion says “Test out this device in the BRICK Home Builder template” — this island has **0** Assembly devices; they live on action_adventure_sample (**29**).
- `find_actors` `"Point Light"` is empty; labels are **Customizable Light**.
- `find_actors` `"Water"` hits watering cans, not `Device_Water`.
- `spawn Item` live struct is empty `(DefaultHandlerFunctions=(),Instance=())` — item class from OFPA strings only.
- Doorbells are interactable props, not Creative devices.

## Capture backlog

Waived: which BRICK Item Spawner is Goo vs Snowball (C_0 vs C_1); per-light spatial mapping to rooms; doorbell `houseType`; Reset/Backup editor commands. `experimental.*` Unverified.

## Live notes

- Cursor `user-unreal-mcp` was down; `tools/mcp.py` on `:8000` worked.
- Mount `/9dbbf9de-6efd-47dc-9191-ccdf575cc87f` from `ListFiles("")`.
- Billboard `get_properties` fails the whole call if any requested name is missing; request `text` only.
- Wrong `refPath` `/mount/map.ActorName` is invalid; use `actorPath`.
- Did not flip experimental flags. Did not switch maps.
