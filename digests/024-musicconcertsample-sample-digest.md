---
id: digest-024
kind: digest
---

# music_concert_sample — sample digest

> **Extractor provenance.** `tools/extract-template.py` (Pass A) + UnrealMCP on
> `localhost:8000` (Pass B). Default level
> `/641aa8cb-7c80-4d55-94e6-d1feb90e8b11/music_concert_sample` (**Observed**).
> Hub: Building BRICK Islands
> + index BRICK Templates
> + island tutorial Music Concert
> + companion Working with BRICK Islands
> + theme Rules (publishing, not a device syllabus).
> Discipline hub (not this island): Composing with Fortnite Patchwork.
> `docsUrl` empty. Not on the spreadsheet. Compat **43.00**. `featureSet`: **BrickGameplay**.

| | |
|---|---|
| Classification | **BRICK learning island** — Patchwork concert lab with a beacon/popup trail, not a Festival jam match |
| Content type | **catalog wrap** (2 authored vs 704 OFPA; `/Musical_Walk_Assets` + modular desert + Fabric devices) |
| Verse | **0 files** |
| Fab `.uref` | **0** |

## What it is

The picker name is **Music Concert** (BRICK). Docs H1 is **Music Concert**. Disk folder is `music_concert_sample`. The island is a **desert festival sampler**: numbered beacons + animal popups teach Drum Sequencer / Note Sequencer / Instrument Player / Echo, then main-stage **Switch stems** mute Drum / Bass / Melody / Chords. Teaching that survives deletion is the **Volume → Popup → Beacon / Tracker trail** and **Switch Enable/Disable on Patchwork sequencers**. Do not duplicate: jam-mutator-crowd-aura.md (0 Crowd / 0 Mutator here; `enableJam=Off`), tracker-gated-encounter.md (combat Increment vs visit Increment), homestead-claim-spawn.md (Village `Patchwork` is a room folder, not Fabric). ExtraAssets HUD stays **four** — this HUD Controller has **null** widget overrides. Deeper UMG stays on user_interfaces_template.

## Scale (**Observed** — `summary.txt` + live)

| | |
|---|---|
| Authored assets | **2** (`GameFeatureData`, `music_concert_sample_DefaultHLODLayer`) |
| OFPA | **704** |
| Maps | `music_concert_sample.umap` (only) |
| Verse files | **0** (`VerseToolset.ListFiles` on `/641aa8cb-…` — folders only; do not recurse `__ExternalActors__`) |
| Verse devices placed | **0** |
| Data layers | **0** |
| Fab `.uref` | **0** |
| Compatibility | **43.00** |
| Matchmaking | **16** / live `teamSize=Dynamic` / `teamFillOption=MustFill` (`.uefnproject` + Island Settings) |
| Streaming | Partitioned + actor folders + OFPA. **`LevelHasStreamingDisabled=1`** (Walker/VFX/music-fantasy camp). Sampled descriptors `bIsSpatiallyLoaded=false`. `load_actors` required for billboard `text` and device values. `find_actors` by `"Device_"` returned **0** — use label needles + `collision_channels=[]`. `folderPath` often **None**. Live `get_folders` returned only `Background` (16 landscape actors). |

Empty `Developers/felixkupis/Collections/`. First-party scratch folder.

`experimental.*` (Python / toolsets / Scene Graph) in `.uefnproject` is **Unverified**. Content-side: **no** `.verse`, **no** `__ExternalEntities__`. Plugin `EnableSceneGraph: true` with no Scene Graph entities.

Largest file well under GHES 100 MB. Whole folder ~**10 MB**.

## Level organisation (**Observed** — live `get_folders` + finds)

```
Background/     16 (landscape / cliffs — only folder path the editor returned)
(unfoldered)    devices + Musical Walk props; folderPath often None
```

**25** `Device_Billboard_V2`. Syllabus is both the public Music Concert page **and** the boards (`_raw/music_concert_sample/billboard-texts.md`). `text` is on the **actor**.

## Device inventory (the design)

| Count | Class | Role |
|---|---|---|
| 25 | `Device_Billboard_V2` | Station signs (D-SEQR / MUSIC_SEQR / I-PLAY / FX-ECHO / Music Manager / switches / HUD / DSD / mixer) |
| 8 | `Device_ModularEchoEffect` | FX-ECHO (Vibe Turtle + graph) |
| 6 | `Device_FabricSequencer` | Note Sequencer (MUSIC_SEQR) |
| 6 | `Device_ModularProgressor` | Note Progressor |
| 6 | `Device_Volume` | Popup Show/Disable |
| 5 | `Device_Beacon_V2` | Golden path (`enabledOnPhase=None` until dismiss) |
| 5 | `Device_FabricInstrumentPlayer` | I-PLAY |
| 5 | `Device_PopUpDialog_V2` | Welcome + four animals |
| 5 | `BP_Device_ModularEmitter` | OFPA histogram; live label find **0** |
| 4 | `Device_VFX_Creator_V2` | Stem-tied stage FX |
| 4 | `Device_Switch_V2` | Drum / Bass / Melody / Chords |
| 4 | `BP_BRICK_Player_Spawner_Prop` | All `useAsIslandStart=true` |
| 2 | `Device_FabricDrumSequencer` | D-SEQR |
| 2 | `Device_ModularDrumPlayer` | Drum Player |
| 2 | `Device_ModularWaveGenerator` | OFPA histogram; live label find **0** |
| 2 | `Device_BRICKItemSpawner` | Patchwork Tool pads |
| 2 | `Device_CRD_AudioMixer` | Footstep / ambience duck |
| 1 | `Device_ModularGlobalMusicController` | Music Manager; tempo **140**, key **C** Major |
| 1 | `Device_ModularBassSynth` | Omega Synth |
| 1 | `Device_FabricDistortionEffect` | Distortion |
| 1 | `Device_HUDController_V2` | HideElement minimap/health/shields; overrides **null** |
| 1 | `Device_ExperienceSettings_BRICK` | `enableJam=Off`, `allowBuilding=None` |
| 1 | `Device_TeamSettings_V2` | Grant Patchwork Tool (billboard) |
| 1 | `Device_Tracker_V2` | Visit counter `targetValue=4` |
| 1 | `BP_DaySequenceModifier_Device` | Priority **2000**, FixedTime **8.5**, `use Volume=false` |
| 0 | Crowd Volume / Mutator / VerseDevice / ExtraAssets HUD widgets | music-fantasy jam stack or theme HUD copy |

Desert dressing is `/Musical_Walk_Assets` (stage, animals, speakers, tents) + `/CRG_Modular` cliffs + `/NEIGHBORHOOD_Assets` signs + `/CR_Legacy` planks.

## Verse (0 files)

No `snippets/verse/` folder. Valid.

## Authored kit (**Observed** — `assets.txt` + live)

```
Content/GameFeatureData.uasset
Content/music_concert_sample_DefaultHLODLayer.uasset
```

**Tier 2 mounts:** `/Game` 1399, `/FMInWorldKnobs` 284, `/CRG_Modular` 281, `/Musical_Walk_Assets` 224, `/CR_Legacy` 200, `/CreativeCoreDevices` 190, `/NEIGHBORHOOD_Assets` 110, plus Fabric `/CRD_Modular*` / `/CRD_Fabric*` / `/BRICKPlayerSpawn` / `/CRD_BRICKItemSpawner`. No Fab `.uref`.

## What this sample actually teaches

1. **A learning trail is Volume Show popup + dismiss Enable Beacon + popup Increment Tracker.** Not Verse, not a combat tracker. techniques/beacon-volume-popup-trail.md.
2. **Main-stage remix is Switch `On Turned On/Off` Enable/Disable on Patchwork sequencers (+ VFX Creator).** Music Manager holds tempo/key. Not Mutator jam. techniques/patchwork-switch-stems.md. Contrast jam-mutator-crowd-aura.md.
3. **HUD Controller can HideElement without ExtraAssets or authored Widgets.** Live overrides **null**; minimap/health/shields hidden; `showHUDMessages` on for the tracker. ExtraAssets HUD stays four (Social / Mastery / Walker / music-fantasy). Deeper viewmodels: umg-viewmodel-device-binding.md.
4. **BRICK Island Settings turn Jam off and building off.** `enableJam=Off`, `allowBuilding=None`, `locomotionPreset=Custom`, `allowSprint=0`, `bUseLatestHUD=false`. Working with BRICK Islands: Minifigures do not hangglide / hurdle / slide.
5. **DSD is 2000-only FixedTime, not a volume stack.** `priority=2000`, `dayNightCycleTime=8.5`, `overrideSunlight=true`, `use Volume=false`. Billboard still calls it the Skydome device.
6. **The festival is Musical Walk + modular desert catalog, not 2 authored meshes.** Docs “cute BRICK animal friends” are `/Musical_Walk_Assets` `BP_MW_*` (Mascot / Rabbit / Frog / Turtle / Vulture).
7. **25 billboards are the in-island syllabus** (Patchwork Tool grant, Audio Mixer buses, Music Manager, VIP backstage cave, gallery island **5806-7083-7937**). Tutorial prose is the beacon walkthrough; boards add Edit-mode device cards the page only sketches.

## Official tutorial vs shipped

**External** hub + BRICK Templates index + Music Concert + Working with BRICK Islands + theme Rules (headed Chrome 2026-09-15). Verbatim: `_raw/music_concert_sample/tutorial.txt`. Spreadsheet: no row. `docsUrl` empty. Guessed `/documentation/fortnite/BRICK-islands-in-fortnite` redirected to the Fortnite docs hub; live hub slug is `building-BRICK-islands-in-fortnite`.

### Lines up

| Tutorial | Shipped |
|---|---|
| Picker / templates index **Music Concert**; Patchwork devices in a desert festival | picker **Music Concert**; `BrickGameplay`; `/Musical_Walk_Assets` |
| Beacon golden path + Popup Dialog + Tracker visit box | 5 Beacons + 5 Popups + Tracker `targetValue=4` |
| DJ Mascot / Piano Rabbit / Organ Frog / Vibe Turtle | popup titles match |
| D-SEQR / Note Sequencer / I-PLAY / FX-ECHO | Drum Sequencer, Fabric Sequencer, Instrument Player, Echo Effect |
| Main-stage Switch Drum / Bass / Melody / Chords | 4 Switches Enable/Disable sequencers + VFX |
| Billboard + blank sign prop | 25 `Device_Billboard_V2` |
| Patchwork Tool from menu / Item Spawner / Team Settings | 2 BRICK Item Spawners; Team Settings grant; billboards |
| Audio Mixer ducks footsteps / ambience | 2 mixers + board |
| Music Manager tempo/key/mode | live tempo **140**, key **C** Major |
| Some Creative devices unavailable on BRICK | `featureSet=BrickGameplay`; Experience Settings BRICK |
| Start Game / End Game from menu | welcome board |

### Does not line up

| Tutorial | Shipped |
|---|---|
| **Six** numbered beacons (incl. Behind the Scenes) | live **5** `Device_Beacon_V2` |
| Tracker “how many areas you have visited” | Increment on **4** animal OK buttons (not Welcome); `statToTrack=Eliminations` leftover |
| Device nicknames **MUSIC_SEQ** / **I-PLY** | boards **MUSIC_SEQR** / **I-PLAY** |
| “learn about about” (double about) | docs typo |
| Create a Festival-like **Jam Stage** as shipped | `enableJam=Off`; **0** Crowd / Mutator |
| Accolades / Prop Mover / Note Trigger as placed | **0** of those; customize section is homework |
| Skydome device (board) | live `BP_DaySequenceModifier_Device` priority **2000** |
| Widget-override theme HUD | HUD overrides **null** |

Training rule: shipped bindings win. Tutorial *prose* is the syllabus (plus boards). Do not emit Mutator jam / Crowd / ExtraAssets HUD / six live beacons as this island’s setup. Composing with Patchwork is a **discipline hub**, not this walkthrough.

## Negative findings

- Picker **Music Concert** / docs **Music Concert** / folder `music_concert_sample` are the same tile — not music-fantasy streets’ plaza jam.
- Name does **not** mean Verse concert scripting or Fortnite Festival Jam.
- Village Starter `Patchwork` folder is **not** Fabric Patchwork.
- `statToTrack=Eliminations` is leftover; progress is popup Increment.
- OFPA Wave Generator ×2 / Modular Emitter ×5 did not appear under those live labels.
- All four BRICK player spawners have `useAsIslandStart=true`.

## Capture backlog

Waived: Patchwork cable graph / sequencer grid contents (not in `get_properties`); Team Settings / Item Spawner item class arrays; Beacon0–4 ↔ docs 1–6 spatial mapping; Wave Generator / Emitter live instances; Fabric knob subcomponents (`FabricCarousel_*` from a `"Fabric"` find). `experimental.*` Unverified.

## Live notes

- Cursor `user-unreal-mcp` was down; `tools/mcp.py` on `:8000` worked.
- Mount `/641aa8cb-7c80-4d55-94e6-d1feb90e8b11` from `ListFiles("")`.
- Billboard `get_properties` fails the whole call if any requested name is missing; request `text` only.
- Wrong `refPath` `/mount/map.ActorName` is invalid; use `actorPath` / `/{mount}/{map}.{map}:PersistentLevel.{name}`.
- Did not flip experimental flags. Did not switch maps.
