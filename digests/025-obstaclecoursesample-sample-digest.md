---
id: digest-025
kind: digest
---

# obstacle_course_sample — sample digest

> **Extractor provenance.** `tools/extract-template.py` (Pass A) + UnrealMCP on
> `localhost:8000` (Pass B). Default level
> `/6386e9a2-8c92-45bb-af84-7a98ac89a4f4/obstacle_course_sample` (**Observed**).
> Hub: Building BRICK Islands
> + index BRICK Templates
> + island tutorial Obstacle Course
> + companion Working with BRICK Islands
> + Collectible Devices (device ref; example is Action Adventure)
> + theme Rules (publishing, not a device syllabus).
> `docsUrl` empty. Not on the spreadsheet. Compat **43.00**. `featureSet`: **BrickGameplay**.

| | |
|---|---|
| Classification | **BRICK learning island** — primitive-block obby with device-wired platforms and a 5-coin collect win |
| Content type | **catalog wrap** (2 authored vs 1291 OFPA; `/Primitives_Assets` + modular desert + Musical Walk dressing) |
| Verse | **0 files** |
| Fab `.uref` | **0** |

## What it is

The picker name is **Obstacle Course** (BRICK). Docs H1 is **Obstacle Course** (also “obby template”). Disk folder is `obstacle_course_sample`. The island is a **mountain course of catalog primitives**: easy jumps, Switch-driven moving platforms, a shoot-to-raise gallery, then five **Gold Coin** collectibles that Increment a Tracker and Activate End Game. Teaching that survives deletion is **any-order Collectible→Tracker→End Game** and **Prop Mover wiring without Verse**. Do not duplicate: sequential-collect-race.md (Verse ordered spawners + timer), cubegrid-then-color-grids.md (CubeGrid authored MIs; **0** CubeGrid here), array-sync-race-primitive.md (not a persistent race), beacon-volume-popup-trail.md / patchwork-switch-stems.md (**0** Beacon / Sequencer / Jam). ExtraAssets HUD stays **four** — this HUD Controller has **null** widget overrides (HideElement, same as Music Concert).

## Scale (**Observed** — `summary.txt` + live)

| | |
|---|---|
| Authored assets | **2** (`GameFeatureData`, `obstacle_course_sample_DefaultHLODLayer`) |
| OFPA | **1291** |
| Maps | `obstacle_course_sample.umap` (only) |
| Verse files | **0** (`VerseToolset.ListFiles` on `/6386e9a2-…` — folders only; do not recurse `__ExternalActors__`) |
| Verse devices placed | **0** |
| Data layers | **0** |
| Fab `.uref` | **0** |
| Compatibility | **43.00** |
| Matchmaking | **8** / live `teamSize=Dynamic` / `teamFillOption=MustFill` (`.uefnproject` + Island Settings) |
| Streaming | Partitioned + actor folders + OFPA. **`LevelHasStreamingDisabled=1`** (Walker/VFX/music-fantasy/Music Concert camp). Sampled descriptors `bIsSpatiallyLoaded=false`. `load_actors` required for billboard `text` and device values. `find_actors` by `"Device_"` returned **0** — use label needles + `collision_channels=[]`. `folderPath` often **None**. Live `get_folders` returned only `Background` (16 landscape actors). |

Empty `Developers/felixkupis/Collections/`. First-party scratch folder.

`experimental.*` (Python / toolsets / Scene Graph) in `.uefnproject` is **Unverified**. Content-side: **no** `.verse`, **no** `__ExternalEntities__`. Plugin `EnableSceneGraph: true` with no Scene Graph entities.

Largest file well under GHES 100 MB. Whole folder ~**10.8 MB**.

## Level organisation (**Observed** — live `get_folders` + finds)

```
Background/     16 (landscape / cliffs — only folder path the editor returned)
(unfoldered)    devices + primitive cubes; folderPath often None
```

**21** `Device_Billboard_V2`. Syllabus is both the public Obstacle Course page **and** the boards (`_raw/obstacle_course_sample/billboard-texts.md`). `text` is on the **actor**. Event graph: `_raw/obstacle_course_sample/event-graph.md`.

## Device inventory (the design)

| Count | Class | Role |
|---|---|---|
| 687 | `BP_Primitive_*` (cubes/tris, blue/red/green) | Catalog massing — not CubeGrid |
| 21 | `Device_Billboard_V2` | Station signs (primitives, coins, movers, HUD, DSD-as-lighting, barrier sky) |
| 18 | `Device_PropMover_V2` | Moving / raise platforms (`Translation`, Gameplay Only) |
| 8 | `BP_BRICK_Player_Spawner_Prop` | All `useAsIslandStart=true` |
| 7 | `Device_Trigger_V2` | 4 damage-targets + 3 ping-pong helpers |
| 5 | `Device_Switch_V2` | Hidden (`visibleDuringGame=false`) mover bus |
| 5 | `Device_Collectible_Object_V2` | Gold Coin ×5 |
| 4 | `Device_Teleporter_V2` | Cliff hops |
| 4 | `Device_Water_V2` | Terrain pools; Button0 empties Water2 |
| 3 | `Device_Button_V2` | Toggle switches / drain water |
| 2 | `Device_MovementModulator_V2` | Jump-pad props; live `affect Movement Speed=false`, `speed=1` |
| 2 | `Device_VendingMachine_V2` | Docs Burst Pulse Rifle; item class arrays waived |
| 1 | `Device_ConditionalButton_V2` | Coconut key → Mover7 Advance |
| 1 | `Device_Tracker_V2` | `targetValue=5`, title Collect the Golden Coins |
| 1 | `Device_GameEnd` | Tracker `When Complete` |
| 1 | `Device_HUDController_V2` | HideElement minimap/health/shields/team/build; overrides **null** |
| 1 | `Device_ExperienceSettings_BRICK` | `enableJam=Off`, `allowBuilding=None` |
| 1 | `BP_Device_DamageVolume_V2` | Under-island; `damage=10`, `damageType=Elimination` |
| 1 | `Device_Barrier_V2_Placed` | Custom sky `M_Barrier_Sky_Noon` (not a block) |
| 1 | `BP_DaySequenceModifier_Device` | Priority **2000**, FixedTime **8**, `use Volume=false` |
| 1 | `Device_PopUpDialog_V2` | `autoDisplay=Game Start`, Welcome! / 5 coins |
| 0 | Beacon / Fabric Sequencer / Mutator / Crowd / CubeGrid / Timer / Item Spawner / VerseDevice | Music Concert Patchwork, Parkour collect, Greyboxing 101 tool |

Dressing: `/Primitives_Assets` + `/Primitives_Composition`, `/CRG_Modular` rocks/cliffs, `/Musical_Walk_Assets` (steps, vultures, snail, frog, rainbow), `/NEIGHBORHOOD_Assets` signs, `/BRICKCoconut`.

## Verse (0 files)

No `snippets/verse/` folder. Valid.

## Authored kit (**Observed** — `assets.txt` + live)

```
Content/GameFeatureData.uasset
Content/obstacle_course_sample_DefaultHLODLayer.uasset
```

**Tier 2 mounts:** `/6386e9a2-…` 2602, `/Primitives_Assets` 1385, `/Game` 815, `/CRG_Modular` 755, `/NEIGHBORHOOD_Assets` 242, `/CreativeCoreDevices` 241, `/CR_Legacy` 103, `/Primitives_Composition` 101, `/Musical_Walk_Assets` 90, `/CRD_PropMover` 54. No Fab `.uref`.

## What this sample actually teaches

1. **Any-order collect win is Collectible `On Collected` → Tracker Increment → `When Complete` End Game.** Not Verse, not ordered spawners. techniques/collectible-tracker-endgame.md. Contrast sequential-collect-race.md.
2. **Moving platforms are hidden Switch + Prop Mover Advance/Reverse; shoot-to-raise is `triggeredByDamage` Trigger → Advance.** Coconut Conditional Button and Button→Water emptying are the same wiring family. techniques/prop-mover-device-wiring.md.
3. **Greybox with catalog `BP_Primitive_*` colors, not CubeGrid.** Live **0** CubeGrid. Link cubegrid-then-color-grids.md for the tool curriculum.
4. **HUD Controller can HideElement without ExtraAssets or authored Widgets.** Overrides **null**; minimap/health/shields/team/build hidden; `showHUDMessages` on for the tracker. ExtraAssets HUD stays four. Same null-override HideElement as Music Concert (coincidence 2).
5. **BRICK Island Settings turn Jam off, building off, sprint off, fall damage off.** `enableJam=Off`, `allowBuilding=None`, `locomotionPreset=Custom`, `allowSprint=0`, `bFallDamage=false`, `bUseLatestHUD=false`. Billboard: destruction disabled. Working with BRICK Islands: Minifigures do not hangglide / hurdle / slide.
6. **DSD is 2000-only FixedTime, not a volume stack.** `priority=2000`, `dayNightCycleTime=8`, `overrideSunlight=true`, `sunlightIntensity=10`, `use Volume=false`. Billboard still calls it a lighting device.
7. **Barrier can be a skydome material, not a block.** `barrierMaterial=/Game/Items/Traps/Materials/M_Barrier_Sky_Noon`.
8. **21 billboards are the in-island device syllabus** (Start Game, coins, movers, triggers, water, HUD, Island Settings, coconut, teleporters, damage volume). Tutorial prose is the obby pacing walkthrough; boards name the devices.

## Official tutorial vs shipped

**External** hub + BRICK Templates index + Obstacle Course + Working with BRICK Islands + Collectible Devices + theme Rules (headed Chrome 2026-09-15). Verbatim: `_raw/obstacle_course_sample/tutorial.txt`. Spreadsheet: no row. `docsUrl` empty. Guessed `/documentation/fortnite/collectible-devices-in-fortnite` redirected to the Fortnite docs hub; live slug is `using-BRICK-collectible-devices-in-fortnite-creative`.

### Lines up

| Tutorial | Shipped |
|---|---|
| Picker / templates index **Obstacle Course**; obby up a mountain; 5 golden coins | picker **Obstacle Course**; 5 Gold Coin collectibles |
| Start Game from Sidebar | welcome board |
| Game-start message with the 5-coin goal | Popup `autoDisplay=Game Start`, title Welcome! |
| Jump pad into water, then easy jumps | 2 Movement Modulators + 4 Water + primitives |
| Moving platforms after simple jumps | 18 Prop Movers, Switch ping-pong |
| Shoot targets to raise platforms; vending Burst Pulse Rifle | Triggers 3–6 `triggeredByDamage=true` → mover Advance; 2 vending machines |
| Collectibles complete the course; two are hidden | Tracker `targetValue=5` → End Game |
| Teleporters on cliffs; water levels; buttons change the course | 4 teles; Button0 empties Water2; 3 buttons |
| Some Creative devices unavailable on BRICK | `featureSet=BrickGameplay`; Experience Settings BRICK |
| Next learning template: Home Builder | theme queue |

### Does not line up

| Tutorial | Shipped |
|---|---|
| Collectible Devices companion: **Gold Stud** + **Action Adventure** unlock/quest example | live **Gold Coin**; this island is the obby, not Action Adventure |
| Billboard “maybe an egg?” | live all five `Gold Coin` |
| Players can't mantle / vault / crouch / slide | Working with BRICK: hangglide / hurdle / slide (wording drift) |
| Jump pad launches onto the course | Movement Modulator `affect Movement Speed=false`, `speed=1` — launch impulse **not** in the properties we could read |
| Time-limit jumps as a shipped variation | **0** Timer devices (customize homework) |
| Lighting devices (board) | live `BP_DaySequenceModifier_Device` priority **2000** |
| Widget-override theme HUD | HUD overrides **null** |

Training rule: shipped bindings win. Tutorial *prose* is the syllabus (plus boards). Do not emit Parkour Verse collect, CubeGrid halls, Patchwork/Jam, Gold Stud Action Adventure unlock, or a Timer as this island’s setup. Collectible Devices is a **device reference**, not this walkthrough.

## Negative findings

- Picker **Obstacle Course** / docs **Obstacle Course** / folder `obstacle_course_sample` are the same tile — not Parkour, not Music Concert.
- Name does **not** mean Verse race logic or CubeGrid greybox curriculum.
- **0** Beacon / Fabric / Mutator / Crowd / Jam — do not copy Music Concert techniques here.
- `statToTrack=None` (not Eliminations leftover).
- All eight BRICK player spawners have `useAsIslandStart=true`.
- Damage Volume is **10** Elimination / 2s tick, not Parkour’s Verse `Damage(500)`.
- `find_actors` `"Volume"` is physics/water children, not `Device_Volume` popup volumes.

## Capture backlog

Waived: vending item class arrays; Movement Modulator launch/impulse fields; Barrier material graph; per-mover translation distance; coconut item class on the Conditional Button; spatial mapping of the five coins to “two hidden”. `experimental.*` Unverified.

## Live notes

- Cursor `user-unreal-mcp` was down; `tools/mcp.py` on `:8000` worked.
- Mount `/6386e9a2-8c92-45bb-af84-7a98ac89a4f4` from `ListFiles("")`.
- Billboard `get_properties` fails the whole call if any requested name is missing; request `text` only.
- Wrong `refPath` `/mount/map.ActorName` is invalid; use `actorPath`.
- Did not flip experimental flags. Did not switch maps.
