---
id: digest-026
kind: digest
---

# lighting_template — sample digest

> **Extractor provenance.** Offline evidence from `tools/extract-template.py` (reconstruction of
> the pipeline in `00-EXTRACTION-PROCESS.md`; the original `.sh` is absent from this repo).
> Live-editor facts read via UnrealMCP on `localhost:8000` with the project open; currently
> loaded level is `/lighting_template/lighting_template`. Property values are **Observed**
> on the loaded actors (camelCase toolset names). Scene Graph / experimental flags are
> **Unverified** — the operator confirmed they flip those on in Project Settings, and they
> are off by default.

| | |
|---|---|
| Classification | **Walkable lighting curriculum** — six named rooms, each demonstrating one technique |
| Recommendation | **KEEP** |

## Verdict first

**Keep it.** Authored-asset count (120) already says "kit, not catalog", but the stronger
reason is what those assets *are*: a complete, labelled lighting lesson that does not exist
anywhere else in this KB.

| | equip_furniture_props | playset_sample | **lighting_template** |
|---|---|---|---|
| Authored assets | 2 | 130 | **120** |
| Placed actors (OFPA) | 14,116 | 18 | **388** |
| Reference mounts | 117 | 5 | **23** |
| Question to ask | *is the curation maintainable?* | *what does it teach?* | *what does it teach?* |
| Verdict | retire | keep | **keep** |

The furniture catalog was a pile of references with nothing unique. This is the inverse: a
small authored project whose *subject* is lighting in UEFN, organised as rooms you walk
through, with the lesson written on the wall. Deleting it loses the only worked example of
an environment light rig, of engine-vs-device lights, and of dual Lumen/Basic exposure.

## What this sample is *for*

It is a **walkable lighting curriculum**. Three pieces of evidence converge and nothing
contradicts them:

**1. The outliner is the syllabus** (**Observed** — live editor, `SceneTools.get_folders`):

```
Environment                         the reusable rig + scalability manager
Landscaping                         trees / logs / bamboo — throwaway set dressing
Replay_Cams                         11 CineCameraActors (Shot_01..Shot_11)
Room_01 / Geometry_01 + Lights_01   engine lights; SSS + emissive
Room_02 / Geometry_02 + Lights_02   engine lights; rect-light interior
Room_03 / Geometry_03 + Lights_03   Creative light *devices*; dim / undim
Room_04 / Geometry_04 + Lights_04   Creative light devices; "Spot Lights"
Room_05 / Lights_05                 Creative light devices (no Geometry_05 folder)
Room_Exterior / Geometry_Ext + Lights_Ext
```

Each room is a paired `Geometry_*` / `Lights_*` folder. The geometry is Modelling-Mode
demo mesh (`Box_*`, `Extrude_*`, `Boolean_*`, `Path_*` — hashed names, not authored art).
The lights are the lesson.

**2. The billboards are the captions** (**Observed** — live editor, `Device_Billboard_V2.text`):

| Room | Billboard text |
|---|---|
| 01 | `Point Light + Sub Surface Scattering` |
| 01 | `Point Light + Self Iluminating Material` |
| 01 | `1. Play Cinematic Below` / `2. Walk through Map` |
| 03 | `Light Device Introduction` / `Dim Light + Undim Light` |
| 04 | `Spot Lights` |
| Exterior | `Exterior Lighting` |

Room 02 has no captured caption (doors / glass, 6 rect lights — likely the "area light /
interior bounce" lesson). Room 05 has lights but no `Geometry_05` folder and no billboard.

**3. A cinematic is the tour** (**Observed**). 15 `Device_CinematicSequence` actors, 8
buttons, 11 cine cameras, and 16 Level Sequences including `LS_Sun_01`..`LS_Sun_04`.
Room 01's billboard tells you the intended use: press play, then walk the map. The four
sun sequences are day-phase transitions driven off the environment rig, not a game loop.

The original name is still in the tree: `CC_LightingDev_POC.umap` (1,566 bytes) plus
`CC_LightingDev_POC_sharedassets/` (12 landscape LayerInfos). This started as a lighting
dev POC and was promoted into a template without being renamed. The POC map is empty
enough to ignore; the live map is `lighting_template.umap`.

## The environment light rig — the reusable finding

`BP_Environment_Light_Rig` is one actor in the `Environment` folder. **Seven lighting
systems live on it as named components** (**Observed** — live editor,
`ActorTools.get_components` + `ObjectTools.get_properties`):

```
DirectionalLight        intensity 10, 6500 K (temp unused), AtmosphereSunLight ON,
                        lightSourceAngle 1.5, indirectLightingIntensity 2
SkyAtmosphere           rayleighScatteringScale 0.0331, mieAnisotropy 0.8,
                        multiScatteringFactor 1
SkyLight                intensity 0.2, SLS_SpecifiedCubemap, bRealTimeCapture ON
ExponentialHeightFog    density 0.03, heightFalloff 0.2, volumetric fog ON
                        (scatteringDistribution 0.2, albedo white, extinctionScale 1)
LumenExposure           PostProcess, ENABLED, unbound, priority 0
BasicExposure           PostProcess, DISABLED, unbound, priority 0
ColorGrading            PostProcess, ENABLED, unbound, priority 0
```

The three post-process components are **named for their job, not their class**. Filtering
on `PostProcess` in the component name finds nothing — they are `LumenExposure`,
`BasicExposure`, `ColorGrading`. Exposure / grading live in the nested `settings` struct
(`FPostProcessSettings`); `bOverride_*` flags mark which fields are actually set.

**LumenExposure** (the live look):

| | |
|---|---|
| Method | `AEM_Histogram` |
| Min / max brightness | **0.075 / 1.0** |
| Bias | 0.5 |
| Speed up / down | 3.75 / 5.5 |
| Histogram log | −20 … 10 |
| GI / reflections | **Lumen / Lumen** |
| Lumen diffuse boost | 1.25 |
| Skylight leaking | 0.02 |
| Bloom | 1.5 |
| Chromatic aberration | 0.3 |

**BasicExposure** (the fallback — disabled):

| | |
|---|---|
| Method | `AEM_Basic` |
| Min / max brightness | **1 / 1** (locked) |
| Bias | **3** |
| White temp | **11000** |
| GI / reflections | **None / None** |
| Bloom | 0.675 |

**ColorGrading** (always on, stacked): vignette 0.6, bloom 0.675, bloom threshold −1,
scene fringe 0.4, local-exposure detail 1.2.

The pattern: **one unbound post-process per concern, stacked by name, with a disabled
fallback for the no-Lumen path.** Toggling scalability is then "enable BasicExposure,
disable LumenExposure" — not rewriting a 200-field settings blob. Promoted into
techniques/environment-lighting-rig.md.

`BP_LightingScalabilityManager` sits next to the rig. Its class path is
`/Game/TimeOfDay/TODM_Sequence/UEFN/BP_LightingScalabilityManager` — Epic's Time-of-Day
scalability manager, **referenced, not authored**. The placed instance has only a Billboard
component and no interesting properties of its own; it is a hook, not a lesson.

## Engine lights vs Creative light devices — the other lesson

The rooms are a progression, not a palette (**Observed**):

| Room | Light actors | Kind | Temps (K) | Intensity range |
|---|---|---|---|---|
| 01 | 2 Spot + 2 Point + 1 Rect | **Engine** (`Engine.SpotLight` etc.) | 3800 / 5000 / 5500 | 0.5 – 4 cd |
| 02 | 4 Spot + 2 Point + 6 Rect | **Engine** | 3800 / 5000 / 6500 | 0.2 – 8 cd |
| 03 | 5 `Device_PointLight_V2` | **Creative device** | 6500 | 0 – 1.56 |
| 04 | 14 `Device_PointLight_V2` | **Creative device** | 6500 | 0 – 1.56 |
| 05 | 10 `Device_PointLight_V2` | **Creative device** | 6500 | 0 – 1.25 |
| Exterior | 19 `Device_PointLight_V2` | **Creative device** | 6500 | 0 – 5 |

Rooms 01–02 use native engine lights, which expose the full property set (cone angles,
source radius, inverse-squared falloff, volumetric scattering, colour temperature). Room 01
SpotLight6 is the worked volumetric example: `volumetricScatteringIntensity = 4` against a
rig default of 1, 5500 K, 25°/65° cone, source radius 20.

Rooms 03–05 and Exterior switch to `Device_PointLight_V2` — Fortnite Creative light
*devices*. They still own a `SpotLightComponent` + `PointLightComponent` internally, but
the authorable surface is the device (dim / undim via gameplay messages), which is what
Room 03's billboard says. Intensity 0 on the unused inner component is normal: a Creative
point-light device carries both shapes and enables one.

**Do not mix the two in production without a reason.** Engine lights are for authored
looks you fully control. Creative devices are for looks players (or Verse) can dim. The
template demonstrates both, in that order, which is the curriculum.

Units are **candelas** on every sampled engine light (`intensityUnits = Candelas`),
inverse-squared falloff **on**, falloff exponent 8. Copy those three together.

## Sequences

16 Level Sequences under `Content/Sequences/` (**Observed** — `assets.txt` + live
`AssetTools.load_asset`). The four sun phases (`LS_Sun_01`..`04`) plus a supporting cast
(`LS_Bob`, `LS_Rotation`, `LS_Replay`, `LS_OpenDoor` / `_02`, `LS_Glass_Door`,
`LS_Button_Color` / `_02`..`_06`).

The copies at the content root (`Content/LS_*`) are **`ObjectRedirector`s left by a
move** — 16 redirectors, 16 real sequences. 19 `/Script/CoreUObject.ObjectRedirector`
references in `script-modules.txt`. The sequences also still reference an old GUID mount
(`/8b62d94b-42be-79af-508a-03b76d6545c1/Sequences/LS_*`), which is the previous project
id. Redirector leftover + stale mount is the same "we moved this and didn't clean up"
story.

Sequencer track contents (which light property each sun sequence actually drives) were
not unpacked — `list_properties` on a `LevelSequence` does not enumerate tracks. Cheap
next session via the Sequencer toolset; does not affect KEEP.

## What is throwaway

- **`Geometry/`** — 40 Modelling-Mode intermediates (`Box_<hash>`, `Extrude_<hash>`,
  `Boolean_<hash>`, `Path_<hash>`). Demo mesh for the rooms, not an art kit.
- **`Landscaping/`** — 31 stock Fortnite trees / logs / bamboo (`Athena_Tree_Large_2` × 20).
- **`CC_LightingDev_POC.umap`** + **`CC_LightingDev_POC_sharedassets/`** — leftover POC
  map (1.5 KB) and 12 landscape LayerInfos.
- **`Sky_Sphere`** — a `FortStaticMeshActor` next to the rig; the sky look is the
  `SkyAtmosphere` + `SkyLight`, not this mesh.
- **Prop lights with `NoLight` in the name** (`CP_Apollo_MountainBase_WallLight_04_NoLight`
  × 41, `CP_SharkIsland_Chandelier_NoLight` × 3) — meshes that *look* like fixtures, with
  the actual light as a sibling engine/device light. The `_NoLight` suffix is the tell.

None of this is a reason to retire. The teaching content is the rig, the room lights, and
the sequences.

## No Verse, no data layers, no Fab — and that is fine here

**Observed:** 0 `.verse` files, 0 data layers, 0 `.uref` Fab dependencies. 23 reference
mounts (the furniture catalog had 117; the playset kit had 5). Most of the extra mounts
here are Creative device packs (`/CRD_PointLight` 144, `/CRD_CinematicSequence` 45,
`/CreativeCoreDevices` 122) — expected, given rooms 03–Exterior *are* a device lesson.

A lighting curriculum has no runtime logic to express in Verse. The devices (button →
cinematic sequence → sun Level Sequence) are the glue, the same pattern as character_environment's
one-button-one-sequence island. Cross-link:
cinematic-sequences.md.

## Project configuration (⚠ flags Unverified)

> The project was open in UEFN during extraction. The operator has confirmed they flip
> Scene Graph / experimental flags on in Project Settings; those flags are **off by
> default** and UEFN rewrites the descriptor on disk the moment one is flipped. Guard:
> 00-EXTRACTION-PROCESS.md Step 2.
> **The KEEP verdict does not depend on these flags.**

```
compatibilityVersion               42.10           (Observed)
requiredRedirectorStartingVersion  42.10.57566230  (Observed)
apiAccess                          Private         (Observed)
matchmaking  maxPlayers 16 / maxTeamCount 16 / maxTeamSize 16   (Observed)
VersePath  /epictemplates@fortnite.com/lighting_template        (Observed)
VerseVersion 1 / VerseAPI 1                        (Observed)
ValkyrieSettings.libraryFolderMode  "Absent"       (Observed)

⚠ experimental.sceneGraph.bIsSceneGraphSystemAllowed  true   <- Unverified
⚠ experimental.sceneGraph.bEnableOneFilePerEntity     false  <- Unverified
⚠ experimental.pythonExperimental.bEnablePythonForProject true <- Unverified
⚠ experimental.toolsets.bEnableToolsetsForProject     true   <- Unverified
⚠ .uplugin EnableSceneGraph                           true   <- Unverified
⚠ .uplugin EnableVerseAssetReflection                 true   <- Unverified
```

`maxTeamCount: 16` is the default unrestricted island, not a teaching signal (the playset
template's `maxTeamCount: 1` *was*). Nothing in the matchmaking block is load-bearing.

## Pre-publish cleanup

- **`Developers/felixkupis/` ships again, empty** — plus a nested empty `Collections/`.
  **Five for five** across first-party templates (character_environment `jojitsuruga`, then four
  `felixkupis`). Pipeline gap, not carelessness.
- **16 `ObjectRedirector`s at the content root** (`Content/LS_*`) left by moving the
  sequences into `Content/Sequences/`. Plus stale references to GUID mount
  `/8b62d94b-42be-79af-508a-03b76d6545c1`. Sweep redirectors and old mounts together.
- **`CC_LightingDev_POC*` leftover** — a 1.5 KB map and 12 landscape LayerInfos under
  the pre-rename name. Harmless, noisy.
- **`MI_Old_Plywood_WorldAligned_Inst`** — `_Old` is in the *source* material name
  (`/EpicBaseMaterials/.../MI_Old_Plywood_WorldAligned`), not a leftover of ours. Do
  not flag it in the same sweep as `_WIP` / `Example`.

## Assessment

| Criterion | Status |
|---|---|
| Digest written | **Met** |
| Verse preserved | **N/A** — zero Verse |
| Unique content captured | **At risk if deleted** — 120 authored assets, including the environment rig, 16 sequences, 32 materials |
| Capture backlog | **1 open** (below) |
| Extractor provenance | **Qualified** — reconstructed extractor |

### Capture backlog

1. **What do `LS_Sun_01`..`04` actually animate?** Almost certainly the directional
   light's rotation / intensity / temperature, but `list_properties` on a LevelSequence
   does not enumerate tracks. The Sequencer toolset can read them. Cheap next session;
   does not affect KEEP. The four-phase naming plus the button → cinematic-sequence
   wiring already establish them as the day-cycle tour.

### Recommendation

**Keep.** It is the only artifact in this KB that documents three genuinely non-obvious
workflows:

1. **An environment lighting rig as one Blueprint** — sun + sky + atmosphere + fog +
   *named* post-process jobs (Lumen exposure / Basic exposure / color grading), with the
   no-Lumen path sitting disabled on the same actor.
2. **Engine lights vs Creative light devices**, taught as a progression (rooms 01–02
   then 03–Exterior), not as interchangeable actors.
3. **A walkable lighting curriculum** — billboard captions, paired geometry/lights
   folders, a cinematic as the tour.

Its 120 authored assets are replaceable only by rebuilding the lessons from scratch.
Promote (1) into
techniques/environment-lighting-rig.md; (3)
stays here as template-specific structure.
