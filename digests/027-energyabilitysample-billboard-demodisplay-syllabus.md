---
id: digest-027
kind: digest
---

# energy_ability_sample — billboard / DemoDisplay syllabus

**Observed** 2026-09-15 via UnrealMCP `get_properties`. 27 `Device_Billboard_V2` (Always, One Sided) + `BP_DemoDisplay` titles. Empty `description Text[0]` on many title boards — title-only is shipped.

## Workshop titles (`BP_DemoDisplay`)

| Label | Title | Description (if any) |
|---|---|---|
| 0 Welcome | WELCOME | this sample showcases: 1. EnergyBlade customization using Scene Graph and Verse. 2. An interactive EnergyBlade Workbench for players built using Verse. 3. Scene Graph customization support for telekinetic abilities. 4. telekinetic abilities as items. |
| 1 | 1 EnergyBlades in Scene Graph | |
| 1.1 | 1.1 Create a EnergyBlade Entity | |
| 1.2 | 1.2 Customize Blade Color | |
| 1.3 | 1.3 Adjust the Arc Intensity | Use the Arc Intensity float value in the kyber_crystal component to adjust the intensity of the arcing energy VFX on the blade. 0 = No arcing (Default). 1 = Maximum arcing. |
| 1.4 | 1.4 Adjust the Stability | |
| 1.5 | 1.5 Verse Customization of EnergyBlades | |
| 2 | 2 Example EnergyBlade Workbench | This workbench is an example of an interactive system built using Verse. Players can use the bench to build their own EnergyBlades. |
| 2.1 | 2.1 EnergyBlade Workbench Verse | The workbench UI was built using UMG with Verse Fields to handle player interaction and real-time visual updates. Verse manages the state of the customization process and dynamically assembles the final EnergyBlade based on selected parts. |
| 2.2 | 2.2 Expansion Opportunities in Verse | |
| 3 | 3. telekinetic abilities in Scene Graph | |
| 3.1 | 3.1 Create telekinetic ability Entities | |
| 3.2 | 3.2 Adjust Cooldown and Energy | All telekinetic abilities have either a Cooldown or Energy property. These properties control how often and how long a player can use a telekinetic ability. |
| 3.3 | 3.3 Additional Properties | The properties available to modify depend on the telekinetic ability selected. … Detailed breakdowns … are in the Training Grounds. |
| 3.4 | 3.4 Verse Customization of telekinetic abilities | All the telekinetic ability parameters available in Scene Graph are available in Verse for customization and manipulation. |
| 3.5 | 3.5 telekinetic abilities and EnergyBlades | |
| 4 | 4 telekinetic abilities as Items | |

## Workshop body (`Device_Billboard_V2`)

| Label | Text |
|---|---|
| 1.0.1 | EnergyBlades are now fully customizable entities … not visible in the viewport … appear as EnergyBlade hilts. |
| 1.1.1 / 3.1.1 | Add a blank entity using the toolbar. Place Actors > Entities > Entity |
| 1.1.2 | Right click the blank entity … add a EnergyBlade entity as a child. … hilt … different base EnergyBlade entity. |
| 1.1.3 | Note: EnergyBlade entities have no mesh in the editor, but do in gameplay. Here, a placeholder mesh is added … |
| 1.1.4 Floor | This is a base EnergyBladeHeroC entity. |
| 1.2.1 | … kyber_crystal_[color]_component … replace the kyber crystal component … |
| 1.2.2 Floor | This is a EnergyBladeHeroC entity with the kyber_crystal_green_component swapped out for a **kyber_krystal_red_component**. |
| 1.3.1 Floor | This is the same EnergyBlade with Arc Intensity set to 0.5 on the kyber_crystal_red_component. |
| 1.4.1 | Use the Stability float … 1 = No flickering (Default). 0 = Maximum flickering. |
| 1.4.2 Floor | This is the same EnergyBlade with Stability set to 0.0 on the kyber_crystal_red_component. |
| 1.5.1 | … this Verse device deletes all previous energy blades and telekinetic abilities … then creates a custom EnergyBlade … when they press a button. |
| 2.2.1 | Using Verse, you can expand the workbench … randomly generated EnergyBlade / personality test / shop where you buy individual components. |
| 3.1.2 | Right-click the blank entity … add a telekinetic ability entity as a child. |
| 3.1.3 | telekinetic ability entities … not visible in the viewport … weapon-type … appear as Holocrons. |
| 3.1.4 | Each telekinetic ability has two entity types, weapon and EnergyBlade augment. Weapon = quick bar pickup. Augment = Verse grant only, requires a EnergyBlade. |
| 3.1.5 Floor | This is a base TelekineticAbilitieshield_Weapon entity. |
| 3.3.1 Floor | This is a TelekineticAbilitieshield_Weapon entity with the energy reduced to 20 and the cooldown reduced to 2s. |
| 3.5.1 Floor | Verse device … awards the customized EnergyBlade from billboard 1.4 and the EnergyBlade augment version of the customized TelekineticAbilitieshield_Weapon entity from billboard 3.3. |
| 3.5.1 Floor2 | Verse device … grants … Telekinetic Training EnergyBlade augment … (placed 3.5 spawner is Shield; Training is 4.11). |
| 4.1 | Weapon type entity … place custom telekinetic abilities … or grant through Verse. Standard Holocrons … Visit the Training Grounds … |
| 4.2 | Using Verse, you can equip a Telekinetic_power_[name]_EnergyBlade_augment … does not appear in the quick bar. Only one augment at a time. |

## Training Grounds titles + body

| Station | Title | Body / Device_Billboard |
|---|---|---|
| 4.1 Telekinetic Push | Push all nearby physics objects … Adjustable: Push Telekinetic, Cooldown | (DemoDisplay description) |
| 4.2 Telekinetic Pull | Pull all nearby physics objects … Adjustable: Cooldown | |
| 4.3 Telekinetic Lightning | Channel lightning … Adjustable: Damage, Structure Damage, Cooldown, Duration | |
| 4.4 Telekinetic Throw | Use the Telekinetic to throw nearby objects … **There are no adjustable components.** | (docs reference lists Damage / BuildingDamage / Cooldown) |
| 4.5 Telekinetic Saber Throw | Throw a EnergyBlade … Requires a EnergyBlade. Adjustable: Damage, Structure Damage, Cooldown | |
| 4.6 Telekinetic Mind Trick | (title only) | Transform into a target character … Adjustable: Duration, Cooldown |
| 4.7 Telekinetic Distraction | (title only) | Spawn a decoy enemy … Adjustable: Decoy Health, Decoy Duration |
| 4.8 Telekinetic Awareness | (title only) | Mark all enemies … Adjustable: Mark Duration, Cooldown |
| 4.9 Telekinetic Heal | Create a healing aura … Heals ALL characters … Adjustable: Heal Strength, Drain Rate, Cooldown, Recharge Rate, Recharge Delay | |
| 4.10 Telekinetic Barrier | (title only) | Create a shield … Adjustable: Cooldown, Drain Rate, Recharge Rate, Recharge Delay |
| 4.11 Telekinetic Training | (title only) | Passive double jump + sprint. **Does not exist as an item.** Adjustable: Jump Stamina Cost, Sprint Stamina Cost |

## Wayfinding signs

| Sign | Title | Description |
|---|---|---|
| EnergyBlade Workbench Training Grounds | EnergyBlade Workbench | Customize a New EnergyBlade Here |
| Combat Simulation | Combat Simulation | Press to Start |
| Sign Workshop / Sign Training Grounds | (title unread — `description Text[0]` empty) | |

HUD Message: “Test out telekinetic abilities and EnergyBlades in the Training Grounds!” `showOnRoundStart=true`.
