---
id: digest-028
kind: digest
---

# energy_ability_sample — sample digest

> **Extractor provenance.** `tools/extract-template.py` (Pass A) + UnrealMCP on
> `localhost:8000` (Pass B). Default level
> `/92f0b488-49b0-4e1f-8d58-5df8b19e40ed/EnergyBlades_and_Telekinetic_Powers` (**Observed**).
> Docs: energy blades and telekinetic abilities
> + hub Working With space-opera™ Islands.
> `docsUrl` empty. Guessed `…-in-unreal-editor-for-fortnite` **hub-redirects**.
> Compat **43.00**. `featureSet`: **CustomGameplayFeature**.

| | |
|---|---|
| Classification | **Walkable Scene Graph + Verse lab** — numbered workshop + Training Grounds, not IceBase/Narrative/Team Conquest |
| Content type | **kit** (75 authored vs 204 OFPA) |
| Verse | **12 files** (5 placed VerseDevices) |
| Fab `.uref` | **0** |

## What it is

The picker name is **energy blades and telekinetic abilities**. The island is a **syllabus hall** (Workshop 0–4) plus a **Training Grounds** of stock holocrons, a duplicate workbench, and IO Guard combat dummies. Teaching that survives deletion is **kyber component swap**, **weapon vs EnergyBlade-augment grants**, and the **UMG+Verse workbench session**. Marketplace entitlements stay on entitlement-vs-offer. ExtraAssets HUD coincidence stays three (Social / Mastery / Walker). Guard Character Definitions stay on IceBase. Clickable `button`+`button_slot` originating write-up: verse-clickable-button.

## Scale (**Observed** — `summary.txt` + live)

| | |
|---|---|
| Authored assets | **75** |
| OFPA | **204** |
| Maps | `EnergyBlades_and_Telekinetic_Powers.umap` + leftover `Akiat_Template_1.umap` (1.5 KB) |
| Verse files | **12** (live `ListFiles` on `/92f0b488-…`; do not recurse `""`) |
| Verse devices placed | **5** (2× `menu_EnergyBlade_customization_device`, 3× `EnergyBlade_spawner_device`) |
| Data layers | **0** |
| Fab `.uref` | **0** |
| Compatibility | **43.00** |
| Matchmaking | Live Island Settings **16** / Solos; `.uefnproject` `maxPlayers` **1** |
| Streaming | Partitioned + actor folders + OFPA. **`LevelHasStreamingDisabled=1`**. Sampled actors still needed `load_actors` (`bIsSpatiallyLoaded=false`). |

Empty `Developers/felixkupis/Collections/`. First-party scratch folder.

`experimental.*` (Python / toolsets / Scene Graph / custom inventory) in `.uefnproject` is **Unverified**. Content-side: **12** `.verse`, `EnableSceneGraph: true` in `.uplugin`, EnergyBladePrefabs referenced by the map.

Whole folder ~**5.3 MB**. No file near the GHES **100 MB** cap.

## Level organisation (**Observed** — live `get_folders` + folder counts)

```
Workshop/                              2
  Billboards/                         39 — Device_Billboard_V2 + BP_DemoDisplay syllabus
  EnergyBlade Workbench/               23 — menu device + HUD hide + camera + mutator
  Screenshots/1.1 / 3.1                2+2 FortStaticMesh placeholders (not Scene Graph entities)
Training Grounds/
  Architecture/                       23
  Billboards/                         16
  Combat Simulation/                   5 — 4 IO Guard Spawners + button
  EnergyBlade Workbench Training Ground/ 23 — second menu copy
  Physics Objects/                    11
  Tutorial Assets/                     6 + 4.1–4.10 Item Spawners (no 4.11 spawner)
Guidance/Directing Arrows             31
Guidance/Signs                         4
UI/                                    2 — island HUD Controller + HUD Message
Island Settings/                       4
```

Syllabus titles on `BP_DemoDisplay` (0, 1, 1.1–1.5, 2, 2.1–2.2, 3, 3.1–3.5, 4, 4.1–4.11). Body copy on `Device_Billboard_V2` (Always, One Sided). Verbatim: billboard-texts.md.

## Device inventory (the design)

| Count | Class | Role |
|---|---|---|
| 32 | `BP_DemoDisplay` | Numbered titles / wayfinding signs |
| 27 | `Device_Billboard_V2` | Syllabus body |
| 11 | `Device_ItemSpawner_V3` | Stock holocrons + VillainA saber; timer **10 s** |
| 5 | `VerseDevice` | 2 workbenches + 3 saber spawners |
| 4 | `Device_Button_V2` | Combat sim + three spawner buttons (Verse `InteractedWithEvent`; classic bindings **[]**) |
| 4 | `Device_GuardSpawner_V2` | Type=**IO**, Columbus cosmetic, Button On Interact → Spawn, Volume OnEnter → Despawn |
| 3 | `Device_HUDController_V2` | Island PlayerInfo override; two workbench Hide HUD |
| 2 | `Device_Trigger_V2` | Workbench enter |
| 2 | `Device_CameraModes_Perspective` | Fixed-angle workbench cams |
| 2 | `Device_MutatorVolume_V2` | No build / no weapon fire |
| 1 | `Device_Volume` | Despawn guards |
| 1 | `Device_HUDMessage_V2` | Round-start Training Grounds prompt |
| 1 | `BP_Creative_Player_Spawner_Prop` | Island start |
| 1 | `Device_ExperienceSettings_V2_UEFN` | Island Settings |
| 93 | `FortStaticMeshActor` | Hall + screenshot placeholders |
| 0 | Marketplace / Item Granter / Hero / Conversation | other tiles |

## Authored kit (**Observed** — `assets.txt` + live)

```
Content/EnergyBladePrefabs/     EnergyBlade_Green, EnergyBlade_Red  (map referencers)
Content/EnergyBladeWorkbenchCore/  item_utils.verse
Content/UI/Widgets/            UW_Overlay, UW_Header, UW_Tab, UW_Item, UW_Details, …
Content/UI/Materials/          M_Hilt_* + M_TelekineticAbility_*
Content/Materials/             MI_Billboard_* screenshot placeholders
```

**Tier 2 mounts:** `/ENV_SPACE_THEME_UI_Assets` (PlayerInfo widget), `/ENV_SPACE_THEME_UI_Renders`, `/SPACE_THEME_C_Assets`, `/SPACE_THEME_A_Comp`, `/SPACE_THEME_EntityItems`, `/ContentHall` (demo displays), `/CRD_HenchmanSpawner`, `/COSMETIC_ASSETS` (IO Columbus). No Fab `.uref`.

## What this sample actually teaches

1. **Kyber swap is remove + 0.1 s sleep + add, because replicate order races.** techniques/kyber-component-swap.md.
2. **Holocron weapon vs EnergyBlade augment are different entity types.** Verse `AddItemDistribute`; Training is augment-only. techniques/weapon-vs-EnergyBlade-augment.md.
3. **Workbench is a per-player Verse session over authored UMG, not Marketplace.** techniques/workbench-umg-verse-menu.md. `button`+`button_slot`: verse-clickable-button.
4. **Custom Scene Graph items are Verse grants; stock holocrons can sit on Item Spawners.** Docs warn spawners cannot take custom components — this island follows that split.
5. **Island HUD is a mounted space-opera PlayerInfo widget, not ExtraAssets and not IceBase’s in-project PlayerHUD.** Live: `/ENV_SPACE_THEME_UI_Assets/PlayerHUD/UW_HUD_PlayerInfo_C`; HideElement minimap / team / build / equipped; ShowElement HUD / health / shields.
6. **Combat dummies are Guard Spawner Type=IO, not IceBase Guard CDs.** Button spawn + volume despawn. Contrast guard-patrol-volume-swap.
7. **Sandbox flags, not a round.** `allowBuilding=None`, `bInfiniteResources=true`, `bFallDamage=false`, `maxItemSlots=5`, BRStyle itemization, `matchmakingType=Off`. `hUDInfoType=AI Enemy Elimination` leftover.

## Official tutorial vs shipped

**External hub + dedicated chapter + two component references + Custom UI** (headed Chrome 2026-09-15). Verbatim: `_raw/energy_ability_sample/tutorial.txt`. Spreadsheet `docsUrl` empty.

### Lines up

| Tutorial | Shipped |
|---|---|
| theme Templates → energy blades and telekinetic abilities | picker; `featureSet` CustomGameplayFeature |
| Workshop + testing/training grounds | `Workshop/` + `Training Grounds/` |
| Steps 1–1.5 Scene Graph saber; 2 workbench; 3 Telekinetic entities; 4 items | DemoDisplay titles 1 / 2 / 3 / 4 |
| Entities invisible in viewport; placeholder meshes | Billboard 1.1.3; `Workshop/Screenshots/` FortStaticMesh |
| `EnergyBlade_item_component` + `kyber_crystal_[color]_component`; Arc/Stability 0–1 | `kyber_crystal.verse`; boards 1.3 / 1.4 |
| Workbench = UMG + Verse session + `ui_config` + menu device | two placed `menu_EnergyBlade_customization_device` |
| Weapon Holocron vs Verse-only augment; one augment at a time | boards 3.1.4 / 4.2; `GivePower` augments |
| Telekinetic Training not an item | no 4.11 Item Spawner; Verse `TrainingAugment` |
| Combat simulation button → troopers | 4× Guard Spawner IO, Button On Interact Spawn |
| Custom UI premade `UW_HUD_PlayerInfo_C` | island HUD override to `/ENV_SPACE_THEME_UI_Assets/…/UW_HUD_PlayerInfo_C` |
| Itemization required | live `customInventoryConfiguration` BRStyle |

### Does not line up

| Tutorial | Shipped |
|---|---|
| Verse using `/Fortnite.com/Items/EnergyBlade/SPACE_OPERA` and `/Fortnite.com/Items/TelekineticAbility/SPACE_OPERA` | `/Fortnite.com/Weapons/SPACE_OPERA` + `/Fortnite.com/Items/SPACE_OPERA` |
| `TelekineticAbility[NAME]_Weapon` (no `_V1`) | `TelekineticAbility*_Weapon_V1` / `*_EnergyBladeAugment_V1` |
| Migrate `items_utils.verse` | `item_utils.verse` |
| Workbench includes color/stability/arcing changer devices | files exist; **0 placed** |
| Add Telekinetic augment as **child of the saber entity** | Verse second `AddItemDistribute` |
| Custom HUD “preconfigured on Flat Grid / XL Grid”; HUD in cantina / The Warren; Player Info = project `UI > Widgets > PlayerHUD` | This island ships a HUD Controller in `UI/`; widgets are workbench `UW_Overlay` etc., **not** `UW_HUD_PlayerVitals` |
| Player Info override `UW_HUD_Vitals` (Custom UI editable table) | Live premade `UW_HUD_PlayerInfo` |
| Physics-enabled objects for Push/Pull | Island `bPhysicsEnabled=false` (PIE **Unverified**) |
| Hub list “Lighting” telekinetic ability | Lightning |
| Stamina on Heal/Barrier (dedicated prose) | Boards + reference: **Energy** (Training still uses Stamina costs) |
| Demo 4.4 “no adjustable components” for Throw | Reference: Damage, BuildingDamage, Cooldown |
| `.uefnproject` matchmaking 1 | Live maxPlayers **16** |
| Billboard 3.5.1 Floor2 “Telekinetic Training augment” | Placed 3.5 spawner `powerOverride=Shield` (4.11 is Training) |

Training rule: `snippets/verse/` win over tutorial Expand snippets. Do not emit the stale EnergyBlade/TelekineticAbility using-paths.

## Negative findings

- **Name implies Marketplace Telekinetic/EnergyBlade entitlements; files show Itemization grants.** No `BuyOffer`.
- **12 Verse files, 5 placed.** Remover / color / stability / arcing / `Telekinetic_power_spawner_device` are library, not Outliner exhibits.
- **`Telekinetic_power.Sense` → Heal augment** on `EnergyBlade_spawner_device` vs Awareness weapon on the unplaced Telekinetic-power spawner vs Awareness augment in `ui_config.GivePower`.
- **TabID 2 skipped** (`Crystal=3`).
- **Billboard 1.2.2** `kyber_krystal_red_component` typo.
- **1.5 / 3.5 interactive spawners live under Training Grounds/Tutorial Assets**, not next to Workshop boards.
- **`Akiat_Template_1.umap` leftover.**
- **`get_referencers` on `UW_Overlay` / `UW_Header` = []** — Verse constructs widgets at runtime.
- **Classic `ListEventBindings` on Buttons / Triggers = []** — Verse subscribe + camera bindings only.
- **`GetDeviceProperties` @editable device refs are subobject paths.**

## Capture backlog

| Item | Status |
|---|---|
| PIE kyber replicate race without `Sleep(0.1)` | **Waived** — comment in snippets is the lesson |
| PIE Heal/Barrier 0 energy / Awareness once / Saber Throw cooldown | **Waived** — documented Known Issues; not re-PIE’d |
| Item Spawner `spawn Item` asset paths | **Waived** — labels + folders 4.1–4.10; ObjectTools `itemToSpawn` failed |
| Sequencer / Blueprint graphs | **N/A** |
| Sign Workshop / Sign Training Grounds `title Text` | **Waived** — `description Text[0]` empty; wayfinding not syllabus |
| Guessed Custom UI slug `creating-a-custom-ui-in-fortnite` | **Closed** — empty hub chrome; real page is `star-wars-custom-ui-in-fortnite` |

## Live notes

- Mount `/92f0b488-49b0-4e1f-8d58-5df8b19e40ed`. Project Verse path `/epictemplates@fortnite.com/energy_ability_sample`. Plugin filename is that GUID. `projectId` `f976f411-a6b1-4b51-abeb-dc78d79a3886`.
- Did **not** `load_level` away from the default map. Used `load_actors` despite streaming-disabled tag.
- `ListFiles("")` is the digest package — list `/92f0b488-…` for sources.
- Folder on disk / title is **`energy_ability_sample`** (picker “energy blades and telekinetic abilities”). Island umap `EnergyBlades_and_Telekinetic_Powers`.
