---
id: digest-029
kind: digest
---

# materials_template — sample digest

> Offline `_raw/materials_template/`. Live
> `/7fbcab1a-4e37-44b5-aa3f-ad1c6d9a3fde/CE_MaterialExamples` (**Observed** — umap is `CE_MaterialExamples`, not `materials_template`).
> Docs: Materials discipline hub (not a template walkthrough). Spreadsheet Keep=Yes: *“excellent material tutorial, still useful.”* **No Verse.**

| | |
|---|---|
| Classification | **Walkable Material Editor curriculum** — 4 numbered halls, same pattern as Modeling |
| Content type | **kit**: **197 authored vs 357 placed** |

## Scale

| | |
|---|---|
| Authored | 197 (`M_` 116, `T_` 45, `SM_` 17, `ML_` 4 material-layer, `MI_` 1) |
| OFPA | 357 |
| Verse | **0** |
| `BP_DemoDisplay` | 144 — one caption per exhibit |
| `FortStaticMeshActor` | 128 |
| `Device_MutatorVolume_V2` | **30** — biggest device class, one per exhibit variant zone |
| Teleporters | 13 |
| `BP_DemoRoom` | 5 |
| Camera modes (Perspective) | 13 |
| Switches | 4 |
| Lumen Exposure Manager | 3 |
| Compatibility | **42.30** |
| Matchmaking | 16 |

Empty `Developers/felixkupis/Collections/`. Twenty-fourth first-party hit.

## What it is *for*

Four numbered halls, same structure as `modeling_template`:

| Hall (live Outliner) | Teaches |
|---|---|
| `1 - Mat Properties` | Blendmode, ShadingModel, TranslucentLightingMode, TwoSided, MaterialDomain, SeperateTranslucency (typo ships), VolumetricDirectional, OpacityMaskClipValue, CastTranslucentShadowMasked, RefractionDepthBias — **10 numbered exhibits** |
| `2 - Materials` | Main Material node pins: BaseColor, Metallic, Specular, Roughness, Emissive, Opacity, Opacity & Subsurface, OpacityMask, Normal, WPO, SubSurfaceColor, Refraction — **12 exhibits** |
| `3 - Instances` | Material → Instance → Instance-of-Instance → Groups (parameter override chain) |
| `4 - Advanced` | MasterMaterial, MaterialFunctions, Layers, VertexAnimation |

Each numbered exhibit folder pairs a `1-N_Layout/Camera+Switch` rig with a `Device_MutatorVolume_V2` that swaps the live material variant when you step in/switch.

## Device inventory (the design)

| Count | Class | Role |
|---|---|---|
| 30 | `Device_MutatorVolume_V2` | swap material property per exhibit |
| 144 | `BP_DemoDisplay` | captions |
| 128 | `FortStaticMeshActor` | preview meshes |
| 13 | `Device_CameraModes_Perspective` | per-exhibit camera |
| 13 | teleporter | hall navigation |
| 5 | `BP_DemoRoom` | the 4 halls + lobby |
| 4 | `Device_Switch_V2` | manual toggle exhibits |
| 3 | `BP_LumenExposureManager` | scalability parity (Day Sequence pattern) |
| 11 | `StaticMeshActor` | non-Fort statics |
| rest | decal / spotlight / PPV | dressing |

Authored kit worth copying: `ExampleContent/{Blueprint_RenderToTarget, Effects, Landscapes, Material_Nodes, MaterialInstances, MaterialLayers, MAXScripts?}`, `Global/Materials`, `PreviewMesh/`.

## What this sample actually teaches

1. **Numbered Outliner = syllabus**, same convention as `modeling_template`. See techniques/mutator-volume-material-swap.md.
2. **Mutator Volume, not Verse, drives the live-preview swap** — 30 volumes, zero script.
3. **Instance chain is a real hierarchy exhibit:** Material → Instance → Instance-of-Instance → Groups, not just a diagram.
4. **Property hall (10) is the settings a mesh author actually flips** (Blend Mode, Shading Model, Two Sided, Material Domain) — separate from the node-pin hall (12).
5. **Lumen Exposure Manager reused from Day Sequence pattern** — 3 copies, one per lighting-sensitive room.

## Official tutorial vs shipped

**External hub:** Materials → cards: Material Nodes and Settings, UI Materials, **Material Library**, Two-sided Materials, Material Effects.

This island is **not** a single walkthrough page — like Modeling, the **numbered halls are the tutorial**, cross-referenced by the Material Library (catalog of Epic/Fortnite base materials, device materials, landscape materials, material functions) and Material Nodes and Settings (Domain/BlendMode/ShadingModel defaults, node table).

`docsUrl` empty. Spreadsheet `/en-us/uefn/material-library-in-unreal-editor-for-fortnite` → the catalog page, not this island directly. Retrieved signed-in Chrome 2026-09-09 (browser had died mid-session; restarted headed, same profile, no re-login needed).

### Lines up

| Tutorial (Material Nodes and Settings) | Shipped hall 1 |
|---|---|
| Material Domain / Blend Mode / Shading Model defaults (Surface / Masked / Default Lit) | `1_5_MaterialDomain`, `1_1_Blendmode`, `1_2_ShadingModel` |
| Two Sided for hollow/thin meshes | `1_4_TwoSided` |
| Node table (Constant3Vector, Lerp, Coordinate, Constant4Vector, BumpOffset…) | hall 2 pin exhibits (BaseColor/Normal/Opacity/WPO) |

| Tutorial (Material Library) | Shipped |
|---|---|
| Device materials (Barrier/Switch) | not an exhibit here — catalog-only |
| Fortnite Base Materials (Simple Colors, Tiling) | `Global/Materials` |
| Epic Base Materials + Material Functions catalog | `ExampleContent/*` folders |
| Landscape Material (`MI_Landscape_Ch4`, `MI_Fortnite_Landscape_Customizable_01`) | present in authored set |

### Does not line up

| Topic | Tutorial | Shipped |
|---|---|---|
| Structure | discipline reference pages | **numbered walkable halls** (own syllabus, like Modeling) |
| Map name | — | `CE_MaterialExamples` |
| Typo | — | `SeperateTranslucency` folder |
| Device materials exhibit | catalog mentions them | **not** a hall here |
| Mutator Volume as the swap mechanism | not mentioned in either doc | 30 `Device_MutatorVolume_V2` — the actual lesson |

Train on halls + `ExampleContent/` kit + the Mutator Volume pattern. Do not emit “Barrier/Switch device materials” as this island — that’s catalog only.

## Negative findings

- Scratch `Developers/` folder.
- Compatibility 42.30, same generation as Modeling/voice_conversation_sample.
- 357 preview meshes are exhibit props, not a kit to reuse wholesale (the `M_*`/`T_*`/`ML_*` assets are the kit).

## Capture backlog

1. Live Mutator Volume settings on one `1_1_Blendmode` exhibit (which material param it flips). *Open, medium.*
2. `BP_DemoDisplay` caption text vs hall numbering. *Waive — numbering is the syllabus.*

## Live notes

Level `CE_MaterialExamples` confirmed. 4 numbered halls + Lobby + WorldSettings. Streaming off (not re-verified). Chrome browser process died mid-session and was restarted headed; session cookie persisted (no re-login needed) — future templates should expect this and just restart if `browser status` shows `running: false`.
