---
id: digest-030
kind: digest
---

# minigame_mastery_template — sample digest

> **Extractor provenance.** `tools/extract-template.py` (Pass A) + UnrealMCP on
> `localhost:8000` (Pass B). Default level
> `/c1f6b7ea-b147-4d63-9c93-ee233c903f40/minigame_mastery_template`. Verse QTE devices live on
> sibling `/…/minigame_mastery_template_Verse` (loaded then restored this pass).
> Hub: Minigame Mastery Template.
> `docsUrl` empty; spreadsheet Link in Desc is this page. Compat **43.00**. `featureSet`: **SOCIAL_WAVEGFS**.

| | |
|---|---|
| Classification | **theme mechanic gallery** — five social-elimination rooms (progress mesh / shove / throwables / carryable / multiplayer QTE), not a full social-elimination match |
| Content type | **catalog wrap** (54 authored vs 4916 OFPA across two maps + leftover `Arena_Export`) |
| Verse | **3 files** (all used on the `_Verse` map) |
| Fab `.uref` | **0** |

## Verdict first

The picker name is **Minigame Mastery**. The island is a **loading room with five labeled doors**, the same syllabus shape as social_deduction_sample but different hero devices. Teaching that survives deletion is the **device-only vs `_Verse` QTE collapse**, the **Fillable Container / progress mesh** progress mesh, and the **third ExtraAssets `UW_HUD_*` HUD Controller**. Voting, giftbox occupancy, and `SubscribeAgent` stay on Social Deduction. Greasy Grove galleries stay on social_game_sample.

## Scale (**Observed** — `summary.txt` + live)

| | |
|---|---|
| Authored assets | **54** |
| OFPA | **4916** — default **2371** + `_Verse` **2341** + leftover `Arena_Export` **204** |
| Maps | `minigame_mastery_template.umap` (Project Browser default) + `minigame_mastery_template_Verse.umap` |
| Verse files | **3** (`Scripts/qte_game_device.verse`, `qte_towers_device.verse`, `utils.verse`) |
| Verse devices placed | **0** on default; **3** on `_Verse` (`QTE Verse Game Logic Device` + two tower devices) |
| Data layers | **0** |
| Fab `.uref` | **0** |
| Compatibility | **43.00** |
| Matchmaking | 16 players; Island Settings `teamSize=Dynamic`, `teams.teamType=TeamIndex` |
| Streaming | Partitioned + actor folders + OFPA. **No** `LevelHasStreamingDisabled` tag (Social Deduction camp). Billboard `text` readable without `load_actors` (`bIsSpatiallyLoaded=false`) |

Empty `Developers/felixkupis/Collections/`. First-party scratch folder.

`experimental.*` (Python / toolsets / Scene Graph) in `.uefnproject` is **Unverified**. Content-side: **no** `__ExternalEntities__`. Docs’ Scene Graph tower-prefab tip is not shipped.

Largest file ~5.3 MB (`VFX/NS_LavaSplash_01.uasset`). Under the GHES **100 MB** cap.

## Two maps (**Observed**)

| Asset | Role |
|---|---|
| `minigame_mastery_template.umap` | Device-only. Five rooms. **11** SID in QTE (per-floor). Default open. |
| `minigame_mastery_template_Verse.umap` | Same rooms. Adds 3 Verse devices. QTE: **1 SID per tower** + corridor sample. |

See techniques/skilled-interaction-queue.md and twin-level-device-vs-verse.md.

## Level organisation (**Observed** — live `get_folders`)

```
StarterRoom/                 hub: 5 labeled doors + 10 player spawners
Progress Based Mesh Device/  jars + tree + progress mesh + CTL doll surprise
Shove Arena/                 shrinking platform + Shove item
Throwables/                  TileToss / Flying Stone vs moving dummies + Mascot score
Carryables/                  PebbleGame / VIP-mask throwers vs carrier
Multiplayer QTEs/            two lava towers (device-only or Verse)
Introduction Room/           empty leftover folder (0 actors)
Background/                  Island Settings
```

Billboard captions (70/70 on default): `_raw/minigame_mastery_template/billboard-texts.md`. `text` is on the **actor**.

## Device inventory (the design)

Counts are **project-wide OFPA** (both maps). Live default-map rooms are ~half.

| Count | Class | Role |
|---|---|---|
| 143 | `Device_Billboard_V2` | corridor syllabus (70 on default) |
| 72 | `Device_Teleporter_V2` | room routing (groups M/N/O/P/Q/W/X/Y/Z + arena) |
| 46 | `Device_Barrier_V2_Placed` | balcony / lava / arena walls |
| 37 | `Device_Button_V2` | start / RTS / grow-tree / ready |
| 28 | `Device_Timer_V2` | QTE floors, carryable respawn, throwables clock |
| 25 | `Device_HUDMessage_V2` | winner / loser / carrier / thrower |
| 24 | `Device_CinematicSequence` | QTE floor hide (10 sequences ×2 maps) + shove shrink + carryable hazard |
| 20 | `Device_CRD_CarryableItemSpawner` | PebbleGame + carrier/thrower spawners |
| 20 | player spawners | hub |
| 18 | `Device_TargetDummy_Track_V2` | throwables moving targets |
| 15 | `Device_PlayerCounter` | QTE floors (device map) / one per tower (Verse) |
| 14 | `Device_SkilledInteractionDevice` | QTE (11 default + 3 Verse, overlapping sample) |
| 10 | `Device_Fillable_Container` | Progress Based Mesh jars/tree |
| 10 | `Device_VFXSpawner_V2` | Mascot / sample feedback |
| 8 | `Device_PinballFlipper_V2` | carryable obstacles |
| 6 | `BP_ProgressMesh_Device` | ExtraAssets progress mesh + throwables score |
| 6 | `Device_PinballBumper_V2` | carryable obstacles |
| 3 | `VerseDevice` | **Verse map only** |
| 2 | `Device_HUDController_V2` | social-elimination UMG overrides (one per map) |
| 2 | `PlaysetRoot` | OFPA; not in live Outliner search — likely `Arena_Export` leftover |
| 0 | Disguise / Voting Group / Giftbox hiding prop | Social Deduction, not this island |

## Verse (3 files)

Copied to `snippets/verse/`.

| File | Job |
|---|---|
| `qte_game_device.verse` | Two-tower round: start buttons → `ArrayRace` first `FailedEvent` → offset timer → `EndGame` |
| `qte_towers_device.verse` | One SID + one timer + `[]` floor cines; `[agent]logic` success map; `BeginInteraction` / `EndInteraction` |
| `utils.verse` | Community `ArrayRace` (header: snippets/oYRX). **No** `SubscribeAgent` — that file is Social Deduction’s `utils.verse` |

## What this sample actually teaches

1. **Twin maps for the QTE room, not a second Outliner stack.** Default = per-floor SID graph. `_Verse` = `qte_game_device` + two `qte_tower_device`. Same pattern as Social Deduction’s quiz/giftbox. twin-level-device-vs-verse.md (now two sources). Detail: techniques/skilled-interaction-queue.md.
2. **`ArrayRace` is the first-to-fail primitive — and this island actually calls it.** Speedway shipped `ArrayRace` unused next to `ArraySync`. `qte_game_device.WaitForEnd` races `Tower.FailedEvent`. array-sync-race-primitive.md.
3. **Progress Based Mesh = Fillable Container; progress mesh is the Social skin.** Threshold meshes + Instant vs ContinuousRate. techniques/progress-mesh-threshold-swap.md.
4. **social-elimination HUD is ExtraAssets `UW_HUD_*` on a HUD Controller — third source (confirmed).** Live: PlayersInfoStack, EquippedInfoBlock, **and** QuickbarSlot. `showMinimap` / `showTeamInfo` = **HideElement**. Health/shields **DoNotOverride** (Social Deduction hid those too). Text chat MiddleRight / **TopLeft** (Social Deduction CenterRight). Island `nameplateDisplayMode=Never` matches the Custom UI chapter; Social Deduction shipped `Default`. Docs: both Mastery and Social Deduction; you **cannot copy** the UI folder. Originating write-up: social_deduction_sample ExtraAssets HUD. Walker NPC is `/EnvironmentPack_ExtraAssets/UI/Widgets/UW_HUD_*`.
5. **VIP masks are a VFX Powerup Custom Effect on the head socket, duration Infinite** (corridor billboards). Authored `Content/NS/VIPMask` has **0** asset referencers — live Niagara actors are labeled `NS SwashTrace Cat` / ExtraAssets, not that asset.
6. **PebbleGame is a branded Carryable Spawner.** Sample `Device Caryable - Sample` ships `canBeThrown=false` so the billboard’s “edit it to explode / throw” exercise is real. Pinball Flipper×4 + Bumper×3 are the moving obstacles.
7. **Shove arena = Switch → Cinematic Sequence shrink**, not Verse. Same no-Verse cine trick as trigger-cinematic-quake.md. Item Granter grants Shove; Damage Volume under the pad; Item Remover on fall.

## Official tutorial vs shipped

**External hub + 5 chapters + shared Custom UI** (headed Chrome 2026-09-14):

1. Minigame Mastery Template
2. Tracking Game Progression
3. Shove Item for Pushing Players
4. TileToss and Flying Stone Target Practice
5. Carry Items Through Obstacles
6. Multiplayer Skill Checks
7. social-elimination Custom UI — “Both the Minigame Mastery and Social Deduction templates feature the custom UI.”

Verbatim capture: `_raw/minigame_mastery_template/tutorial.txt`. Spreadsheet `docsUrl` empty; Link in Desc is the hub.

### Lines up

| Tutorial | Shipped |
|---|---|
| theme Templates → social-elimination → Minigame Mastery | picker; `featureSet` SOCIAL_WAVEGFS |
| Loading room, five doors, tutorial then minigame | Outliner folders match (Progress / Shove / Throwables / Carryables / QTEs) |
| `_Verse` sibling for QTE; `qte_game_device` + `qte_tower_device` | 3 VerseDevice OFPA only under `__ExternalActors__/minigame_mastery_template_Verse/` |
| Verse: one SID per tower, one timer per tower | Verse QTE: 3 SID (2 towers + sample), 3 timers, 2 player counters |
| Queue types Synchronous / Sequential / Random | corridor billboards; live tower SID `Synchronous` |
| progress mesh = variant of Progress Based Mesh | `BP_ProgressMesh_Device` ExtraAssets + `Device_Fillable_Container` tree/jars |
| CTL doll via Prop Mover or Cinematic | `Device_PropMover_V2` in progress room |
| Shove: Switch → cine shrink, granter, damage volume | `Switch - Start Sequencer` + `Arena Shrinking Cinematic Sequence Device` |
| Throwables progress mesh ×2 as score | `progress mesh - Throwables Score` / `Score2` |
| Carryable: PebbleGame + VFX Powerup VIP mask + pinball | 10 carryable spawners; VFX Powerup sample; Flipper×4 Bumper×3 |
| Custom UI ExtraAssets widgets; cannot copy UI folder | `/SOCIAL_WAVE_ExtraAssets/UI/Widgets/UW_HUD_*`; HideElement minimap/team |
| Custom UI: hide nameplates in Island Settings | `nameplateDisplayMode=Never` |

### Does not line up

| Tutorial | Shipped |
|---|---|
| QTE device-only SID **×10** | Live **11** (corridor sample + 10 floors) |
| QTE Player Counter **×11**, Volume **×1**, Teleporter **×8**, Button **×3** | **13** / **3** / **10** / **5** on default QTE folder |
| Shove Item Granter **×1**, Remover **×1**, Teleporter **×7**, Button **×2** | **2** / **2** / **10** / **3** |
| Throwables Billboard **×3**, Barrier **×6**, Track Dummy **×6**, Player Ref **×1**, VFX **×1** | **9** / **7** / **9** / **2** / **2**; extra Item Spawners labeled `Ddajki` |
| Carryable Spawner **×8**, Billboard **×5**, Timer **×2**, Teleporter **×4** | **10** / **20** / **5** / **7**; Target Dummy (not Track) in tutorial room |
| Scene Graph prefab towers | **0** entities; experimental Scene Graph flag **Unverified** |
| List item **“Minigames Mastery Template”** (theme page) | H3 / picker **Minigame Mastery** |
| Custom UI H1 “**Lean** to add…” | Typo in docs |

Training rule: corridor billboards + `snippets/` = ground truth. Do not emit docs SID×10 or “Scene Graph generates the towers.”

## Negative findings

- **Name implies mastery of a full social-elimination; files show five exhibits.** No round manager, no disguise/vote/giftbox.
- **`Introduction Room/` is empty.** Folder path exists; 0 actors. Same leftover class as Social Deduction’s empty `Room - Tutorial/`.
- **`Content/__ExternalActors__/Arena_Export/` (204 OFPA)** is not an Outliner folder — leftover export next to the two umaps.
- **`Content/NS/VIPMask`, `Content/NPCs/Guard`, `Content/Meshes/SM_Arena_Floor`, authored `NS_LavaBurst` / `NS_LavaSplash_01`:** `get_referencers` **[]**. Live Niagara actors are still labeled `NS LavaBurst*` (OFPA often does not show as asset referencers — same trap as Social Deduction lava). Guard definition and VIPMask asset are the ones to treat as unused.
- **Billboard spelling `Niagra`.** Twice, VIP-mask corridor.
- **Actor labels `Device Caryable - Sample` and Item Spawner `Ddajki`.** Typos shipped.
- **`GetDeviceProperties` returns Verse subobject paths** (`Scripts-qte_tower_device_0.cinematic_sequence_device_4`), not placed cine actors. Same unanswerable as Social Deduction’s voting device. Towers array **length 2** and `towerFloors` **length 5** are readable.
- **`hUDInfoType=AI Enemy Elimination` with one shove NPC.** Leftover default, same family as Starter/Prison.

## Capture backlog

| Item | Status |
|---|---|
| Sequencer keys on `QTE_Tower_*_Floor_*` and `ShoveArena_ChangeSize` | **Waived** — keyframes unreachable; floor count Observed via `towerFloors` |
| VFX Powerup Custom Effect asset path | **Waived** — billboards + ExtraAssets SwashTrace Niagara labels sufficient |
| PIE all five minigames | **Waived** — structure + bindings sufficient |
| `Arena_Export` actor list | **Waived** — leftover OFPA, not the syllabus |
| Fillable `thresholdMeshes[0..3]` mesh refs | **Waived** — occupancy Observed; meshes are the lesson not the names |

## Live notes

- Mount `/c1f6b7ea-b147-4d63-9c93-ee233c903f40`. `VerseToolset.ListFiles("")` also lists `__ExternalActors__` if recursive — list `Scripts/` instead.
- Default map has **no** ScriptDevices. Load `_Verse` to call `ListDeviceProperties`. Restore default after.
- Billboard `text` did **not** need `load_actors` on this map (`bIsSpatiallyLoaded=false`).
- `get_properties` rejects the whole call if `ActorLabel` is in the list.
- Folder on disk / title is **`minigame_mastery_template`** (picker “Minigame Mastery”).
