---
id: digest-031
kind: digest
---

# modeling_template — sample digest

> Offline `_raw/modeling_template/`. Live
> `/cca3b104-5c9e-48d6-9542-2c8bd0340014/modeling_templateTools` (**Observed** — umap is `modeling_templateTools`, not `modeling_template`).
> Docs: Modeling *discipline* hub (not a template walkthrough). Spreadsheet Keep=Yes. **No Verse.**

| | |
|---|---|
| Classification | **Walkable Modeling Mode curriculum** — numbered halls for create/edit/UV/collision/units |
| Content type | **kit**: **87 authored vs 170 placed** |

## Scale

| | |
|---|---|
| Authored | 87 (`SM_MT_*` steps, collision variants, locker cooked/uncooked, TNT/keg) |
| OFPA | 170 |
| Verse | **0** |
| `FortStaticMeshActor` | 86 |
| `BP_DemoDisplay` | 64 — one exhibit stand each |
| Teleporters | 10 |
| Demo rooms | 4 (`BP_DemoRoom`) |
| Compatibility | **42.30** (older than Wave 1’s 43.00) |
| Matchmaking | 16 |

Empty `Developers/felixkupis/Collections/`. Twenty-second first-party hit.

Content is nested `Content/Content/{ModelingTools,StaticMesh,Units_Grid}` plus leftover `Content/Meshes/`.

## What it is *for*

Three halls. You walk them; each numbered Outliner folder is one Modeling Mode tool (or one static-mesh inspector setting). 10 teleporters hop exhibits. Not a game.

| Hall (live Outliner) | Teaches |
|---|---|
| `1-Tools` | Modeling Mode: Create → Edge Loop → Shaping → Extrude → E&L → Bevel → Material ID → Greebles → Pattern → Merge → Polygroups → UVs → Finished → Sculpt / Deform / Stylize → TNT / keg extras |
| `2 - Static Meshes` | Inspector: UVs good/bad, extra UV channel, Material IDs, collision **None / Box / Capsule / Sphere / Auto Convex / Complex as Simple / 24-DOP** |
| `3 - Units and Grids` | Units, grid snap, world scale, **cooked vs uncooked** locker (`SM_Locker_Uncooked`) |

Meshes: `SM_MT_One` … `Fifteen_*` (the TNT barrel build), `SM_Mesh_Collision*`, `SM_Unit_*`.

## Device inventory (the design)

| Count | Class | Role |
|---|---|---|
| 86 | `FortStaticMeshActor` | the meshes |
| 64 | `BP_DemoDisplay` | captions / stands |
| 4 | `BP_DemoRoom` | halls |
| 10 | teleporter | hop exhibits |
| 1 | trash can | leftover basketball mini |
| 1 | HUD controller | |
| rest | World* / spawner | boilerplate |

No Animated Mesh, no Day Sequence, no VerseDevice.

## What this sample actually teaches

1. **Number the Outliner like a textbook.** `1.1 Create` … `3.6 DOP` is the syllabus. See techniques/walkable-modeling-curriculum.md.
2. **Collision is seven exhibits, not one checkbox** — None / primitive / Auto Convex / Complex as Simple / DOP.
3. **Cooked vs uncooked** locker pair — same prop, different cook.
4. **UVs: bad vs good vs extra texcoord** as adjacent meshes, not a lecture.
5. Modeling Mode *tools* live in hall 1; *asset inspector* lives in hall 2. Don’t conflate.

## Official tutorial vs shipped

**External** hub: Modeling

This is the **discipline landing**, not “walk this island.” Cards: Read-Only Asset Editors, Modeling Mode, Creating a Column, Greyboxing, Architectural Guidelines, Modeling Tips.

`docsUrl` empty. Spreadsheet `/en-us/uefn/modeling-in-unreal-editor-for-fortnite` → same hub. Retrieved signed-in Chrome 2026-09-09. Bare `/fortnite/modeling-in-unreal-editor-for-fortnite` **403’d** until the `/en-us/uefn/` form (session-dependent).

### Lines up

| Tutorial (Modeling Mode page) | Shipped halls |
|---|---|
| Shapes / Create (extrude, path, revolve, merge, duplicate, pattern) | `1.1 Create`, `1.2 Edge Loop`, `1.4 Extrude`, `2.3 Merge`, `2.2 Pattern` |
| PolyEdit / Deform (sculpt, warp, lattice) | `1.3 Shaping`, `4.1 Sculpting`, `4.2 Sculpt Deform`, `4.3 Stylize` |
| Attribute / UVs / Baking | `2.0 Material ID`, `2.4 Polygroups`, `2.5 UV's`, `1.6`/`2.x` unwrap |
| Volumes = collision | hall `3.0 - Collision` |
| Greyboxing mentioned | hall 3 units/grid; Greyboxing **template** is a later Wave 2 island |

### Does not line up

| Topic | Tutorial | Shipped |
|---|---|---|
| Island walkthrough | **none** — tool reference | the numbered halls *are* the walkthrough |
| Creating a Column | separate how-to | **not** an exhibit here |
| Map name | “Modeling” | `modeling_templateTools` |
| Nested `Content/Content/` | — | ships |
| Compatibility | — | **42.30** vs Wave 1 **43.00** |
| Basketball trash can | — | leftover `B_DowntownTrashCan_BasketballMinigame` |

Train on halls + authored `SM_MT_*` / collision meshes. Do not emit “Creating a Column” as this island.

## Negative findings

- Scratch `Developers/` folder.
- Nested Content path + `CubeGridToolOutput_*` scratch meshes.
- Compatibility 42.30 — don’t assume 43 APIs.

## Capture backlog

1. Live collision method on each `SM_Mesh_Collision*` vs folder name. *Open, low — names already match.*
2. `BP_DemoDisplay` text. *Waive — Outliner is the syllabus.*

## Live notes

Level `modeling_templateTools`. Folders 1-Tools / 2 Static Meshes / 3 Units. Streaming off (not re-verified).
