---
id: digest-032
kind: digest
---

# my_first_island_template — sample digest

> Offline `_raw/my_first_island_template/`. Live
> `/caa35cfd-0df6-40fb-b38b-b0834e99c62e/my_first_island_template` (**Observed**).
> Docs: 6-chapter “Build Your First Island” hub + a **follow-on** “Level Up with Verse” that is **not** this island’s Verse.

| | |
|---|---|
| Classification | **First-project shooting gallery** — devices + one Verse manager |
| Content type | catalog wrap: **2 authored vs 86 placed** |

## Scale

| | |
|---|---|
| Authored | 2 (GameFeatureData + umap) |
| OFPA | 86 |
| Verse | **1** (`shooting_range_manager_device.verse`) |
| Matchmaking | 1 player |
| Fab | 0 |
| Live folders | **none** (empty `get_folders`) |
| Compatibility | 43.00 |

Spreadsheet: *Should replace Intro to UEFN.* Same job as `intro_to_uefnfk1` (walk an island) with **one working Verse device** instead of 35 boards.

Empty `Developers/felixkupis/Collections/`. Fifteenth first-party hit.

## What it is *for*

Single-player range: 3 track dummies (good, +100) and 3 teddy Target Dummy (bad, −100). Hit → pop down in **one shot** (Verse, not device HP). Score Manager awards via `SetScoreAward` + `Activate` on player[0]. Item Granter for the gun. Barrier keeps the player back. Theme: neon / bunny / Durr posters.

## Device inventory

| Count | Class | Role |
|---|---|---|
| 3 | `Device_TargetDummy_Track_V2` | good targets (`shooting_range_target_track_device`) |
| 3 | `Device_TargetDummy_V2` | bad targets (`shooting_range_target_device`) |
| 1 | `Device_ScoreManager_V2` | ±score |
| 1 | `Device_ItemGranter_V2` | weapon |
| 1 | `Device_Barrier_V2_Placed` | keep-off line |
| 1 | `VerseDevice` | `shooting_range_manager_device` |
| 11 | billboards | captions |
| 1 | audio player | ambience |
| rest | floors/walls/props | throwaway |

## Verse

Snippet: `snippets/verse/shooting_range_manager_device.verse`

Six `@editable` targets + ScoreManager + `@editable` scores (100 / −100).

`OnBegin`: each `HitEvent.Subscribe(OnXHit)`. Callback: `AdjustScore` then `PopDown()`.

```verse
AdjustScore(Value:int):void=
    ScoreManager.SetScoreAward(Value)
    if (Player:player = GetPlayspace().GetPlayers()[0]):
        ScoreManager.Activate(Player)
```

One Score Manager for both signs — Verse changes the award, then Activate. First-player only (matches 1-player island).

**Not in this file:** combo, timer, bonus target, weapon leveling — those are the *Level Up with Verse* follow-on (player writes them).

## What this sample actually teaches

1. **A Verse device is a puppet for Creative devices.** `@editable` refs + `HitEvent.Subscribe` + `PopDown` — first-lesson shape. See techniques/hit-subscribe-score.md.
2. **One Score Manager, many awards:** `SetScoreAward` then `Activate` — don’t place two managers.
3. **Track vs dummy types:** good = `shooting_range_target_track_device`, bad = `shooting_range_target_device`. Mixing them is a compile/wiring error.
4. **Tutorial is a *build* path; the template is the *finished* mini-game.** Ch.1–4/6 are “you place this”; the island already has it.
5. **Follow-on docs are not this packet.** Don’t train combo/timer Verse as shipped here.

## Official tutorial vs shipped

**External** hub: Build Your First Island in Fortnite

Chapters (`/documentation/fortnite/first-island-0N-…-in-fortnite?lang=en-US`):

1. Create and Explore a Project
2. Place and Customize a Device
3. Build a Shooting Gallery
4. Playtest Your Island
5. Spice Up the Gameplay with Verse
6. Use Props to Set a Theme

Follow-on (player writes; **not** in this repo): Level Up with Verse — combo / timer / bonus / weapons / refactor.

`docsUrl` empty. Spreadsheet link is the hub (`/en-us/fortnite/build-your-first-island-in-fortnite`).

### Lines up

| Tutorial | Shipped |
|---|---|
| Class `shooting_range_manager_device` | same file |
| Good = track devices, bad = target dummies | OFPA 3+3 |
| `HitEvent.Subscribe` → `PopDown` | same |
| Score Manager + `@editable` ±scores | 100 / −100 |
| Item granter, barrier, playtest loop | devices present |
| Hide Verse device in-game | **External** User Options |

### Does not line up

| Topic | Tutorial | Shipped |
|---|---|---|
| Purpose | empty project, *you* add devices | **finished** gallery |
| Docs type name slip | once `shooting_range_track_device` | `shooting_range_target_track_device` |
| Combo/timer/bonus/weapons | Level Up hub | **absent** — don’t emit as this sample |
| Outliner folders | — | none live |

Train on snippet. Emitting combo Verse or `shooting_range_track_device` is doc regurgitation.

## Negative findings

- Almost no authored art; don’t treat as a kit.
- First-player `GetPlayers()[0]` — not multiplayer-safe (island is 1P).
- Scratch `Developers/` folder.

## Capture backlog

1. Live `@editable` wiring on the Verse device (which dummy is GoodTarget1). *Waive — counts match.*
2. Billboard captions vs ch.1–4 syllabus. *Open, low.*

## Live notes

Level confirmed. No actor folders. Streaming off (partitioned OFPA).
