---
id: digest-033
kind: digest
---

# narrative_roleplaying_template — billboard syllabus

**Observed** 2026-09-15 via UnrealMCP `get_properties` `text`. 26 `Device_Billboard_V2`. Syllabus boards are **Create Only**; room signs are **Always**, One Sided.

## Syllabus (Create Only)

| Label | Text |
|---|---|
| Billboard2 | On player spawn, the Disguise device triggers, changing the player's outfit to a AllyFaction trooper. |
| Billboard3 | The Disguise device doesn't replace the player portrait, so the player portrait is disabled with the HUD Controller device. |
| Billboard4 | This Pop-up Dialog device welcomes the player, sets the scene, and, provides guidance on what to do. The text box that displays has a new look that captures the space-opera™ theme. |
| Billboard5 | This HUD Controller device includes custom widgets to capture the space-opera™ theme in the UI. |
| Billboard6 | There are over 30 new characters to create with, like rebels, Guards, and civilians. |
| Billboard10 | This NPC is unfriendly, refusing to talk until the player completes specific dialog branches with DJ R0-B0. |
| Billboard16 | This Verse device tracks quest objectives, progress, and rewards, and updates Conversation devices when NPCs receive new dialog. |
| Billboard17 | Player decisions that impact the game state should return feedback to the player. These HUD Message devices create notifications for raising or lowering attitude and reputation. |
| Billboard26 | This NPC uses multiple branches of unique dialog. The Conversation device is using the random variant node and Verse to deliver unique random responses. |
| Billboard29 | This Stat Creator device creates a Reputation stat and tracks it. Completing quest objectives increases this stat, and reaching a certain Reputation score is needed to unlock the final quest objective. |
| Billboard30 | This NPC acts as a quest reward--an example of rewarding players with interesting story experiences instead of loot. |
| Billboard31 | Like the starting message, you can use the Pop-up Dialog device to end a quest, continue narrative, steer the player toward new objectives, or offer the player a choice of rewards. |
| Billboard32 | This Lock device on this door only unlocks when the player completes certain objectives. |
| Billboard33 | Anything can be an NPC. This NPC utilizes an invisible Character Device alongside a Junk Droid Head asset. |
| Billboard34 | The Verse devices for the NPCs manage at which stage of the quest a Conversation device is enabled. |
| Billboard47 | It's helpful to have an on-screen indicator directing the player toward quest objectives, especially if a quest has multiple steps. |

## Room signs (Always)

| Label | Text |
|---|---|
| Billboard27 | CHANGING ROOM |
| Billboard28 | DJ BOOTH |
| Billboard35 | HYDRATION HUB |
| Billboard37 / 39 / 41 | SCAVENGER VAULT |
| Billboard36 / 38 / 40 / 42 | CAPTAIN'S LOUNGE |

No board named **The Warren** — that name is the welcome popup title + tracker copy, not a room sign.
