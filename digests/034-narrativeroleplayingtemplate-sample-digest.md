---
id: digest-034
kind: digest
---

# narrative_roleplaying_template — sample digest

> **Extractor provenance.** `tools/extract-template.py` (Pass A) + UnrealMCP on
> `localhost:8000` (Pass B). Default level
> `/1987d6e3-afde-480f-ba1b-c87d024ab3f2/Cantina_Roleplay` (**Observed**).
> Docs: hub Working With space-opera™ Islands
> + dedicated Narrative and Roleplaying
> + Custom UI
> + Faction Reputation
> + Warren NPC Conversations
> + Conversations and Quests.
> `docsUrl` empty. Guessed `…-in-unreal-editor-for-fortnite` **hub-redirects**.
> Compat **43.00**. `featureSet`: **CustomGameplayFeature**.

| | |
|---|---|
| Classification | **Walkable quest cantina lab** — 1-player scripted Conversation + Verse, not IceBase/EnergyBlades/Persona |
| Content type | **kit of authored UI+Dialogue wrapping a catalog cantina** (101 authored vs 955 OFPA) |
| Verse | **3 files** (5 placed VerseDevices) |
| Fab `.uref` | **0** |

## What it is

The picker name is **Narrative and Roleplaying**. The island is **The Warren** — a walkable cantina with four talkable NPCs (DJ R0-B0, Caris, Scavenger Vex, Captain Journ), 29 scenery mannequins, and one scripted quest (`cantina_objectives`). Teaching that survives deletion is **Conversation Device trees enabled/disabled from Verse**, and **space-opera Custom UI in `Content/UI/Widgets`**. Persona stays on persona-voice-structured-output.md. IceBase PlayerHUD contrast: ice_base_sample. EnergyBlades workbench UMG+Verse: workbench-umg-verse-menu.md. ExtraAssets HUD coincidence stays **three** (Social / Mastery / Walker). Marketplace stays on entitlement-vs-offer — this island has none.

## Scale (**Observed** — `summary.txt` + live)

| | |
|---|---|
| Authored assets | **101** |
| OFPA | **955** |
| Maps | `Cantina_Roleplay.umap` + leftover `Akita_Template_3.umap` (1.5 KB) |
| Verse files | **3** (live `ListFiles` on `/1987d6e3-…/Scripts`; do not recurse `""`) |
| Verse devices placed | **5** (1× `quest_manager_device`, 4× `dialog_device`) |
| Data layers | **0** |
| Fab `.uref` | **0** |
| Compatibility | **43.00** |
| Matchmaking | Live Island Settings **1** / Solos; `.uefnproject` `maxPlayers` **1** |
| Streaming | Partitioned + actor folders + OFPA. **`LevelHasStreamingDisabled=1`**. Sampled actors still `bIsSpatiallyLoaded=false`. |

Empty `Developers/felixkupis/Collections/`. First-party scratch folder.

`experimental.*` (Python / toolsets / Scene Graph / Persona) in `.uefnproject` is **Unverified**. Content-side: **3** `.verse`, **0** Persona devices, **18** `Device_Conversation`, `EnableSceneGraph: true` in `.uplugin`, `bEnablePersonaDevice: false`.

Whole folder ~**20 MB**. No file near the GHES **100 MB** cap.

## Level organisation (**Observed** — live `get_folders` + folder counts)

```
Billboards/                            26 — Device_Billboard_V2 (syllabus Create Only + room signs Always)
Characters/                            64
  Captain/                              1 + 2 Conversation
  Caris/                                1 + Attitude 3 + Conversation 6
  DJ Droid/                             2 + Attitude 3 + Conversation 7 + SFX 3
  Scavenger/                            1 + Attitude 3 + Conversation 3
  Side Characters/                     29 — Do Not Interact mannequins
Environment/
  Cantina/                            711 — IcedTea / Apollo / PlaysetRoot “Cantina Prefab”
  Landscape/                            1
  Vista/                               69 + Periphery 9
Guidance/                              15 — arrows 5, lights 5, map indicators 5
Interactive Elements/                   4 — Billboard Toggle 2, Changing Room (Button + Disguise Device2)
Setup/                                  7 — Character & Game Setup 3 (spawn pad + Disguise Device + …), Island Setup 4
UI/                                    14 — HUD Controller + 2 popups + notifications + Tracker
Verse Devices/                          5
```

Syllabus body on `Device_Billboard_V2`. Verbatim: billboard-texts.md.

## Device inventory (the design)

| Count | Class | Role |
|---|---|---|
| 33 | `Device_Character_V2` | 4 talkable (Talk / Send Event Only) + 29 scenery (Do Not Interact) |
| 26 | `Device_Billboard_V2` | Syllabus + room signs |
| 18 | `Device_Conversation` | One bank per quest stage; Custom + Lightside widget |
| 10 | `Device_HUDMessage_V2` | Reputation / attitude notifications (default HUD Message widget) |
| 5 | `VerseDevice` | Quest manager + 4 dialog devices |
| 5 | `Device_MapIndicator_V2` | Pulse targets; `enabledOnGameStart=true` |
| 4 | `Device_Keylock_V2` | All `starts Locked=Locked`; only Lock4 has a binding |
| 4 | `Device_StatCreator` | Reputation Always + 3 NPC Attitude `enabled Phase=None` |
| 4 | `Device_Switch_V2` | Per-NPC first-talk gates |
| 3 | `Device_Powerup_Stat` | Attitude pickups mag **40 / 30 / 10** |
| 2 | `Device_DisguiseDevice` | Spawn Vector (`applyOnSpawn=true`); Setup Prime (`false` + Button) |
| 2 | `Device_PopUpDialog_V2` | Welcome **The Warren** Game Start; Quest Complete Lightside |
| 2 | `Device_CRD_AudioPlayer` | DJ SFX |
| 1 | `Device_HUDController_V2` | In-project PlayerInfo; HideElement minimap/team/build/inventory |
| 1 | `Device_Tracker_V2` | `Quest: Anchor Point`; Verse assigns (`assignOnGameStart=false`) |
| 1 | `SPACE_THEME_SongDevice_CantinaBand` | Cantina music |
| 1 | `PlaysetRoot` | Cantina Prefab |
| 1 | `BP_Creative_Player_Spawner_Prop` | Also **Disable**s conversations on spawn |
| 1 | `Device_Button_V2` | Changing-room ApplyDisguiseToInstigator |
| 1 | `Device_Barrier_V2_Placed` | Bindings **[]** |
| 1 | `Device_InputTrigger` | Billboard toggle |
| 1 | `Device_ExperienceSettings_V2_UEFN` | Island Settings |
| 82 | `FortStaticMeshActor` | Dressing |
| 0 | Persona / Guard Spawner / Marketplace / Item Granter / Hero | other tiles |

## Authored kit (**Observed** — `assets.txt` + live)

```
Content/Dialogue/          Captain 2, Caris 10 (incl. unused Bartender_*), DJDroid 7, Scavenger_Vex 3
Content/Scripts/           dialog_device.verse, quest_manager_device.verse, utility_module.verse
Content/UI/Widgets/
  Conversation/            UW_UI_Conversation_{Lightside,Darkside,Neutral} + UW_BTN_Conversation
  PlayerHUD/               UW_HUD_{PlayerInfo,Equipped,Quickbar,PlayerVitals,TeamVitals}
  Popup/                   UW_UI_Popup_{Lightside,Darkside,Neutral} + UW_BTN_Popups
  Stats/                   UW_UI_Stats
  Tracker/                 UW_UI_Tracker_Lightside
Content/UI/Materials/      M_Character_{Caris,DJDroid,Scavenger,Player}
```

**Tier 2 mounts:** `/SPACE_THEME_A_Comp` 581 + `/SPACE_THEME_A_Assets` 436 (cantina gallery), `/THEME_PACK` 1101, `/CRD_Conversation_Device` 116, `/ENV_SPACE_THEME_UI_Assets` 67, `/ENV_SWCosmetics` 62, `/CRD_Mannequin` 203, leftover GUID `/735ba659-…` (stale Tracker widget). No Fab `.uref`.

## What this sample actually teaches

1. **Conversation Device banks + Verse Enable/Disable by quest enum, not Persona.** techniques/conversation-device-verse-quest.md. Character `OnInteractedWith` → `InitiateConversation`. Contrast voice persona-voice-structured-output.md.
2. **space-opera Custom UI is authored under `Content/UI/Widgets`, HUD Controller in Outliner `UI/`.** techniques/custom-ui-widgets-folder.md. This island **is** the Custom UI chapter’s cantina / The Warren example. IceBase contrast: `Content/UI/PlayerHUD` (no `Widgets/`), Outliner `Devices/`, Equipped + Quickbar **on**. EnergyBlades PlayerInfo is a **mount**. ExtraAssets HUD stays three. Deeper viewmodels: umg-viewmodel-device-binding.md. Clickable `button`+`button_slot` is **not** this island — verse-clickable-button / EnergyBlades workbench.
3. **Reputation is one Always Stat Creator; NPC attitude is three Phase=None stats + HUD Messages.** Verse `SetValue` + `hud_message_device.Show`. Not Marketplace `BuyOffer`.
4. **Talkable NPCs are Character Device V2, not Guards.** `interactType=Send Event Only`, `interactText=Talk`. DJ `character=null` + Junk Droid Head (billboard 33). Caris cosmetic is BR **Frogman** (`CID_663`). Captain `BravePebble_B2`. Scavenger `CID_518`. 29 side NPCs `Do Not Interact`. Guard CDs stay on IceBase.
5. **The cantina mesh is a mounted `/SPACE_THEME_A_*` + IcedTea/Apollo catalog, not 101 authored rooms.** Authored budget is UI + Dialogue. One `PlaysetRoot`.
6. **1-player sandbox flags, not an open RPG.** `maxPlayers=1`, `allowBuilding=None`, `bFallDamage=false`, `bInfiniteResources=true`, `maxItemSlots=5`, `matchmakingType=Off`, `customInventoryConfiguration=null`. `hUDInfoType=AI Enemy Elimination` leftover.
7. **Starting NPC is not free.** Verse `StartQuestWithDJ`: Caris-first awards `QuestAccepted` then jumps to `TalkToDJAboutEskay1_2`. Docs say who-first has no impact.

## Official tutorial vs shipped

**External hub + dedicated Narrative + Custom UI + Faction Reputation + Warren NPC Conversations + Conversations and Quests** (headed Chrome 2026-09-15). Verbatim: `_raw/narrative_roleplaying_template/tutorial.txt`. Spreadsheet `docsUrl` empty.

### Lines up

| Tutorial | Shipped |
|---|---|
| theme Templates → Narrative and Roleplaying | picker; `featureSet` CustomGameplayFeature |
| Conversation Device + Verse; faction reputation | 18 Conversation + 5 VerseDevices + Stat Creators |
| Cantina / The Warren; HUD in Outliner **UI** | Welcome title **The Warren**; HUD Controller in `UI/` |
| Hide minimap / build / inventory | `HideElement` |
| Player Info override in-project | `…/PlayerHUD/UW_HUD_PlayerInfo_C` |
| Popup Lightside chrome | both popups `templateOverrideClass` → `UW_UI_Popup_Lightside_C` |
| Editable widgets under `UI > Widgets` | `Content/UI/Widgets/{Conversation,PlayerHUD,Popup,Stats,Tracker}` |
| Character Talk → InitiateConversation | classic bindings on all 18 |
| DJ as invisible Character + Junk Droid Head | `character=null`; `BP_IcedTea_JunkDroidHead_C_B` |
| Tracker “Anchor Point” / explore The Warren | live title + description |
| Reputation Always; attitude on interact | Reputation `enabled Phase=Always`; Attitude `None` |

### Does not line up

| Tutorial | Shipped |
|---|---|
| Player Info override `UW_HUD_Vitals` | Live **`UW_HUD_PlayerInfo`** |
| Disguise table Civilian:Random | Live **SnowSurgeon_Vector** (spawn, Changing Room folder, `applyOnSpawn=true`) + **SnowSurgeon_Prime** (Setup, button, `false`) |
| Starting NPC has no impact | Verse `StartQuestWithDJ=false` awards `QuestAccepted` then `TalkToDJAboutEskay1_2` |
| Objectives list omits `TalkToDJAboutEskay1_2` | Enum includes it |
| `quest_manager.verse` | Shipped **`quest_manager_device.verse`** |
| Lock unlocks on quest objectives (billboard 32 / docs) | Lock4 Unlocks on **Disguise** `OnDisguiseApplied`; locks 1–3 bindings **[]**; Verse has no `keylock` API |
| Custom HUD “preconfigured on Flat Grid / XL Grid” | **This** island ships the Widgets folder + HUD Controller |
| IceBase Custom UI chapter (cantina / Widgets) | IceBase is `Content/UI/PlayerHUD`, no `Widgets/`; **this** packet owns that chapter |

Training rule: `snippets/verse/` win over tutorial filenames / who-first prose. Do not emit Persona `PromptToTalk` as this island. Do not emit Civilian:Random as the shipped disguise.

## Negative findings

- **Name implies an open RPG; files show a 1-player scripted cantina quest.** No Persona, no Guard combat, no Marketplace, `maxPlayers=1`.
- **4 `Bartender_*` ConversationDatabases under `Dialogue/Caris/` are unused.** Live banks are `Caris_Conversation_*` only. Outliner `UI/Notifications/…/Bartender Attitude` is **0** actors.
- **Darkside / Neutral conversation + popup widgets are authored, unused.** All 18 conversations + both popups are Lightside.
- **Caris cosmetic is BR Frogman**, not a space-opera CID.
- **`Akita_Template_3.umap` leftover.** Stale mount `/735ba659-…` still referenced by Tracker widget class.
- **`get_referencers` on widgets / dialogue = []** despite live overrides — same IceBase trap; confirm on placed devices.
- **`GetDeviceProperties` @editable arrays are subobject `refPath`s** (`Scripts-cantina_quest_N`, `Scripts-dialogpath_N`). ObjectTools cannot read those Verse struct fields by friendly name. Subobject names skip `cantina_quest_3` (9 rows: 0,1,2,4–9).
- **Three Keylocks unwired.** Billboard 32 overclaims.
- **Map indicators `enabledOnGameStart=true` while Verse also `ActivateObjectivePulse`.** Dual system.

## Capture backlog

| Item | Status |
|---|---|
| PIE full quest / who-first branch | **Waived** — Verse `StartQuestWithDJ` is the lesson |
| Conversation graph internals (random variant node) | **Waived** — billboard 26; graphs are binary |
| Per-row `cantina_quest` Objective enum values | **Waived** — ObjectTools cannot read Verse structs |
| Sequencer / Blueprint graphs | **N/A** |
| Guessed UEFN-suffix Custom UI slug | **Closed** — hub-redirect; real page `star-wars-custom-ui-in-fortnite` |

## Live notes

- Mount `/1987d6e3-afde-480f-ba1b-c87d024ab3f2`. Project Verse path `/epictemplates@fortnite.com/narrative_roleplaying_template`. Plugin filename is that GUID. `projectId` `aa154122-674d-4459-8f0d-e2ba606d17d2`.
- Did **not** `load_level` away from the default map.
- `ListFiles("")` is the digest package — list `/1987d6e3-…` for sources.
- Island settings live on the **actor**, not `ToyOptionsComponent.playerOptionData` (Social Deduction / EnergyBlades trap).
- `folderPath` often `None` on ActorDesc — use `get_actors_in_folder`.
- Conversation `indicator Bubble Range`: DJ **15**, Caris **10**, Vex/Captain **5**.
- Folder on disk / title is **`narrative_roleplaying_template`** (picker “Narrative and Roleplaying”). Island umap `Cantina_Roleplay`.
- `_raw/.../live-values.json` is curated. Noisy Pass B dumps: `live-values-b.json`, `live-values-c.json`.
