---
id: digest-035
kind: digest
---

# volcanic_town_sample — sample digest

> **Extractor provenance.** `tools/extract-template.py` (Pass A) + UnrealMCP on
> `localhost:8000` (Pass B). Default level
> `/9423f55a-0791-428c-bf9b-16317a048a6d/volcanic_town_sample` (**Observed**).
> Docs: Working With space-opera™ Islands
> (hub; VolcanicTown is one environment-starter bullet) + Vistas and Biome Environments
> + Custom UI.
> `docsUrl` empty. Guessed UEFN-suffix hub slug and dedicated
> `star-wars-VolcanicTown-inspired-starter-island-in-fortnite` **hub-redirect**.
> Compat **43.00**. `featureSet`: **CustomGameplayFeature**.

| | |
|---|---|
| Classification | **theme environment starter** — VolcanicTown volcanic city sandbox, not EnergyBlades/Narrative/Team Conquest |
| Content type | **catalog wrap** (38 authored vs 7409 OFPA; `/SPACE_THEME_C_*` city + `/ENV_PACK_Biome_*` volcanic) |
| Verse | **0 files** |
| Fab `.uref` | **0** |

## What it is

The picker name is **VolcanicTown Inspired Starter Island**. The island is a **16-player lava-town**: alleys, houses, magister building, spaceport, N-1 Starfighter, roaming civilians and guards. Teaching that survives deletion is the **same Guard CD + patrol + volume swap as IceBase**, plus a **Civilian `CharacterType_Custom`** cohort and **three doorway zones** instead of one hangar pair. Do not duplicate: guard-patrol-volume-swap.md. Landscape paint layers stay on landscape-weight-blend-layers.md (this island’s `layers[]` is empty). Day Sequence priority stays on day-sequence-priority-stack.md. Environment lighting rig stays on environment-lighting-rig.md. Custom skyboxes stay on night-atmosphere-stack.md. Conversation / Widgets-folder HUD stay on narrative_roleplaying_template.

## Scale (**Observed** — `summary.txt` + live)

| | |
|---|---|
| Authored assets | **38** |
| OFPA | **7409** |
| Maps | `volcanic_town_sample.umap` (only) |
| Verse files | **0** (`VerseToolset.ListFiles` on `/9423f55a-…` — do not recurse `__ExternalActors__`) |
| Verse devices placed | **0** |
| Data layers | **0** |
| Fab `.uref` | **0** |
| Compatibility | **43.00** |
| Matchmaking | **16** (`maxPlayers=16`, `teamSize=Dynamic`, `teams=FreeForAll`, `locomotionPreset=Current BR`) |
| Streaming | Partitioned + actor folders + OFPA. **`LevelHasStreamingDisabled` omitted** (IceBase / Social Deduction / Arcade camp). Sampled `bIsSpatiallyLoaded=false`; `load_actors` not required. `find_actors` returns **0** — use `get_actors_in_folder`. |

Empty `Developers/felixkupis/Collections/`. First-party scratch folder.

`experimental.*` (Python / toolsets / Scene Graph) in `.uefnproject` is **Unverified**. Content-side: **no** `.verse`, **no** `__ExternalEntities__`.

Largest file ~**5.91 MB**. Whole folder ~**135 MB**. Under the GHES **100 MB** per-file cap (IceBase OFPA was the large sibling).

## Level organisation (**Observed** — live `get_folders` + folder counts)

```
Devices/                              3 — Player Spawner, N-1 Starfighter, HUD Controller
Environment/                       6916
  Town/                            2473
  Houses/                           720
  Rocks & Cliffs/                   635
  MagisterBuilding/                 481
  SpacePort/                        459
  Alley_A–D/                       1310
  Tower/ / Tower2/ / Tower 3/       526
  MainSquare/                       281
  Background/                        76 (Vista/ 64)
  Foliage/                           35
  Landscape/                         32 (InstancedFoliage/ 31)
  Outskirts/                         52
NPCs/                               216 — 120 spawners, 90 patrol, 6 volumes
Lighting/                             7 — DSD + DSA PPV; TestLighting/ leftover engine lights
VFX/                                 12 — ENV_PACK NS Biome Steam Niagara
HLOD/                               186
LevelData/                            5 — IslandSettings0, cine camera, bounds, minimap
09/ 10/ Alley_A/ Alley_B/ Background/ Houses/ Landscape/ SpacePort/ Tower/   0 leftover roots
```

No `Device_Billboard_V2`. No syllabus boards.

## Device inventory (the design)

| Count | Class | Role |
|---|---|---|
| 120 | `Device_CharacterSpawner` | civilians + 27 Guard-labeled; 3-zone volume swap |
| 90 | `AIPatrolPath_Device` | groups **1–13** match folder names; mostly BackAndForth (09/12 Loop) |
| 6 | `Device_Volume` | Hallway / Hallway4 / MainSquare Enter+Exit pairs |
| 32 | `PlaysetRoot` | 6 named city buildings ×2 + `Nv Prefab *` |
| 12 | `NiagaraActor` | volcanic steam |
| 2 | `PostProcessVolume` | unbound DSA + Custom volcanic grade |
| 1 | `Device_HUDController_V2` | in-project `UW_HUD_*` overrides |
| 1 | `BP_DaySequenceModifier_Device` | 18:30, priority **2000**, `use Volume=false` |
| 1 | `CRD_SpawnerDevice_THEME_PACK` | N-1 Starfighter, health **600**, phase Always |
| 1 | `BP_Creative_Player_Spawner_Prop` | island start |
| 1 | `CineCameraActor` | leftover in LevelData |
| 0 | Billboard / Verse device / Hero / EnergyBlade / Conversation / turret | other space-opera tiles |

## Authored kit (**Observed** — `assets.txt` + live)

```
Content/NPCs/CD_CivilianNPC
Content/NPCs/CD_CivilianNPC_GuardBehavior
Content/UI/PlayerHUD/   UW_HUD_PlayerInfo, Equipped, Quickbar, PlayerVitals, TeamVitals
Content/Landscape/      SM_Nev_Landscape_01–12 + SM_Landscape_Proxy + SM_4K_Plane
Content/Materials/      ENV_PACK_MI_Biome_Volcanic_Landscape_A/B + MI/M_CloudShadows
Content/VolcanicTown_Starter_sharedassets/  Layer_1–6_LayerInfo_1  (0 asset-registry refs; Landscape.layers[] empty)
Content/Foliage/        SM_TREE_PACK_* FoliageTypes
```

**Tier 2 mounts:** `/SPACE_THEME_C_Assets` 10646 + `/SPACE_THEME_C_Comp` 5032 (city gallery), `/VolcanicTown_Starter` 3591, `/THEME_PACK` 2724, `/ENV_PACK_Biome_Assets` 2226 + `/ENV_PACK_Biome_Comp` 636 (volcanic), `/ENV_SWCosmetics` 120 (SnowSurgeon / LongBandit CIDs), `/ENV_SPACE_THEME_UI_Assets` 40 (HUD materials), `/ENV_SWVehicleClasses` + `/ENV_SWVehicleSpawners` (N-1 / THEME_PACK). Leftover GUID `/cb093564-…` still referenced next to authored CDs. No Fab `.uref`.

## What this sample actually teaches

1. **Guard + Civilian CDs + three doorway swaps, no Verse.** Guard Type=`Guard` + Trespassers preset (same as IceBase). Civilian Type=`Custom`. guard-patrol-volume-swap.md (Sources: IceBase, VolcanicTown).
2. **The city is a mounted `/SPACE_THEME_C_*` + `/ENV_PACK_Biome_*` catalog plus 32 playset roots, not 38 authored meshes.** Same catalog-vs-kit split as ice_base_sample (`/SPACE_THEME_E_*`) and prison_sample (`/ENV_Meadow_*`). IceBase shipped **1** playset (WalkerVehicle); this island ships **32**.
3. **HUD is authored UMG in `Content/UI/PlayerHUD`, not ExtraAssets.** Live overrides: PlayerInfo, Equipped, Quickbar. HideElement minimap + build menu. Same layout as IceBase, not Narrative’s `UI/Widgets`. ExtraAssets HUD coincidence stays **three** (Social Social, Mastery, Walker). space-opera in-project PlayerHUD is now **two** (IceBase, VolcanicTown).
4. **Dusk is one Day Sequence at 2000, not a volume stack.** FixedTime **18:30**, sun **20** `overrideSunlight`, fog **0.003** `overrideFog`, stock `MI_DaySequenceSkydome`. IceBase was 18:00 / sun 10 / fog 0.02 / no sunlight override. **Not** a fourth 2000-vs-higher source. Leftover `Lighting/TestLighting` engine lights are unused next to the DSD.
5. **Landscape layers are leftover LayerInfos, not a weight-blend demo.** Live `layers[]` **empty**; volcanic MI parent `/ENV_PACK_Biome_Assets/…`. Distant meshes `SM_Nev_Landscape_01–12` in Background. Do not train as landscape-weight-blend-layers.md.
6. **Sandbox flags, not a round.** `allowBuilding=All`, `bInfiniteResources=true`, `bFallDamage=false`, `timeOfDayOverride=NoOverride` (DSD owns 18:30). One spawn pad, `spawnLocation=SpawnPads`.
7. **N-1 Starfighter is a THEME_PACK vehicle spawner, not a turret turret.** Health **600**, atmospheric effects on, flight ceiling **250**. IceBase’s CopperPuck turret stays on that packet.

## Official tutorial vs shipped

**External hub + Vistas + Custom UI** (headed Chrome 2026-09-15). Verbatim: `_raw/volcanic_town_sample/tutorial.txt`. Spreadsheet `docsUrl` empty. No dedicated VolcanicTown chapter.

### Lines up

| Tutorial | Shipped |
|---|---|
| theme Templates → space-opera → VolcanicTown Inspired Starter Island | picker; `featureSet` CustomGameplayFeature |
| “Volcanic terrain and urban structures” | `Environment/Town` 2473, `Rocks & Cliffs` 635, volcanic MI |
| VolcanicTown City prefab/gallery; vista assets; volcanic props | `/SPACE_THEME_C_*`, `/ENV_PACK_Biome_*`, `Background/Vista` 64 |
| Roaming NPCs + NPC Spawners + AI Patrol Path | 120 spawners, 90 patrol devices |
| Smoke VFX | 12 Niagara steam |
| CoastalWorld N-1 among space-opera vehicles | 1 THEME_PACK spawner |
| Custom UI widgets `UW_HUD_Equipped` / `PlayerInfo` / `Quickbar` | same names in `Content/UI/PlayerHUD`; HUD Controller overrides |
| Hide minimap | `showMinimap=HideElement` |
| Named city buildings (Molten Mug, Gearpit, Char Vault, Ember Quarters, Sootmark Atelier, Corewright's Bay) | those six as `PlaysetRoot` (each ×2) |

### Does not line up

| Tutorial | Shipped |
|---|---|
| City paths with **swappable material layers**; Volcanic **6 paint layers** | Authored LayerInfo×6; live `Landscape.layers[]` **[]** |
| Custom HUD “preconfigured on Flat Grid / XL Flat Grid” | **This** island also ships a HUD Controller (Devices/, not UI/) |
| Editable widgets under `UI > Widgets > PlayerHUD`; Player Info = `UW_HUD_Vitals` | Shipped `Content/UI/PlayerHUD`; override `UW_HUD_PlayerInfo`. Widgets-folder layout is Narrative |
| HUD Controller in Outliner **UI** folder; cantina / The Warren; no building | That chapter’s example is Narrative. This island: folder **Devices**; `allowBuilding=All` |
| Hub building roster (Council, Starhome, Obsidian Flask, Glowfang Den, …) | Only six names as `PlaysetRoot`; rest are gallery pieces inside Town/ |
| Dedicated VolcanicTown walkthrough | Hub bullet + Vistas list; guessed dedicated slug hub-redirects |
| Planet VolcanicTown background asset | Not a placed device; decorative Content mention |

Training rule: live Guard/Civilian CDs + volume bindings + in-project HUD paths win. Do not emit “VolcanicTown ships 6 painted volcanic layers / HUD only on Grid islands / Widgets-folder HUD.”

## Negative findings

- **Name implies a VolcanicTown *game*; files show a kitbash sandbox.** No round manager, no Verse, no EnergyBlades, no Hero / Conversation device.
- **`hUDInfoType=AI Enemy Elimination` with Neutral unkillable NPCs** (`isVulnerable=false`). Leftover default, same family as IceBase/Prison/Starter.
- **Root `Alley_A/` / `Houses/` / `SpacePort/` / `Tower/` / `09` / `10` are empty leftovers** after the nested move under `Environment/`.
- **`NPCs/OutskirtSettlements/02` is dead:** 10 spawners `enabledAtGameStart=false` and **no** volume bindings. Nothing Enables them.
- **Hallway3 / Hallway4 stay Always-on** (no volume). Only MainSquare* + SecondarySquare start disabled.
- **Guard inventory is empty** (`items=[]`). IceBase’s Guard CD grants an AR.
- **Guard leash is off** (`bEnableLeash=false`) with unused distance **30**. IceBase’s leash is on.
- **`get_referencers` on HUD widgets / both CDs / LayerInfo returns []** despite live overrides — confirm on the placed HUD Controller / spawners, same trap as IceBase.
- **`GetDeviceProperties` N/A** — no Verse `creative_device`s. `playerOptionData` unreadable; use actor `get_properties` with `list_properties` names (a missing name fails the whole call).
- **`Lighting/TestLighting` leftover engine sun/sky/fog.** DSD owns 18:30.
- **Default GuardSpawn cosmetic is `CID_NPC_Athena_Commando_M_AlienRobot`.** OFPA still references `/ENV_SWCosmetics` SnowSurgeon / LongBandit CIDs on placed spawners.

## Capture backlog

| Item | Status |
|---|---|
| Per-alley child lists (6916 Environment) | **Waived** — folder counts + `/SPACE_THEME_C_*` mounts are the lesson |
| LayerInfo ↔ material slot wiring | **Waived** — empty `layers[]`; don’t train as weight-blend |
| PPV blendable graphs | **Waived** — unbound grade flags + sampled ColorContrast/Gamma sufficient |
| PIE volume swap | **Waived** — bindings + enabledAtStart sufficient |
| Per-spawner cosmetic CID | **Waived** — unread on spawner `list_properties`; OFPA `/ENV_SWCosmetics` is Observed |
| Dedicated VolcanicTown docs page | **Waived** — slug hub-redirects; Vistas list Observed |

## Live notes

- Mount `/9423f55a-0791-428c-bf9b-16317a048a6d`. Project Verse path `/epictemplates@fortnite.com/volcanic_town_sample`. Plugin filename is that GUID.
- Did **not** `load_level` away from the default map.
- Billboard `text` N/A. `GetDeviceProperties` N/A.
- `find_actors` by name/class returns **[]** while folders are full — World Partition streaming-on + use `get_actors_in_folder`.
- `folderPath` on descriptors is often `"None"` even when `get_actors_in_folder("Devices")` returns the HUD / N-1 / pad. Count from the folder query, not `folderPath`.
- Character Definition modifiers: `load_asset` then `get_properties` on `:CharacterModifier_*_C_0` suffixes with **that modifier’s** property names (a wrong name fails the whole call).
- Folder on disk / title is **`volcanic_town_sample`** (picker “VolcanicTown Inspired Starter Island”).
