---
id: digest-036
kind: digest
---

# martial_arts_sample — sample digest

> **Extractor provenance.** `tools/extract-template.py` (Pass A) + UnrealMCP on
> `localhost:8000` (Pass B). Default level
> `/f8efbe30-2700-4061-9332-aecd9cbcc022/HotSauce_Template` (**Observed**).
> Hub: Building BRICK Islands
> + index BRICK Templates
> (Starter Island Templates row — **no** dedicated island chapter).
> Companion Working with BRICK Islands
> + theme Rules (publishing, not a device syllabus)
> + BRICK Brick Editor (how the modules were made, not this walkthrough).
> `docsUrl` empty. Not on the spreadsheet. Compat **43.00**. `featureSet`: **BrickGameplay**.

| | |
|---|---|
| Classification | **BRICK theme art starter** — three-biome BrickAssembled MartialArts wrap (NinjaBase / Valley / Whispering Woods), not Action Adventure / Brick Editor / PvP Extraction |
| Content type | **kit + wrap** (390 authored vs 6274 OFPA; in-project BrickAssembled `SM_*` + `/BRICKHotSauce` / `/BRICKFairy` / catalog foliage) |
| Verse | **0 files** |
| Fab `.uref` | **0** |

## What it is

The picker name is **BRICK MartialArts Starter Template** (BRICK). Docs H1 on the Templates index is **BRICK MartialArts Starter**. Disk folder is `martial_arts_sample`. The live map is **`HotSauce_Template`** (internal HotSauce / BRICKHotSauce name). The island is a **16-player art sandbox**: monastery, dojo, pagoda, houses, a concert-venue leftover, a valley, and Whispering Woods. Teaching that survives deletion is **shipping BrickAssembly Atom modules as a FortStaticMeshActor world kit**, a **Switch → Audio Player music bed**, and **BRICK Guard CDs dressed with Fortnite CIDs**. Do not duplicate: BrickAssembly-atom-static-mesh.md / modular-BrickAssembly-building-modules.md (viewport lab that *makes* the meshes), phone-copy-paste-prefab-quickbar.md / BRICK-fortnite-mix-no-scale.md (**0** Phone; find `"Phone"` is a SpaceFriends microphone), persistable-quest-line-verse-dialog.md / tagged-collectible-stud-granter.md / npc-attack-stasis-aoe.md (**0** Verse / Assembly / Tracker), collectible-tracker-endgame.md (3 Coins, **0** Tracker / End Game), patchwork-switch-stems.md (**0** Patchwork / Jam; concert venue is Musical Walk dressing), TURTLE-character-spawner.md / hireable-theme-character-spawner.md / guard-patrol-volume-swap.md (**0** theme enum, **0** patrol, **0** doorway volumes). ExtraAssets HUD stays **four** — this island places **0** HUD Controller.

Docs: starter islands “do not include gameplay mechanics.” PLAY still has light toys (weapons, grappler, three coins, five citizens, a music switch).

## Scale (**Observed** — `summary.txt` + live)

| | |
|---|---|
| Authored assets | **390** (`Meshes/BRICKModels` Medieval + MartialArts BrickAssembled/Props/WhisperingWoods, `SetupAssets/Props/Monastery`, `Trees/`, `NPCDefinitions/` ×5, `Music/` ×2, 7 LandscapeLayers, GameFeatureData) |
| OFPA | **6274** (5 unclassified) |
| Maps | `HotSauce_Template.umap` (only) |
| Verse files | **0** (`VerseToolset.ListFiles` on `/f8efbe30-…` — folders only; do not recurse `__ExternalActors__`) |
| Verse devices placed | **0** |
| Data layers | **0** |
| Fab `.uref` | **0** |
| Compatibility | **43.00** |
| Matchmaking | **16** / live `teamSize=Dynamic` / `teamFillOption=MustFill` / `teams=FreeForAll` |
| Streaming | Partitioned + actor folders + OFPA. **`LevelHasStreamingDisabled=1`** (Walker/VFX/music-fantasy/Music/Obstacle/Home Builder/Brick Editor camp). Sampled `bIsSpatiallyLoaded=false` (6199). `load_actors` used for device values. `find_actors` by `"Character Spawner"` / `"Item Spawner"` / `"Phone"` returned **0** / microphone — labels are **NPC Spawner_Citizen1–5**, **Weapon1–5**, **Grappler**. `folderPath` often **None**. Live `get_folders` returned the **full** Outliner (unlike Music/Obstacle/Home Builder Background-only). |

Empty `Developers/felixkupis/Collections/`. First-party scratch folder.

`experimental.*` (Python / toolsets / Scene Graph) in `.uefnproject` is **Unverified**. Content-side: **no** `.verse`, **no** `__ExternalEntities__`. Plugin `EnableSceneGraph: true` with no Scene Graph entities.

Largest authored file ~**24.1 MB** (`T_mossC_grey_COL`). Whole folder ~**1,061 MB**. Under the GHES **100 MB** per-file cap (same class as IceBase ~917 MB).

## Level organisation (**Observed** — live `get_folders` + folder counts)

```
NinjaBase/                 3063
  Buildings/Dojo           335   (+ Dojo_SetDress 19)
  Buildings/Monastery      645
  Buildings/Pagoda         116
  Buildings/ResidentialHouse_A–F   230 / 160 / 144 / 141 / 151 / 70
  Buildings/ConcertVenue    47   Musical Walk speakers + SpaceFriends mic
  Buildings/BridgeSmall      3
  Buildings/MiniShrine      15
  Buildings/ShrineGarden    21
  Foliage                  358
  OuterWalls               251
  NinjaBase_SetDress       170
  Rocks                    141
Valley/                    559   foliage 411 / rocks 88 / bridges / shrines
Whispering Woods/         2488
  Monastery                654
  Background               245
  Abandoned Village         39
  Village/                 huts in subfolders (Wood Hut / Stone Hut / Unique Wood Hut)
Background/                 63
VFX/                        15   moths + fog (plus unfoldered)
Landscape/                   6
Lighting/                    1   DSD
World Setup/                 4   Music Control (billboard / switch / audio / speech bubble)
PlaceholderAssets/           0   leftover
```

**1** `Device_Billboard_V2` (`Billboard_Music`: “Turn on/off Music”). Syllabus is the **Templates index row**, not a street of boards (`_raw/martial_arts_sample/billboard-texts.md`). Event graph: `_raw/martial_arts_sample/event-graph.md`.

## Device inventory (the design)

| Count | Class | Role |
|---|---|---|
| 4531 | `FortStaticMeshActor` | BrickAssembled MartialArts / Medieval / Fairy modules |
| 165 | `BP_BambooTree_A` | Catalog bamboo |
| 110 | `StaticMeshActor` | Extra meshes |
| 22 | `Device_VFXSpawner_V2` | Fog ×5 (`AreaOfFog`, Enable on player spawn) + moths ×17 (`Insect_Flies`) |
| 17 | `Device_PointLight_V2` | `Light_Lantern*` warm (`r=1,g=0.96,b=0.27`), Always |
| 5 | `Device_CharacterSpawner` | `NPC Spawner_Citizen1–5`; live spawn 1 / infinite limit 1 / timer **3** |
| 5 | `Device_BRICKItemSpawner` | Labels **Weapon1–5**; timer 10 / respawn; `base Visible During Game=false` |
| 5 | `BP_HotSauce_Rack_Katana_A` | Set-dress racks (not the spawners) |
| 3 | `BRICK_Device_Collectible` | `BRICKCollectibleObject=Coin`; **0** Tracker bindings |
| 1 | `Device_ItemSpawner_V3` | **Grappler**; BRICK-like timer 10; `base Visible During Game=true` |
| 1 | `Device_Billboard_V2` | Music Control |
| 1 | `Device_Switch_V2` | `Switch_Music`; start **On**; Turn on/off Music |
| 1 | `Device_CRD_AudioPlayer` | `AudioPlayer_BKG`; cue `MUSCSong_MartialArtsArtTemplate_Cue`; autoPlay Gameplay; volume **2**; not looping |
| 1 | `Device_ExperienceSettings_BRICK` | `allowBuilding=None`, `enableJam=Off`, Custom locomotion |
| 1 | `BP_BRICK_Player_Spawner_Prop` | `useAsIslandStart=true`, Always, hidden |
| 1 | `BP_DaySequenceModifier_Device` | Priority **2500**, FixedTime **17**, fog **0.04**, Always, `use Volume=false` |
| 0 | HUD / Tracker / End Game / Assembly / Phone / Patchwork / Conversation / VerseDevice / AIPatrolPath / CubeGrid / Playset | Action Adventure / Home Builder / Music Concert / IceBase |

Extractor histogram is close; live class for Island Settings is **BRICK**. Use live labels for spawners (`NPC Spawner`, not `Character Spawner`).

Dressing mounts (tier 2): `/BRICKBasicBrickGallery_Assets` 4055, `/BRICKAtomAssets` 1130, `/CRG_Modular` 876, `/ENV_PACK_Nature_Ch6_Comp` 535, `/BRICKFairy` 256, `/BRICKHotSauce` 252, `/TREE_PACK` 240, `/Musical_Walk_Assets` 20. Weapon mounts (1 each): `/BRICKTabascoKatana`, `/TabascoShuriken`, `/TabascoNunchucks`, `/BRICKHandachi`, `/BRICKWeaponScythe`, `/BRICKEvade`, `/BRICKGrappler`. No Fab `.uref`.

## Verse (0 files)

No `snippets/verse/` folder. Valid. `ListFiles` on `/f8efbe30-…` is Content dirs only.

## Authored kit (**Observed** — `assets.txt` + live)

```
Content/Meshes/BRICKModels/Medieval/BRICK_Stone_*          115
Content/Meshes/BRICKModels/MartialArts/BrickAssembled/SM_BRICK_MartialArts_*   18  AtomGeneratedAsset=True
Content/Meshes/BRICKModels/MartialArts/Props/SM_BRICK_MartialArts_Torii_A
Content/Meshes/BRICKModels/MartialArts/WhisperingWoods/Monastery/SM_Fairy_*   66
Content/Meshes/BRICKModels/MartialArts/WhisperingWoods/Village/               71
Content/SetupAssets/Props/Monastery/SM_HotSauce_*         89
Content/Trees/                                            15  M_wWoods_Tree + textures
Content/NPCDefinitions/NewNPCCharacterDefinition_Citizen1–5
Content/Music/MUSCSong_MartialArtsArtTemplate(.ogg import) + _Cue
Content/LandscapeLayers/Layer1–6_LayerInfo (+ Layer1_LayerInfo_1)
```

Sampled BrickAssembled pillar (`get_asset_tags`): `AtomGeneratedAsset=True`, `/BRICKAtomAssets` LUT materials (`MI_Primitive_Standard_VC_LUT_FakeBevel` + `_CP`), `CollisionComplexity=CTF_UseSimpleAsComplex`, Nanite + 6 LODs, 2 UV channels.

**Tier 2 mounts:** `/f8efbe30-…` 18146, `/BRICKBasicBrickGallery_Assets` 4055, `/Game` 2712, `/BRICKAtomAssets` 1130. No Fab `.uref`.

## What this sample actually teaches

1. **Ship BrickAssembly Atom modules as a FortStaticMeshActor world, not a viewport lab.** Three Outliner biomes of reused `SM_BRICK_MartialArts_*` / `SM_Fairy_*` / `BRICK_Stone_*` / `SM_HotSauce_*`. How to BrickAssembly stays on Brick Editor. techniques/BrickAssembled-modular-world-kit.md.
2. **A starter-island music bed is Switch On/Off → Audio Player Play/Stop**, plus `autoPlay - Gameplay` on an in-project cue. Not Patchwork stems. techniques/switch-audioplayer-music-bed.md.
3. **BRICK Guard Character Definitions wear Fortnite CIDs** (`characterLook=FortniteCharacter`). Labels say Citizen; Type is `CharacterType_BRICKGuard`. Not a MartialArts enum, not IceBase patrol. techniques/BRICK-guard-fortnite-cosmetic.md.
4. **`allowBuilding=None` + `enableJam=Off` on BRICK Experience Settings.** Brick Editor is the `allowBuilding=All` exception. No HUD Controller.
5. **DSD is 2500-only FixedTime 17** (`use Volume=false`, Always) — late-day fog, not a 2000-vs-higher volume stack. Other extracted BRICK learning islands sit at 2000.
6. **Light PLAY toys are not a quest line.** 5 BRICK weapon spawners + Grappler V3 + 3 Coins with empty `On Collected`. Docs “no gameplay mechanics” is the art-starter *intent*; files still place devices.
7. **1 music billboard is the only in-island sign.** The Templates index row is the syllabus.
8. **ConcertVenue is Musical Walk dressing** (speakers, lute, pink guitar, SpaceFriends mic), not a Jam Stage.

## Official tutorial vs shipped

**External** hub + Templates index Starter Island Templates row + Working with BRICK Islands + theme Rules (headed Chrome 2026-09-16). Verbatim: `_raw/martial_arts_sample/tutorial.txt`. Spreadsheet: no row. `docsUrl` empty. Guessed slugs `BRICK-MartialArts-starter-template-in-fortnite` and `BRICK-MartialArts-starter-in-fortnite` **hub-redirect** to `fortnite-root-documentation`. No dedicated island chapter. MartialArts is **not** in Learning and Gameplay Templates (those tiles have links). Brick Editor hub is how the modules were made.

### Lines up

| Tutorial | Shipped |
|---|---|
| Picker / index **BRICK MartialArts Starter**; art template | picker **BRICK MartialArts Starter Template**; BrickAssembled POIs |
| Points of interest | NinjaBase monastery / dojo / pagoda / houses; Valley; Whispering Woods |
| Audio | `MUSCSong_MartialArtsArtTemplate` + Switch/AudioPlayer |
| MartialArts-themed modular pieces from Brick Editor | `AtomGeneratedAsset=True` `SM_BRICK_MartialArts_*` + Fairy/HotSauce/Stone |
| Reassemble / add to modules within theme guidelines | FortStaticMeshActor copies; `allowBuilding=None` |
| Starter vs Learning split | no DemoDisplay street; 1 music board |

### Does not line up

| Tutorial | Shipped |
|---|---|
| Starter islands “do not include gameplay mechanics” | 5 NPC spawners, 5 BRICK weapons, Grappler, 3 Coins, music switch |
| Dedicated island tutorial | **none** (index blurb only) |
| Ninja / minifig NPCs | BRICK Guard CDs + Fortnite CIDs (`ScallopLava`, `TicketPoke`, `RedSilk`, `BlessFlan`, `OvenDrastic`) |
| Map / folder name MartialArts | live umap **`HotSauce_Template`**; `BP_HotSauce_*` / `/BRICKHotSauce` |
| Concert as music teaching | Musical Walk props; **0** Patchwork; `enableJam=Off` |
| `find_actors` `"Character Spawner"` / `"Item Spawner"` | labels **NPC Spawner_Citizen1–5** / **Weapon1–5** |

Training rule: shipped bindings win. Tutorial *prose* is the art-starter intent. Do not emit Action Adventure quests, Brick Editor playground Verse, Home Builder Phone, Patchwork stems, TURTLE/IceBase patrol, or ExtraAssets HUD as this island’s setup. Modeling / Greyboxing / fort_playset hubs are other packets.

## Negative findings

- Picker **BRICK MartialArts Starter Template** / docs **BRICK MartialArts Starter** / folder `martial_arts_sample` / map `HotSauce_Template` are the same tile — not Brick Editor, not Action Adventure, not PvP Extraction, not skipped Brick Grid / Starter Island.
- Name does **not** mean ninja minifig spawners, quests, or Assembly puzzles.
- **0** Assembly / Tracker / End Game / Prop Mover / Patchwork / Conversation / VerseDevice / Phone / HUD Controller / patrol path.
- 3 Coins are dressing (`On Collected` **[]**).
- `get_properties` fails the whole call if any name is missing (`enabledDuringPhase` is often absent; BRICK item fields are `enabled At Game Start` with spaces).
- `ListEventBindings` on the Switch is **[]** — Play/Stop live on the **Audio Player**.
- NPC Character Definition `get_referencers` **[]**; assignment is the `NPC Spawner_CitizenN` label + matching `NewNPCCharacterDefinition_CitizenN`.
- BRICK weapon **item class** arrays waived (same trap as Obstacle vending). Mount names are the identity hint.
- Do not flip `experimental.*`. Do not treat ConcertVenue as Patchwork (`enableJam=Off`).

## Capture backlog

Waived: per-weapon item definition paths; Citizen3 nested modifier objects that failed to load; Sequencer (none placed); which FortStaticMeshActor instance is which monastery bay; landscape paint weights. `experimental.*` Unverified.

## Live notes

- Cursor `user-unreal-mcp` was down; `tools/mcp.py` on `:8000` worked.
- Mount `/f8efbe30-2700-4061-9332-aecd9cbcc022` from `ListFiles("")`. Do not recurse `ListFiles("")`.
- Billboard text is on the **actor**.
- Default map is `HotSauce_Template`; did not switch maps. Did not flip experimental flags.
