---
id: digest-037
kind: digest
---

# parkour_template — sample digest

> Offline `_raw/parkour_template/`. Live
> `/878afab7-c63e-4c32-9081-a2fc86486ddd/parkour_template` (**Observed**).
> Docs: one page. Tutorial says *complex concepts like concurrency will not be used* — true: **Subscribe**, not `race`.

| | |
|---|---|
| Classification | Sequential collect-race (4 batteries + secret time bonus) |
| Content type | **environment-heavy**: 38 authored vs **1910** placed (Cyber City) — take the Verse |

## Scale

| | |
|---|---|
| Authored | 38 (billboard/decal/fog/skydome + 2 sequences) |
| OFPA | 1910 |
| Verse | **1** (`parkour_race_device.verse`) |
| Item spawners | **5** (4 path + 1 secret) |
| Timer / HUD Message / End Game / Damage Volume | 1 / 1 / 1 / 2 |
| Plugin `Hidden` | true |
| `LevelHasStreamingDisabled` | 1 |
| Compatibility | 43.00 |

Spreadsheet: *billboards don't go invis when playing; good Verse race device / FAB.* Billboards 11 + Cybercity frames — Visible In Game is a known miss, not a Verse bug.

`Developers/felixkupis/Collections/` empty. Eleventh first-party hit.

Live folders: playspace buildings First/Middle/Last, Devices, Instruction Billboards, BG, Atmosphere.

## What it is *for*

Parkour through Cyber City. Timer 30s (Island Settings + Timer device). Collect batteries **in array order** (only the next `SpawnItem()` after pickup). Secret battery adds time capped at `GetMaxDuration()`. All four path batteries → HUD + `Sleep(3)` → End Game. Timer `FailureEvent` → `FortCharacter.Damage(500)` → respawn resets timer.

Fall death is **Damage Volume 500**, not Verse.

## Device inventory (the design)

| Count | Class | Role |
|---|---|---|
| 5 | `Device_ItemSpawner_V3` | 4 sequential + secret — **spawners, not granters** |
| 2 | `BP_Device_DamageVolume_V2` | kill floor |
| 1 | `Device_Timer_V2` | per-player countdown |
| 1 | `Device_HUDMessage_V2` | pickup copy |
| 1 | `Device_GameEnd` | victory |
| 1 | `VerseDevice` | `parkour_race_device` |
| 1 | player spawner | start |
| 11 | `Device_Billboard_V2` | captions (stay visible in-game — known) |
| rest | Cyber City | throwaway |

## Verse

Snippet: `snippets/verse/parkour_race_device.verse`

`@editable`: PlayerSpawnDevice, EndGameVictoryDevice, TimerDevice, HUDMessageBattery, BatteryItemSpawners[], SecretBatteryItemSpawner, SecretBatteryTimeReward (default 10).

`OnBegin`: Subscribe spawn, timer Failure, **each** path spawner `ItemPickedUpEvent`, secret pickup. Comment notes you *could* keep cancelable handles.

`HandleAgentSpawned`: `TimerDevice.Reset` + `Start` (full duration after death).

`HandleBatteryPickedUp`: increment; if `>= Length` `spawn{EndGame}` else HUD grammar (battery vs batteries) and `BatteryItemSpawners[BatteriesCollected].SpawnItem()` — **index after increment = next**.

Secret: `GetActiveDuration` + reward, `SetActiveDuration(Min(TimeToAdd, GetMaxDuration()), Agent)`.

`HandleTimerExpired(?agent)`: Damage 500.

`EndGame`<suspends>: SetText all-collected, Show, Sleep(3), Activate.

`<localizes>(Amount:int):message` for parameterized HUD.

## What this sample actually teaches

1. **Sequential collect = subscribe all spawners, only `SpawnItem()` the next index.** Length of the `@editable []` is the win count (same scaling idea as Elimination). See techniques/sequential-collect-race.md.
2. **Timer API:** Reset/Start on spawn; FailureEvent on expire; GetActiveDuration / SetActiveDuration / GetMaxDuration / Min for a capped bonus.
3. **`<localizes>(Amount:int):message`** — parameterized copy without a widget.
4. **Docs “no concurrency”** except `spawn{EndGame}` + `Sleep` — fire-and-forget async, not `race`.
5. **Item Spawner vs Item Granter:** this mode *places* pickups in the world; Elimination *grants to inventory*. Tutorial device table says Granter — **wrong for this file.**

## Official tutorial vs shipped

**External:** Verse Parkour Template

Single page. `docsUrl` empty.

### Lines up

| Tutorial | Shipped |
|---|---|
| Class `parkour_race_device`, `log_parkour` | same |
| 4 batteries + secret time, timer eliminate, Sleep(3) victory | same |
| Subscribe spawn / FailureEvent / ItemPickedUpEvent | same |
| `Min(..., GetMaxDuration())` | same |
| Cyber City gallery | OFPA |
| Locomotion island settings (mantle/sprint/slide…) | **External** |
| “No concurrency” as a teaching frame | no `race`; only `spawn`+`Sleep` |

### Does not line up

| Topic | Tutorial | Shipped |
|---|---|---|
| Pickup devices | **Item Granter** in device list and Details photo labels | **`item_spawner_device` / `Device_ItemSpawner_V3`** |
| Details names | BatteryItemSpawners “Battery Spawner” mixed with Granter prose | spawners |
| Typos in docs | “frelly”, “cutom”, “Ch 4 Movemenet”, “generage”, “determine sif”, HandlebatteryPickedUp in commented alt loop | ignore as identifiers |
| Billboard visibility | spawn pads “not visible”; user’s note that **billboards stay visible** | 11 billboards, not hidden in Verse |

Train on snippet. Emitting Item Granter for this collect-race is doc regurgitation.

## Negative findings

- 1910-actor city is not the lesson.
- Plugin Hidden.
- Billboards visible in play (user + OFPA).
- Scratch Developers folder; streaming off.

## Capture backlog

1. Confirm spawners start disabled except index 0 (editor User Options). *Open, medium — sequential spawn depends on it.*
2. Billboard Visible In Game. *Observed as problem, waived as Verse.*
3. Island locomotion table. *External.*

## Live notes

Level confirmed. Folders named the course (First/Middle/Last Building + Devices). Streaming off.
