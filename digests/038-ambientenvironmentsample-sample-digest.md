---
id: digest-038
kind: digest
---

# ambient_environment_sample — sample digest

> **Extractor provenance.** `tools/extract-template.py` (Pass A) + UnrealMCP on
> `localhost:8000` (Pass B); `SceneTools.get_current_level` returned
> `/73e107ba-67b6-4edd-88b1-7484b058e78c/ambient_environment_sample`.
> Picker-only template (Game Examples), **not on the spreadsheet** — no Keep column.

| | |
|---|---|
| Classification | **Living daytime village** — cozy farm/town/beach starter with ambient NPC life |
| Verse | **0 files** |
| Fab `.uref` | **0** |
| Compatibility | **43.00** — newest in corpus (everything else so far: 42.30) |

## Verdict first

The daytime counterpart to `night_environment_sample`: same shape (small authored
atmosphere layer + thousands of placed stock/mounted actors), but where NightEnvironment teaches
*night mood*, AmbientEnvironment teaches **ambient life** — wildlife spawners, fishing zones,
fireflies, named NPC characters, and a Sidekick-companion example. If the campaign keeps
only one environment starter per mood, NightEnvironment = night, AmbientEnvironment = day.

## Scale (**Observed** — `summary.txt`)

| | |
|---|---|
| Authored assets | **42** — foliage/landscape materials, 12 LayerInfos, 4 Sidekick NPC assets |
| Placed actors | **5,568** OFPA packages, **892 distinct classes** (largest in corpus) |
| Reference mounts | **117 distinct** (widest in corpus; NightEnvironment: 78) |
| Verse files | **0** |
| Data layers | **0** (live `get_data_layers` → `[]`) |
| Feature flags | SceneGraph allowed (unused), Python, toolsets; `editorPermissions` empty (public) |

## Authored tier (**Observed** — `assets.txt`)

```
Content/
  Foliage/        11 FoliageTypes (Helios grass, Hera flowers, TREE_PACK wildflower/leaves…)
                  + LGT/ 5 LGT_AmbientEnvironment_{Beach,ForestScatter,Grass,Gravel,Soil} layer groups
  Materials/      M_/MI_AmbientEnvironment_Landscape_A, ENV_PACK_MI_Biome_Vista{Outcrop,Hilla}_A_Grassy,
                  ENV_PACK_MI_JungleFerns_A, Hera grass MIs, T_UI_IconLibrary_boat_64_Mat
  AmbientEnvironment_Starter_Island_sharedassets/   12 LayerInfos — **Layer1..6 AND Layer_1..6**
                  (duplicate naming; see negative findings)
  Sidekick_NPC_Example{,_Cat,_Flopsticks,_FollowPlayer}   NPC data assets (see below)
  KBFN_CozyPrefabs_DefaultHLODLayer   ← name leaks a different project ("CozyPrefabs")
```

**Sidekick NPC assets** (readable via `ObjectTools` — Observed): each carries
`type` → `CharacterType_Custom`, `behavior` → `CharacterModifier_VerseBehavior`,
`modifiers[]` → `CharacterModifier_CosmeticSpawn` + `CharacterModifier_Health`.
These are **data-asset NPC definitions** (no `.verse` in the project; the Verse behavior
lives behind the modifier reference). Cat = companion creature, FollowPlayer = follow
behaviour demo. First NPC-data-asset examples in the corpus — compare voice_conversation_sample
(persona voice) and characters_cameras_npc (`npc_behavior` Verse subclass).

## Level organisation (**Observed** — live `get_folders`)

```
00_Lighting/        Day Sequence Device, BP_LightingScalabilityManager,
                    PPV_Low/Med/High (three tier volumes), ExponentialHeightFog
Buildings/          Town/{Cafe, 6(Tavern), 2(Pet Shop), 4(Town Hall), 7, 8},
                    Farm/{Bakery, Farm_Shed, Prefab1}, Park/{Tent, Playground},
                    HerbalistHut, Prefab3 (Fishing Hut), Prefab5 (Forest Home)
                    → per-building split: PrefabN_Building / PrefabN_SetDressing
Beach/  + Fish Market      Nature/{Cliffs,Foliage,Rocks}      Water/
VFX/      Level Data/      HLOD/      Background/
**Temp (Remove At End)** — shipped      **Preview_Grid** — 10 GridPlanes shipped
```

## Device inventory — the ambient-life layer (**Observed** — actor classes + folders)

| Count | Device | Notes |
|---|---|---|
| 9 | CharacterSpawner | Named NPC outfits ("NPC Spawner - CID 320 Commando M Nautilus", SnowFairy, WeepingWoods…) |
| 7 | WildlifeSpawner V2 | Scattered across farm/forest/beach |
| 4 | FishingZone V2 | Coast + river (`VK_Waterbody_River`, `Lake`, `Ocean` all present) |
| 1 | FireflySpawner V2 | Night accent |
| 12 | Decal V2 | Ground detail |
| 1 ea | Item Spawner V3, Vending Machine, Powerup Grind, Character (Mannequin) V2, DaySequenceModifier, ExperienceSettings V2 | singletons |

Farm dressing uses **crop lifecycle prop families** from `/AmbientFarm`:
`*_{Seedling,Grown,Produce,Dying}` per crop (wheat, carrot, cabbage, tomato, sweetcorn,
pumpkin, apple) — growth stages placed as static dressing, not simulated.

## What this sample actually teaches

1. **Ambient-life layer** (`techniques/ambient-life-layer.md`) — daytime village: Day
   Sequence + scalability PPV trio + wildlife/fishing/firefly/character spawners + Sidekick
   NPC data assets. The "alive without Verse" recipe.

## Official tutorial vs shipped

**Official docs: none found** (picker-only, not on spreadsheet, `docsUrl` empty; guessed
slug `…/uefn/AmbientEnvironment-environment-template-…` → empty hub page).

## Negative findings

- **No Verse, no Fab, no data layers.** SceneGraph flag on, unused.
- **Shipped cleanup leftovers:** a `Temp (Remove At End)` outliner folder (contains 3 named
  NPC spawners + a mannequin device — i.e. *work-in-progress content*, not just empties) and
  a `Preview_Grid` folder with 10 GridPlanes. Same class of leftover as
  `Developers/` scratch folders — but this one contains live devices.
- **Duplicate LayerInfo naming:** `Layer1..6_LayerInfo` and `Layer_1..6_LayerInfo` both
  shipped in `AmbientEnvironment_Starter_Island_sharedassets/` — a rename that orphaned the old set.
- **Cross-project name leak:** `KBFN_CozyPrefabs_DefaultHLODLayer` — the island began life
  as (or alongside) a "CozyPrefabs" project.
- Village architecture mounts in (`/AmbientEnvironment_Comp` 917 refs, `/AmbientEnvironment_Assets` 2,591) —
  not authored here. Same lesson as NightEnvironment; also `/AmbientFarm`, `/AmbientPetProps`,
  `/AmbientTraining` mounts carry the themed prop families.
- `PlaysetRoot` present (village placed as playset), `LevelIsPartitioned=1`, **no**
  `LevelHasStreamingDisabled` tag → runtime streaming ON (second such map after NightEnvironment).

## Capture backlog

1. WildlifeSpawner option payloads (`ToyOptionsComponent.playerOptionData` unreadable on
   this class — "properties could not be read"; try DeviceToolset or string-scan) — waived.
2. Which crops/NPCs are wired vs decorative — needs PIE; waived.
3. PPV_Low/Med/High property deltas — `ActorTools.diff` candidate; waived.

## Live notes

- Level: `/73e107ba-67b6-4edd-88b1-7484b058e78c/ambient_environment_sample`.
- `ValkyrieToolset.VerseToolset` is **absent** in this session (project has no Verse module
  loaded) — don't assume the toolset exists on every project.
- Named-NPC CharacterSpawners live under `Temp (Remove At End)`; `find_actors` by name
  misses them (WP-unloaded) — use `get_actors_in_folder`.
