---
id: digest-039
kind: digest
---

# prison_sample — sample digest

> **Extractor provenance.** `tools/extract-template.py` (Pass A) + UnrealMCP on
> `localhost:8000` (Pass B). Default level
> `/000aa9d3-2913-4ad6-ac6c-71847da65b2f/prison_sample` (**Observed**).
> Docs: Working With TWDU Islands
> (single page, Prison + Walker NPC together). `docsUrl` empty; spreadsheet Documentation=Yes.
> Guessed `prison-starter-island-in-unreal-editor-for-fortnite` **hub-redirects**.
> Compat **43.00**. `featureSet`: **CustomGameplayFeature**.

| | |
|---|---|
| Classification | **theme environment starter** — TWDU prison kitbash, not a Walker lab |
| Content type | **catalog wrap** (13 authored vs 9582 OFPA; meadow kit + 12 playsets) |
| Verse | **0 files** |
| Fab `.uref` | **0** |

## What it is

The picker name is **Prison Starter Island**. The island is a **16-player sandbox prison**, not Walker NPC’s three-room syllabus: cell blocks, cafeteria/gym, garage, medbay, underground, watchtowers, chain-link yard. Teaching that survives deletion is the **interior Day Sequence volume** over a mounted `/ENV_Meadow_*` kit. Walker Character Definition, volume/checkpoint hordes, FPS switch, ExtraAssets HUD, and Heavy Linework live in horde_npc_sample. Verse wander is npc-behavior-navigatable-focus.md — **not** this island.

## Scale (**Observed** — `summary.txt` + live)

| | |
|---|---|
| Authored assets | **13** |
| OFPA | **9582** |
| Maps | `prison_sample.umap` (only) |
| Verse files | **0** (`VerseToolset.ListFiles` on `/000aa9d3-…` — do not recurse `__ExternalActors__`) |
| Verse devices placed | **0** |
| Data layers | **0** |
| Fab `.uref` | **0** |
| Compatibility | **43.00** |
| Matchmaking | **16** (`maxPlayers=16`, `teamSize=Dynamic`, `locomotionPreset=Current BR`) |
| Streaming | Partitioned + actor folders + OFPA. **`LevelHasStreamingDisabled=1`** |

Empty `Developers/felixkupis/Collections/`. First-party scratch folder.

`experimental.*` (Python / toolsets / Scene Graph) in `.uefnproject` is **Unverified**. Content-side: **no** `.verse`, **no** `__ExternalEntities__`.

Largest file ~0.24 MB OFPA. Whole folder ~89 MB. Under the GHES **100 MB** per-file cap.

## Level organisation (**Observed** — live `get_folders` + folder counts)

```
Background/                         7 — BackDrop FortStaticMeshActors (vista cards)
Clutter/                          122 — leftover root (also Exterior Set Dressing/Clutter)
Exterior Set Dressing/           2158
  Clutter/                        487
  Grass/                          211
  OuterFence/                     923
  Sidewalks & Streets/            322
  Utility Sheds/                  215
Landscape/                        191 — Landscape + 150 LandscapeStreamingProxy (+ foliage)
Level Data/                         6 — IslandSettings0, bounds, minimap, inspector + cine cams
Lightblocker/                       5 — GridPlane sun occluders under the yard
Lighting & Post-Process/           87 — 2 DSD + unbound PPV + Lumen Exposure; Wall Lights 83
Prefabs/                         5839 — prison interiors (flat; 12 PlaysetRoot mixed in)
OuterFence/  Sidewalks/  Wall Lights/   0 leftover roots after the nested move
```

No `Device_Billboard_V2`. Sampled devices `bIsSpatiallyLoaded=false`; `load_actors` not required.

## Device inventory (the design)

| Count | Class | Role |
|---|---|---|
| 12 | `PlaysetRoot` | named rooms (Quarters / Garage / Underground / Cafe-Gym / Watchtowers / Medbay) |
| 2 | `BP_DaySequenceModifier_Device` | Outside **2000** unbound; Underground **2050** volume |
| 1 | `PostProcessVolume` | `Global PostProcessVolume` — unbound, **empty** blendables |
| 1 | `BP_LumenExposureManager` | `LumenExposureManager C` |
| 1 | `Device_ExperienceSettings_V2_UEFN` | IslandSettings0 |
| 42 | `PointLight` | plus meadow `BP_ENV_Meadow_WallLight_A` ×65 in Wall Lights |
| 5 | `GridPlane` | `Lightblocker/` |
| 7 | `FortStaticMeshActor` | `BackDrop 7–12, 14` |
| 0 | Billboard / HUD / NPC Spawner / Player Spawner / Verse device / Item Spawner | — |

PlaysetRoot labels (**Observed** `find_actors` by class):

`EnvironmentPack Quarters A/B/C`, `EnvironmentPack Garage`, `EnvironmentPack Underground`, `EnvironmentPack Cafe/Gym`, `EnvironmentPack Watchtower`, `EnvironmentPack Watchtower3/4/5`, `TWD Prison Medbay`, `TWDU Prison Watchtower`.

## Authored kit (**Observed** — `assets.txt` + live)

```
Content/EnvironmentPack_Starter_Island_sharedassets/  Layer1–7_LayerInfo  (0 asset-registry refs)
Content/Foliage/   CP_CH4_Pine1_Bake_01_FoliageType, CP_CH4_GravelRocksSmall_01_FoliageType
                   (0 refs; 41 InstancedFoliageActor still place /FortniteFoliage mounts)
Content/Materials/ MI_ENV_Prison_Landscape_A   ← live Landscape.landscapeMaterial
                   MI_Creative_Landscape_Asteria_Inst  (0 refs; leftover Creative default)
Content/           GameFeatureData, prison_sample_DefaultHLODLayer
```

`MI_ENV_Prison_Landscape_A` parent: `/FortniteLandscape/Materials/MaterialInstances/MI_FortniteLandscape_Customizable_01`. Confirm on the placed `Landscape` actor — `get_referencers` returns 0.

**Tier 2 mounts:** `/ENV_Meadow_Assets` 13548 + `/ENV_Meadow_Comp` 8878. Also `/CRP_GroundFloors`, `/Drome_Comp`, `/FortniteFoliage`. No Fab `.uref`.

## What this sample actually teaches

1. **Interior vs yard lighting is a Day Sequence volume at 2050, not a second Time of Day.** techniques/interior-volume-day-sequence.md. Third source for 2000-vs-higher: day-sequence-priority-stack.md.
2. **The prison geo is a mounted meadow kit + 12 playsets, not 13 authored meshes.** `/ENV_Meadow_*` + `PlaysetRoot` labels. Same catalog-vs-kit split as dimension_x_template (`/SCI_FI_*`) and city_starter_template (`/City_*`). Prison does **not** ship a custom `SM_SkySphere` — DSD skybox + `BackDrop` cards.
3. **Walkers are not on this island.** Character Definition Type=`Walker`, gated hordes, FPS: walker-character-definition.md, volume-checkpoint-gated-horde.md, switch-gated-first-person.md. Do not subclass `npc_behavior` here (npc-behavior-navigatable-focus.md).
4. **No ExtraAssets HUD.** Zero HUD Controller. ExtraAssets coincidence stays two sources (Social + Walker NPC). Same “no HUD” as alternate dimension — not a third ExtraAssets source.
5. **Sandbox flags, not a round.** `allowBuilding=All`, `bInfiniteResources=true`, `bFallDamage=false`, `timeOfDayOverride=NoOverride` (DSD owns 20:00). Contrast City `allowBuilding=None`.
6. **Island Settings `spawnLocation=SpawnPads` with zero spawn-pad actors.** Spreadsheet note “player spawns in the sky” matches. `spawnLimit=1`. Do not cite this island as a spawn-pad example.
7. **Galleries live in the theme Content Browser, not this project’s `Content/`.** Docs table (Floor/Stair, Roof, Wall, Prop + named prefabs) is the `/ENV_Meadow_*` mount. Outliner labels say `EnvironmentPack Quarters *`, not “Cell Block”.

## Official tutorial vs shipped

**External:** Working With TWDU Islands (headed Chrome 2026-09-14). Single page covering **both** Prison Starter and Walker NPC. Verbatim: `_raw/prison_sample/tutorial.txt`. Spreadsheet `docsUrl` empty; `Link in Desc` is this URL. Linked Walker chapters belong to horde_npc_sample (already extracted).

### Lines up

| Tutorial | Shipped |
|---|---|
| theme Templates → zombie-survival Universe → Prison Starter Island + Walker NPC | picker; `featureSet` CustomGameplayFeature |
| Prison = pre-designed environment to build gameplay on | 9582 OFPA, almost no devices |
| Content Browser → zombie-survival galleries/prefabs | `/ENV_Meadow_*` mounts; 12 `PlaysetRoot` |
| Prefabs: Cafeteria and Gym, Garage, Medbay, Underground, Watchtower | `EnvironmentPack Cafe/Gym`, `EnvironmentPack Garage`, `TWD Prison Medbay`, `EnvironmentPack Underground`, watchtower roots |
| Galleries: Floor/Stair, Roof, Wall, Prop | `BP_ENV_Meadow_Wall_*` / `LowRoof_*` / `FlatRoof_*` / props dominate `actor-classes.txt` |
| Combine galleries with lighting and post-process | 2 DSD + unbound PPV + Lumen Exposure + wall lights |
| Walker NPC / Character Definition is UEFN-only | **not placed here**; lab is Walker NPC |
| Age-rating table for publishing TWDU | theme rules — not island content |

### Does not line up

| Tutorial | Shipped |
|---|---|
| “Take advantage of the many Walker-fighting scenarios” on Prison Starter | **0** `Device_CharacterSpawner`, **0** item spawners |
| FPS toggle + Heavy Linework (same page, Walker paragraph) | Walker NPC only; Prison has neither camera nor `Device_PostProcess` |
| Prefab names **Cell Block A/B/C** | Live labels **`EnvironmentPack Quarters A/B/C`** |
| Prefabs table: **one** Watchtower | Live **five** watchtower playset roots (`EnvironmentPack Watchtower`, `3/4/5`, `TWDU Prison Watchtower`) |
| Dedicated Prison walkthrough | **one** theme blurb; guessed `prison-starter-island-…` slug hub-redirects |

Training rule: live playset labels + zero Walker devices win. Do not emit “Prison Starter ships Walker fights / FPS / Heavy Linework.” Those are Walker NPC.

## Negative findings

- **Name implies a prison *game*; files show a kitbash sandbox.** No round manager, no Verse, no Walkers, no HUD.
- **`spawnLocation=SpawnPads` but no player spawn pads exist.** Spreadsheet: “player spawns in the sky.” `hUDInfoType=AI Enemy Elimination` with no AI.
- **`OuterFence` / `Sidewalks` / `Wall Lights` at root are empty leftovers** after the nested move. Root `Clutter/` still has **122** actors.
- **`MI_Creative_Landscape_Asteria_Inst` and both authored FoliageTypes have 0 referencers.** Live landscape material is `MI_ENV_Prison_Landscape_A`. `Landscape.layers[]` is **empty** despite seven authored `LayerInfo`s (spreadsheet: “117 landscape errors”).
- **`GetDeviceProperties` N/A** — no Verse `creative_device`s.
- **837 `BP_Blockout_Floor_A_A`** still placed (greybox floors shipped, unlike Walker NPC’s unused CubeGrid assets).

## Capture backlog

| Item | Status |
|---|---|
| Per-playset child actor lists (5839 in `Prefabs/`) | **Waived** — 12 `PlaysetRoot` labels + `/ENV_Meadow_*` mounts are the lesson |
| LayerInfo ↔ material slot wiring (7 assets, empty `layers[]`) | **Waived** — spreadsheet already flags landscape errors; don’t train as weight-blend demo |
| Lumen Exposure Manager knob values | **Waived** — day-sequence-priority-stack.md already covers when to use it |
| PIE sky-spawn / fall from origin | **Waived** — zero pads + spreadsheet note + `spawnLocation=SpawnPads` |
| Sequencer / cine camera | **Waived** — one leftover `CineCameraActor`; not the lesson |

## Live notes

- Mount `/000aa9d3-2913-4ad6-ac6c-71847da65b2f`. Project Verse path `/epictemplates@fortnite.com/prison_sample`. Plugin filename is that GUID.
- Did **not** `load_level` away from the default map.
- Billboard `text` N/A (no billboards). `GetDeviceProperties` N/A.
- `get_referencers` on authored landscape MI / LayerInfos / FoliageTypes returns 0; confirm `landscapeMaterial` on the placed `Landscape` actor.
- `get_actors_in_folder` descriptors often have `folderPath=None`; counts still match.
- Folder on disk / title is **`prison_sample`** (picker “Prison Starter Island”).
