---
id: digest-040
kind: digest
---

# prop_hunt_tutorial — sample digest

> Offline `_raw/prop_hunt_tutorial/`. Live
> `/39f596ad-46fd-4ee8-b188-992907205b29/prop_hunt_tutorial` (**Observed**).
> Disk folder is **`prop_hunt_tutorial`**, not `prop_hunt_template`.
> Docs: 10-chapter hub. Island code **0259-6053-5824**. Needs **two players** to start (spreadsheet).

| | |
|---|---|
| Classification | **Asymmetric hide-and-hunt** — Verse owns teams, round loop, idle heartbeat |
| Content type | catalog wrap: **20 authored vs 3306 placed** (mansion/prop dump) |

## Scale

| | |
|---|---|
| Authored | 20 (VFX, SFX, materials, sequences) |
| OFPA | **3306** |
| Verse | **8** |
| VerseDevices | 3 |
| Matchmaking | 16 players, **2 teams** |
| Compatibility | 43.00 |
| `Device_PropOMaticManager_V2` | 1 — Creative disguise |

Spreadsheet: *elevator lobby works; graphics filter hurts; FAB devices candidate.* Verse is the product; 3306 props are the map.

Empty `Developers/felixkupis/Collections/`. Seventeenth first-party hit.

## What it is *for*

Lobby → wait ≥2 participants → shuffle: `Ceil(N / HunterAgentPerNumberOfPlayers)` hunters (default 5.0, floor 1.1 so ≥1 prop). Props teleport out first; hunters wait **15s**. Props score per second; hunters score `Floor(MaxEliminationScore / propsRemaining)` (last prop worth most). Idle **15s** (`MinimumMoveDistance` 100) → warning UI then heartbeat VFX/SFX. Eliminated prop respawns as hunter. Round ends on empty prop team, empty hunter team, or `round_timer`.

Join mid-round → hunter. Leave → drop from team arrays (`TeamEmptyEvent` if last).

## Device inventory (the design)

| Count | Class | Role |
|---|---|---|
| 1 | `Device_PropOMaticManager_V2` | become a prop |
| 2 | class designer / class selector / team settings | teams+classes |
| 16 | VFX spawners | one heartbeat per player |
| 9 | teleporters | lobby / spawn |
| 2 | score manager | prop tick / hunter elim |
| 2 | timer | hunter wait / prop score |
| 1 | round settings | `EndRound` |
| 3 | Verse devices | `prop_hunt`, `waiting_for_more_players`, `round_timer` |
| 25 | player spawners | 16P |
| 29 | billboards | captions |
| rest | furniture/plants | throwaway |

## Verse (8 files)

Copied to `snippets/verse/`. Each file header cites the tutorial URL.

| File | Job |
|---|---|
| `base_team.verse` | abstract: ClassSelector, ScoreManager, HUD, TeamManager, `[]agent`, `TeamEmptyEvent` |
| `prop_team.verse` | idle loop, heartbeat, tracker |
| `hunter_team.verse` | ratio, 15s wait, MaxEliminationScore |
| `prop_hunt.verse` | round orchestration |
| `heartbeat.verse` | VFX teleport + radio SFX + warning UI |
| `round_timer.verse` | Verse countdown (no Timer device) + canvas clock |
| `waiting_for_more_players.verse` | min 2 + timer then start |
| `billboard_cinematic_flythrough.verse` | optional intro (same family as Commander/Stronghold) |

**Round end `race`:** `PropTeam.TeamEmptyEvent` vs `HunterTeam.TeamEmptyEvent` vs `RoundTimer.AwaitEnd()`.

## What this sample actually teaches

1. **Abstract `base_team` + two concrete teams.** Private `[]agent`; `TeamEmptyEvent` when Count<1. See techniques/asymmetric-teams-round-loop.md.
2. **Hunter count from player count:** `Ceil[N / Max(1.1, HunterAgentPerNumberOfPlayers)]`, Shuffle, Slice.
3. **Idle heartbeat:** `race` AwaitStopMoving vs AwaitStartMoving; 15s then VFX. `NumberOfHeartBeats` gates one radio. See techniques/idle-heartbeat.md.
4. **Hunter payout scales as props die:** `Floor(MaxEliminationScore / PropTeam.Count())`.
5. **Wait-for-N + Verse HUD** (`waiting_for_more_players`, `round_timer`) reusable devices.
6. **Join/leave mid-round** via `ParticipantAdded/RemovedEvent`. Mid-join → hunter.

## Official tutorial vs shipped

**External** hub: Prop Hunt Template

10 chapters (`prop-hunt-0N-…`). Ch.3 idle VFX, ch.5 game loop, ch.6 full script are the Verse ones.

`docsUrl` empty. Spreadsheet `/en-us/uefn/prop-hunt-template-…`.

### Lines up

| Tutorial | Shipped |
|---|---|
| 15s hunter wait; 15s idle heartbeat | `SpawnDelay`, `MoveTime` |
| Asymmetric teams, shuffle, join/leave | `prop_hunt.verse` |
| Score per second vs elim payout vs remaining props | same |
| Fewer devices / `@editable` tweaks without Verse | class fields |
| Needs 2 players | `MinimumNumberOfPlayers = 2` |
| Creative Prop Hunt vs this | hub: Verse enables the above |

### Does not line up

| Topic | Tutorial | Shipped |
|---|---|---|
| Graphics filter | user hates it | PPV present — not Verse |
| Folder name | “Prop Hunt” | `prop_hunt_tutorial` |
| Ch.6 “full script” | one dump | **8 files** — snippets win |

Train on snippets. Don’t emit Creative-only Prop Hunt wiring as this island.

## Negative findings

- 3306 OFPA is not a kit.
- Filter/PP is a taste issue, not a Verse lesson.
- Scratch `Developers/` folder.

## Capture backlog

1. Live Prop-o-Matic / class designer settings. *Waive — Creative, not Verse.*
2. Heartbeat VFX array length vs max players 16. *Open, low.*

## Live notes

Level confirmed. Streaming off.
