---
id: digest-041
kind: digest
---

# gift_factory_sample — sample digest

> **Extractor provenance.** `tools/extract-template.py` (Pass A) + UnrealMCP on
> `localhost:8000` (Pass B). Default level
> `/a490a843-f5cf-4e49-9684-5469e9b6a913/gift_factory_sample` (**Observed**).
> Hub: Building BRICK Islands
> + index BRICK Templates
> + island tutorial BRICK gift factory
> + companion Working with BRICK Islands.
> `.uefnproject` `docsUrl` empty; listed in `Templates list.xlsx` row 90. Compat **43.00**. `featureSet`: **BrickGameplay**.

| | |
|---|---|
| Classification | **BRICK learning island & tycoon framework** — 4-player cooperative gift factory with persistent real-time calendar, offline progression simulation, and farmable resource fields |
| Content type | **kit + syllabus** (170 authored assets vs 1016 placed actors; 57 billboards + 5 NPC booths form an interactive classroom) |
| Verse | **42 files** (38 placed Verse device actors across 18 unique roles) — ties `user_interfaces_template` for 2nd richest Verse packet in corpus |
| Fab `.uref` | **0** |

## What it is

The picker name is **BRICK → gift factory** / **gift factory** (BRICK). Docs H1 is **BRICK gift factory**. Disk folder is `gift_factory_sample`. The live map is **`gift_factory_sample`**.

This is the flagship educational template for **Verse Persistence** and **real-time simulation** in UEFN. It demonstrates how to build an idle tycoon game that continues producing while players are offline, manage real-world Gregorian calendar dates and midnight UTC rollovers, animate 3D clock hands via trigonometry, operate multi-station factory assembly lines, and harvest respawning crops using Prop Manipulators.

The island includes a structured 3-tier creator curriculum with in-game guide characters:
- **Beginner**: Tweak `@editable` regrowth parameters on `candy_cane_field`.
- **Intermediate**: Modify 2D matrix economic upgrade curves in `upgrade_config.verse`.
- **Expert**: Complete deliberately unfulfilled daily reward logic in `player_info.verse` and `daily_gift.verse` using calendar rollover checks.

## Scale (**Observed** — `summary.txt` + live)

| | |
|---|---|
| Authored assets | **170** (`Audio/` 48, `Textures/` 40, `Props/` 24, `Animations/` 21, `Effects/` 20, `Materials/` 14, `Factory/` 10, `Persistence/` 7, `Time/` 3, `Config/` 2, `Meshes/` 1) |
| External actor packages (OFPA) | **988** (1016 placed actors returned live) |
| External object packages | **68** |
| Total `.uasset` | **1226** |
| Maps | `Content/gift_factory_sample.umap` (14,099 bytes) |
| Verse files | **42** |
| Placed Verse devices | **38** (12 factory stations + 11 UI managers + 5 speaking characters + 10 gameplay managers) |
| Data layers | **0** |
| Fab `.uref` | **0** |
| Compatibility | **43.00** (`requiredRedirectorStartingVersion: 43.00.57945210`) |
| Matchmaking | `maxPlayers=4`, `maxTeamSize=4`, `maxTeamCount=1`, JIP `true` |
| FeatureSet | **BrickGameplay** |
| Total folder size | **49.76 MB** (well under the 100 MB GHES cap) |

## Level organisation (**Observed** — live `get_folders`)

```
Background / Background DONE
Call To Action
Candy Cane Field / Candy Cane Field DONE / Creator Info
Challenge Beginner / Creator Info
Challenge Intermediate / Creator Info
Challenge Expert / Challenge Expert DONE / Creator Info
Daily Gift / Daily Gift DONE / Creator Info
Environment / Lodge (Building, Indoor, Outdoor) / Rusty's Den / Sunrise View
Factory / Factory DONE / Line 0..3 (Belt, Buttons, Mould, Wrapper) / Storage / Creator Info
Game / Game DONE / Sir Pawlar / Button / Creator Info
Jukebox / Jukebox DONE / Button / Creator Info
Order / Button / Creator Info
Real Time / Creator Info
Spawners
Test / Test DONE / Sub Zero / Creator Info
Tutorial / Tutorial DONE / Creator Info
UI / Cameras / Creator Info
```

57 billboards live in the level, heavily concentrated in the `Creator Info` booths beside each system's creator pod.

## Device inventory (the design)

| Count | Class | Role |
|---|---|---|
| 88 | `BP_Cane_BeltTrack_A` | Modular conveyor track segment |
| 64 | `BP_Cane_BeltAxel_A` | Conveyor mechanical rollers |
| 55 | `Device_Billboard_V2` | Syllabus boards, challenge cards, and file citations |
| 55 | `Device_CRD_AudioPlayer` | Ambient, machine stamps, conveyor hums, UI jingles |
| **38** | `VerseDevice_C` | Full game logic stack (all 38 placed devices) |
| 38 | `PointLight` | Level lighting and interior accents |
| 21 | `Device_CinematicSequence` | Machine cycle sequences (`LS_Factory_Mould_*`, `LS_Factory_Wrapper_*`, `LS_Order_Completion`) |
| 20 | `BP_Light_Green` | Machine operational step indicators |
| 19 | `BP_Light_Red` | Machine blocked / error step indicators |
| 18 | `BP_Light_Orange` | Machine standby step indicators |
| 16 | `Device_Trigger_V2` | Station interaction triggers and factory load trigger |
| 15 | `Device_CameraModes_FixedPoint` | Cinematic inspect views for factory stations and UI |
| 15 | `Device_Tracker_V2` | Tutorial steps (`TalkToSirPawlar`, `LoadFactory`, `OfflineProductionPerformedTracker`) + debug time offset |
| 14 | `Device_MutatorVolume_V2` | Station zones and lobby bounds |
| 11 | `Device_PropManipulation_V2` | Candy cane harvesting and Daily Gift show/hide control |
| 8 | `Device_MapIndicator_V2` | Tutorial directional map beacons guiding players |
| 7 | `Device_PointLight_V2` | Creative light devices |
| 7 | `NiagaraActor` | Snowfall, steam vents, confetti bursts |
| 4 | `BP_BRICK_Player_Spawner_Prop` | Island player spawn locations |
| 2 | `Device_Button_V2` | Sub Zero debug buttons (Skip 1 Hour, Grant 1000 Gold) |
| 1 | `Device_HUDController_V2` | Game HUD configuration |
| 1 | `Device_PopUpDialog_V2` | Factory reset warning when owner disconnects |
| 1 | `Device_ExperienceSettings_BRICK` | Island settings |
| 1 | `VK_Waterbody_Ocean_Parent` | Surrounding island ocean |

## Placed Verse Devices (**Observed** — 38 instances)

| Device Label | Verse Class | Role |
|---|---|---|
| `game_manager` | `game_manager` | Central coordinator; joins/leaves; loads factory for single/owner player |
| `factory_manager` | `factory_manager` | 4-line production coordinator; offline progression simulation; conveyor animation |
| `order` | `order` | Customer delivery contract system; evaluates storage, triggers delivery cine |
| `tutorial` | `tutorial` | Step-by-step questline using Trackers and Map Indicators |
| `jukebox` | `jukebox` | Paid audio player; unlocks festive tracks for gold |
| `test` | `test` | Creator debug tools; wires Skip Hour & Grant Gold buttons |
| `daily_gift` | `daily_gift` | Calendar-gated present; midnight UTC rollover reward |
| `real_time` | `real_time` | UTC time provider; Common Era calendar math; 3D clock hands rotation |
| `candy_cane_field` | `candy_cane_field` | 10 harvestable cane props with individual regrowth cycles |
| `speaking_character` (×5) | `speaking_character` | Proximity state machine guides (Sir Pawlar, Sub Zero, Flopkins, Flapjack, Blubberchops) |
| `factory_storage` | `factory_storage` | Aggregated silo capacity; visual level upgrades |
| `factory_belt 0..3` | `factory_belt` | 4 conveyor lines; speed upgrades; waypoint movement |
| `factory_mould 0..3` | `factory_mould` | 4 stamping machines; multiplier upgrades; kinematic press |
| `factory_wrapper 0..3` | `factory_wrapper` | 4 packaging units; pneumatic ram animation; mesh swap |
| `ui_manager` | `ui_manager` | Master UI coordinator; instantiates and drives all player UI widgets |
| `factory_ui` | `factory_ui` | Main factory status canvas |
| `resource_ui` | `resource_ui` | Persistent gold & inventory counter HUD |
| `load_factory_ui` | `load_factory_ui` | Island entry factory claim prompt |
| `offline_production_ui` | `offline_production_ui` | "While you were away" modal dialog displaying offline gains |
| `factory_belt_ui` | `factory_belt_ui` | Conveyor belt upgrade canvas |
| `factory_mould_ui` | `factory_mould_ui` | Mould upgrade canvas |
| `factory_wrapper_ui` | `factory_wrapper_ui` | Wrapper upgrade canvas |
| `factory_storage_ui` | `factory_storage_ui` | Storage expansion canvas |
| `order_ui` | `order_ui` | Delivery order details canvas |
| `jukebox_ui` | `jukebox_ui` | Music selection canvas |

## Verse modules and structure (42 files)

All 42 files copied verbatim to `snippets/verse/`:
- `Content/` root (9 files): `game_manager.verse`, `factory_manager.verse`, `order.verse`, `tutorial.verse`, `jukebox.verse`, `test.verse`, `daily_gift.verse`, `candy_cane_field.verse`, `speaking_character.verse`, `utility_module.verse`.
- `Content/Factory/` (10 files): `factory_element.verse`, `factory_belt.verse`, `factory_mould.verse`, `factory_wrapper.verse`, `factory_storage.verse`, `factory_product.verse`, `factory_progress_lights.verse`, `factory_rotor_light.verse`, `factory_intensity_lights.verse`.
- `Content/UserInterface/` (11 files): `ui_manager.verse`, `resource_ui.verse`, `order_ui.verse`, `load_factory_ui.verse`, `offline_production_ui.verse`, `jukebox_ui.verse`, `factory_wrapper_ui.verse`, `factory_ui.verse`, `factory_storage_ui.verse`, `factory_mould_ui.verse`, `factory_belt_ui.verse`.
- `Content/Persistence/` (7 files): `player_info.verse`, `factory_wrapper_info.verse`, `factory_storage_info.verse`, `factory_belt_info.verse`, `factory_mould_info.verse`, `order_info.verse`, `order_line_info.verse`.
- `Content/Time/` (3 files): `real_time.verse`, `date_and_time.verse`, `time_interval.verse`.
- `Content/Config/` (2 files): `upgrade_config.verse`, `order_config.verse`.

## What this sample actually teaches

1. **Persistable real-time calendar and offline catch-up simulation** (persistable-realtime-calendar-offline-progression.md) — converting Unix epoch to Common Era, implementing a custom `<persistable>` Gregorian date struct with leap year rules, computing offline production catch-up bounded by capacity, and injecting debug time-travel offsets via persistent trackers.
2. **Analog 3D clock hands driven by Verse trigonometric rotation** (analog-clock-hands-trigonometric-teleport.md) — calculating hour, minute, and second angles in radians and applying `RotateBy` with `TeleportTo` to rotate dial pointers cleanly in 3D level space.
3. **Proximity state machine for speaking NPCs with animated prop cues** (proximity-state-machine-speaking-character.md) — continuous squared-distance player character polling driving a 4-state machine (`None`, `Idle`, `Attract`, `Speak`) with animated indicator props and audio chimes without trigger volume bloat.
4. **Resource harvesting and regrowth cycle via prop manipulators** (prop-manipulator-harvest-regrowth-cycle.md) — configuring pickaxe-damageable props with `prop_manipulator_device`, hiding them on harvest, playing cosmetic effects, granting persistent gold, and scheduling timed regrowth that resumes across game sessions.
5. **Modular factory conveyor pipeline and upgrade matrix** (modular-factory-conveyor-production-pipeline.md) — unifying conveyor stations under an abstract `factory_element` base class, animating physical product props along belt waypoints, and managing balancing curves via 2D matrices in `upgrade_config.verse`.
6. **Immutable persistent player data updating** — maintaining a `weak_map(player, player_info)` and executing atomic state updates via constructor cloning (`MakePlayerInfo<constructor>(SourceInfo)`).

## Official tutorial vs shipped

Hub: BRICK Templates
Dedicated chapter: BRICK gift factory.

### Lines up

| Topic | Tutorial | Shipped |
|---|---|---|
| Three creative challenges | Beginner (Candy Cane regrowth), Intermediate (Upgrade prices), Expert (Daily gift) | Identical structure, designated yellow pads, and character guides |
| Guide characters | Sir Pawlar, Freshman Flapjack, Blubberchops, Professor Flopkins, Sub Zero | All 5 placed as `speaking_character` devices with matching roles |
| 4-line factory layout | Belts, moulds, wrappers, storage | Exact 4-line setup in `factory_manager.verse` and level outliner |
| Offline progression | Simulates offline production using real time clock | Implemented in `factory_manager.CalculateOfflineProduction()` |
| Debug time travel | Sub Zero button skips 1 hour ahead to test offline production | Placed `SkipHourButton` hooked to `RealTime.AddDebugTimeOffsetHours(1)` |
| Regrowth parameter | `CandyCaneRegrowthTime` parameter | Shipped `@editable CandyCaneRegrowthTime:float = 45.0` in `candy_cane_field.verse` |
| Economy tables | `upgrade_config.verse` multi-column tables | Shipped `FactoryBeltPrices: [][]int`, etc. |
| Daily rollover check | Year/Month/Day comparison | Shipped `IsDateRolledOver()` snippet matches tutorial solution |

### Does not line up

| Topic | Tutorial | Shipped |
|---|---|---|
| Time device filename | Tutorial cites `real_time_device.verse` | Shipped filename is `real_time.verse` |
| Expert challenge completeness | Tutorial provides working solution code | Shipped `player_info.verse` and `daily_gift.verse` deliberately leave gaps with "Professor Flopkins" comments for student authoring |
| Platform scope | Tutorial claims "This UEFN-only template" | Shipped with `featureSet: BrickGameplay` under BRICK Islands theme with `/BRICK_Cane` asset mounts |
| Project docs URL | `.uefnproject` `docsUrl: ""` | Public tutorial exists and is live on Epic Developer Community |

## Negative findings

- **"gift factory" implies GiftGiver is an NPC or present in the level** — **No GiftGiver NPC exists**. The character guides are animals: Sir Pawlar (polar bear), Sub Zero (penguin), Professor Flopkins (rabbit), Freshman Flapjack (walrus), and Blubberchops (seal). GiftGiver is represented only by the vacant lot billboard ("Vacant gift factory Lot") and customer delivery orders.
- **"Offline progression" implies background cloud compute** — Verse cannot run when no server instance is active. The progression is entirely a **catch-up simulation** calculated upon player rejoin by evaluating elapsed Common Era seconds against production intervals and storage capacity.
- **"Factory conveyor belts" implies physics or splines** — Conveyor belts do not use physics or SplineComponents. Belt tracks are static meshes (`BP_Cane_BeltTrack_A`), and product props are moved along discrete waypoints using kinematic `MoveTo()`.

## Capture backlog

All items resolved live via MCP:
- Level path confirmed: `/a490a843-f5cf-4e49-9684-5469e9b6a913/gift_factory_sample`
- All 57 billboard texts extracted (`_raw/gift_factory_sample/billboard-texts.md`)
- All 38 Verse devices classified and mapped to source files
- All 42 Verse files deposited in `KB/templates/gift_factory_sample/snippets/verse/`
- Full tutorial comparison completed (`_raw/gift_factory_sample/tutorial.txt`)

## Live notes

- MCP connection: localhost:8000 responsive via `tools/mcp.py`.
- No Data Layers are referenced (0 data layers).
- `bIsSpatiallyLoaded`: World Partition actors loaded in editor without streaming hitches.
