---
id: digest-042
kind: digest
---

# scene_graph_template — sample digest

> **Extractor provenance.** Offline evidence from `tools/extract-template.py` → `_raw/scene_graph_template/`.
> Live-editor facts from UnrealMCP on `localhost:8000` with the project open
> (`ValkyrieToolset.EntityToolset` + `SceneTools` + `AssetTools` + `DeviceToolset`).
> Loaded level: `/8e9a56a2-05cf-4ef8-bc34-4fc22186bac3/scene_graph_template`
> (**Observed**). Experimental Scene Graph flags in `.uefnproject` are **Unverified**
> as shipped state — they are writable from Project Settings. Content-side evidence
> (6 `.verse` files, 3 `EntityPrefab` assets, 65 live entities) is the stronger
> signal that this sample actually uses Scene Graph.

| | |
|---|---|
| Classification | **Walkable Scene Graph curriculum** — one hall, two sequences of booths |
| Content type | *reusable prefabs + reference lesson* — `P_LightPost`, `P_Cow`, `P_UFO` are authored here |
| Official docs | Scene Graph Sample Tutorial (**External**) — three chapters. Docs and shipped Verse have drifted; prefer the files. |

## Scale

**Observed** — `_raw/scene_graph_template/summary.txt` + live `AssetTools.find_assets` + `EntityToolset.FindEntities`.

| | |
|---|---|
| Authored assets (`.uasset` outside `__External*__`) | **67** |
| OFPA actor packages | 28 |
| Maps | 1 (`Content/scene_graph_template.umap`) |
| Verse files | **6** |
| Fab refs (`.uref`) | **0** |
| Distinct actor classes (OFPA histogram) | 14 |
| Classified actors | 28 |
| Live Scene Graph entities | **65** (incl. children) |
| Reference mounts | 27 |
| Compatibility | `43.00` |
| VerseAPI / VerseVersion | 1 / 1 |
| `EnableSceneGraph` | true (plugin) |
| `EnableVerseAssetReflection` | true |
| `bEnableOneFilePerEntity` | false |
| Data layers | **0** |
| `LevelHasStreamingDisabled` | **1** (**Observed** — `AssetTools.get_asset_tags`) |
| Matchmaking | 1–16 players |

Authored-vs-placed: **67 authored / 28 OFPA actors / 65 live entities**. Kit, not catalog. The subject is Scene Graph prefabs + Verse components; the hall geometry (`BP_DemoRoom`, hashed `Combined_*` / `Sphere_*` meshes) is throwaway set dressing.

## What this sample is *for*

A **walk-the-hall curriculum**. Two sequences, left-to-right.

### Sequence A — iterate a prefab by adding components

Five instances of `P_LightPost`, each a later layer of the same prefab (**Observed** — live entity list + docs chapter 01). Demo boards captured from OFPA FNames (**Observed**):

| Station (approx. X) | What is on the instance | Board text |
|---|---|---|
| x ≈ −6462 | Post mesh + Pivot + Lantern mesh + sounds | `Basic hierarchy of Entities and Mesh Components` |
| x ≈ −5472 | + `sphere_light_component` + hum | `Light Component Added` / `Light Component Added <br><br>Sound Component Added` |
| x ≈ −4440 | + `particle_system_component` (`NSYS_LampFlies`) | `Particle System Component adds the moths VFX` |
| x ≈ −3443 | Pivot gets `simple_movement_component` (`MovementMode = PingPong`, `AutoPlay = true`) + keyframed movement | *(C13/C14 titles not in the FName table — backlog)* |
| x ≈ −2470 | + `lantern_interaction_component` on the prefab root, `interactable_component` (`Enabled = true`) on Lantern | *(same)* |

Prefab hierarchy (**Observed** — `EntityToolset.GetComponents` on instance `P_LightPost_nmkpc0kwvy1o3_1507597414`):

```
P_LightPost                          mesh SM_Lightpost_01
  + lantern_interaction_component    (only on the last instance)
  Pivot                              keyframed_movement + simple_movement
    Lantern                          mesh SM_Lightpost_Lantern_01 + interactable
      SphereLight                    sphere_light_component
    Sound_*                          MetaSound sources (hum / ignite / creak)
    VFX                              NSYS_LampFlies
```

Child entities of a prefab class as `entity` (not the prefab `_C`). Instantiating `P_LightPost` brings the whole tree.

### Sequence B — query, spawn, abduct

Three UFOs + nine cows + one spawn pad, with volume devices as the player-facing trigger.

| Station | Live wiring (**Observed**) | Board text |
|---|---|---|
| Spawn pad (`entity_u2aa9f36qalr3_1240676800` at x=−919, y=9084) + `Volume: HMD01` | `SpawnPrefabDevice` (Verse `creative_device`) | `Add & Remove Entities From the World` |
| UFO + cow, cow has `entity_entered_example_component` | mesh `EntityEnteredEvent` | `Entity Entered / Exited` |
| UFO has `find_overlaphits_example_component`; `Volume: HMD2` | `@editable Trigger: volume_device` | `Step into the Trigger to stop the UFO and attempt to abduct a cow.` |
| UFO has `find_sweephits_example_component`; `Volume: HMD3`; two `Globe*` spheres sit over cows | sweep down `Up := -300` | same prompt text on C16/C17 |

Three Creative volumes are the *player* interface. Scene Graph queries are the *entity* interface. Mixing them is the lesson, not an accident.

## Project configuration

**Observed** — `_raw/scene_graph_template/config.txt`.

- `.uefnproject` `bindings.projectId` `46398064-fc8f-4a5f-acfe-ef9936c1c20a`; Verse module GUID `8e9a56a2-05cf-4ef8-bc34-4fc22186bac3` (content mount).
- `docsUrl` is **empty** in the descriptor. Spreadsheet and Epic docs still point at the tutorial (**External**).
- Plugin: `CanContainVerse` true, `EnableSceneGraph` true, `VerseAPI` 1, `EnableVerseAssetReflection` true. Prefabs show up in `Assets.digest.verse` as classes (`LightPost.P_LightPost`, `Meshes.SM_Toy_Cow`, …) — that reflection flag is load-bearing for the Verse in this sample.
- `experimental.sceneGraph.bIsSceneGraphSystemAllowed = true`, `pythonExperimental` on, `toolsets` on. Treat as **Unverified** shipped; the 65 live entities are the content-side proof.
- No Fab packs.

## Device / actor inventory (classic)

**Observed** — `_raw/.../actor-classes.txt` (OFPA) + live `get_actors_in_folder("Room 01")`.

| Count | Class | Role |
|---|---|---|
| 9 | `BP_DemoDisplay` | Curriculum captions (same construction-script signage as intro_to_uefnfk1) |
| 3 | `Device_Volume` | HMD01 / HMD2 / HMD3 — player steps on these |
| 3 | `Device_Decal_V2` | Floor marks on the pads |
| 2 | `Device_Billboard_V2` | *(text not captured — backlog)* |
| 1 | `VerseDevice` | `SpawnPrefabDevice` |
| 1 | `BP_Creative_Player_Spawner_Prop` | Spawn |
| 1 | `BP_DemoRoom` | Hall geometry |
| 1 | `BP_DaySequenceModifier_Device` | Lighting |
| 1 | `Device_HUDController_V2` | HUD off |
| 1 | `Device_ExperienceSettings_V2_UEFN` | Island settings |
| 2 | `FortStaticMeshActor` | Set dressing |
| + | WorldPartitionMiniMap, LevelBounds, WorldDataLayers | Engine scaffolding |

**Very few Creative devices; Scene Graph does the gameplay.** Volumes exist because Verse components `@editable` a `volume_device` — there is no player-overlaps-entity input API used here.

## Content organisation

```
Content/
  EntityEnteredExampleComponent.verse
  FindOverlapHitsExampleComponent.verse
  FindSweepHitsExampleComponent.verse
  LanternInteractionComponent.verse
  SimpleMovementComponent.verse
  SpawnPrefabDevice.verse          # the only creative_device
  LightPost/                       # P_LightPost prefab + SM/MI/T_
  Prefabs/                         # P_Cow, P_UFO
  Meshes/                          # cow, UFO, beam, hashed modelling-mode leftovers
  Sounds/ + Sounds/Wavs/           # MetaSounds wrap WAVs
  VFX/                             # NSYS_LampFlies
  ExampleContent/                  # material-layer demos (likely leftover UE samples)
  Developers/felixkupis/           # empty scratch — sixth first-party hit
```

Naming: `P_` prefab, `SM_` mesh, `MI_` material instance, `T_` texture, `MSS_` MetaSound, `SFX_` wave, `NSYS_` Niagara. Hashed `Combined_*` / `Sphere_*` / `Box_*` are Modelling Mode leftovers, not the lesson.

Art source paths in import metadata (**Observed** — FName scan): light post from `CR8V/.../PrefabExample/Lightpost`; cow/UFO FBX from `C:/Users/RodneyHoule/Desktop/`; audio from `L:/Audio/Projects/Scenegraph/Example_0625/`. Rodney Houle is the spreadsheet Dev Owner.

## Verse (6 files)

Copied verbatim to `snippets/verse/`.

| File | Class | Attached to (live) |
|---|---|---|
| `SimpleMovementComponent.verse` | `simple_movement_component` | Light-post Pivot (PingPong), UFO roots (Loop / PingPong), UFO TopMesh (Loop spin) |
| `LanternInteractionComponent.verse` | `lantern_interaction_component` | Last `P_LightPost` root |
| `EntityEnteredExampleComponent.verse` | `entity_entered_example_component` | One `P_Cow` |
| `FindOverlapHitsExampleComponent.verse` | `find_overlaphits_example_component` | One `P_UFO` |
| `FindSweepHitsExampleComponent.verse` | `find_sweephits_example_component` | One `P_UFO` |
| `SpawnPrefabDevice.verse` | `SpawnPrefabDevice` (`creative_device`) | The single Verse device actor |

Shared API surface this sample actually exercises:

- `component` / `creative_device`
- `keyframed_movement_component` + `keyframed_movement_delta` + `oneshot_` / `loop_` / `pingpong_keyframed_movement_playback_mode`
- `Entity.AddComponents`, `FindDescendantComponents`, `GetComponent[]`, `GetEntities()[0]`
- `GetSimulationEntity[]` + `AddEntities` + `RemoveFromParent` + `SetGlobalTransform`
- `mesh_component.EntityEnteredEvent` (`Await` in shipped code)
- `entity.FindOverlapHits(transform, collision_capsule)` + typed mesh cast `Meshes.SM_Toy_Cow[Overlap.TargetComponent]`
- `entity.FindSweepHits(DisplacementVector)` + `sweep_hit.TargetComponent.Entity`
- `interactable_component.SucceededEvent`
- `sphere_light_component` / `spot_light_component` / `particle_system_component` / `sound_component` Enable/Disable/Play/Stop
- generated asset classes from `/epictemplates@fortnite.com/scene_graph_template`
- `Entity.GetFortRoundManager[].SubscribeRoundStarted` so logic waits for round start, not simulation begin
- SpatialMath `vector3{Left, Up, Forward}` (**not** XYZ) — SpawnAt-namespace rule applies here too

## What this sample actually teaches

1. **A Scene Graph prefab is a tree of `entity` + `component`, saved as one asset.** Iterate by adding a component (or a child entity) at each booth; don't rebuild the mesh. See techniques/prefab-iteration.md.
2. **`keyframed_movement_component` does not author its own path.** A tiny Verse wrapper (`simple_movement_component`) owns `@editable` keyframes + playback enum and calls `SetKeyframes`/`Play`. Same wrapper on a lantern pivot (PingPong sway), a UFO hull (Loop patrol), and a UFO top mesh (Loop spin). See techniques/simple-movement-component.md.
3. **Find descendants by component type, not by name.** `lantern_interaction_component` lives on the prefab *root* and `FindDescendantComponents(light_component)` / `(particle_system_component)` / `(interactable_component)`. Extra lanterns would work with no code change.
4. **Three query styles, increasing precision:** mesh `EntityEnteredEvent` (this entity overlapping) → `FindOverlapHits` (any hits in a volume you place) → `FindSweepHits` (first hit along a vector, so a globe can block). See techniques/scene-queries.md.
5. **Spawn/despawn is `GetSimulationEntity[].AddEntities` / `RemoveFromParent`, not a Creative spawner.** A `creative_device` is only glue to a `volume_device` the player can step on. See techniques/spawn-remove-prefabs.md.
6. **Generated mesh/prefab classes are how Verse names content.** `Meshes.SM_Toy_Cow[Overlap.TargetComponent]` is a typed cast, not a string compare. Requires `EnableVerseAssetReflection`.
7. **Player input still comes from Creative devices.** Volumes + `interactable_component`. There is no raw input API on entities in this sample.

## Official tutorial vs shipped code

**External** — three chapters under Scene Graph Sample Tutorial:

1. Iteration in Scene Graph
2. Movement and Interaction
3. Flexible Gameplay

Docs are marked **Beta**. `.uefnproject` `docsUrl` is empty — the spreadsheet still links the hub page.

**Use the tutorial as a syllabus outline only.** Concept sequence matches the hall (five light posts, then spawn pad, then three UFO queries). The Verse *in the tutorial* is an older draft. Train on `snippets/verse/`. Downstream models that ingest both will emit the stale APIs unless this table is in the corpus.

### What lines up (safe to learn from the tutorial)

| Claim in docs | Live / files |
|---|---|
| Scene Graph = entity tree + components, not actors | **Observed** — 65 entities, 3 prefab `_C` roots |
| Prefab instance shares meshes/textures | 5× `P_LightPost` against one asset |
| Five light-post booths: mesh → light → VFX → sway → interact | Live instances + demo boards |
| Overrides are per instance; Prefab Editor propagates | Stations 4–5 add components the earlier ones lack |
| `EnableVerseAssetReflection` exposes `LightPost.P_LightPost`, `Meshes.SM_Toy_Cow` | Required for every snippet to compile |
| `FindDescendantComponents(type)` from the prefab root | `lantern_interaction_component` |
| `interactable_component.SucceededEvent` on the lantern, logic on the root | **Observed** |
| Player volumes (`volume_device`) start UFO queries / spawn | 3× `Volume: HMD*` |
| Overlap uses `collision_capsule` + typed mesh cast | shipped |
| Sweep uses `FindSweepHits` + first-hit helper so a globe can block | two `Globe*` spheres **Observed** |
| Despawn is `RemoveFromParent` | shipped |
| `vector3{Left, Up, Forward}` | shipped |

### What does **not** line up — do not train on the tutorial's code

Compared **Observed** against chapters 02–03 (raw page text, 2026-09-09). Shipped wins.

| Topic | Tutorial (stale) | Shipped file |
|---|---|---|
| Movement class name | `basic_movement_component` in prose | `simple_movement_component` |
| Movement lifecycle | `OnSimulate<override>()<suspends>` + `Sleep(0.1)` then Play | `OnBeginSimulation` (no Sleep) + `GetFortRoundManager[].SubscribeRoundStarted` → `OnRoundStarted` does `Play()` |
| Adding the keyframed component | `Entity.AddComponents of array { New… }` (invalid Verse) | `Entity.AddComponents(array{ NewKeyFramedMovementComponent })` with `{ Entity := Entity }` |
| AutoPlay | Play inside `Initialize…` | Play only in `OnRoundStarted` gated on `AutoPlay?` |
| Lantern toggle | one `ToggleLight()`, branch on `MaterialInstance.Emissive_Multiply = 0.0` | `var IsLightOn:logic = true`; `TurnLightOn` / `TurnLightOff`; `OnInteractFinished` flips the bool. Emissive is written, not read, as state |
| Lantern audio | docs snippet omits it | shipped Play/Stop on `MSS_LightBuzz_Loop`, `MSS_LightBuzzMoths_Loop`, ignite one-shot |
| Cow entered-event | `Mesh.EntityEnteredEvent.Subscribe(OnEntityEntered)` + misspelled `PlayCowAbdcutionAnimation` | no Subscribe. `spawn{PlayCowAbductionAnimation()}` then `loop{ … Mesh.EntityEnteredEvent.Await() … hide … Sleep(2) }` — hide, don't remove; loops forever |
| Cow round gate | not in the tutorial snippet | `SubscribeRoundStarted` so the lift waits for the match, not entity sim |
| Spawn prefab | construct `PrefabToSpawn` **once**, `race` of enter-block vs exit-block, re-`AddEntities` the same object | `loop { SpawnedPrefab := LightPost.P_LightPost{} ; Await enter; Add; Await exit; RemoveFromParent }` — **new instance every enter** |
| Spawn concurrency | `race` so enter/exit "always run" | sequential Awaits; you cannot re-enter until exit has been seen |
| Overlap capsule offset | `Translation.Up - 248.0` | `− 164.0` (half of Length 328 — centered under hull) |
| Overlap `OnTriggerExited` | not described | `Play()` hull + `DisableAbductionBeam()` even if still abducting — can cut the beam on step-off |
| Sweep first-child | "UFO's mesh entity" | `Entity.GetEntities()[0]` — order-dependent |
| Sweep abduct target | docs imply mesh entity | `HitEntity.GetComponent[Meshes.SM_Toy_Cow]` then `AbductCow(HitEntity)` — HitEntity is the mesh's entity (the cow prefab root if the mesh is on the root) |

Also in the tutorial, not bugs in the template: `Intractable component` typo, `gameplya` tag, `PlayCowAbdcutionAnimation` misspelling. Those must not enter a training set as identifiers.

### Training rule

```
IF source == official Scene Graph Sample Tutorial AND text is Verse
    → treat as UNTRUSTED / outdated draft
IF source == KB/templates/scene_graph_template/snippets/verse/*.verse
    → treat as ground truth for this sample
Tutorial prose (booth sequence, prefab-iteration idea, query taxonomy)
    → still valid as outline
```

If a model is trained on both, prefer the snippet whenever they conflict. Emitting `OnSimulate` + `Sleep(0.1)` or `AddComponents of array` is a doc-regurgitation failure.

## Negative findings (**Observed**)

- **Official tutorial has drifted from shipped Verse.** Full table above. Short version: `OnSimulate`/`Sleep`/`Subscribe(OnEntityEntered)`/`race`+reuse vs round-subscribe/`Await`/construct-in-loop. Train on snippets.
- **`find_sweephits` UFO's `simple_movement_component.Keyframes` point at the overlap-UFO's delta objects** (live `ListComponentProperties`). Copy-paste of the keyframe array across instances. It still plays; it is not a second authored path.
- **Native `keyframed_movement_component` lists no editable properties** through `EntityToolset.ListComponentProperties`. The wrapper's `@editable` Keyframes/AutoPlay/MovementMode are what the outliner actually exposes.
- **`lantern_interaction_component` itself lists no editable properties.** All behaviour is FindDescendant + a local `IsLightOn`. Designers don't tune it per instance.
- **`ExampleContent/` material-layer assets** look like UE sample leftovers (cow/UFO textures originate in `ArtSource/Rocket/Samples/Demolets`). Do not cite them as the Scene Graph lesson.
- **Empty `Developers/felixkupis/Collections/`** ships. Sixth first-party template with a scratch folder.
- **World Partition on, streaming off** (`LevelHasStreamingDisabled = 1`). Same as every other UEFN template in this KB so far.

## Capture backlog

Ranked by teaching value. Waive freely.

1. **DemoDisplay C13 / C14 (and remaining description bodies).** Title strings for the sway / interact stations were not in the OFPA FName table. Construction-script text lives as property values. `ObjectTools.get_properties` on those actors — folder query returned the volumes/decals, not the boards (boards may sit in a different folder / at root). *Open.*
2. **`Device_Billboard_V2` captions** (2 actors). Same. *Open.*
3. **Keyframe delta transforms** (duration, translation, easing) for lantern PingPong and UFO Loop. `ListComponentProperties` only returned `refPath`s to `keyframed_movement_delta` subobjects. `ObjectTools.get_properties` on those paths. *Open — high value for the movement technique.*
4. **`SpawnPrefabDevice.TriggerVolume` live binding** (which of HMD01/2/3). Code hardcodes spawn translation `(-9084, -8, -919)` in Left/Up/Forward. *Open, low.*
5. **`ExampleContent` referencers** — confirm unused. *Waived unless a cow/UFO material actually points there.*

## Live-editor notes (this pass)

What the **dev-build** UnrealMCP added over retail:

- `ValkyrieToolset.EntityToolset` — `FindEntities`, `GetComponents`, `ListComponentProperties` by **friendly Verse name**. This is how the 65-entity tree and `@editable` values were read. No mangled `__verse_0x` names needed on those properties.
- `VerseToolset.VerseToolset.ListFiles("")` with `bRecursive=true` also listed `__ExternalActors__` directories; it is a content-tree listing, not "Verse files only". Filter on `.verse` suffix.
- `DeviceToolset.ListDeviceAssets` returns the **whole Creative catalog**, not project-authored devices. Useless as an inventory of this sample; use OFPA + entity list instead.
- `get_actors_in_folder("")` returned empty; `"Room 01"` returned 6 (3 volumes + 3 decals). Boards/Verse device/spawner live outside that folder — don't treat one folder query as the full actor list. OFPA histogram is the complete count.

Unreachable this pass: Blueprint construction-script *values* on `BP_DemoDisplay` (need `get_properties` with the right field names); Sequencer N/A (none).
