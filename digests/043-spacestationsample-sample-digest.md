---
id: digest-043
kind: digest
---

# space_station_sample — sample digest

> **Extractor provenance.** `tools/extract-template.py` (Pass A) + UnrealMCP on
> `localhost:8000` (Pass B). Default level
> `/d1fedf73-6faf-4ae5-8d26-0fc3711445cd/space_station_sample` (**Observed**).
> Docs: Working With space-opera™ Islands
> (hub; Space is one environment-starter bullet) + Vistas and Biome Environments
> (Hyperspace chapter + feature list) + Custom UI.
> `docsUrl` empty. Guessed `star-wars-space-starter-island-in-fortnite` **hub-redirect**.
> Compat **43.00**. `featureSet`: **CustomGameplayFeature**.

| | |
|---|---|
| Classification | **theme environment starter** — Death Star hangar + hyperspace shot, not EnergyBlades/Narrative/Team Conquest |
| Content type | **catalog wrap** (28 authored vs 1624 OFPA; `/SPACE_THEME_B_*` hangar + `/THEME_PACK` destroyers) |
| Verse | **1 file** (`Attach_Hyperspace.verse` → `hyperspace_sequence_device`) |
| Fab `.uref` | **0** |

## What it is

The picker name is **Space Starter Island**. The island is a **16-player hangar in open space**: one `PID ENV_PACK SW HangarRoom` playset, Star Destroyer / DesertTown backdrop meshes, roaming Guards, an starfighter, and a **Verse-driven hyperspace jump** that hides the hangar data layer. Teaching that survives deletion is hyperspace-attach-datalayer.md. Guard patrols are **always-on** (no doorway swap) — do not duplicate guard-patrol-volume-swap.md. Day Sequence priority stays on day-sequence-priority-stack.md (this island is **2000-only**). Environment lighting rig stays on environment-lighting-rig.md. Custom skyboxes stay on night-atmosphere-stack.md — this sky is a DSD **Space** look, not `SM_SkySphere`. Data-layer *naming* stays on world-partition-data-layers.md; this island’s layer is a **cinematic unload**, not a streaming choreography. Sibling catalog wraps: ice_base_sample (`/SPACE_THEME_E_*`), volcanic_town_sample (`/SPACE_THEME_C_*`).

## Scale (**Observed** — `summary.txt` + live)

| | |
|---|---|
| Authored assets | **28** |
| OFPA | **1624** |
| Maps | `space_station_sample.umap` (only) |
| Verse files | **1** (`VerseToolset.ListFiles` on `/d1fedf73-…` — do not recurse `__ExternalActors__`) |
| Verse devices placed | **1** (`hyperspace sequence device`) |
| Data layers | **1** (`DL_Hangar`, **1554** actors, `initialRuntimeState=Activated`) |
| Fab `.uref` | **0** |
| Compatibility | **43.00** |
| Matchmaking | **16** (`maxPlayers=16`, `teamSize=Dynamic`, `teams=FreeForAll`, `locomotionPreset=Current BR`) |
| Streaming | Partitioned + actor folders + OFPA. **`LevelHasStreamingDisabled=1`** (Walker / VFX camp, unlike IceBase/VolcanicTown omit). Sampled Environment descriptors `bIsSpatiallyLoaded=false`; playset `get_properties` says **true** — moot because streaming is off. `load_actors` not required. |

Empty `Developers/felixkupis/Collections/`. First-party scratch folder.

`experimental.*` (Python / toolsets / Scene Graph) in `.uefnproject` is **Unverified**. Content-side: **one** `.verse`, **no** `__ExternalEntities__`. Plugin `EnableSceneGraph: true` with no Scene Graph entities.

Largest file well under GHES **100 MB**. Whole folder ~**16 MB**.

## Level organisation (**Observed** — live `get_folders` + folder counts)

```
Devices/                              5 — Player Spawner, Audio Mixer, IslandSettings, DSD, starfighter
Environment/                       1560
  (hangar kit, unfoldered under Environment)  1554
  Background/                         4 — Sailboat Bombardment A/A2/A3 + Nebula Planet PlTa
  Foliage/                            1 — InstancedFoliageActor
  VFX/                                1 — ENV_PACK NS THEME_PACK SkySwarm
Hyperspace/                           6 — Verse device, 3 cinematics, volume, BP_Hyperspace plane
LevelData/                            3 — WorldDataLayers, minimap, bounds
Lighting/                             1 — leftover PointLight
NPCs/                                47
  01/                                47 — 15 spawners + 32 patrol
```

No `Device_Billboard_V2`. No syllabus boards. No HUD Controller in any folder.

## Device inventory (the design)

| Count | Class | Role |
|---|---|---|
| 32 | `AIPatrolPath_Device` | hangar patrol; group **1** (24) / 11 (4) / 3 (2) / 6 / 12; BackAndForth **30**, StopAtEnd **2** |
| 15 | `Device_CharacterSpawner` | 13 Guard + 2 AWR Trooper; all `enabledAtGameStart=true` |
| 3 | `Device_CinematicSequence` | Enter / Exit / Data Layer (`LS_*` in `Content/Hyperspace/Cinematics`) |
| 1 | `VerseDevice` | `hyperspace_sequence_device` |
| 1 | `Device_Volume` | Hyperspace Trigger; Verse `AgentEntersEvent`, **not** NPC Enable |
| 1 | `BP_Hyperspace` | curved plane + enter/exit Niagara + cyan PointLight |
| 1 | `Device_CRD_AudioMixer` | `Hyperspace_Mix`, `bActivateOnGameStart=true` |
| 1 | `BP_DaySequenceModifier_Device` | FixedTime **1:00**, priority **2000**, `use Volume=false` |
| 1 | `CRD_SpawnerDevice_THEME_VEHICLE_A` | starfighter, health **600**, atmospheric effects **off** |
| 1 | `PlaysetRoot` | `PID ENV_PACK SW HangarRoom Prefab A` (on `DL_Hangar`) |
| 1 | `NiagaraActor` | `ENV_PACK NS THEME_PACK SkySwarm` (space-battle backdrop) |
| 1 | `BP_Creative_Player_Spawner_Prop` | island start |
| 1 | `PointLight` | leftover in Lighting/ (1000 cd, 8000 uu) |
| 0 | Billboard / HUD Controller / Hero / EnergyBlade / Conversation | other space-opera tiles |

## Authored kit (**Observed** — `assets.txt` + live)

```
Content/Attach_Hyperspace.verse          hyperspace_sequence_device
Content/Hyperspace/BP_Hyperspace
Content/Hyperspace/SM_Hyperspace_Plane + MI_Hyperspace
Content/Hyperspace/Cinematics/           LS_Enter_Hyperspace, LS_Exit_Hyperspace, LS_DataLayerController
Content/Hyperspace/Audio/                Hyperspace_Mix + Launch/Arrive/Travel cues
Content/Datalayers/DL_Hangar
Content/NPCs/CD_Guard
Content/UI/PlayerHUD/                    UW_HUD_* (5 widgets; no placed HUD Controller)
```

**Tier 2 mounts:** `/SPACE_THEME_B_Assets` 2815 + `/SPACE_THEME_B_Comp` 1554 (hangar gallery; playset `PID_ENV_PACK_SPACE_THEME_HangarRoom_Prefab_A`), `/THEME_PACK` 539 (`SM_Sailboat_Bombardment_A` Star Destroyers), `/ENV_SPACE_THEME_UI_Assets` 40, `/SPACE_THEME_D_Assets` 5 (`NS_Enter_Hyperspace` / `NS_Exit_Hyperspace` / `SM_Nebula_Planet_PlTa`), leftover `/c5745a77-…` 11 + `/Space_Starter` 8 (old GUID next to authored sequences). No Fab `.uref`.

## What this sample actually teaches

1. **Hyperspace is Verse `MoveTo` + three Cinematic Sequence devices + one data layer, not a Sequencer possessable of the player.** Volume subscribe → Play enter + Play data-layer cine → `race` attach-plane vs `Sleep(6)` → Play exit. techniques/hyperspace-attach-datalayer.md. Snippet: snippets/verse/Attach_Hyperspace.verse.
2. **The hangar is a mounted `/SPACE_THEME_B_*` catalog plus one playset root, not 28 authored meshes.** Same catalog-vs-kit split as IceBase (`/SPACE_THEME_E_*`) and VolcanicTown (`/SPACE_THEME_C_*`). **1554 / 1624** OFPA sit on `DL_Hangar` so the jump can unload them.
3. **Hangar NPCs are Guard CD + patrol with no doorway swap.** Type=`Guard` + Trespassers + AR + leash **on** 30 (IceBase-like, not VolcanicTown empty/unleashed). All **15** spawners start enabled; the only volume is hyperspace. Perception modifiers **disable** sight/hearing/touch. Link guard-patrol-volume-swap.md; do not train this island as a third volume-swap source.
4. **Space sky is one Day Sequence at 2000, FixedTime 1:00, not NightEnvironment’s custom sphere.** `overrideSky=true`, stock `MI_DaySequenceSkydome`, `starBrightness=0`. Docs: World Time of Day Manager **Space** actor + DSD refine; sun size / clouds “have no effect.” **Not** a fourth 2000-vs-higher source. Leftover `Lighting/` PointLight is unused next to the DSD.
5. **HUD widgets ship without a HUD Controller.** `Content/UI/PlayerHUD` matches IceBase/VolcanicTown filenames; live Outliner has **0** `Device_HUDController_V2`. space-opera in-project PlayerHUD coincidence stays **two** (IceBase, VolcanicTown). ExtraAssets HUD stays **three**.
6. **Sandbox flags, not a round.** `allowBuilding=All`, `bInfiniteResources=true`, `bFallDamage=false`, `timeOfDayOverride=NoOverride` (DSD owns 1:00). One spawn pad, `spawnLocation=SpawnPads`.
7. **starfighter is a THEME_VEHICLE_A vehicle spawner, not THEME_PACK N-1 / CopperPuck turret.** Health **600**, flight ceiling **250**, atmospheric effects **false** (VolcanicTown N-1 had them on).

## Official tutorial vs shipped

**External hub + Vistas (Hyperspace chapter) + Custom UI** (headed Chrome 2026-09-15). Verbatim: `_raw/space_station_sample/tutorial.txt`. Spreadsheet `docsUrl` empty. No dedicated Space chapter (guessed slug hub-redirects). Tutorial Verse expanded to 62 lines — **matches** shipped `Attach_Hyperspace.verse`.

### Lines up

| Tutorial | Shipped |
|---|---|
| theme Templates → space-opera → Space Starter Island; “celestial visuals and hyperspace effect” | picker; `featureSet` CustomGameplayFeature; Hyperspace/ folder |
| Death Star prefab and gallery assets | `/SPACE_THEME_B_*` + `PID ENV_PACK SW HangarRoom Prefab A` (docs name vs HangarRoom label) |
| Hyperspace effect; Enter/Exit Niagara in space-opera Content > VFX > Hyperspace | `/SPACE_THEME_D_Assets/…/NS_Enter_Hyperspace` + `NS_Exit_Hyperspace` on `BP_Hyperspace` |
| Verse `hyperspace_sequence_device`; 62-line snippet | same class; Attach_Hyperspace.verse |
| Three Cinematic Sequence devices + Volume + `BP_Hyperspace` + Details object refs | live `savedActor` wires plane / volume / enter / exit / data-layer cine |
| Data layer unloads hangar / “Death Star prefab” during the shot | `DL_Hangar` 1554 actors; `LS_DataLayerController`; `get_referencers` on the layer → enter + data-layer sequences |
| Extra assets in project `Hyperspace/` including sound | Audio mix + Launch/Arrive/Travel cues; mixer `bActivateOnGameStart` |
| Space Skybox; DSD refine; sun size / clouds no effect | DSD `overrideSky=true`, `overrideClouds=false`, `overrideSunlight=false` |
| Planet DesertTown + Star Destroyer background | `Nebula Planet PlTa` (`SM_Nebula_Planet_PlTa`); 3× `Sailboat Bombardment` (`SM_Sailboat_Bombardment_A`) |
| Roaming NPCs + spawners + patrol | 15 spawners, 32 patrol |
| Space Battle VFX | `ENV_PACK NS THEME_PACK SkySwarm` |
| starfighter among space-opera vehicles | 1 THEME_VEHICLE_A spawner |
| Custom UI widgets `UW_HUD_*` exist in PlayerHUD | authored in `Content/UI/PlayerHUD` |

### Does not line up

| Tutorial | Shipped |
|---|---|
| Verse file `hyperspace_sequence_device.verse`; “create a new Verse device” then paste | Shipped filename **`Attach_Hyperspace.verse`** already in the project; class name matches |
| Unload the **Death Star prefab** | Playset label **HangarRoom**; data layer **`DL_Hangar`** |
| Custom HUD “preconfigured on Flat Grid / XL Flat Grid”; “other space-opera island templates feature this custom UI”; HUD Controller in Outliner **UI** / cantina | **This** island authors widgets but places **0** HUD Controller. IceBase/VolcanicTown place one in **Devices/** |
| Editable widgets under `UI > Widgets > PlayerHUD`; Player Info = `UW_HUD_Vitals` | Shipped `Content/UI/PlayerHUD` (no `Widgets/`). Widgets-folder layout is Narrative |
| Dedicated Space walkthrough | Hub bullet + Vistas Hyperspace chapter; guessed dedicated slug hub-redirects |
| Docs typo `Sequencher` | Sequencer |

Training rule: shipped `Attach_Hyperspace.verse` + live `savedActor` bindings + `DL_Hangar` win. Do not emit “Space ships a HUD Controller / Death Star-named playset / `hyperspace_sequence_device.verse` filename.” Tutorial *prose* is the syllabus; tutorial Verse happens to match — still train on `snippets/`.

## Negative findings

- **Name implies a space *game*; files show a hangar sandbox plus one hyperspace gag.** No round manager, no EnergyBlades, no Hero / Conversation device.
- **`hUDInfoType=AI Enemy Elimination` with Neutral unkillable Guards** (`isVulnerable=false`) **and perception off**. Leftover default, same family as IceBase/VolcanicTown/Prison.
- **Authored HUD widgets have no placed controller.** OFPA class histogram listed `Device_HUDController_V2` ×1 — that string lives in a package, not the Outliner.
- **`GetDeviceProperties` returns mangled `__verse_0x…` field objects**, not the bound actors. Resolve `savedActor` on those objects (Bindings on the volume/cines are empty — Verse subscribes in `OnBegin`).
- **`get_referencers` on HUD widgets / CD_Guard / sequences returns []** except layer→sequences and PlayerVitals→PlayerInfo — confirm on placed devices, same trap as IceBase.
- **`Lighting/` PointLight leftover.** DSD owns 1:00.
- **Guard inventory is IceBase’s AR**, not VolcanicTown’s empty list. Perception-off is Space-only among the three environment starters.
- **`despawnAIsWhenDisabled=false`** on every spawner (IceBase/VolcanicTown used true for the swap hose). Irrelevant here: nothing Disables them.

## Capture backlog

| Item | Status |
|---|---|
| Sequencer keys on `LS_Enter/Exit/DataLayerController` | **Waived** — unanswerable; bindings + data-layer referencers sufficient |
| World Settings TODM “Space” actor vs DSD | **Waived** — DSD `overrideSky` / FixedTime 1:00 Observed; docs say both |
| Per-hangar child lists (1554 Environment) | **Waived** — folder counts + `/SPACE_THEME_B_*` mounts are the lesson |
| Per-spawner cosmetic CID | **Waived** — unread on spawner `list_properties`; CD default is Columbus `CID_613` |
| PIE hyperspace | **Waived** — Verse + `savedActor` + layer cine sufficient |
| Dedicated Space docs page | **Waived** — slug hub-redirects; Vistas Hyperspace chapter Observed |

## Live notes

- Mount `/d1fedf73-6faf-4ae5-8d26-0fc3711445cd`. Project Verse path `/epictemplates@fortnite.com/space_station_sample`. Plugin filename is that GUID.
- Did **not** `load_level` away from the default map.
- Billboard `text` N/A.
- `folderPath` on descriptors is often `"None"` even when `get_actors_in_folder("Devices")` returns the pad / mixer / DSD / starfighter. Count from the folder query, not `folderPath`.
- `ListFiles` `bRecursive=true` on the mount walks `__ExternalActors__` — list `""` then the `.verse` at `/d1fedf73-…/Attach_Hyperspace.verse` only.
- Character Definition modifiers: `load_asset` then `get_properties` on `:CharacterModifier_*_C_0` suffixes with **that modifier’s** property names.
- Cursor `user-unreal-mcp` was down; `tools/mcp.py` on `:8000` worked.
- Folder on disk / title is **`space_station_sample`** (picker “Space Starter Island”).
