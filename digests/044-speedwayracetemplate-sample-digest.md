---
id: digest-044
kind: digest
---

# speedway_race_template — sample digest

> Offline `_raw/speedway_race_template/`. Live
> `/d8024f4f-87e6-4438-b07b-63ebfcd82733/speedway_race_template` (**Observed**).
> Docs: hub + 3 linked deep-dives (Leaderboard, Custom Round Logic, Sorting Algorithms). All retrieved.

| | |
|---|---|
| Classification | **Full race game** — persistence, leaderboards, round-order, cinematic lineup, generic sorting |
| Content type | kit + environment: 20 authored files, **2447 OFPA actors** (real racetrack, not a factory shell) |
| Richest Verse system in this KB so far | 12 files across 5 modules |

## Scale

| | |
|---|---|
| Authored assets | 20 |
| Verse files | **12** |
| OFPA actors | 2447 (1 unclassified) |
| Max players | 8 (matchmaking) |
| `LevelHasStreamingDisabled` | 1 |
| Compatibility | 43.00 |

This is the **Design a Speedway Race** Creative template converted to UEFN and rebuilt with Verse persistence (**External**, hub). Unlike Detonation/Elimination/Parkour, the environment (real track, landscape, water, trees) **is** part of the lesson — not throwaway.

Live outliner (**Observed** `get_folders`) already shows the documented reorg: `AAA_Design/Devices/Player Start Positions/Start Position 1..8`, `TutorialAndWinnersZone`, `TutorialZone` — confirms the tutorial's "Cleaning up the Outliner" section is Observed, not aspirational. Stray `NewFolder1` / `Temp` folders remain — even a showcase reorg has leftovers.

`Developers/felixkupis/Collections/` empty. Twelfth first-party hit.

## Module map (5 folders, 12 files)

| Folder | Files | Owns |
|---|---|---|
| `PlayerStatistics/` | `player_stats_table.verse`, `player_stats_manager.verse` | Persistable stats + comparisons |
| `PlayerLeaderboard/` | `player_leaderboards.verse` | Billboard/UI/player-reference display |
| `RaceManagement/` | `race_order.verse`, `race_end_manager.verse`, `starting_game_sequence.verse`, `sync_array.verse` | Round persistence, race lifecycle, pregame cinematic, concurrency helper |
| `PlayerVehicle/` | `player_vehicle_manager.verse` | Assign agents to spawned cars |
| `SortingAlgorithms/` | `merge_sort.verse`, `quick_sort.verse` | Generic parametric sorting |
| `Test/` | `test_sorting.verse` | Self-test + `profile()` benchmark device |
| root | `submodules.verse` | empty (unlike other templates — no asset stubs needed here) |

## What this sample actually teaches

1. **Persistable stats table + comparator functions, kept separate from display.** `player_stats_table<final><persistable>` (Version/Points/Podiums/BestLapTime) plus six comparator functions (`MorePointsComparison`, `LessPointsComparison`, `MorePodiumsComparison`, `LessPodiumsComparison`, `SlowerLapTimeComparison`, `FasterLapTimeComparison`) that pass **whichever operand wins** back out — same shape MergeSort expects. See techniques/persistable-stats-and-sorted-leaderboard.md.
2. **Session `weak_map` resets every round; player `weak_map` is persistable.** Round bookkeeping (`LastCompletedRound`, finish order) lives on `weak_map(player, player_circuit_info)` so it survives round boundaries. `weak_map(session, round_info)` is a per-round cache (`GetRound` without recomputing). A project can have **at most two** player weak maps — that's why circuit info is the second map, not mixed into stats. See techniques/round-scoped-vs-persistent-state.md.
3. **Generic parametric sorting** (`MergeSort<public>(Array:[]t, Compare(L:t,R:t)<decides><transacts>:t where t:type)`) reused for both leaderboard order and race-start order — one algorithm, two comparator functions, two call sites.
4. **`ArraySync`/`ArrayRace` — a reusable divide-and-conquer concurrency primitive** over `sync`/`race`, not hand-rolled per feature. Powers both "wait for all players to finish the race" and "play N sequential player intros without knowing N up front." See techniques/array-sync-race-primitive.md.
5. **Self-testing Verse with `profile()`.** `test_sorting.verse` is a placeable device that shuffles known-good arrays, sorts them, and asserts the result — plus timing via `profile(Description): ...`. A pattern the earlier single-file templates never needed.
6. **Cinematic-driven UI sync via Trigger devices as signals**, not polling. `StartPlayerIntroEvent`/`EndPlayerIntroEvent` triggers fired *from Sequencer* are awaited in Verse to know which player's stats to show.
7. **`type {Rounds:int where 1 <= Rounds}`** — a refined type constraining an `@editable` int at the type level, not just a runtime check.

## Official tutorial vs shipped

**External** hub: Speedway Race with Verse Persistence Template

Linked deep-dives (all retrieved):
1. Make Your Own In-Game Leaderboard in Verse
2. Custom Round Logic Using Verse — **no `-in-unreal-editor-for-fortnite` suffix** (that variant is empty). Bare `…/custom-round-logic-using-verse?lang=en-US` is the live page.
3. Sorting Algorithms in Verse

### What lines up

| Tutorial | Shipped |
|---|---|
| `player_stats_table<final><persistable>` with Version/Points/Podiums(Wins in generic tutorial)/BestLapTime | matches (template adds Podiums + BestLapTime beyond the generic leaderboard tutorial's Points/Wins) |
| `player_and_stats` struct pairing player + stats table | identical |
| `MorePointsComparison` returns `Left` after the `>` check | identical pattern |
| `MergeSort<public>(Array:[]t, Compare(...)... where t:type)<transacts>:[]t` recursive divide-Slice-Merge | **identical**, including comments |
| `UpdateStatsBillboards` / `UpdatePlayerReferences` sort-then-apply pattern | identical shape (template's version threads through `PlayerLeaderboard`/`PlayerStatistics` modules instead of one flat file) |
| `race_manager_device.RaceCompletedEvent` / `LapCompleted` awaited per player, cancel on `PlayerRemovedEvent` | `WaitForPlayerToFinishRace` / `WaitForPlayerToFinishLap` in `race_end_manager.verse` match the leaderboard tutorial's `WaitForPlayerToFinishRace` almost line for line |
| `GetPlayerStartOrder` shuffles on first round, else sorts by last finish order (hub prose + snippet) | `race_order.verse` matches |
| Custom Round Logic: `player_circuit_info<final><persistable>`, `CircuitInfo:weak_map(player,…)`, `RoundInfo:weak_map(session,…)`, `MakePlayerCircuitInfo` constructor, `RecordPlayerFinishOrder` / `UpdateRound` / `GetLastCompletedRound` / `ResetCircuitInfo`, `GetRound`, `RecordCurrentRound`, `IsFirstRound`, `TotalRounds:type {Rounds:int where 1 <= Rounds} = 3`, reset on `PlayerRemovedEvent`, no round API so round # is recorded by you | **matches `race_order.verse`.** Teaching the hub only sketches: session weak_map **resets every round** (not every session), so last-completed-round must be on the player map; max **two** player weak maps per project. |
| `ArraySync()` divide-and-conquer over array elements (hub prose, links to a Community snippet) | `sync_array.verse` — **note:** hub links this as an external Community snippet, not part of this tutorial's "Complete Code"; template ships its own copy plus the `ArrayRace` sibling the tutorial page never shows |
| Sorting-algorithms page: merge sort O(n log n), parametric `where t:type`, stability discussion, `profile()` testing pattern | `merge_sort.verse` + `test_sorting.verse` match structurally; quick_sort.verse is **not** covered on this page (linked only as "on your own" — separate Community snippet, not reproduced here) |

### What does not line up

| Topic | Tutorial | Shipped |
|---|---|---|
| Generic leaderboard tutorial stats | Points + Wins | Points + **Podiums** + **BestLapTime** — hub-documented racing extension, not drift |
| `quick_sort.verse` | Sorting chapter only *links out* to a Community snippet | File ships here; no in-chapter source to diff |
| Custom Round Logic device | walks a `test_round_info_device` teaching sample | Production split: logic in `race_order.verse`, wired from `starting_game_sequence` / `race_end_manager` |
| URL shape | working = `/documentation/fortnite/custom-round-logic-using-verse?lang=en-US` | `-in-unreal-editor-for-fortnite` suffix is empty |

**Training rule:** For stats table, leaderboards, merge sort, **and** `race_order.verse` — tutorial and shipped agree; prefer snippets (superset + production wiring). `quick_sort.verse` = Observed file only.

## Negative findings

- Fetcher note: Custom Round Logic lives at the **unsuffixed** URL. Adding `-in-unreal-editor-for-fortnite` returns empty.
- Live outliner still has `NewFolder1` and `Temp` — the documented cleanup wasn't total.
- `submodules.verse` is empty here (unlike UI/Scene Graph/Verse-Device templates) — this sample needs no asset-reflection stub tree, since it references only Creative devices and native Verse types, not generated widget/mesh classes.
- `sync_array.verse`'s `ArrayRace` function is shipped but never called anywhere in this sample's other 11 files (**Observed** — no references found in `race_end_manager`, `starting_game_sequence`, or elsewhere) — it travels with `ArraySync` as a pair but only `ArraySync` is used.

## Capture backlog

1. Live Race Manager / Timer / Analytics `@editable` values. *Open, low — file defaults Observed.*
2. TutorialZone billboard captions. *Waive — hub already narrates.*

## Live notes

Level confirmed. Folders match the tutorial's described outliner reorg almost exactly (`AAA_Design/Devices/Player Start Positions/Start Position N`). Streaming disabled.
