---
id: digest-045
kind: digest
---

# social_game_sample — sample digest

> **Extractor provenance.** `tools/extract-template.py` (Pass A) + UnrealMCP on
> `localhost:8000` (Pass B). Default level
> `/cdd93f61-ae73-4861-a6f3-72668d63a04b/social_game_sample` (**Observed**).
> Docs: Working With social-elimination Islands
> (single page, Starter + Minigame Mastery + Social Deduction together). `docsUrl` empty;
> spreadsheet Link in Desc is the `uefn/working-with-Social-game-islands-…` slug (200-redirects
> here). Guessed `Social-game-starter-island-in-unreal-editor-for-fortnite` **hub-redirects**.
> Compat **43.00**. `featureSet`: **SOCIAL_WAVEGFS**.

| | |
|---|---|
| Classification | **theme environment starter** — Greasy Grove leftover + Social galleries, not a minigame lab |
| Content type | **catalog wrap** (2 authored vs 5406 OFPA; `/SOCIAL_WAVE_*` mounts) |
| Verse | **0 files** |
| Fab `.uref` | **0** |

## What it is

The picker name is **social-elimination Starter Island**. The island is a **16-player sandbox** (docs: “deviously fun twist on Greasy Grove”), not Social Deduction’s four-door syllabus and not Minigame Mastery’s device rooms. Teaching that survives deletion is **theme galleries as Outliner folders over a leftover BR POI**, with **zero** `PlaysetRoot`. Voting, giftbox occupancy, `SubscribeAgent`, twin `_Verse` map, and ExtraAssets HUD widgets live in social_deduction_sample — **not** this island.

## Scale (**Observed** — `summary.txt` + live)

| | |
|---|---|
| Authored assets | **2** (`GameFeatureData`, `social_game_sample_DefaultHLODLayer`) |
| OFPA | **5406** |
| Maps | `social_game_sample.umap` (only) |
| Verse files | **0** (`VerseToolset.ListFiles` on `/cdd93f61-…` — do not recurse `__ExternalActors__`) |
| Verse devices placed | **0** |
| Data layers | **0** |
| Fab `.uref` | **0** |
| Compatibility | **43.00** |
| Matchmaking | **16** (`maxPlayers=16`, `teamSize=Dynamic`, `locomotionPreset=Current BR`) |
| Streaming | Partitioned + actor folders + OFPA. **`LevelHasStreamingDisabled=1`** |

Empty `Developers/felixkupis/Collections/`. First-party scratch folder.

`experimental.*` (Python / toolsets / Scene Graph) in `.uefnproject` is **Unverified**. Content-side: **no** `.verse`, **no** `__ExternalEntities__`.

Largest file ~0.88 MB OFPA. Whole folder ~38 MB. Under the GHES **100 MB** per-file cap.

`keyArt` is `Resources/SocialDeductionTemplateThumbnail.png` — Social Deduction’s thumbnail reused.

## Level organisation (**Observed** — live `get_folders` + folder counts)

```
FN_Existing/                      2757 — leftover Greasy Grove
  FN_Assets/                      2733 (PowerStation/ 12 nested)
  Background/                       16
  Landscape/                         1
  Special/                           7
GreasyGames/                      2634 — Social galleries
  Colorful/                        824
    Colorful Building 1–5          125/73/38/123/83
    Colorful Entrance 1–2          191/191
    Building 1–4 / Entrance 1        0 leftover
  Cornfield/                       691
  Dorm/                            938
  Playground/                      180
Background/                          0 leftover root
Colorful Entrance 2/                 0 leftover root
```

No `Device_Billboard_V2`. Sampled devices `bIsSpatiallyLoaded=false`; `load_actors` not required.

## Device inventory (the design)

| Count | Class | Role |
|---|---|---|
| 11 | `Device_Decal_V2` | spray decals (`M_SprayDecal_Master`, opacity 100, distance 2) |
| 4 | `Device_Ascender` | zipline ascenders (`Ascender`, `2`, `3`, `4`) |
| 2 | `B_BasketballHoopWithMinigame` | playground toys |
| 2 | `B_DowntownTrashCan_BasketballMinigame` | playground toys |
| 2 | `PointLight` | `Playground Light`, `Mascot Light` |
| 2 | `RectLight` | `DormLight 1`, `DormLight 2` |
| 1 | `BP_Creative_Player_Spawner_Prop` | `Player Spawner`, `useAsIslandStart=true` |
| 1 | `BP_ProgressMesh_Device` | ExtraAssets progress mesh (see values) |
| 1 | `Device_VendingMachine_V2` | `Vending Machine0` |
| 1 | `Device_ExperienceSettings_V2_UEFN` | IslandSettings0 |
| 1 | `BP_LightingScalabilityManager` | `LightingScalabilityManager C` |
| 1 | `Landscape` + 6 `LandscapeStreamingProxy` | Apollo Creative landscape MI; **empty** `layers[]` |
| 1 | `VK_Waterbody_Ocean_Parent` + `WaterZone` | background ocean |
| 1 | `FortUndergroundVolume` | |
| 0 | Billboard / HUD Controller / DSD / PPV / PlaysetRoot / Character Spawner / Hiding Prop / Voting / Disguise / Verse device | — |

progress mesh (**Observed**): `fillTarget=100`, `gameStartFillAmount=0`, `fillType=ContinuousRate`, `fillRate=5`, `drain Rate=5`, `enabledIndex=0`. Class lives in `/SOCIAL_WAVE_ExtraAssets/SetupAssets/`. Docs call it “Progress Based Mesh.”

CTL doll pieces (**Observed**): `SOCIAL_WAVE Core Doll Head A` / `Body A` → `staticMesh` `/SOCIAL_WAVE_ExtraAssets/PROP_THEME/Meshes/SM_PROP_THEME_Head` / `_Body`. Static meshes, not a Sequencer Control Rig.

## Authored kit (**Observed** — `assets.txt` + live)

```
Content/GameFeatureData.uasset                         (0 referencers)
Content/social_game_sample_DefaultHLODLayer   (umap refers)
```

**Tier 2 mounts:** `/SOCIAL_WAVE_Beacon_Assets` 1986 + `/SOCIAL_WAVE_Matrix_Assets` 1696 + `/SOCIAL_WAVE_Matrix_Comp` 835 + `/SOCIAL_WAVE_Beacon_Comp` 815 + `/SOCIAL_WAVE_Core_*` + `/SOCIAL_WAVE_Quill_*`. `/SOCIAL_WAVE_ExtraAssets` 10 (progress mesh + PROP_THEME doll meshes + graffiti texture FNames). No Fab `.uref`.

Live landscape material: `/Game/Creative/Materials/Landscape/Creative_Landscape_Apollo/M_Creative_Landscape_Apollo`. `layers[]` empty. Spreadsheet: “6 Landscape actors with physical materials need to be rebuilt.”

## What this sample actually teaches

1. **theme galleries over a leftover POI, not playset roots.** techniques/theme-galleries-over-poi.md. Docs gallery names (Cornfield / Dorm / Playground / colorful stairway) are the `GreasyGames/` folders. Prison’s assembly is the inverse: prison_sample.
2. **Social Deduction / Mastery devices are not on this island.** Twin `_Verse` map, `SubscribeAgent`, giftbox occupancy, ExtraAssets **HUD** widgets: twin-level-device-vs-verse.md, subscribe-agent-extra-data.md, hiding-prop-occupancy-inspect.md, social_deduction_sample ExtraAssets HUD (now **3** ExtraAssets HUD sources with Mastery). Dorm coffins are Matrix **props**. This island’s ExtraAssets uses are progress mesh + PROP_THEME doll meshes, not `UW_HUD_*`.
3. **The one Social *device* that ships is progress mesh**, plus sandbox toys (ascenders, basketball, vending, spray decals). Docs’ Gift Box / PebbleGame / Disguise / Voting / Guard Spawner belong to the other Social tiles.
4. **CTL doll here is two static ExtraAssets meshes**, not the Control Rig + Sequencer workflow the theme page describes. No `CinematicSequence` actor.
5. **Sandbox flags, not a round.** `allowBuilding=All`, `bInfiniteResources=true`, `bFallDamage=false`, `timeOfDayOverride=NoOverride` (no Creative DSD). Contrast City `allowBuilding=None`. **One** player spawner with `useAsIslandStart=true` — unlike Prison’s `spawnLocation=SpawnPads` with **zero** pads.
6. **Galleries live in the theme Content Browser, not this project’s `Content/`.** `/SOCIAL_WAVE_*` mounts. Outliner labels say `GreasyGames/Dorm`, not “social-elimination Dorm Gallery.”
7. **No authored atmosphere stack.** 0 DirectionalLight / SkyLight / SkyAtmosphere / PPV / DSD. Not the NightEnvironment / alternate dimension / City `SM_SkySphere` pattern. `DSA_UEFN_Ch5S1` appears once in FNames; lighting is the scalability manager + a handful of Point/Rect lights.

## Official tutorial vs shipped

**External:** Working With social-elimination Islands (headed Chrome 2026-09-14). Single page covering **three** Social tiles. Verbatim: `_raw/social_game_sample/tutorial.txt`. Spreadsheet `docsUrl` empty; `Link in Desc` is this page (uefn slug redirects). Linked Minigame Mastery / Social Deduction / Custom UI / Dynamic Team Emotes chapters belong to those packets (Social Deduction already extracted).

### Lines up

| Tutorial | Shipped |
|---|---|
| theme Templates → social-elimination → Starter Island | picker; `featureSet` SOCIAL_WAVEGFS |
| “twist on Greasy Grove” | `FN_Existing/` 2757 leftover POI + `GreasyGames/` 2634 |
| Galleries: cornfield, dorm, colorful stairway, playground | `GreasyGames/Cornfield`, `Dorm`, `Colorful`, `Playground` |
| Stairway colors Blue/Green/Pink/Yellow | Beacon colored-light / doorway kit in `Colorful/` |
| progress mesh (Progress Based Mesh) | 1× `BP_ProgressMesh_Device` in ExtraAssets |
| Loot chests as an item source | 34 `Tiered_Chest_Athena` + 37 ammo in the leftover POI |
| Age-rating table for publishing social-elimination | theme rules — not island content |
| Minigame Mastery / Social Deduction are separate tiles | **0** of those devices here |

### Does not line up

| Tutorial | Shipped |
|---|---|
| Gift Box / PebbleGame / Disguise / Voting / Guard Spawner as Social devices | **0** placed (Social Deduction / Mastery) |
| Dynamic Team Emotes room | Social Deduction only |
| CTL doll = Control Rig + Sequencer, no Prop Mover | two **static** `SM_PROP_THEME_*` meshes; **0** cine devices |
| VIP Masks + Niagara (see Mastery Carryable Spawner room) | **0** on this island |
| Dedicated Starter walkthrough | **one** sentence; guessed `Social-game-starter-island-…` slug hub-redirects |
| List item **“Minigames Mastery Template”** | H3 / picker **Minigame Mastery** |

Training rule: live Outliner folders + zero Social Deduction devices win. Do not emit “Starter Island ships voting / giftboxes / disguises / CTL Sequencer.” Those are other tiles or the theme overview.

## Negative findings

- **Name implies a social-elimination *match*; files show a kitbash sandbox.** No round manager, no Verse, no HUD, no Giftboxes.
- **`hUDInfoType=AI Enemy Elimination` with no AI.** Same leftover as Prison.
- **`keyArt` is Social Deduction’s thumbnail.**
- **593 `BP_Blockout_Floor_A_A` still placed** (greybox floors shipped).
- **Empty leftover Outliner roots** after the `GreasyGames/Colorful/` nest.
- **Landscape `layers[]` empty** on Creative Apollo MI; spreadsheet flags 6 landscape rebuild warnings (6 streaming proxies).
- **`GetDeviceProperties` N/A** — no Verse `creative_device`s.
- **Decal devices use stock `M_SprayDecal_Master`**, not the ExtraAssets graffiti textures that appear in OFPA FNames.

## Capture backlog

| Item | Status |
|---|---|
| Per-building actor lists inside `Colorful Building *` | **Waived** — folder names + `/SOCIAL_WAVE_Beacon_*` mounts are the lesson |
| progress mesh thresholdMeshes[0..4] mesh refs | **Waived** — fillTarget/rate observed; ExtraAssets HUD lesson stays on Social Deduction |
| Vending Machine item/cost rows | **Waived** — not the lesson |
| Sequencer / CTL Control Rig on disk in ExtraAssets | **Waived** — this island places static meshes; Mastery may own the rig demo |
| PIE spawn pad vs Island Start | **Waived** — one pad, `useAsIslandStart=true` |

## Live notes

- Mount `/cdd93f61-ae73-4861-a6f3-72668d63a04b`. Project Verse path `/epictemplates@fortnite.com/social_game_sample`. Plugin filename is that GUID.
- Did **not** `load_level` away from the default map.
- Billboard `text` N/A (no billboards). `GetDeviceProperties` N/A.
- `get_actors_in_folder` descriptors often have `folderPath=None`; counts still match.
- Island Settings `get_properties` fails if `ActorLabel` is in the list (whole call rejected). ToyOptionsComponent has no `playerOptionData` on this class — read fields on the actor.
- Folder on disk / title is **`social_game_sample`** (picker “social-elimination Starter Island”).
