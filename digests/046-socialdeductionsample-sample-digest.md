---
id: digest-046
kind: digest
---

# social_deduction_sample — sample digest

> **Extractor provenance.** `tools/extract-template.py` (Pass A) + UnrealMCP on
> `localhost:8000` (Pass B). Default level
> `/c3e2dd68-affc-4e30-862d-a3cdc5999136/social_deduction_sample`. Verse devices live on
> sibling `/…/social_deduction_sample_Verse` (loaded then restored this pass).
> Hub: Social Deduction Template.
> `docsUrl` empty. Compat **43.00**. `featureSet`: **SOCIAL_WAVEGFS**.

| | |
|---|---|
| Classification | **theme mechanic gallery** — four social-elimination rooms (disguise / vote / giftbox / dynamic teams), not a full social-deduction match |
| Content type | **catalog wrap** (36 authored vs 3000 OFPA across two maps) |
| Verse | **4 files** (3 used + 1 leftover NPC stub) |
| Fab `.uref` | **0** |

## Verdict first

The picker name implies an Among-Us-style match. The island is a **loading room with four doors**. Each door is a tutorial corridor plus a minigame that isolates one hero device. Teaching that survives deletion is the device/Verse pairing (especially the twin `_Verse` map), not the Matrix dorm geo.

## Scale (**Observed** — `summary.txt` + live)

| | |
|---|---|
| Authored assets | **36** |
| OFPA | **3000** total — **1500** on each umap |
| Verse files | **4** (`Scripts/` ×3 + root `new_npc_behavior.verse`) |
| Verse devices placed | **0** on default map; **7** on `_Verse` (1 `voting_device` + 6 `coffin_controller`) |
| Data layers | **0** |
| Fab `.uref` | **0** |
| Compatibility | **43.00** |
| Matchmaking | 16 players; Island Settings `teamSize=Dynamic`, `teams.teamType=Custom` |
| Streaming | Partitioned + actor folders + OFPA. **No** `LevelHasStreamingDisabled` tag (unlike VFX). Billboard `get_properties` needed `load_actors` |

Empty `Developers/felixkupis/Collections/`. First-party scratch folder, same family as prior packets.

## Two maps (**Observed**)

| Asset | Role |
|---|---|
| `social_deduction_sample.umap` | Device-only. Project Browser default. |
| `social_deduction_sample_Verse.umap` | Same Outliner rooms + Verse devices. |
| `social_deduction_sample_Verse_v2.umap` | Redirector → Verse map |

See techniques/twin-level-device-vs-verse.md.

## Level organisation (**Observed** — live `get_folders`)

```
StarterRoom/                 hub: 4 labeled doors + PebbleGame carryables + default class/team
Disguise/
  Room - Tutorial/           billboards + sample Disguise device + class/sniper logic
  Room - MiniGame/           dorm: 8 NPC spawners, sniper perch, locks, AI nav blockers
Voting Device/
  BBs/                       corridor syllabus (BB - Voting Intro 1–7)
  Minigame/                  glass cubes over lava (quiz)
  PedestalVotingExample/     X vs O buttons
  VolumeVotingExample/       stand-on-shape Red vs Blue
Giftbox Coffins/
  Room - Tutorial/
  Room - MiniGame/           6 SwashTide giftboxes
Dynamic Team Emotes/         two class pads + team settings (pistol/low-G vs sword/speed)
Background/                  Island Settings
Room - Tutorial/             empty leftover folder
```

Billboard captions (81 loaded this pass): `_raw/social_deduction_sample/billboard-texts.md`. `text` is on the **actor** (same as VFX); actors were spatially unloaded until `load_actors`.

## Device inventory (the design)

Counts are **project-wide OFPA** (both maps). Live default-map rooms are ~half.

| Count | Class | Role |
|---|---|---|
| 162 | `Device_Billboard_V2` | corridor syllabus + quiz faces |
| 66 | `Device_Button_V2` | vote / inspect / start / return |
| 53 | `Device_HUDMessage_V2` | role intros, vote results |
| 52 | player spawners | hub + hide/seek |
| 48 | `Device_Keylock_V2` | corridor / dorm doors |
| 47 | `Device_Teleporter_V2` | room routing (groups M/N/O/A/Q/W/X/Y/Z) |
| 33 | `Device_Timer_V2` | vote countdown, hide timer, seeker freeze |
| 16 | `Device_Simple_Voting_Option` | quiz + X/O + Red/Blue |
| 16 | `Device_CharacterSpawner` | disguise-room NPCs (8 on live minigame) |
| 14 | `Device_HidingProp_SwashTide` | giftboxes (6 minigame + 1 tutorial per map) |
| 14 | class selector / team settings | roles + dynamic teams |
| 12 | `Device_Switch_V2` | quiz visual + Red/Blue win |
| 10 | `Device_ClassDesigner_V2` | loadouts |
| 7 | `VerseDevice` | **Verse map only** |
| 6 | `Device_DisguiseDevice` | contestant / guard / sample |
| 6 | `Device_Simple_Voting_Group` | quiz + pedestal + volume examples |
| 2 | `Device_HUDController_V2` | social-elimination UMG overrides (one per map) |
| 2 | `Device_CinematicSequence` | `LS_DollHeadsRotation`, loop, GameplayOnly |
| 2 | Day Sequence modifier | atmosphere |

## Verse (4 files)

Copied to `snippets/verse/`.

| File | Job |
|---|---|
| `voting_device.verse` | Quiz round: `[]question`, billboard `SetText`, winners teleport / losers `HideProps` |
| `coffin_controller.verse` | One giftbox: occupancy events + one inspect button + freeze zone |
| `utils.verse` | `SubscribeAgent` extra-data wrapper (community snippet header) |
| `new_npc_behavior.verse` | **Unused leftover** — same `npc_behavior` wander stub as characters_cameras_npc_template. Lives at Content root, not `Scripts/`. Print helpers commented out. |

## What this sample actually teaches

1. **Twin maps, not twin Outliner stacks** — techniques/twin-level-device-vs-verse.md.
2. **`SubscribeAgent` to carry an enum into `InteractedWithEvent`** — techniques/subscribe-agent-extra-data.md.
3. **Giftbox inspect = hide events + one button**, not Valid/Invalid pair — techniques/hiding-prop-occupancy-inspect.md.
4. **Voting Group ID is a shared string**, not a reference. Live quiz group = **`"2"`** (tutorial says `"3"`). Options bind `On Interact → CastVoteWhenReceivingFrom`; group `BeginVote` from the center button; timer `OnSuccess → EndVote`. Wrong answers: Prop Manipulator hides the glass floor; Damage Volume over lava. Winners: teleporter in the correct cube. `allowVoteSwitching=false`, `timeLimit=20`, `maxVotesPerPlayer=1` (**Observed**).
5. **Disguise + Class Selector is the sniper/hider split.** `OnClassSwitched → ApplyDisguiseToInstigator`. Sniper disguise: breaks **Off/Off**, `replaceExistingDisguise=true`. NPC/sample disguises: break **On Damage Opponent / On Damaged** (corridor billboard: damage makes the mannequin dance). 8 NPC Spawners + 4 AI Navigation Modification blockers + locks keep the crowd in the dorm. Elimination Manager grants sniper ammo.
6. **Dynamic Team Emotes need Island Settings, not a device.** `teamSize=Dynamic` + `teams=Custom`. Team 1: `gravityPreset=Low`. Team 2: `movementMultiplier=2`. Invite/leave emotes come from those settings (tutorial: also enable Dynamic Team Leave on Team Settings). No Verse. Both umaps share this room.
7. **social-elimination HUD is HUD Controller widget overrides, files stay in `/SOCIAL_WAVE_ExtraAssets/`.** Live: `UW_HUD_PlayersInfoStack`, `UW_HUD_EquippedInfoBlock`; FNames also list `UW_HUD_QuickbarSlot`. `showMinimap` / `showTeamInfo` / `showHealth` / `showShields` = **HideElement**. `bModifyTextChatLayout=true`, alignment **MiddleRight**, anchor **CenterRight**. Matches the Custom UI chapter. Docs: you **cannot copy** those widgets into the project — reference them. Same HUD stack as Minigame Mastery / Walker / **music-fantasy streets** (fourth ExtraAssets source — **confirmed**). **Sources:** social_deduction_sample (originating); horde_npc_sample `/EnvironmentPack_ExtraAssets/UI/Widgets/UW_HUD_*`; minigame_mastery_template `/SOCIAL_WAVE_ExtraAssets/UI/Widgets/UW_HUD_*` including a live QuickbarSlot override; music_fantasy_sample `/DemonFactionExtraAssets/UI/Widgets/UW_HUD_*` including live QuickbarSlot (authored `Content/UI/Widgets/UW_*` unused). Mastery differences: health/shields **DoNotOverride**, chat anchor **TopLeft**, Island `nameplateDisplayMode=Never`. music-fantasy: health/shields **DoNotOverride** (Mastery-like), plus a second cutscene HUD Controller. Deeper UMG/viewmodel lesson stays with user_interfaces_template.

## Official tutorial vs shipped

**External hub + 5 chapters** (headed Chrome 2026-09-14; Sign in chrome still visible, article body readable):

1. Social Deduction Template
2. Deception Through Disguises
3. Create Voting Opportunities
4. Giftbox Device for Hiding and Teleportation
5. Create Temporary Teams Through Emotes
6. social-elimination Custom UI

Verbatim capture: `_raw/social_deduction_sample/tutorial.txt`. Spreadsheet `docsUrl` empty; hub exists anyway.

### Lines up

| Tutorial | Shipped |
|---|---|
| theme Templates → social-elimination → Social Deduction | Project Browser tile; `featureSet` SOCIAL_WAVEGFS |
| Loading room, four doors, tutorial corridor then minigame | Outliner folders match |
| Disguise minigame device table (8 NPC spawners, 2 class selectors, …) | Live `Disguise/Room - MiniGame/Gameplay` + tutorial gameplay folder |
| Voting device-only table (Group×1, Options×4, Billboard×18, …) | Exact match on `Voting Device/Minigame` |
| Giftbox ×6, max 3 hiders | 6 minigame props, `numberOfOccupants=3` |
| `_Verse` sibling for voting + giftbox | 7 VerseDevice OFPA only under `__ExternalActors__/social_deduction_sample_Verse/` |
| Verse giftbox: one inspect button | `Inspect Valid` ×6, no Invalid |
| Dynamic teams: Island Settings Dynamic + Custom; no Verse | `teamSize=Dynamic`, `teams=Custom`; Team 1 low-G / Team 2 speed |
| HUD Controller + three social-elimination widgets; cannot copy UI folder | `/SOCIAL_WAVE_ExtraAssets/UI/Widgets/UW_HUD_*`; HideElement on minimap/team/health/shield; text chat Middle Right / Center Right |
| Doll-head cinematic on giftbox | `LS_DollHeadsRotation`, loop, GameplayOnly |

### Does not line up

| Tutorial | Shipped |
|---|---|
| Voting Group ID example **"3"** | Live quiz group **`"2"`** |
| Device-only: “only answer C can be correct” | Neptune quiz boards 12/14/16/18; C labeled 16 — plausible, not independently proven as the *only* winner binding |
| Verse voting snippet **277** lines / giftbox **137** | Shipped **236** / **111** (`voting_device.verse` / `coffin_controller.verse`) |
| Giftbox chapter **Next** → Custom UI | Hub still has an Emotes chapter; skip in that page's footer |
| Custom UI H1 “**Lean** to add…” / tag `Social gane` | Typos in docs |
| Disguise “Disguise ×2” | Live tutorial folder has **3** (NPC, Guard Sniper, **Sample** exhibit) |
| Nameplates hidden for anonymity | Island Settings `nameplateDisplayMode=Default`; HUD hides team/minimap instead |
| `new_npc_behavior` / lava Niagara | Not in any chapter. Leftovers (see negatives) |

Training rule: corridor billboards + `snippets/` = ground truth. Do not emit tutorial line counts or Voting Group `"3"`.

## Negative findings

- **Name implies a match; files show four exhibits.** No round manager, no crew/impostor Verse loop.
- **`Content/VFX/` (lava Niagara) has 0 asset referencers.** Copy debris from `vfx_template`. Do not cite as a worked example.
- **`new_npc_behavior.verse` is not assigned.** Root leftover + `NPC_Imposter` ObjectRedirector at Content root → `NPCs/NPC_Imposter`.
- **`Room - Tutorial/` is empty.** Folder path exists; 0 actors.
- **PebbleGame carryable spawners** sit in the hub and are not part of the four-room syllabus.
- **GetDeviceProperties** on Verse devices returns **subobject paths** (`Scripts-voting_device_0.billboard_device_0`), not the placed billboard actors. Bindings for Verse `@editable` device refs are not auditable that way. Question `description`/`a`/`b`/`c`/`d` listed on the struct but `get_properties` cannot read them — **unanswerable**.

## Capture backlog

| Item | Status |
|---|---|
| Quiz question string values on the Verse device | **Waived** — listed, unreadable |
| Sequencer keys on `LS_DollHeadsRotation` | **Waived** — keyframes unreachable |
| PIE the four minigames | **Waived** — structure + bindings sufficient |
| HUD `quickbar Slot Widget Override Class` (spaced name) | **Closed as FNames** — `UW_HUD_QuickbarSlot` |

## Live notes

- Mount `/c3e2dd68-affc-4e30-862d-a3cdc5999136`. `VerseToolset.ListFiles("")` also lists `__ExternalActors__` if recursive — list `Scripts/` instead.
- Default map has **no** ScriptDevices. Load `_Verse` to call `ListDeviceProperties`.
- Billboard `text` needs `load_actors` first on this map.
- Mutator freeze zone: Enable/Disable warm-up in `coffin_controller.OnBegin`.
