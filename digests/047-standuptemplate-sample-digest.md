---
id: digest-047
kind: digest
---

# stand_up_template — sample digest

> Offline `_raw/stand_up_template/`. Live
> `/d0a4b218-67d1-448f-9983-69c57b9f56ba/Map/stand_up_template` (**Observed** — map is nested under `Content/Map/`, not content root).
> Docs: **one page**, no chapter split.

| | |
|---|---|
| Classification | **Sit-down cinematic show** — chair + Verse camera director |
| Content type | authored sequences/characters (77) inside a kitbashed club (576 OFPA) |

## Scale

| | |
|---|---|
| Authored | **77** (`seq/` 6, `characters/` 25, materials/textures, `animations/Chair`) |
| OFPA | 576 |
| Verse | **1** (`show_template_device.verse`) |
| Matchmaking | **1** player |
| `LevelHasStreamingDisabled` | 1 |
| Compatibility | 43.00 |
| Fab | 0 |

Spreadsheet: *Pretty impressive… devices maybe getting updated.* Teaching is the Verse camera director, not the club geometry.

Empty `Developers/felixkupis/Collections/`. Thirteenth first-party hit.

## What it is *for*

Player sits. Chair locks them. Main cinematic runs the MetaHuman act; a second “TV mode” sequence is the multi-angle view. Input Triggers cycle alt cameras, return to TV (seek-synced to main), or free look. Show ends → `TheChair.Eject`. Sit again while the show is still running → jump back to TV, don’t restart the act.

## Device inventory (the design)

| Count | Class | Role |
|---|---|---|
| 1 | `Device_Chair` | Sit to start; `SeatedEvent` / `ExitedEvent` / `Eject` |
| 4 | `Device_InputTrigger` | Free look / TV / next / prev — **Register only while seated** |
| 4 | `Device_Character_V2` | On-stage + waiting MetaHumans |
| 6 | `Device_CinematicSequence` | 1 main + 1 TV + 4 alt cameras (`CameraSwitches`) |
| 1 | `VerseDevice` | `show_template_device` |
| 8 | cine cameras | sequencer shots |
| rest | club kitbash | throwaway |

Most device User Options stay default; Verse owns the interactions (**External** tutorial, matches slim `@editable` list).

## Verse

Snippet: `snippets/verse/show_template_device.verse`

`OnBegin` → `TheChair.SeatedEvent.Subscribe(OnSeated)` only.

`OnSeated`: Register Next/Prev/FreeLook; **Subscribe** TV + `ExitedEvent` (Register TV later when leaving TV). If `MainSequencePlaying = false` then `spawn{RunSequence}` else `DoReturnToTVMode`.

`RunSequence`<suspends>: `Sleep(1)` (sit anim) → `MainSequence.Play()` + `TVModeSequence.Play(Agent)` → `AwaitMainSequencingEnding` → `StoppedEvent.Await()` → `TheChair.Eject`.

Camera:

- `?cinematic_sequence_device` = current alt cam; stop before Play next.
- Wrap index: `Mod[CurrentCameraIndex ± 1, CameraSwitches.Length]`.
- Return to TV: `GetPlaybackTime()` on main → `SetPlaybackTime` + `Play` on TV.

Subscriptions stored as `?cancelable`; `CancelSubscription` + Unregister on chair exit.

## What this sample actually teaches

1. **Chair is the gate.** Seated starts the show; Eject ends it; Exited tears down input. See techniques/chair-gated-cinematic.md.
2. **Two sequences, one clock.** Main drives the act; TV is the directed camera. Re-enter TV by **seeking** to main’s playback time — don’t Play from 0.
3. **Input Triggers only while eligible.** `Register`/`Unregister` + keep `?cancelable` and `Cancel()` on exit so Sprint-rebinds don’t leak into free roam.
4. **`option{PressedEvent.Subscribe(...)}`** stored, cancelled later — same idea as Parkour’s “you can drop the handle,” inverted: here you **must** keep it.
5. **Circular array index via `Mod[]`** so next/prev never walk off `CameraSwitches`.

## Official tutorial vs shipped

**External:** Stand-Up Comedy Club Template

Single page. Spreadsheet `/en-us/uefn/verse-standup-comedy-club-template-…`. `docsUrl` empty.

### Lines up

| Tutorial | Shipped |
|---|---|
| Class `show_template_device`, log channel | same |
| 1 chair, 4 input triggers, 4 characters, 6 cinematic sequences | OFPA exact |
| Seated → Sleep(1) → Play main+TV; Stopped → Eject | same |
| `GetPlaybackTime` / `SetPlaybackTime` to rejoin TV | same |
| `Mod` wrap on camera index | same |
| Max players 1 | matchmaking |
| Device interactions in Verse, thin User Options | `@editable` list |

### Does not line up

| Topic | Tutorial | Shipped |
|---|---|---|
| Await helper name | typo **`AwaitMainSequecingEnding`** in one heading | `AwaitMainSequencingEnding` |
| One trigger label | “PreviousMode” in a field list | `PreviousCamera` |
| Input Trigger HUD | Creative Input Custom 5 (Sprint) = Previous Camera | **External** per-device; 4 distinct bindings |

Train on snippet. Emitting the Sequecing typo is doc regurgitation.

Linked **Importing MetaHuman Animations** / Sequencer pages are adjacent, not this packet.

## Negative findings

- Club mesh is kitbash; **seq + character anims are the authored kit**.
- **`SU2_AnimCarter.uasset` (~134 MB)** — gitignored (GHES 100 MB cap). Clone has `SU2_AnimCarter.uasset.zip` (~35 MB) in the same `anim/` folder; unpack to that filename before PIE.
- Map path `Content/Map/` — extractor/MCP must not assume umap at content root.
- Scratch `Developers/` folder.

## Capture backlog

1. Confirm Input Trigger Creative Input bindings vs HUD labels (live ToyOptions). *Open, low.*
2. Sequencer shot list / MetaHuman anim assets. *Waive — tutorial points at Sequencer docs.*

## Live notes

Level confirmed under `Map/`. Streaming off.
