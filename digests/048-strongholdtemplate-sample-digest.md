---
id: digest-048
kind: digest
---

# stronghold_template — sample digest

> Offline `_raw/stronghold_template/`. Live
> `/b678911b-ea3d-4925-b272-ff87200310bb/stronghold_template` (**Observed**).
> Docs: 5-chapter hub. Island code **2218-1545-9149** (**External**).

| | |
|---|---|
| Classification | **Stealth / assault against guards** — Verse orchestrates spawners, leashes, alert mood |
| Content type | kit + jungle wrap: 31 authored vs **1137** placed |

## Scale

| | |
|---|---|
| Authored | 31 (`Barks/` 20, `Cinematic/` 3) |
| OFPA | 1137 |
| Verse | **7** |
| VerseDevices | **11** |
| Guard spawners | 14 |
| AI patrol paths | 41 |
| Billboards | 58 |
| Matchmaking | **4** players, 1 team |
| Compatibility | 43.00 |

Spreadsheet: *really solid… lots of usable verse.* Game manager is the product; jungle is dressing.

Empty `Developers/felixkupis/Collections/`. Sixteenth first-party hit.

Live folders (**Observed**): `Bunker01..03`, `Devices/AI/{IntroEvent,Leash,PatrolPath}`, `Devices/Alarm/{Orange,Red}`, Audio, Background, Cover.

## What it is *for*

Solo or squad of 4. Sneak or fight. Eliminate all guards. **First alert** → reinforcements spawn, leash, `ForceAttackTarget` the player who tripped it. When 3 guards remain after detection → **fallback** leash on the keep. Wipe without ever alerting → undetected End Game. Two player lives then fail.

Intro: truck explode + a guard walks to the wreck (`NavigateTo` / `MaintainFocus`).

## Device inventory (the design)

| Count | Class | Role |
|---|---|---|
| 14 | `Device_GuardSpawner_V2` | initial + additional + reinforcement |
| 41 | `AIPatrolPath_Device` | default patrol groups |
| 11 | `VerseDevice` | game / leash / alert / bark / intro / cinematic |
| 12 | audio player | barks |
| 12 | damage volumes | kill/fire |
| 3 | end game | undetected / detected / fail |
| 1 | tracker | objective count |
| 2 | HUD message | reinforcement / fallback |
| 1 | explosive | intro + `StartGameplay` |

## Verse (7 files)

Copied to `snippets/verse/`.

| File | Job |
|---|---|
| `stronghold_game_manager.verse` | spawners, perception events, reinforcement, fallback, lives, victory |
| `stronghold_leash_position.verse` | `SetLeashPosition` inner/outer cm; clear while on patrol path |
| `stronghold_alert_manager.verse` | combat vs alert mood (radio + lights + flare) |
| `stronghold_bark_manager.verse` | `audio_npc_bark`: delay/cooldown/`race` stop |
| `stronghold_intro_explosion.verse` | wreck investigation nav |
| `stronghold_billboard_cinematic.verse` | tutorial flythrough (same family as Verse Commander) |
| `stronghold_log.verse` | `Log(...)` channel |

**Event bus on the game manager** (other devices Subscribe):

`ReinforcementsCalledEvent`, `FallbackEvent`, `GuardsSuspiciousEvent`, `GuardsUnawareEvent`, `PlayerDetectedEvent`, `PlayerLostEvent`

Perception from spawners: `Spawned` / `Eliminated` / `Suspicious` / `Alerted` / `TargetLost` / `Unaware`. Payload `device_ai_interaction_result` (`Source` = guard, `Target` = player).

**Scaling:** if active players **> 2**, append `GuardsInitialSpawnersAdditional` + `GuardsReinforcementSpawnersAdditional`.

## What this sample actually teaches

1. **Guard spawners are the perception API.** Subscribe once; keep arrays of suspicious/alerted/investigating agents; fire bus events on 0↔1 transitions. See techniques/guard-alert-reinforcement.md.
2. **Leash = Verse position + two radii (cm), not a Creative volume.** Dummy Verse device as origin. Clear leash while patrol path runs; restore on stop. Fallback = swap origin. See techniques/ai-leash-position.md.
3. **Two victory End Games** (undetected vs detected) from `ReinforcementTriggered?`.
4. **Squad scale by extra spawners**, not by difficulty scalars.
5. **Alert mood is a subscriber**, not inside the game manager — radio + customizable lights.
6. **Barks:** concrete class + delay/cooldown + `race` vs death/`StopBark`.

## Official tutorial vs shipped

**External** hub: Stronghold Template

Chapters (`stronghold-0N-…-in-unreal-editor-for-fortnite`):

1. Create a New Project
2. Add Devices
3. Add Verse Script to Devices
4. Set Up Leash Devices
5. Set Up Audio and Visual Effects

`docsUrl` empty.

### Lines up

| Tutorial | Shipped |
|---|---|
| Classes `stronghold_game_manager`, `stronghold_leash_position` | same |
| Initial / additional / reinforcement spawners, tracker, two HUD, three End Games, 2 retries | `@editable` names |
| Leash inner/outer, patrol path group match on spawner | `SetLeashPosition`, 2300/2400 default cm |
| Squad of 4, sneak or fight | matchmaking 4 / 1 team |
| Dummy Verse device as leash origin | leash devices |

### Does not line up

| Topic | Tutorial | Shipped |
|---|---|---|
| Details typo | `ReinforcementLeashReferernce` | `ReinforcementLeashReference` |
| Details typo | `Spiper tower 2` | Sniper |
| Initial spawners array | two elements numbered **4** | shipped array, no duplicate index |
| Game manager Visible in Game | **True** | unusual — most Verse devices hide |
| Docs dump of game manager | 259-line paste | 16k file — same architecture; prefer snippet |

Train on snippets. Emitting `Referernce` is doc regurgitation.

## Negative findings

- 1137 jungle actors are not the lesson.
- `Visible in Game True` on the manager is a tutorial default, not a pattern to copy.
- Scratch `Developers/` folder.

## Capture backlog

1. Live leash inner/outer vs 2300/2400 defaults. *Waive.*
2. Billboard 58 captions. *Waive — mechanic is Verse.*
3. Ch.5 audio/VFX vs `alert_manager` / `bark_manager`. *Open, low — files already Observed.*

## Live notes

Level confirmed. Folders named bunkers + `Devices/AI/Leash`. Streaming off.
