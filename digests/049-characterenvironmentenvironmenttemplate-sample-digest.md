---
id: digest-049
kind: digest
---

# character_environment_environment_template — sample digest

> Offline `_raw/character_environment_environment_template/`. Live
> `/e3c5a857-eea8-4430-b43a-66db8f987496/character_environment_environment_template` (**Observed**).
> Docs: long companion (kitbash / memory / orbit camera). Spreadsheet Keep=Yes. **2 Verse.**

| | |
|---|---|
| Classification | **AAA environment showcase** — modular ship, data layers, orbit-camera stack |
| Content type | full island: **286 authored vs 24985 placed** |

## Scale

| | |
|---|---|
| Authored | 286 (`BP_*` kit, 39 `DL_*` data layers, cinematics, lighting BPs) |
| OFPA | 24985 (**111 unclassified**) |
| Verse | **2** |
| Fab | **4** — `character_environment_{Audio,Props,UI,VFX}` |
| Maps | **2** — `character_environment_environment_template` + `AssetArtGallery` |
| `FortStaticMeshActor` | 12522 |
| Decals | 3598 |
| Audio players | 105 |
| Compatibility | **42.30** |
| Matchmaking | default 1 (config `maxPlayers` not 16 — check live) |

Empty `Developers/felixkupis/Collections/`. Twenty-ninth first-party hit. Sibling of extracted `character_environment_sample`.

**Why 286 vs 25k:** same Deserted pattern — look lives in Fab + thousands of kitbash `BP_*` instances. See fab-uref-content-packs.md and the four stubs below.

## Fab (4 packs, not 3)

```
References/character_environment_Audio_….uref  → /EpicGames@Fab.com/character_environment_Audio  projectId 77934947-…
References/character_environment_Props_….uref  → /EpicGames@Fab.com/character_environment_Props  projectId 4aa8e375-…
References/character_environment_UI_….uref     → /EpicGames@Fab.com/character_environment_UI     projectId 8ed653c6-…
References/character_environment_VFX_….uref    → /EpicGames@Fab.com/character_environment_VFX    projectId 31850f48-…
```

Deserted had Audio/Props/VFX. **character_environment adds UI.** `.uref` here also stores `projectId` (Deserted’s were empty). Filename hash ≠ mount UUID.

## Live rooms (not numbered teaching halls)

`1_Environment/{CargoHold, CargoHold LowerDeck, Common_Room, Crew_Quarters, Helm, Corridor_East, Navigations_Room, Shoulder_Room_East/West, WorkshopCorridor, CargoCorridor}` + `2_Global_Lighting` + `Audio` + `Cameras` + `Gameplay` + `MapControllers`. Kitbash set-dressing, not Modeling-style exhibits.

Folder leftovers ship: `LowerDeck_keep`, `CranLift` typo, `DisableUntilScalabilityMGRFixIsIn`, `DecoractionLights_*`.

## Verse (the design)

| File | Class | Role |
|---|---|---|
| `camera_switch_mode_device.verse` | `camera_switch_mode_device` | cycle `gameplay_camera_orbit_device` stack on Input Trigger; hide/show pawn with delay |
| `starting_sequence_device.verse` | `starting_sequence_device` | enable ambient audio; **Await** crew-quarters button → stop welcome VO → enable upper-deck map + minimap HUD + quest message |

See techniques/orbit-camera-stack-hide-pawn.md.

## Device / actor inventory (high-signal)

| Count | Class | Role |
|---|---|---|
| 10 | `BP_LightingScalabilityManager` | per-room scalable lighting (Day Sequence / Materials Lumen manager family) |
| 2 | `BP_LumenExposureManager` | Lumen/non-Lumen parity |
| 2 | `Device_CameraModes_Orbit` | 3P ↔ simulated 1P |
| 1 | `Device_InputTrigger` | right-click / LT to cycle cameras |
| 2 | `VerseDevice` | the two scripts |
| 105 | `Device_CRD_AudioPlayer` | ambients / spots |
| 40 | `Device_Volume` | audio / lighting volumes |
| 6 | `Device_MapIndicator_V2` + MapController | quest map after console |
| 9 | `Device_CinematicSequence` | crane / coffee / robot anim |
| rest | kitbash `BP_*` (walls, crates, vents, doors) | the ship |

## What this sample actually teaches

1. **Orbit camera is a stack, not a mode flag.** `AddTo` / `RemoveFrom` agent; empty stack = default Fortnite cam. Hide pawn **after** `HideDelay` (0.20s) so the blend doesn't pop.
2. **Console gate before HUD.** `InteractedWithEvent.Await()` on crew-quarters button, then map+HUD — matches spreadsheet “door/computer first.” Shipped `starting_sequence` does the HUD/map, not the door (spreadsheet still flags the door).
3. **Kitbash + Mesh Decals + face-weighted normals** instead of unique baked-normal hero meshes (tutorial: 400 MB / 100k streaming units).
4. **Data layers + HLOD + WP** for a 25k-actor ship. 40 `DL_*` assets.
5. **Lighting Scalability Manager × 10** — per room, HighEpic vs Scalable vs FakeVolumetrics folders.
6. **Fab UI pack** — fourth stub vs Deserted’s three.

## Official tutorial vs shipped

**External:** character_environment Environment Template

Companion to this island (not a recreate-from-scratch). Sections: memory limits, custom UI, lighting large spaces, mesh construction, dynamic materials, **Using the Orbit Camera Device** (pastes `camera_switch_mode_device`). Island code 7100-3544-3074 for the GDC experience.

`docsUrl` empty. Spreadsheet `/en-us/uefn/character_environment-environment-template-in-unreal-editor-for-fortnite`. Retrieved signed-in Chrome 2026-09-11.

### Lines up

| Tutorial | Shipped |
|---|---|
| Kitbash / mid-poly / mesh decals | thousands of `BP_*` kit parts |
| Orbit Camera + Verse cycle | `camera_switch_mode_device` + 2 Orbit devices + Input Trigger |
| Hide character when cam needs it | `ShouldHideCharacter` + `HideDelay`/`ShowDelay` |
| HLOD / WP / Data Layers | 40 `DL_*`, 2 maps |
| Lowest-end first | `HighEpic_Lighting` vs `ScalableLighting` vs `FakeVolumetrics` |

### Does not line up

| Topic | Tutorial | Shipped / spreadsheet |
|---|---|---|
| Recreate | companion overview | **does not** document `starting_sequence_device` |
| Camera at edit time | — | spreadsheet: camera **outside the level** |
| First spawn | — | spreadsheet: player spawns **outside the ship**, fixes on restart |
| First-room door | — | should stay closed until computer; **not** in Verse (button only unlocks map/HUD) |
| Wants Lumen | implied | spreadsheet: run in Lumen to match the page |

Train on orbit-stack Verse + Fab/kitbash. Do not emit “paste only camera_switch” as the whole island. Do not treat Scene Graph cameras as this (Orbit **device**).

## Negative findings

- Scratch `Developers/` folder.
- 111 unclassified OFPA.
- Folder typos / WIP names ship (`CranLift`, `DecoractionLights`, `DisableUntilScalabilityMGRFixIsIn`).
- Compatibility 42.30.
- Mechanical extract ~108s (25k OFPA) — expected.

## Capture backlog

1. Live Orbit device properties vs `Cameras[]` order. *Open, low — Verse is the lesson.*
2. Which DataLayer is on which room. *Waive — 40 DLs, tutorial is the policy.*

## Live notes

Level confirmed. Room tree loaded (not empty like Deserted WP). Chrome still signed in. No file >80MB in a quick scan.
