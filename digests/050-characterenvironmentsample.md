---
id: digest-050
kind: digest
---

# character_environment_sample

Epic first-party showcase, found under **Project Browser → Feature Examples**. A
single-player, walk-through sci-fi ship interior ("character_environment") built to demonstrate
**MetaHuman characters in UEFN**, custom Chaos Cloth clothing, and cinematic camera work.
Originally the GDC 2024 character_environment experience — playable in Fortnite at island code
**7100-3544-3074**.

**Sources:** template extraction in `_raw/character_environment_sample/`, plus
Epic's documentation page.
Downloadable learning assets (Captain Roux + Captain Elli, Marvelous Designer garments,
Photoshop textures, UE projects): `https://fn.gg/character_environment-Learning-Assets`, linked from a
comment at the top of the Verse file and reachable via **Verse Explorer**.

## Scale (Observed)

| | |
|---|---|
| Authored assets | 307 (155 of them MetaHuman) |
| External actor packages | 2474 |
| External object packages | 294 |
| Verse files | **1** |
| Maps | 1 |

## Project configuration (Observed — `config.txt`)

```
compatibilityVersion            41.30
matchmaking.maxPlayers          1        (maxTeamCount 1, maxTeamSize 1)
allowJoinInProgress             true
experimental.sceneGraph         bIsSceneGraphSystemAllowed: true
                                bEnableOneFilePerEntity:    false
apiAccess                       Private
```

`.uplugin`: `VerseVersion 1`, `VerseAPI 1`, `EnableVerseAssetReflection: true`,
`EnableSceneGraph: true`, `CanContainVerse: true`.

The whole experience is configured as a **solo showcase**. Read this as the deliberate
scope choice it is — the template does not solve any multiplayer/replication problem, so
do not cite it as evidence that a technique here is multiplayer-safe.

## External dependencies (Observed — `References/*.uref`)

The template does **not** ship the props, audio, or VFX. It references three Fab content packs:

| Fab id | Project id |
|---|---|
| `/EpicGames@Fab.com/character_environment_Props` | `4aa8e375-4294-1ad0-74fe-3e972bbf4aa6` |
| `/EpicGames@Fab.com/character_environment_Audio` | `77934947-4208-5d20-d5f2-a58568e15aa2` |
| `/EpicGames@Fab.com/character_environment_VFX`   | `31850f48-4544-4899-adba-558244a7057f` |

322 actor references resolve to mount `/04f0ce9d-4b1a-a806-e1b8-6e8582031a95`, i.e. a large
share of the visible world is external pack content, not template content (**Inferred** from
`reference-mounts.txt`; the GUID is not named in any manifest).

There is also a reference to `/Game/TimeOfDay/TODM_Sequence/Blueprint/UEFN_Ch4_DaySequence`
— the stock UEFN time-of-day sequence (**Observed** in the `.umap`).

## Device inventory (Observed — `actor-classes.txt`)

This is the single most surprising finding. The **entire interactive layer is ~12 device
instances**:

| Count | Class |
|---|---|
| 4 | `Device_Billboard_V2` |
| 3 | `BP_ElevatorMover_N1` |
| 2 | `Device_CameraModes_Orbit` |
| 2 | `Light_WallLightFixture_SmallGreenLight` |
| 1 | `VerseDevice` (the Verse device below) |
| 1 | `Device_InputTrigger` |
| 1 | `Device_HUDController_V2` |
| 1 | `Device_CinematicSequence` |
| 1 | `Device_Button_V2` |
| 1 | `BP_LumenExposureManager` |
| 1 | `BP_Creative_Player_Spawner_Prop` |

Against **1136 `FortStaticMeshActor`** and **588 actor folders**. Ratio ≈ 100:1 environment
art to interactivity.

## Content organisation (Observed — `folders.txt`, `naming-prefixes.txt`)

```
Content/
  Animation/        loose anim assets
  Characters/       Captain/{Body,Face} anims + Cloth/Captain/<Garment>/Textures/Lod{1,2}
  Cinematics/       Cap_Reveal/LS_Cap_Helm{,_Anim}
  DataLayers/       39 DL_* assets + LoadingSequences/
  Geometry/         Helm/, Keyboard/, ModularBP/{Props,StructuralComponents/*}, CharacterEnvironmentExt/
  Lighting/         BP_LightSources/
  Materials/        Textures/
  MetaHumans/       Common/ + Cpt_Roux/
  Collections/      (empty)
  Developers/       jojitsuruga/ (local user scratch — leaked into the shipped template)
```

Prefixes: `T_` 65, `BP_` 64, `DL_` 39, `MI_` 32, `MF_` 19, `M_` 15, `LS_` (level sequence),
`LAS_` (level anim sequence), `AS_` (anim sequence), `SSP_` (subsurface profile).
Geometry blueprints carry a `_N1`/`_N2` revision suffix (**Inferred**).

## Verse (Observed)

One file: `camera_switch_mode_device.verse` → `snippets/verse/`. Covered in
gameplay-camera-switching.md.

## Engine subsystems pulled in (Observed — `script-modules.txt`)

`MetaHumanRuntime`, `MetaHumanSDKRuntime`, `MetaHumanUEFNRuntime`, `HairStrandsCore` (groom),
`ChaosClothAssetEngine` (cloth sim), `OptimusCore` (deformer graph), `CinematicCamera`,
`LevelSequence`, `NavigationSystem` (1147 hits), `Niagara`, `VerseDevices`,
`WorldPartitionHLODUtilities`, `DataDrivenGameplayEventRouter`.

## The problem the template exists to solve (External — Epic doc)

Epic names three obstacles to moving MetaHumans from UE into UEFN:

1. **Memory** — UEFN caps upload at 2 GB and **download at 400 MB**; a typical UE MetaHuman
   was ~**800 MB**, i.e. twice the whole download budget.
2. **Performance** — different platforms have different FPS targets, and MetaHumans were
   expensive enough that multi-MetaHuman projects were impractical.
3. **Cross-platform quality** — fidelity had to hold everywhere while staying compatible with
   existing UE MetaHuman meshes and rigs.

The optimisation pass brought a MetaHuman to **~60 MB, down from ~800 MB**, "without
sacrificing fidelity". Roughly 13×, achieved through texture compression (half the memory on
its own), reduced-instruction materials with textures baked in, per-platform LOD tuning
(4 LODs each for body and face, with a **LODSync component** keeping components in step),
animation feature gating via the **MetaHuman Component**, and groom rework (~half the cook
size). Details in metahuman-in-uefn.md.

## What this sample actually teaches

1. **A cinematic-quality UEFN island needs almost no devices.** Twelve device instances
   carried this. Reach for environment art, data layers, and one Verse device before
   reaching for a wall of Creative devices.
2. **MetaHumans in UEFN run on a parallel, reduced material stack** — see
   metahuman-in-uefn.md. The `UEFN/` subfolder pattern
   and `_Baked` textures are the whole story. Budget ~60 MB per MetaHuman against a 400 MB
   island download cap.
3. **Data layers here are an organisational scheme, not a streaming one.** 39 layer assets
   with a two-level area/sub-group naming convention — but the map ships with **World
   Partition streaming disabled**, only 18 layers instanced, and 17 of those `Activated` at
   start. Copy the naming and the structure/props split; do not read it as a streaming
   worked example. See
   world-partition-data-layers.md.
4. **Kitbash via Blueprint clusters, place via static meshes.** 64 modular BPs generate 1136
   static mesh actors. See modular-kitbash-blueprints.md.
5. **Ship the map, reference the art.** Props/audio/VFX live in Fab packs the project
   `.uref`s. Keeps the project small and updatable. See
   project-config-and-dependencies.md.
6. **Camera swapping needs a character-visibility delay — and Epic rejected the built-in one.**
   Both `Device_CameraModes_Orbit` instances have `HidePlayerCharacter` **off**; hiding is
   hand-rolled in Verse with `HideDelay` 0.20s / `ShowDelay` 0.3s against a **0.5s
   Ease-In-Out blend**, i.e. the character is swapped roughly halfway through the camera
   move. Also note the FPS camera is just an orbit camera with `Distance` 0.
   See gameplay-camera-switching.md.
7. **Cloth is authored outside the MetaHuman**, as Chaos Cloth garments with per-LOD ORM
   texture sets, round-tripped through full Unreal Engine (Marvelous Designer → UE clothing
   asset → UEFN). And the headline gotcha: **cloth simulation is off by default and only runs
   during the cinematic sequence** — even Epic could not afford it always-on.
   See character-clothing-and-lods.md.
8. **Interaction design is minimal and physical:** an **invisible** `Device_Button_V2`
   (`VisibleDuringGame` false, 0.7 m radius, unlimited triggers) on the MetaHuman starts a
   `Device_CinematicSequence`. That is the map's **only** event binding. The cinematic does
   not seize control — `DisablePlayerInput: No`, `ShowHUDDuringCinematic: Show`. See
   cinematic-sequences.md.
9. **Counter-example worth keeping:** `Content/Developers/jojitsuruga/` shipped in a
   first-party template. Even Epic leaks scratch folders. Sweep `Developers/` and empty
   `Collections/` before publishing.

## Corrections from the live-editor pass

A UEFN session with this project open was audited on **2026-08-14**, reading actual property
values through the editor bridge. Three claims in earlier revisions of these notes were wrong
and have been fixed. They are listed here because they were all **Inferred from names and
counts** — the exact failure mode this pipeline is supposed to guard against.

| Was | Actually |
|---|---|
| "Streaming is designed as a first-class layer" (point 3 below) | **`LevelHasStreamingDisabled = 1`.** World Partition streaming is off. 17 of the 18 instanced layers start `Activated`. |
| "39 data layers … plus explicit loading-sequence assets" | Only **18** of the 39 assets are instanced. The loading sequence has **zero referencers** and names layers that don't exist in this world — leftovers from the larger GDC build. |
| "Which MetaHuman Component features are disabled" | **None are.** Every feature boolean is true; the saving is done entirely with LOD thresholds. |

A fourth item was not wrong but was materially incomplete: the cloth-sim gating that Epic's
doc describes turns out to be **a Sequencer bool track**, not Blueprint or Verse logic.

## Capture backlog

**Retirement status: ready, pending the content classification below.** Two live-editor passes
cleared the entire numbered backlog and filled the cinematics gap.

**0 open, 19 closed (one as unanswerable), 2 waived**, plus the previously-closed importer
items. One non-blocking gap remains in the new cinematics note: Sequencer keyframe *times and
values*, which require opening Sequencer and are not reachable by any API on the bridge.

### Closed by the second pass (asset-level)

| # | Question | Answer |
|---|---|---|
| 1 | Chaos Cloth solver settings | **Unanswerable — the Dataflow graph is not shipped** (`dataflowAsset: null`). The values do not exist in this sample. Go to Epic's Chaos Cloth docs / learning assets. |
| 2 | Physics asset setup | 4 bodies (`hand_l/r`, `lowerarm_l/r`) — arms only. Tapered-capsule colliders, no constraints, RBAN `positionIterations` 6. |
| 3 | Master cloth material | 236 PS instructions, 3 samplers, two-sided; 3 texture + 9 scalar/vector params + 1 static switch. Full table in the clothing note. |
| 7 | Greeble cluster assembly | **No construction script at all** — empty event graph. 38–90 fixed mesh components incl. bundled mesh decals. Two instances are byte-identical; variety comes from placement transform. |
| 8 | Elevator Blueprint mechanism | **There is none.** 3 components, empty event graph. The elevator is set dressing. |

Items 9 and 10 (wall grid size, light-fixture bundling) were folded into 7/8 as low-value and
are closed with them.

### Closed by the first pass (level-level)

| # | Question | Answer now lives in |
|---|---|---|
| 4 | How cloth sim is enabled for the cinematic and disabled after | clothing — Sequencer bool track on `Cloth_CA` |
| 5 | LOD thresholds the Lod1/Lod2 texture sets switch at | clothing — `minLod` switch/mobile = 1 |
| 6 | Which MetaHuman Component features are disabled; LODSync config | metahuman — none disabled; groom map `[1,3,5,7]` |
| 11 | Which Fab pack meshes the placed actors point at | kitbash — mount `/04f0ce9d-…` confirmed |
| 12 | Per-layer initial runtime state | data layers — full 18-row table |
| 13 | What a loading sequence batches, and how it is triggered | data layers — LevelSequence, unreferenced |
| 14 | HLOD build settings | data layers — MeshMerge, non-spatial |
| 15 | Streaming cell/grid sizes, `bBlockOnSlowStreaming` | data layers — streaming disabled outright |
| 16 | Full layer hierarchy and membership | data layers — 18 instanced of 39 |
| 17 | `@editable` values on the Verse device | cameras — Orbit/FPS, delays unchanged |
| 18 | Camera blend times | cameras — 0.5s Ease-In-Out both ways |
| 19 | Which input is bound | cameras — Creative Input Action, Custom 2 (Target) |
| — | The cinematics gap (button → sequence → cloth) | **new note:** cinematic-sequences.md |

### Still open

Nothing in the numbered backlog. One item in the cinematics note remains, and it is not
blocking: **Sequencer keyframe times and values** (shot cuts, and the exact
`bEnableSimulation` keys). `FMovieSceneBoolChannel` is not exposed as a property, there is no
Sequencer toolset on the bridge, and the channel is a custom-serialised blob in the `.uasset`.
It needs Sequencer opened by hand. `BP_LumenExposureManager` is also unopened, and low value.

**Waived:** `Face_Archetype_LODSettings_High` LOD assignment (importer-produced; current Epic
LOD docs beat a 41.30 template).

### What the bridge could and could not read

Worth recording for the next template, since it changes what "capture backlog" should mean.
Much of what the first pass filed as "needs an asset editor" turned out to be readable.

**Readable directly:**

- Every classic device setting, via `ObjectTools.get_properties` on
  `<actor>.ToyOptionsComponent` → `playerOptionData`.
- Verse `@editable` values (`DeviceToolset`), event bindings, component hierarchies, data
  layer instance state, asset registry tags, dependencies and referencers.
- **Physics assets** — `PhysicsAssetToolset` reads bodies, shapes, constraints, solver
  settings. *Caveat:* `GetBodyShapes` only enumerates Sphere/Capsule/Box, so bodies built from
  `KTaperedCapsuleElem` (the Chaos Cloth collider type) return an empty list. Empty ≠ absent.
- **Material graphs** — `MaterialTools` reads expressions, parameter names and defaults, pin
  wiring, output connections, and compiled shader statistics.
- **Sequencer structure** — MovieScene tracks and sections are addressable as subobjects
  (`…LS_Foo.LS_Foo:MovieScene_0.MovieSceneBoolTrack_1.MovieSceneBoolSection_0`), exposing
  `displayName`, `sectionRange`, `bIsActive`, `evalOptions.completionMode`. Discover the
  object names by string-scanning the `.uasset`.
- **Blueprint component structure** — via `ActorTools.get_components` on a *placed instance*,
  which sidesteps the lack of a Blueprint toolset entirely.
- **Whether a Blueprint has any logic** — string-scan the `.uasset` for `K2Node`. No `K2Node_*`
  names in the table means an empty event graph. This closed two backlog items on its own.

**Readable with a trick:** Verse sub-object fields are stored under mangled names
(`__verse_0xC30798CE_ShouldHideCharacter`). `list_properties` shows the friendly name but
`get_properties` needs the mangled one — recover it by dumping ASCII strings from the actor's
`__ExternalActors__` package. Object references are wrapped in a sub-object carrying a
`savedActor` property.

**Genuinely not readable:** Blueprint node graphs (beyond presence/absence), material node
*positions*, and Sequencer keyframe times and values. Track presence and binding are readable;
channel keys are not.

**And one thing no tool can recover:** a Chaos Cloth Asset's Dataflow construction graph, if
it was stripped before shipping. That is a property of the asset, not a limit of the bridge.

### Does deprecating this sample cost users the MetaHuman optimisations?

**No.** Verified (see metahuman-in-uefn.md). The ~13×
size reduction is performed by **UEFN's MetaHuman Importer / export pipeline**, not by
anything authored in character_environment. A user creates a MetaHuman in MetaHuman Creator (or via Mesh
to MetaHuman), runs **Window → MetaHuman Importer** in their own project, and receives the
same baked textures, `_UEFN` material stack, 4 LODs, LODSync, and MetaHuman Component.

Proof from the template: `MetaHumans/Common/` contains a Medium body archetype Roux doesn't
use and MetaHuman Creator's default **hoodie**, which she doesn't wear. That folder is
importer output, not authored content — and `Common/UEFN/` (the master materials) sits inside
it.

What is *not* automated, and therefore what this sample actually demonstrates:

| | Automated by UEFN? |
|---|---|
| MetaHuman optimisation, baking, LODs, material stack | **Yes** — importer |
| Custom clothing (DCC → UE → clothing asset → migrate) | **No** — user work |
| Gating cloth simulation to the cinematic | **No** — design decision |
| MetaHuman Component feature toggles | **No** — authoring decision |
| Data layers, kitbash organisation, Verse camera logic | **No** — design decisions |

Every row in the "No" column is written up in this sample's `techniques/`. That is precisely the case for
retiring the template: it teaches design decisions, and design decisions survive as prose.

### Before this sample leaves the repository

Decide what its content **is**. The KB can carry the knowledge; it cannot carry the assets.

**The art is already safe — this was verified, not assumed.** All three referenced packs
(`character_environment_Props`, `character_environment_Audio`, `character_environment_VFX`) ship inside the Fortnite/UEFN build's pak
containers and are *exposed* on project open in ~0.00 s, with no download and no user action.
See project-config-and-dependencies.md.
**Retiring the template deletes none of the art**, because none of the meshes, audio, or VFX
were ever in the template.

That narrows the reusable-art question to exactly one thing:

- **`Content/Geometry/ModularBP/` — 64 composition Blueprints.** These *are* template content.
  They hold no logic (empty event graphs) but each embodies 38–90 hand-placed components
  referencing pack meshes. Cheap to preserve, real work to recreate. Deleting the template
  deletes the assembly labour, not the art.
- **`Content/Characters/Cloth/Captain/`** — the garment set. A baked artifact with its
  Dataflow graph stripped, so it can be *used* but not learned from or re-authored. Only worth
  preserving if someone intends to reuse Roux's jacket verbatim; the `fn.gg/character_environment-Learning-Assets`
  download supplies the authorable source, which is strictly better.
- **Everything else is reference only** — the ship, the level, the cinematics, and the
  MetaHuman (importer output any user can reproduce). Retire freely.

A user who wants the sci-fi kit after retirement adds `/EpicGames@Fab.com/character_environment_Props` to
their own project and gets the meshes instantly, since the bytes are already on their disk.
What they lose is Epic's 64 pre-assembled clusters — an inconvenience, not a dependency cliff.

**Remaining unknown:** whether those packs can be added to an arbitrary user project via Fab,
or whether exposure only works because *this* project's `.uref` declares them. Worth one
check before retiring, because the whole "art survives" argument rests on it.

character_environment already demonstrates the correct split: its props, audio, and VFX are packs the
project merely references. Art in packs, demonstration in the KB, nothing left in between —
that is the end state this pipeline is aiming at, and this sample is one folder away from it.
