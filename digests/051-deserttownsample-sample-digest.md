---
id: digest-051
kind: digest
---

# desert_town_sample — sample digest

> **Extractor provenance.** `tools/extract-template.py` (Pass A) + UnrealMCP on
> `localhost:8000` (Pass B). Default level
> `/a15c1ab0-203f-4539-bdd6-00cb2358b4d2/desert_town_sample` (**Observed**).
> Docs: Working With space-opera™ Islands
> (hub; DesertTown is one environment-starter bullet) + Vistas and Biome Environments
> (DesertTown feature list + Binary Suns / heat ripple / sandcrawler spline) + Custom UI.
> `docsUrl` empty. Guessed `star-wars-DesertTown-inspired-starter-island-in-fortnite` **hub-redirect**.
> Compat **43.00**. `featureSet`: **CustomGameplayFeature**.

| | |
|---|---|
| Classification | **theme environment starter** — DesertTown desert town sandbox, not EnergyBlades/Narrative/Team Conquest |
| Content type | **catalog wrap** (45 authored vs 5281 OFPA; `/SPACE_THEME_A_*` IcedTea town + `/ENV_PACK_Biome_*` desert) |
| Verse | **0 files** |
| Fab `.uref` | **0** |

## What it is

The picker name is **DesertTown Inspired Starter Island**. The island is a **16-player desert town**: spaceport, trader dwellings, EnemyFaction tents, sandcrawler playset, starfighter + Fighter, roaming civilians and Guards. Teaching that survives deletion is the **same catalog-wrap + Guard/Civilian CDs as IceBase/VolcanicTown**, with **no doorway volumes**. Do not duplicate: guard-patrol-volume-swap.md (this island is **not** a third volume-swap). Hyperspace stays on hyperspace-attach-datalayer.md. Landscape paint layers stay on landscape-weight-blend-layers.md (`layers[]` empty). Day Sequence priority stays on day-sequence-priority-stack.md (**2000-only**, `Default` cycle not IceBase/VolcanicTown FixedTime dusk). Environment lighting rig stays on environment-lighting-rig.md. Custom skyboxes stay on night-atmosphere-stack.md. Widgets-folder HUD stays on narrative_roleplaying_template. Sibling catalog wraps: ice_base_sample (`/SPACE_THEME_E_*`), volcanic_town_sample (`/SPACE_THEME_C_*`), space_station_sample (`/SPACE_THEME_B_*`).

## Scale (**Observed** — `summary.txt` + live)

| | |
|---|---|
| Authored assets | **45** |
| OFPA | **5281** |
| Maps | `desert_town_sample.umap` (only) |
| Verse files | **0** (`VerseToolset.ListFiles` on `/a15c1ab0-…` — do not recurse `__ExternalActors__`) |
| Verse devices placed | **0** |
| Data layers | **0** |
| Fab `.uref` | **0** |
| Compatibility | **43.00** |
| Matchmaking | **16** (`maxPlayers=16`, `teamSize=Dynamic`, `teams=FreeForAll`, `locomotionPreset=Current BR`) |
| Streaming | Partitioned + actor folders + OFPA. **`LevelHasStreamingDisabled` omitted** (IceBase / VolcanicTown camp). Live descriptors **5191** `bIsSpatiallyLoaded=false` / **25** true. `load_actors` used on Devices/Vehicles/Lighting/LevelData/NPCs. |

Empty `Developers/felixkupis/Collections/`. First-party scratch folder.

`experimental.*` (Python / toolsets / Scene Graph) in `.uefnproject` is **Unverified**. Content-side: **no** `.verse`, **no** `__ExternalEntities__`. Plugin `EnableSceneGraph: true` with no Scene Graph entities.

Largest OFPA ~**53.8 MB**. Whole folder ~**259 MB**. Under the GHES **100 MB** per-file cap.

## Level organisation (**Observed** — live `get_folders` + folder counts)

```
Devices/                              3 — 2 player pads + HUD Controller
Vehicles/                             2 — starfighter, fighter
Environment/                       4886
  Spaceport/                       3125
  Rocks&Cliffs/                     826
  Trader_Dwellings/                 449 (Sandcrawler/ 3, dwellings, perimeter)
  Foliage/                          170
  Large_Dwelling/                   119
  EnemyFaction_Encampment/               97 (four tent sets)
  Vista/                             63
  Minor_Locations/                   19
  Landscape_Vista/                   12 — authored SM_Landscape_01–12
  Landscape/                          1 — Landscape actor
  VFX/                                5 — sand Niagara
  Sandcrawler/                        0 leftover (moved under Trader_Dwellings)
NPCs/                               153 — 64 spawners, 89 patrol, 0 volumes
Lighting/                             2 — DSD + unbound heat-ripple PPV
HLOD/                               166
LevelData/                            4 — WorldDataLayers, leftover cine camera, minimap, bounds
Spaceport/                            0 leftover root
(unfoldered)                          IslandSettings0
```

No `Device_Billboard_V2`. No syllabus boards.

## Device inventory (the design)

| Count | Class | Role |
|---|---|---|
| 89 | `AIPatrolPath_Device` | groups **1–7** spaceport + **8 / 20–25**; Loop and BackAndForth mix |
| 64 | `Device_CharacterSpawner` | civilians + Guards + HeroA + WalkerVehicle; **0** volume bindings |
| 9 | `PlaysetRoot` | 8× `OUTPOST Prefab *` + `WalkerVehicle` |
| 5 | `NiagaraActor` | sand ground-drift / cliff |
| 2 | `BP_Creative_Player_Spawner_Prop` | both `useAsIslandStart=true` |
| 1 | `Device_HUDController_V2` | in-project `UW_HUD_*` overrides |
| 1 | `BP_DaySequenceModifier_Device` | **Default** cycle, priority **2000**, `use Volume=false` |
| 1 | `PostProcessVolume` | unbound heat-ripple blendable |
| 1 | `CRD_SpawnerDevice_THEME_VEHICLE_A` | starfighter, health **1500** |
| 1 | `CRD_SpawnerDevice_THEME_VEHICLE_B` | fighter, health **1500** |
| 1 | `Landscape` | desert MI; `layers[]` **empty** |
| 1 | `CineCameraActor` | leftover in LevelData |
| 0 | Billboard / Verse device / Volume / Hero / EnergyBlade / Conversation | other space-opera tiles |

## Authored kit (**Observed** — `assets.txt` + live)

```
Content/NPCs/CD_CivilianNPC
Content/NPCs/CD_NPC_GuardBehavior
Content/UI/PlayerHUD/   UW_HUD_PlayerInfo, Equipped, Quickbar, PlayerVitals, TeamVitals
Content/Landscape/      SM_Landscape_01–12
Content/Materials/      MI_DesertTown_Landscape + ENV_PACK_MI_PPV_HeatRipple_Day_Inst + M_SandcrawlerTrack
Content/Textures/       T_SandCrawlerTrack_N/S + T_SandDrift_Noise
Content/DesertTown_Starter_sharedassets/  Layer_1–6_LayerInfo (+ `_1` dupes; Landscape.layers[] empty)
Content/Foliage/        SM_TREE_PACK_Gravel/Scree FoliageTypes
```

**Tier 2 mounts:** `/SPACE_THEME_A_Comp` 3728 + `/SPACE_THEME_A_Assets` 2714 (IcedTea town gallery), `/THEME_PACK` 6986, `/DesertTown_Starter` 3035, `/ENV_PACK_Biome_Assets` 3471 + `/ENV_PACK_Biome_Comp` 826 (desert), `/ENV_SWCosmetics` 53 (SnowSurgeon / LongBandit CIDs), `/ENV_SPACE_THEME_UI_Assets` 40, leftover `/92a9947c-…` 60. Small `/SPACE_THEME_B_*` + `/SPACE_THEME_C_*` bleed. No Fab `.uref`.

## What this sample actually teaches

1. **The town is a mounted `/SPACE_THEME_A_*` catalog plus nine playset roots, not 45 authored meshes.** Same catalog-vs-kit split as ice_base_sample (`/SPACE_THEME_E_*`) and volcanic_town_sample (`/SPACE_THEME_C_*`). Playset labels are `OUTPOST Prefab A/B/C/D/D2/E/E2/Crawler` + `WalkerVehicle` (IceBase shipped the WalkerVehicle playset too).
2. **Guard + Civilian CDs with patrol, but no doorway volume-swap.** Guard Type=`Guard` + Trespassers preset; Civilian Type=`Custom` `canSwim=false`. Link guard-patrol-volume-swap.md; **do not** train this island as a third volume source (`Device_Volume` **0**, all `enable.eventSubscriptions` empty). Interior folders look like IceBase’s swap and are **dead** (see negatives).
3. **HUD is authored UMG in `Content/UI/PlayerHUD` + a placed HUD Controller in `Devices/`, not ExtraAssets.** Live overrides: PlayerInfo, Equipped, Quickbar. HideElement minimap + build menu. Same layout as IceBase/VolcanicTown — space-opera in-project PlayerHUD is now **three** (IceBase, VolcanicTown, DesertTown). Space authors the widgets but places **0** controller. Contrast: Narrative `UI/Widgets` (custom-ui-widgets-folder.md).
4. **Day is one Day Sequence at 2000 on `Default` cycle, not IceBase/VolcanicTown FixedTime dusk and not a Binary Suns DSD mode.** `dayNightCycleTime` **12**, `overrideClouds=true`, `overrideSky=false`, stock `MI_DaySequenceSkydome`. **Not** a fourth 2000-vs-higher source. Unbound PPV blendable is authored `ENV_PACK_MI_PPV_HeatRipple_Day_Inst` (parent `/ENV_PACK_Biome_Assets/…/ENV_PACK_MI_PPV_HeatRipple_Day`) — that *does* match the Vistas heat-ripple blurb.
5. **Landscape layers are leftover LayerInfos, not a weight-blend demo.** Live `layers[]` **[]**; `MI_DesertTown_Landscape` parent `/ENV_PACK_Biome_Assets/…/ENV_PACK_MI_Biome_Desert_Landscape_A`. Distant meshes `SM_Landscape_01–12` in `Landscape_Vista`. Docs “Desert (DesertTown) Layers” / sandcrawler spline: authored `M_SandcrawlerTrack` + textures exist; OFPA lists `LandscapeSplineActor` ×1; live folders did not return a spline actor. Do not train as landscape-weight-blend-layers.md.
6. **Sandbox flags, not a round.** `allowBuilding=All`, `bInfiniteResources=true`, `bFallDamage=false`, `timeOfDayOverride=NoOverride`. **Two** spawn pads, both island start, `VisibleInGame=true` (IceBase/VolcanicTown pads were hidden). `spawnLimit=1`.
7. **starfighter + Fighter are THEME_VEHICLE_A / THEME_VEHICLE_B vehicle spawners, health 1500.** Atmospheric effects **off** on the starfighter, **on** on the Fighter. Flight ceiling **250**. Space’s starfighter was health **600**; VolcanicTown’s N-1 was THEME_PACK **600**. No turret.

## Official tutorial vs shipped

**External hub + Vistas + Custom UI** (headed Chrome 2026-09-15). Verbatim: `_raw/desert_town_sample/tutorial.txt`. Spreadsheet `docsUrl` empty. No dedicated DesertTown chapter (guessed slug hub-redirects).

### Lines up

| Tutorial | Shipped |
|---|---|
| theme Templates → space-opera → DesertTown Inspired Starter Island; “Desert terrain and arid environments” | picker; `featureSet` CustomGameplayFeature; `/ENV_PACK_Biome_*` desert MI |
| “Follow the twin suns to the DesertTown-inspired town” | `Environment/Spaceport` 3125 + `Trader_Dwellings` 449 |
| DesertTown styled prefab/gallery; vista assets; desert props | `/SPACE_THEME_A_*`, `Vista/` 63, `Landscape_Vista` 12 |
| Roaming NPCs + NPC Spawners + AI Patrol Path | 64 spawners, 89 patrol |
| Sand VFX | 5× `ENV_PACK NS Biome Sand *` Niagara |
| Heat ripple effect | unbound PPV + `ENV_PACK_MI_PPV_HeatRipple_Day_Inst` |
| starfighter and Fighter among space-opera vehicles | THEME_VEHICLE_A + THEME_VEHICLE_B in `Vehicles/` |
| Custom UI widgets `UW_HUD_Equipped` / `PlayerInfo` / `Quickbar` | same names in `Content/UI/PlayerHUD`; HUD Controller overrides |
| Hide minimap | `showMinimap=HideElement` |

### Does not line up

| Tutorial | Shipped |
|---|---|
| **Binary Suns Skybox**; World TODM Binary Suns; DSD opacity **0** known issue | Live DSD `dayNightCycleMode=Default`, time **12**, `overrideSky=false`, `overrideClouds=true`. No opacity property read. World TODM Binary Suns **unread** |
| Sandcrawler **spline tracks** | Authored track MI+textures; OFPA `LandscapeSplineActor` ×1; live Outliner folders **0** spline; leftover `Environment/Sandcrawler/` empty; crawler is playset `OUTPOST Prefab Crawler` |
| Desert **6 paint layers** | Authored LayerInfo×6 (+ `_1` dupes); live `Landscape.layers[]` **[]** |
| Custom HUD “preconfigured on Flat Grid / XL Flat Grid”; HUD Controller in Outliner **UI**; Player Info = `UW_HUD_Vitals`; `UI > Widgets > PlayerHUD` | **This** island ships a HUD Controller in **Devices/**; override `UW_HUD_PlayerInfo`; widgets in `Content/UI/PlayerHUD`. Widgets-folder layout is Narrative |
| Dedicated DesertTown walkthrough | Hub bullet + Vistas list; guessed dedicated slug hub-redirects |

Training rule: live CDs + empty volume bindings + in-project HUD paths + DSD Default win. Do not emit “DesertTown ships Binary Suns DSD / 6 painted desert layers / HUD only on Grid islands / doorway volume-swap.” Tutorial *prose* is the syllabus.

## Negative findings

- **Name implies a DesertTown *game*; files show a kitbash sandbox.** No round manager, no Verse, no EnergyBlades, no Hero / Conversation / hyperspace.
- **Interior/ImperialOutpost folders look like a IceBase volume-swap and are dead.** `NPCs/Spaceport/Interior` **10** spawners `enabledAtGameStart=false` + empty bindings. `NPCs/ImperialOutpost` **6** Guards same. Cantina01 **8** civilians stay on. Same class of leftover as VolcanicTown Outskirt 02.
- **`hUDInfoType=AI Enemy Elimination` with Neutral unkillable NPCs** (`isVulnerable=false`) **and Guard perception off** (override+disable sight/hearing/touch). Leftover default family as IceBase/VolcanicTown/Space/Prison.
- **Guard inventory is empty** (`items=[]`) like VolcanicTown, **leash on** 30 like IceBase, **perception off** like Space — do not copy any one sibling as “the” DesertTown Guard.
- **Root `Spaceport/` and `Environment/Sandcrawler/` are empty leftovers** after the nested move.
- **`get_referencers` on HUD widgets / both CDs returns []** — confirm on the placed HUD Controller / spawners, same trap as IceBase. LayerInfo_1 references LayerInfo; landscape MI is referenced by the 12 vista meshes.
- **`GetDeviceProperties` N/A** — no Verse `creative_device`s. A missing `get_properties` name fails the whole call (HUD `quickbarSlotWidgetOverride` vs `quickbar Slot Widget Override Class`).
- **Two island-start pads, `VisibleInGame=true`.** Siblings hid the pad.
- **IslandSettings0 is unfoldered** (`folderPath` None). Leftover `LevelData/` CineCamera.
- **OFPA class histogram listed `LandscapeSplineActor` ×1** that live folder queries did not return.

## Capture backlog

| Item | Status |
|---|---|
| Per-building child lists (3125 Spaceport) | **Waived** — folder counts + `/SPACE_THEME_A_*` mounts are the lesson |
| LayerInfo ↔ material slot wiring | **Waived** — empty `layers[]`; don’t train as weight-blend |
| World TODM Binary Suns flag | **Waived** — DSD Default/12 Observed; world setting not read |
| Live LandscapeSpline actor | **Waived** — OFPA ×1 + authored track MI; folders empty |
| PIE dead-interior confirm | **Waived** — `enabledAtGameStart=false` + empty bindings sufficient |
| Dedicated DesertTown docs page | **Waived** — slug hub-redirects; Vistas list Observed |

## Live notes

- Mount `/a15c1ab0-203f-4539-bdd6-00cb2358b4d2`. Project Verse path `/epictemplates@fortnite.com/desert_town_sample`. Plugin filename is that GUID.
- Did **not** `load_level` away from the default map.
- Billboard `text` N/A. `GetDeviceProperties` N/A.
- `folderPath` on descriptors is often `"None"` even when `get_actors_in_folder("Devices")` returns the HUD / pads. Count from the folder query, not `folderPath`.
- Character Definition modifiers: `get_properties` on `:CharacterModifier_*_C_0` / `:CharacterType_*_C_0` suffixes (listing `modifiers[0]` on the CD asset fails).
- Folder on disk / title is **`desert_town_sample`** (picker “DesertTown Inspired Starter Island”).
