---
id: digest-052
kind: digest
---

# team_conquest_sample — sample digest

> **Extractor provenance.** `tools/extract-template.py` (Pass A) + UnrealMCP on
> `localhost:8000` (Pass B). Default level
> `/6ce09c8f-a624-44d0-9d9a-efcd7b110050/team_conquest_sample` (**Observed**).
> Docs: dedicated space-opera™ Team Conquest
> + hub Working With space-opera™ Islands
> + Custom UI.
> `docsUrl` empty. Not on the spreadsheet. Compat **43.00**. `featureSet`: **CustomGameplayFeature**.

| | |
|---|---|
| Classification | **theme game-mode island** — 12-player Rebels vs Empire sequential capture, not IceBase/DesertTown sandbox |
| Content type | **catalog wrap** (129 authored vs 9528 OFPA; `/SPACE_THEME_A_*` + `/THEME_PACK` desert + IcedTea/HarshTone) |
| Verse | **3 files** (1 placed: `Hide_on_Start`) |
| Fab `.uref` | **0** |

## What it is

The picker name is **Team Conquest**. The island is a **linear 3-zone attack/defend**: Rebels (team 1) capture A→B→C; Empire (team 2) holds until a **900s** timer. Teaching that survives deletion is **device `OnControlChange` unlocks** (not Verse spawn weights) and a **Conversation Device respawn picker**. Do not duplicate: weighted-capture-spawns.md, asymmetric-cap-balancer.md, persistable-arena-rotation.md. Widgets-folder HUD: add this island as a source on custom-ui-widgets-folder.md — do not rewrite it. ExtraAssets HUD stays **three**. Guard *volume-swap* stays on IceBase (this island’s Guards Enable with capture, no doorway volumes). Environment catalog wrap is the same `/SPACE_THEME_A_*` family as desert_town_sample.

## Scale (**Observed** — `summary.txt` + live)

| | |
|---|---|
| Authored assets | **129** |
| OFPA | **9528** |
| Maps | `team_conquest_sample.umap` (only) |
| Verse files | **3** (`VerseToolset.ListFiles` on `/6ce09c8f-…` — do not recurse `__ExternalActors__`) |
| Verse devices placed | **1** (`Hide on Start` → `Hide_on_Start`; `propsToHide` **24**) |
| Data layers | **0** |
| Fab `.uref` | **0** |
| Compatibility | **43.00** |
| Matchmaking | **12** / **2** teams (`.uefnproject` + live Island Settings `maxPlayers=12`, `teamSize=SplitEvenly`) |
| Streaming | Partitioned + actor folders + OFPA. **`LevelHasStreamingDisabled` omitted** (IceBase / VolcanicTown / DesertTown camp). Sampled descriptors `bIsSpatiallyLoaded=false`. `load_actors` required for billboard `text` and device values. `find_actors` needs `collision_channels=[]`. `folderPath` often **None** — use `get_folders` + `get_actors_in_folder`. |

Empty `Developers/felixkupis/Collections/`. First-party scratch folder.

`experimental.*` (Python / toolsets / Scene Graph) in `.uefnproject` is **Unverified**. Content-side: **3** `.verse`, **no** `__ExternalEntities__`. Plugin `EnableSceneGraph: true` with no Scene Graph entities.

Largest file well under GHES **100 MB**. Whole folder ~**110 MB**.

## Level organisation (**Observed** — live `get_folders` + folder counts)

```
Billboards/                           48 — 24 Device_Billboard_V2 (syllabus) + 24 authored Billboard_C backs
Gameplay/                              8 + Capture 2, Class 1, Conversation 4, Hero 3, Patrol 20
Lighting/                              3 — DSD + unbound PPV + ExponentialHeightFog
ZoneA/Environment                   2208
ZoneA/SpawnPoints/…                 Rebels 6 pads + 2 NPC; Empire 6 pads + 2 NPC
ZoneB/Environment                   2914
ZoneB/Vehicles                         8
ZoneC/Environment                      0 leftover empty folders
ZoneC/Vehicles                        12
EmpireStronghold/                   1227
Environment/                        1181 + Border/EnemyFaction tents/Rocks/Vista
Decals/                               76 + CrashedShip 13
Background/                            2
Props/                                 5
(unfoldered)                           HUD Controller, IslandSettings, Hide on Start, 4 Ascender, Grind Rail
```

Syllabus boards are readable on `Device_Billboard_V2.text` after `load_actors`. Authored `Billboard_C` has no `text`.

## Device inventory (the design)

| Count | Class | Role |
|---|---|---|
| 60 | `AIPatrolPath_Device` | Empire / AllyFaction groups per zone |
| 24 | `Device_Billboard_V2` | Editor syllabus (Welcome, Capture, Hero, Vehicles, …) |
| 24 | `Billboard` (authored BP) | Mesh backs; `Hide_on_Start` hides **24** props |
| 24 | `BP_Creative_Player_Spawner_Prop` | Team-filtered pads; `useAsIslandStart=false`; docs **25** |
| 21 | `PlaysetRoot` | First Order / desert prefabs (OFPA **30**) |
| 14 | `Device_GuardSpawner_V2` | `guardType=Shadow`; team index 1/2; patrol groups |
| 12 | `Device_Teleporter_V2` | Conversation destinations |
| 8 | `CRD_SpawnerDevice_CloverDog` | AAT; Enable on capture |
| 6 | `Device_HUDMessage_V2` | Contested + capture; copy-pasted message |
| 4 | `Device_HeroDevice` | HeroA (wired to A), HeroB (wired to B), VillainA + VillainC **unwired** |
| 4 | `Device_Conversation` | Field Operator respawn picker |
| 4 | `CRD_SpawnerDevice_CopperPuck` + 4 `IronPuck` | Turbolaser (docs turret — **0** live) |
| 4 | `Device_Ascender` | Traversal after zone A |
| 3 | `Device_CaptureArea_V2` | Sequential A→B→C |
| 3 | `Device_Beacon_V2` | Objective ping; only A Always |
| 2 | `Device_DisguiseDevice` | Guard (Empire) / BravePebble Dust (Rebels), `applyDisguiseOnPlayerSpawn=true` |
| 2 | `Device_GameEnd` | Timer → Empire; Zone C capture → Rebels |
| 2 | `CRD_SpawnerDevice_THEME_VEHICLE_A` / `THEME_VEHICLE_B` | starfighter / Fighter, health **600** |
| 1 | `Device_Timer_V2` | **900s**, start at game start |
| 1 | `Device_HUDController_V2` | Widgets-folder PlayerInfo + Equipped + Quickbar **on** |
| 1 | `Device_ClassDesigner_V2` | Slot 1; `buildingMode=All` (island `allowBuilding=None`) |
| 1 | `Device_Switch_V2` | `S_Dialogue_GameStart` gates Conversation until A falls |
| 1 | `VerseDevice` | `Hide_on_Start` only |
| 1 | `BP_DaySequenceModifier_Device` | Priority **2000**, `FixedTime` **9.5**, `overrideFog=true` |
| 1 | `Device_GrindRail` | Extra traversal |
| 0 | turret / ArmoredHero / HeroD Hero / `Swap_assets` / `vehicle_mod_zone` | docs or leftover Verse |

## Verse (3 files)

Copied to `snippets/verse/`.

| File | Class | Placed? |
|---|---|---|
| `Hide_on_Start.verse` | `Hide_on_Start` | **Yes** — hides `PropsToHide` (**24**) |
| `Swap_assets_on_capture.verse` | `asset_swap_device` | **No** — mutator-zone cycles `[]creative_prop`; filename ≠ capture |
| `Vehicle_Speed_Mutator.verse` | `vehicle_mod_zone` | **No** — movement mod only if agent is in a vehicle |

Docs: “this sample features no Verse devices.” Live: one placed device. Train on snippets; the game mode is classic bindings.

## Authored kit (**Observed** — `assets.txt` + live)

```
Content/DIalogue/          CD_Empire_ZB/ZC_Spawn, CD_Rebels_Z/ZB/ZC_Spawn  (folder typo)
Content/UI/Widgets/
  Conversation/            Lightside / Darkside / Neutral + UW_BTN_Conversation
  HUDMessage/              Darkside / Lightside / Empire|Rebels OnCaptureStart|OnControlContested
  PlayerHUD/               UW_HUD_{PlayerInfo,Equipped,Quickbar,PlayerVitals,TeamVitals}
  Popups/                  Lightside / Darkside / Neutral + UW_BTN_Popups
Content/CD_Guard    leftover CD (Guards use CID cosmetics, not this asset)
Content/Akita_Template2_Test_sharedassets/  Layer_2–6_LayerInfo leftovers
Content/Materials/         blowing/flowing sand, fog cards, decals
Content/VFX/               NS_BurningSmoke
Content/Props/Billboard    authored billboard mesh BP
```

**Tier 2 mounts:** `/THEME_PACK` 14939, `/SPACE_THEME_A_Comp` 8467, `/SPACE_THEME_A_Assets` 5093, `/ENV_PACK_Biome_*`, `/ENV_SPACE_THEME_UI_Assets`, `/ENV_SWCosmetics`, `/SPACE_THEME_HeroTransformations`, `/CRD_HeroDevice`, `/CRD_DisguiseDevice`. Small `/SPACE_THEME_C_*` bleed. No Fab `.uref`.

## What this sample actually teaches

1. **Sequential capture is device `OnControlChange`, not Verse weights.** Only A starts capturable; A `AllowCapture`s B; B `AllowCapture`s C. See techniques/sequential-capture-unlock.md. Contrast Deserted’s three live points.
2. **Respawn picker = Conversation bank + Teleporter, not extra pads.** techniques/conversation-respawn-picker.md. Lightside widget for Rebels, Darkside for Empire.
3. **Team identity is Disguise-on-spawn, not a class selector.** Empire Guard AGID; Rebels BravePebble Dust. Break-on-attack **Off**.
4. **Hero is `ApplyTransformationToInstigator` on the capturing player.** Bindings: A→HeroA (`AGID_HeroTransformation_CH`), B→HeroB (`RS`). VillainA/VillainC placed, **unwired**. Scalers live **0** (docs’ health/speed prose is the Hero catalog, not these instances).
5. **Custom UI is Narrative’s `Content/UI/Widgets` tree plus a `HUDMessage/` folder.** HUD Controller overrides PlayerInfo **and** Equipped + Quickbar (IceBase-like), `showTeamInfo=DoNotOverride` (PvP). ExtraAssets stays three. `get_referencers` on `UW_HUD_PlayerInfo` returns **[]**.
6. **Timer 900s + two End Games is the win table.** Empire: timer `OnSuccess`. Rebels: Zone C `OnControlChange`.
7. **The battlefield is `/SPACE_THEME_A_*` + HarshTone/IcedTea catalog, not 129 authored rooms.** Leftover `Akita_Template2_Test_sharedassets` LayerInfos; landscape MI is stock `/ENV_PACK_Biome_Assets/…/ENV_PACK_MI_Biome_Desert_Landscape_A`.
8. **Guards are capture-gated Shadow spawners with CID cosmetics, not IceBase volume-swap.** Empire `CID_613` Columbus; Rebels `Character_BravePebble_A2`. `allowInfiniteSpawn=true`, `spawnCount=4`.

## Official tutorial vs shipped

**External** dedicated page + hub + Custom UI (headed Chrome 2026-09-15). Verbatim: `_raw/team_conquest_sample/tutorial.txt`. Spreadsheet: no row. `docsUrl` empty.

### Lines up

| Tutorial | Shipped |
|---|---|
| theme Templates → Team Conquest; 12 / 2 teams Split Evenly | picker; live Island Settings |
| Rebels capture, Empire defends; 3 Capture Areas | `canBeCapturedByTeam` 1 / `startingTeam` 2 |
| Timer 900s, start at game start, no interact | live Timer |
| Empire wins on timer; Rebels on last capture | two End Games |
| Only first zone capturable at start | A `captureAllowedOnGameStart=true`; B/C false |
| Spline boundary; No Neutralize; Gameplay only | live |
| Disguise on spawn; Conversation Field Operator | 2 Disguise + 4 Conversation |
| 60 patrol / 14 Guard / 12 Teleporter / 8 AAT / 4 Ascender | live counts |
| HeroA then HeroB on capture | bindings A→CH, B→RS |
| Widgets under `UI > Widgets > Conversation` / `HUDMessage` | authored tree |
| Billboard syllabus | 24 `Device_Billboard_V2` texts |

### Does not line up

| Tutorial | Shipped |
|---|---|
| **No Verse devices** | **3** `.verse`; **1** placed `Hide_on_Start` |
| 6 Hero (incl. ArmoredHero, HeroD) | **4** Hero; ArmoredHero/HeroD **0**; VillainA/VillainC unwired |
| Zone A HeroA / Zone B HeroB (labels) | Bindings match; **labels** `HD_ZoneB_Chewwy` / `HD_ZoneC_HeroB` |
| 4 turret turrets | **0** turret; 8 Turbolaser spawners |
| 25 Player Spawners | **24** |
| Respawn Time **5s** | live **1** |
| Accent Color **Blue**; Progress Decay **No Decay** | **Team Relationship Color**; **Default** |
| Control Time 15s all | Zone C **25** |
| Dialogue folder | **`DIalogue/`** |
| Player icon left off HUD Controller; set priority **High** | PlayerInfo override **on**; Equipped + Quickbar **on**; priority **Normal** |
| Binary suns / DSA Binary Sun | DSD `FixedTime` **9.5**, stock `MI_DaySequenceSkydome` (same DesertTown drift) |
| HUD Message copy per event | all six `message` = **“Resistance now controls the First Town”** |
| `Swap_assets_on_capture` as capture logic | file is a **mutator-zone prop cycle**; unplaced |

Training rule: shipped bindings + `snippets/` win. Tutorial *prose* is the syllabus (billboards match that outline). Do not emit “no Verse” / “6 heroes” / “turret” / “respawn 5s”.

## Negative findings

- Name **Team Conquest** / domination does **not** mean Deserted-style weighted pads or periodic score.
- `Swap_assets_on_capture.verse` does not subscribe to capture — it subscribes to `MutatorZone.AgentEntersEvent`.
- `Content/CD_Guard` is not what Guard spawners reference (live cosmetics are BR CID / BravePebble).
- ZoneC/Environment and several Patrol/Spawn subfolders are **empty leftovers**.
- HUD Message widgets exist (`OnCaptureStart` / `OnControlContested`) but the `message` field was never localized per event.
- Class Designer `buildingMode=All` vs Island `allowBuilding=None`.

## Capture backlog

Waived: Sequencer keyframes (none needed); Blueprint event graphs on authored `Billboard_C`; exact `propsToHide` actor labels (24 creative_prop slots); DBNO device (OFPA 1, `find_actors` 0). Island `eliminatedPlayersItems` name unknown — docs Keep, unverified live.

## Live notes

- Cursor `user-unreal-mcp` was down; `tools/mcp.py` on `:8000` worked.
- Mount `/6ce09c8f-a624-44d0-9d9a-efcd7b110050` from `ListFiles("")`.
- `get_asset_tags` wants `asset_path` string, not `asset.refPath`.
- Billboard `text` readable after `load_actors`; docs’ “reload if missing” matches unloaded OFPA.
- Did not flip experimental flags. Did not switch maps.
