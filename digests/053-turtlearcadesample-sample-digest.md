---
id: digest-053
kind: digest
---

# turtle_arcade_sample — sample digest

> **Extractor provenance.** `tools/extract-template.py` (Pass A) + UnrealMCP on
> `localhost:8000` (Pass B). Default level
> `/4e87564d-2aa4-4e4d-8a38-4b2108e27d89/turtle_arcade_sample` (**Observed**).
> Hub: TURTLE Arcade Template.
> `docsUrl` empty; spreadsheet Documentation=Yes. Compat **43.00**. `featureSet`: **CustomGameplayFeature**.
> On-disk folder / title is **`turtle_arcade_sample`** (missing `e`) — shipped state, like `verse_proceduarl_building`.

| | |
|---|---|
| Classification | **theme playable sidescroller** — lobby hideout → streets → sewers → construction → BossEnemy |
| Content type | **catalog wrap + small HUD/VFX kit** (110 authored vs 4141 OFPA; city/sewer mounts) |
| Verse | **1 file** (`multiplayer_teleporter.verse`), **5** placed copies |
| Fab `.uref` | **0** |

## What it is

The picker name is **Arcade**. The island is a **4-player beat-em-up course**, not a cabinet minigame: Fixed Angle camera + Side Scrolling Controls, EnemyFaction / RobotEnemy / BossEnemy encounters, pizza supply drops, TURTLE sewer-grate travel. Teaching that survives deletion is the device split (TURTLE spawner vs NPC definition), the tracker-gated arena bus, the Verse gather, and in-project HUD widgets.

## Scale (**Observed** — `summary.txt` + live)

| | |
|---|---|
| Authored assets | **110** |
| OFPA | **4141** |
| Maps | `turtle_arcade_sample.umap` (only) |
| Verse files | **1** (`VerseToolset.ListFiles` on `/4e87564d-…` — do not recurse `__ExternalActors__`) |
| Verse devices placed | **5** (same class) |
| Data layers | **0** |
| Fab `.uref` | **0** |
| Compatibility | **43.00** |
| Matchmaking | **4** (`maxPlayers=4`, `teamSize=Squads`) |
| Streaming | Partitioned + actor folders + OFPA. **`LevelHasStreamingDisabled` omitted**; Travel chapter: streaming on so the intro cine is preceded by a teleport |

Empty `Developers/felixkupis/Collections/`. First-party scratch folder.

`experimental.*` (Python / toolsets / Scene Graph) in `.uefnproject` is **Unverified** — do not treat as shipped. Content-side: **one** `.verse`, **no** `__ExternalEntities__`.

Largest file in the project tree is well under the GHES **100 MB** cap (whole folder ~77 MB).

## Level organisation (**Observed** — live `get_folders`)

```
Billboards/                 14 Device_Billboard_V2 — syllabus
Cinematics/                 5 CineCameraActor + explosion / wall-dust VFX Spawners
Gameplay/                   HUD Controllers ×2, Game End, Fade HUD Message, main barriers
  Lobby/                    4 player starts, 4 granters, 4 item spawners, 30 s timer, radio
  Intro Area/               teleporter, Verse gather, cine, Perspective + Sidescroller
  First Encounter/          3 TURTLE spawners, tracker=3, Fixed Point + Perspective, pizza drop
  Sewers/                   2 TURTLE tunnels, 3 RobotEnemy NPC spawners, tracker=9, steam/water volumes
  Construction Area/        hydrant damage + SFX, pizza drop, Perspective, teleporter
  Boss Fight/               cine + Character dummy, BossEnemy TURTLE spawner, 2 Foot supports, End Game
Level/                      city kitbash (2887 actors); Island Settings; customizable lights
  LobbyHideout/             hideout dressing
  Sewers/                   sewer kitbash
Lighting/                   Day Sequence, PPVs, Volume_Sewer / Volume_Street*, LS_LowSunlight On/Off
VFXs/                       hydrant + steam VFX Spawners
```

Billboard captions: `_raw/turtle_arcade_sample/billboard-texts.md`. `text` is on the **actor**. Actors reported `bIsSpatiallyLoaded=false`; `load_actors` was not required this pass.

## Device inventory (the design)

| Count | Class | Role |
|---|---|---|
| 16 | `Device_Barrier_V2_Placed` | arena / cine locks |
| 14 | `Device_Billboard_V2` | syllabus |
| 11 | `Device_VFXSpawner_V2` | steam, hydrant, cine dust |
| 9 | `Device_Trigger_V2` | encounter / fade / gather cues |
| 6 | `Device_Prop_PlayerCheckpoint_V2` | per-area respawn |
| 6 | `Device_TURTLE_CharacterSpawner` | EnemyFaction + BossEnemy |
| 6 | `BP_Device_DamageVolume_V2` | steam 25/2s, hydrant 50/1s, water Elimination |
| 5 | `VerseDevice` | `multiplayer_teleporter` |
| 4 | `Device_CameraModes_Perspective` | Fixed Angle (docs said 5) |
| 4 | `Device_CinematicSequence` | CityIntroSweep, BossFightIntro, LS_LowSunlight On/Off |
| 4 | `Device_ItemGranter_V2` | lobby turtle weapons (per player start) |
| 4 | `Device_ItemSpawner_V3` | lobby weapon pickups |
| 4 | `BP_Creative_Player_Spawner_Prop` | lobby |
| 4 | `Device_Timer_V2` | lobby 30 s, fade, sewer exit, … |
| 3 | `Device_CameraModes_FixedPoint` | arenas |
| 3 | `Device_CharacterSpawner` | Mousers |
| 3 | `Device_Tracker_V2` | FE=3, RobotEnemy=9, Boss=10 leftover |
| 3 | `Device_Radio_V2` | lobby / intro / boss |
| 3 | `Device_Volume` | lighting zones |
| 2 | `Device_CustomControls_Sidescroller` | streets + sewer re-add |
| 2 | `Device_HUDController_V2` | arcade HUD vs cine-off |
| 2 | `Device_HidingProp_Sewer_Tunnel` | grate travel |
| 2 | `Device_LinedNotebookSupplyDropSpawner` | pizza |
| 2 | `Device_Teleporter_V2` | lobby → intro; construction |
| 2 | `Device_PropManipulation_V2` | proceed arrows |
| 1 | `Device_Character_V2` | cine BossEnemy dummy (`Character_PeacefulPoem`) |
| 1 | `Device_GameEnd` | BossEnemy elim |
| 1 | `Device_HUDMessage_V2` | fade (`FadeHUDDevice`) |
| 1 | `Device_CRD_AudioPlayer` | hydrant SFX |
| 1 | Island Settings / Day Sequence / player spawners | session |

## Authored kit (**Observed** — `assets.txt` + referencers)

```
Content/UI/Widgets/     UW_HUD_PlayerInfoBlock, UW_HUD_PlayerInfoStack (live HUD override), WBP_TransitionUI
Content/UI/Materials/   M_UI_PlayerCard, progress bars (use UIMaterialLab MFs)
Content/VFXs/           NS_VFX_FireHydrant / HydrantSplash / DamagingSteam / ExplosionDust
Content/Cinematics/     CityIntroSweep, BossFightIntro, LS_LowSunlightOn/Off
Content/NPCCharacterDefinition_RobotEnemy   live on 3 sewer spawners (Type=BP_Chompy_CharacterType)
Content/Mannequin/      0 referencers (leftover)
```

Root `BossFightIntro` / `IntroSweep` are **ObjectRedirector**s into `Cinematics/`. `NPCCharacterDefinition_Chompy` and `NewNPCCharacterDefinition_Chompy` redirect onto RobotEnemy. `NewWidgetBlueprint` redirects onto `WBP_TransitionUI`.

HUD that actually shows is **in this project**, not ExtraAssets. Live override: `/4e87564d-…/UI/Widgets/UW_HUD_PlayerInfoStack`. HideElement on minimap + build menu + team info; health/shield **DoNotOverride** (custom widgets replace them). That **contradicts** treating Social/Walker ExtraAssets as the only theme-HUD recipe — see Cross-template notes.

## What this sample actually teaches

1. **Squad gather is Verse `TeleportTo` + a 100 uu axis offset**, not more Creative Teleporters. techniques/multiplayer-teleporter-offset.md. Five devices, one class; `@editable` trigger, empty Creative bindings.
2. **Foot/BossEnemy = TURTLE Character Spawner enum; Mousers = NPC Character Definition.** techniques/TURTLE-character-spawner.md. Live Foot `EnemyFaction` 75 HP; BossEnemy 300 HP; RobotEnemy Type `BP_Chompy_CharacterType`, 100 HP, `nPCBehaviorScript` empty.
3. **Sidescroller = Perspective Fixed Angle + Side Scrolling Controls; arenas swap to Fixed Point.** techniques/sidescroller-camera-controls.md. Streets pitch **-15** FOV 100; sewers **-20** FOV 90; Ease-In-Out. Second sidescroller on tunnel exit.
4. **Encounters are Trigger Enable + Tracker Increment, not a round manager.** techniques/tracker-gated-encounter.md. FE target **3**; Mousers **9**; boss win is BossEnemy `On Eliminated` → End Game.
5. **Sewer grate is hiding-prop Hidden Travel**, not a Teleporter. techniques/sewer-tunnel-hidden-travel.md. Groups 0→2 and 2→1; `numberOfOccupants=4`.
6. **Arcade HUD is authored UMG + HUD Controller viewmodels in-project.** Docs walk `UW_HUD_PlayerInfoBlock` → stack → Player Info Widget Override. Bindings: HUD Controller Player Info Viewmodel (`Progress` scalar, `Texture` avatar). Deeper UMG: umg-viewmodel-device-binding.md. Second HUD Controller hides the HUD for cines (`showHUD=HideElement`).
7. **Hazards are Damage Volume + VFX Spawner** (authored Niagara in `Content/VFXs/`). Steam **25 / 2 s**; hydrant **50 / 1 s**; water **Elimination**. Billboard10 still says “VFX Creator” — live and the encounters chapter say **Spawner**.

## Official tutorial vs shipped

**External hub + 4 companion chapters** (headed Chrome 2026-09-14; Sign in chrome still visible, article body readable):

1. TURTLE Arcade Template
2. TURTLE Custom UI: Player Info
3. TURTLE Cameras and Controls
4. TURTLE Enemy Encounters and Obstacles
5. TURTLE Travel Between Areas

Verbatim capture: `_raw/turtle_arcade_sample/tutorial.txt`. Spreadsheet `docsUrl` empty; hub exists anyway. theme Rules / Working with TURTLE Islands / City Starter PP are **not** this island’s walkthrough.

### Lines up

| Tutorial | Shipped |
|---|---|
| theme Templates → Arcade | Project Browser; `featureSet` CustomGameplayFeature |
| 4-player co-op sidescroller | `maxPlayers=4`, `teamSize=Squads` |
| Lobby timer then cine then teleport | Timer `duration=30`, `urgencyModeTime=1`, CityIntroSweep, Verse gather |
| 2× Side Scrolling Controls, 3× Fixed Point | Exact |
| Ease-In-Out camera transitions | Intro Perspective 0.5 / 0.5; FE Fixed Point 0.2 in |
| Second sidescroller on sewer tunnel exit | `OnStopHiding` → `AddTo` |
| FE: 3 TURTLE spawners, tracker, 2 barriers, Fixed Point, Prop Manipulator | Exact on those rows |
| Three Foot Elite then pizza drop | Tracker **3**; supply drop `When Complete` |
| Steam 25 dmg / 2 s; water eliminates | Live Steam **25 / 2 Seconds**; Water `damageType=Elimination` |
| RobotEnemy NPC spawners, infinite 1 s, tracker **9** | Exact |
| Boss: cine, extra barrier, 2 Foot + BossEnemy, End Game on BossEnemy | Support1=`BossEnemy`; Support2–3=`EnemyFaction`; End Game on Support1 elim |
| TURTLE Sewer device for grate travel | `Device_HidingProp_Sewer_Tunnel` |
| Fade HUD then timer then Verse gather on sewer leave | FadeHUDDevice + ExitTeleportTrigger + Exit MultiplayerTeleporter |
| HUD: Show HUD Yes, Show Team Info No; override = stack widget | Live HideElement team/minimap/build; override `UW_HUD_PlayerInfoStack` |
| `multiplayer_teleporter` class + TriggerDevice + TeleportTo | Snippet matches |

### Does not line up

| Tutorial | Shipped |
|---|---|
| **5** Fixed Angle cameras | **4** Perspective devices |
| First Encounter: **1** checkpoint | **2** (`Checkpoint` + `Checkpoint2`) |
| Travel chapter: offset on the **X** axis | Live + snippet default **Y** `{0,1,0}` |
| Create a new widget named `HUDInfoStack` | Shipped `UW_HUD_PlayerInfoStack` |
| Health bar Multiplier **0.01**, then later **0.1** | Docs self-contradict; train 0–1 scalar / 0–100 health ⇒ **0.01** |
| Health binding steps say `Profile_Image > Brush` | Copy-paste; widget is `Health_Image` |
| Hub typo “TNMT Arcade” | — |
| Billboard10: “VFX **Creator**” for steam | Live + encounters chapter: **VFX Spawner** |
| Folder / project title `teenage_…` | Disk `turtle_arcade_sample` |

Training rule: snippets + live island win. Do not emit X-axis gather offset, 5 Fixed Angle cameras, or “VFX Creator for steam.”

## Negative findings

- **Name implies an arcade cabinet; files show a linear 4-player beat-em-up.** No cabinet devices, no high-score Verse, matchmaking 4.
- **`NPCCharacterDefinition_Chompy` is not a second enemy.** Redirector → RobotEnemy. The RobotEnemy asset’s Type subobject is still named `BP_Chompy_CharacterType`.
- **Boss Tracker (`targetValue=10`, assign on game start) is not the win condition.** End Game is BossEnemy elim.
- **Mannequin anims have 0 referencers.** Do not cite as a worked animation example.
- **Cine BossEnemy is `Character_PeacefulPoem` (BR cosmetic), not the TURTLE enum.** The fight uses `Device_TURTLE_CharacterSpawner` `BossEnemy`.
- **`GetDeviceProperties` returns Verse subobject paths**, not placed triggers — expected; `savedActor` closes it.

## Capture backlog

| Item | Status |
|---|---|
| Exact lobby granter WID rows (itemToGrant is an index) | **Waived** — labels + Changing Booth + docs “turtle weapons” |
| Sequencer keyframes (CityIntroSweep / BossFightIntro) | **Waived** — unreachable; sequence asset paths captured |
| PIE: Support2/3 `enabledAtGameStart=true` extra Foot before cine | **Waived** — bindings also `Spawn` them from BossEnemy `On Spawned` |
| Lighting volume → Day Sequence / PPV wiring | **Waived** — not the syllabus |
| UMG view-binding graph inside `UW_HUD_PlayerInfoBlock` | **Waived** — docs chapter is the walkthrough; live override path is the ground truth |

## Live notes

- Mount `/4e87564d-2aa4-4e4d-8a38-4b2108e27d89`. Project Verse path `/epictemplates@fortnite.com/turtle_arcade_sample`.
- Did **not** `load_level` away from the default map.
- Billboard `text` is on the actor (same as VFX / Social / Walker).
- `GetDeviceProperties` on the five Verse devices: `triggerDevice.refPath` is `…multiplayer_teleporter_0.__verse_0x62571B8A_TriggerDevice`; `savedActor` is the Creative trigger. Note as campaign trap, not a missing binding.
