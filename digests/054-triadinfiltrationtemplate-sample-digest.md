---
id: digest-054
kind: digest
---

# triad_infiltration_template — sample digest

> Offline `_raw/triad_infiltration_template/`. Live
> `/cbc2511a-c127-4b75-a8c2-5d947077a982/triad_infiltration_template` (**Observed**).
> Docs: 10-chapter hub. Prerequisites: Team Multiplayer Balancing + Team Elimination (separate pages, not this island).

| | |
|---|---|
| Classification | **3-way asymmetric capture** — relative team caps, infiltrator Hide/flicker, floating objective beacon |
| Content type | catalog wrap: **3 authored vs 2796 placed** (jungle temple) |

## Scale

| | |
|---|---|
| Authored | 3 |
| OFPA | **2796** |
| Verse | **4** |
| VerseDevices | 4 |
| Matchmaking | **10** players, **3** teams |
| Compatibility | 43.00 |

Spreadsheet: *teaches an important asymmetric gameplay concept using verse so may be useful to keep.*

Empty `Developers/felixkupis/Collections/`. Nineteenth first-party hit. **Wave 1 complete.**

## What it is *for*

Three teams, different jobs and **relative** sizes (not even split):

| Team (Island Settings order) | Default cap | Job |
|---|---|---|
| Infiltrators `[0]` | **2** | Invisible; steal Defender objective; flicker on damage |
| Attackers `[1]` | **4** | Steal Defender flag; stop Infiltrators |
| Defenders `[2]` | **4** | Hold until time; stop both |

Cap-difference balancer: assign each player to `argmax (Max − Current)`. Island max players should equal sum of caps (comment in Verse). Per-team weapons via granter **index = team index**. Floating `creative_prop` + map indicator follow the carrier.

## Device inventory (the design)

| Count | Class | Role |
|---|---|---|
| 2 | capture item spawner | Defender / Attacker objectives |
| 2 | capture area | drop-off |
| 3 | team settings | three teams |
| 3 | item granter | per-team kit |
| 3 | teleporter | team start (order = AllTeams) |
| 24 | player spawners | 10P |
| 4 | VerseDevice | triad / invisibility / capture / cinematic |
| 2 | score manager | capture award |
| 54 | billboards | captions |
| rest | Ancients kitbash | throwaway |

## Verse (4 files)

Copied to `snippets/verse/`.

| File | Job |
|---|---|
| `triad_infiltration.verse` | cap map, `FindTeamWithLargestDifference`, grant-by-index, JIP Respawn at team teleporter |
| `invisibility.verse` | `Hide` infiltrators; `DamagedEvent` → flicker `VulnerableSeconds` / `FlickerRateSeconds`; optional shared with teammates |
| `item_capture_manager.verse` | PickedUp → `FollowCharacter` MoveTo loop; Dropped `race` ReturnTime vs pick-up; Captured score |
| `design_cinematic_flythrough.verse` | optional intro (same family as Prop Hunt / Stronghold) |

Shipped comments: `Maxmimum`, `Rerfernece` — don’t copy as names.

## What this sample actually teaches

1. **Asymmetric fill is “largest remaining cap”, not even split.** `[team]int` maxima; Shuffle then `BalancePlayer`. See techniques/asymmetric-cap-balancer.md.
2. **Team index = device-array index** for granters and teleporters. Island Settings team order **is** the contract.
3. **Invisibility is `fort_character.Hide/Show`**, not a mutator. Flicker loop decrements a `[agent]float` map; `IsFlickering` skips stacking. See techniques/infiltrator-hide-flicker.md.
4. **Objective beacon follows the carrier** (`creative_prop.MoveTo` + map indicator, one tick `Sleep(0.0)`); return via `race` timer vs PickedUp.
5. **JIP:** `PlayerAddedEvent` → `BalancePlayer` → `Respawn` at that team’s teleporter → if infiltrator, `OnInfiltratorJoined`.

## Official tutorial vs shipped

**External** hub: Triad Infiltration

10 chapters (`triad-infiltration-0N-…-in-verse`). Ch.5 slug is `…-making-players-invisible-in-verse` (no `05`). Hub also lists **Team Multiplayer Balancing** and **Team Elimination Game** as required reading — those are *other* tutorials, not files in this repo.

`docsUrl` empty. Spreadsheet `/en-us/uefn/triad-infiltration-in-verse`.

### Lines up

| Tutorial | Shipped |
|---|---|
| 2 / 4 / 4 caps; three `?team` | `MaximumInfiltrators/Attackers/Defenders` |
| `FindTeamWithLargestDifference` + `AddToTeam` | same algorithm |
| Hide infiltrators; flicker on damage | `invisibility_manager` |
| Grant on spawn by team index | `WeaponGranters[TeamIndex]` |
| Follow-prop + map indicator | `item_capture_manager` |
| JIP rebalance | `OnPlayerAdded` |

### Does not line up

| Topic | Tutorial | Shipped |
|---|---|---|
| Class name (ch.3 paste) | `triad_infiltration_game` | **`triad_infiltration`** |
| Teams variable | `Teams` | **`AllTeams`** |
| Teleport after balance | docs `Respawn` | shipped `TeamTeleporter.Teleport` (JIP uses `Respawn`) |
| Ch.10 “complete code” | one dump | **4 files** — snippets win |

Train on snippets. Emitting `triad_infiltration_game` is doc regurgitation.

## Negative findings

- 2796 jungle actors are not the lesson.
- Team order in Island Settings **must** match arrays; silent wrong-kit if swapped.
- Scratch `Developers/` folder.

## Capture backlog

1. Live cap vs Island Settings max 10. *Waive — comment says sum the maxima.*
2. Billboard 54 captions. *Waive.*

## Live notes

Level confirmed. Streaming off.
