---
id: digest-055
kind: digest
---

# user_interfaces_template — sample digest

> Offline: `tools/extract-template.py` → `_raw/user_interfaces_template/`.
> Live: UnrealMCP `localhost:8000`, loaded level
> `/6b3c2b58-e3e6-4852-8989-ac4e92070773/user_interfaces_template` (**Observed**).
> Experimental Scene Graph / experimental-content flags in `.uefnproject` are
> **Unverified** as shipped. Content-side: 42 `.verse` files, `EnableSceneGraph`
> true in plugin, one Scene Graph in-world UMG example.

| | |
|---|---|
| Classification | **Walkable UI curriculum** — Creative-device HUD, UMG widgets, Verse UI, UI materials |
| Content type | *kit* — 1278 authored assets (647 textures, 362 MI, 114 UW, 61 MF) |
| Official docs | Hub + 5 chapters. See **Official tutorial vs shipped**. `docsUrl` empty in descriptor; spreadsheet has the hub. |

## Scale

**Observed** — `summary.txt` + live `get_asset_tags` / `get_folders`.

| | |
|---|---|
| Authored assets | **1278** |
| OFPA actors | **515** |
| Verse files | **42** |
| Fab `.uref` | 0 |
| Distinct actor classes | 53 |
| Live outliner folders | 27 (Hall 01–05, Room 00/04/09/11, Environment, Track) |
| Data layers | 0 |
| `LevelHasStreamingDisabled` | **1** |
| Compatibility | 43.00 |
| VerseAPI / VerseVersion | 1 / 1 |
| `EnableVerseAssetReflection` | true |
| Localization | `ugcLocalization` 13 cultures + auto-loc untranslated |
| Matchmaking | 1–16, Public queue |

Kit, not catalog. Question is *what it teaches*. Authored mass is UI: `T_` / `MI_` / `UW_` / `MF_`. Hall geometry is ContentHall + road/garage props.

## What this sample is *for*

A **multi-hall UI course**. Tutorial says three halls + keys 1–4; **live outliner is five halls** (**Observed** `SceneTools.get_folders`):

| Folder | Lesson |
|---|---|
| Hall 01 / 05 — Creative Devices + Feature Examples | Bind UMG User Widgets to Creative devices (HUD Message, Pop-Up, HUD Controller, Stat Creator, Skilled Interaction, Timer, Conversation) |
| Hall 02 / 04 — Verse UI + UI Art | Pure Verse widgets (`button`/`button_quiet`/`material_block`/`text_block`) and the material collection |
| Hall 03 — SGUI | Scene Graph in-world UMG (`ui_in_world_umg_device` + entity widget) |
| Room 11 / Environment/Menu | Full Verse menu (main / settings / upgrade / vehicle / countdown) |
| Track / Garage | Vehicle-select demo set dressing |

Volume devices (65) are the player trigger for almost every booth — same pattern as Scene Graph.

## Project configuration

**Observed** — `config.txt`.

- Mount GUID `6b3c2b58-e3e6-4852-8989-ac4e92070773`; `projectId` `1c55a407-058b-4ea2-97df-c54c2a8f11c3`
- `docsUrl` **empty**. Spreadsheet link is the hub.
- Plugin: `CanContainVerse`, `EnableSceneGraph`, `EnableVerseAssetReflection`, `VerseAPI` 1
- `experimental.experimentalContent.bIsExperimentalContentAllowed` true — **Unverified** (operator-toggle). In-island transactions / Marketplace APIs in `ui_shop*.verse` are the content-side signal.
- `ui_helper.verse` is a **module stub tree** matching Content folders so generated widget/material classes resolve (`UserInterfaces.VerseUIExamples.CustomButton.Widgets…`).

Empty `Developers/felixkupis/Collections/` ships. Seventh first-party hit.

## Device / actor inventory

**Observed** — `actor-classes.txt`. Design *is* this list.

| Count | Class | Role |
|---|---|---|
| 65 | `Device_Volume` | Booth triggers |
| 64 | `FortStaticMeshActor` | Set dressing |
| 54 | `Device_Billboard_V2` | Captions |
| 49 | `Device_Decal_V2` | Floor marks |
| 46 | `BP_DemoDisplay` | Construction-script signage (same family as intro_to_uefnfk1 / Scene Graph) |
| 25 | `Device_HUDController_V2` | Custom HUD samples |
| 25 | `Device_Teleporter_V2` | Hall navigation |
| 23 | `VerseDevice` | Verse UI devices (matches ~23 `*_device.verse` classes) |
| 15 | `Device_Powerup_Stat` | Drive HUD/stat samples |
| 8 | `Device_CameraModes_FixedPoint` | Bookmarks / room cameras |
| 7 | `BP_DemoRoom` | Rooms |
| 4 | `Device_Conversation` | Conversation bank UI |
| 3 each | HUD Message, Pop-Up Dialog, Timer, Skilled Interaction, Stat Creator | One booth per look (materials / textures / both) |
| 5 | `Device_CinematicSequence` | Menu / conversation intros |
| + | Input Trigger, Item Granter/Remover, vehicles, characters | Feature examples |

**Very many Creative devices, on purpose** — Hall 01 *is* “devices that accept a User Widget.” Verse hall uses volumes + one Verse device per booth.

## Content organisation

```
Content/UserInterfaces/
  CreativeDevicesExamples/   HUDController, HUDMessage, PopupDialog, SkilledInteraction, StatCreator, Timer
  VerseUIExamples/           CustomButton, MaterialBlock, PlayerPortrait, VerseFields, VerseEvents, VirtualPointer, Vehicle, Touchscreen, InputActions
  UIFeatureExamples/         Conversations, Hotbar, InIslandTransactions (Demo01/02), Menu
  UIArtExamples/             MaterialCollection
  SceneGraphUIExamples/      InWorldUMG
  Shared/                    MaterialFunctions, SDF textures, 366 SDF icons
Content/*.verse at root      ui_helper, verse_event_helpers
Content/UserInterfaces/VerseClasses/  mirrors the folder tree as empty Verse modules
```

Widget prefix **`UW_`** (114). Materials **`M_`/`MI_`/`MF_`**. Shared `M_ProgressBar_Basic`, `M_Meter`, `M_DropShadow`, `M_Wave` named in the materials chapter.

## Verse (42 files)

Copied to `snippets/verse/`. Clusters:

| Cluster | Files | Lesson |
|---|---|---|
| Buttons | `ui_simple_button_device`, `ui_custom_button_device`, `ui_custom_button_umg_*`, hotbar | `button{ Slot := button_slot{ Widget := … } }` is the clickable-image pattern |
| Fields / events | `ui_verse_fields_demo_device`, `ui_verse_events_device`, `verse_event_helpers` | Drive UMG from Verse (`set Widget.TextField`); `event(tuple())` via `SubscribeToEvent` race helper |
| Material / text / input | `ui_materialblock_*`, `ui_textsize_device`, `ui_hud_keybind_demo_device`, `ui_touchscreen_*`, `ui_player_portrait_device` | `material_block`, `text_block.SetTextSize`, Enhanced Input → Action Widget |
| Shop | `ui_shop*`, `ui_link_volume_to_shop_device`, `ui_shop_setup_device_template` | Hybrid: UMG for chrome, Verse `overlay`/`stack_box`/`button_loud` for offers; Marketplace `offer` / `BuyOffer` |
| Menu | `ui_menu_device` + `ui_*_page` | Multi-page Verse+UMG menu |
| SGUI | `ui_in_world_umg_device` | Volume flag + Reload input increments in-world UMG on an entity |
| Glue | `ui_helper.verse` | Empty public modules = digest path |

Canonical clickable-image (**Observed** `ui_simple_button_device.verse`):

```verse
VerseButton : button = button{ Slot := button_slot{ Widget := TextWidget } }
VerseButton.OnClick().Subscribe(OnClicked)
PlayerUI.AddWidget(RootOverlay, player_ui_slot{ InputMode := ui_input_mode.All })
```

UMG close path (**Observed** `ui_custom_button_umg_device.verse`): `SubscribeToEvent(Widget.CloseEvent, OnCloseEvent)` — dispatcher is `event(tuple())`, no `.Subscribe` on the event itself.

## What this sample actually teaches

1. **Two HUD stacks, don't mix casually.** Creative devices consume a **User Widget** (or Modal Dialog Variant for Pop-Up / Conversation). Verse UI builds `overlay`/`stack_box`/`button` in code and `AddWidget`s with `ui_input_mode.All`. Hybrid shops exist (`ui_shop`) — UMG for art, Verse for rows/buttons.
2. **Clickable custom visuals = wrap them in base `button` via `button_slot`.** Not `button_quiet` over a texture (hover tint, no style API). See techniques/verse-clickable-button.md.
3. **UMG talks to devices through Viewmodels + View Bindings**, not Verse. HUD Controller Health/Shield → `Set Scalar Parameter` named `Progress` on `M_ProgressBar_Basic` (0–1). Equipped-item override is a second viewmodel. See techniques/umg-viewmodel-device-binding.md.
4. **Verse talks to UMG through Verse fields** (`set DemoWidget.IntField`) or Event Dispatchers (`event(tuple())` + the `SubscribeToEvent` race helper). Custom Fortnite button widgets often **don't** fire OnClicked into Verse — this sample uses a standard path + dispatcher.
5. **`material_block` is the Verse equivalent of an Image+MI.** Parameters become writable from Verse; that's how meters react without a Creative device.
6. **UI materials are User-Interface domain + MI parameters; textures/SDF for shapes materials can't do.** Collection vs custom parent. SDF + SmoothStep for glow/outline/scale without extra maps.
7. **Show UI on `volume_device.AgentEntersEvent`; free the cursor with `ui_input_mode.All`; `SetFocus` the first widget.** Tear down on exit / `PlayerRemovedEvent` and `Cancel()` subscriptions (`verse_event_helpers`).

## Official tutorial vs shipped

**External** hub (canonical, `?lang=en-US`):
User Interfaces Feature Template

Chapters — all live. Hub cards are relative (`/documentation/fortnite/<slug>`); the working absolute form is `/documentation/fortnite/<slug>-in-unreal-editor-for-fortnite?lang=en-US` except Verse Fields, which already has `-in-fortnite`. **Fetcher 404s on `/en-us/uefn/<slug>` are not broken docs.** Browser without Epic SSO hits 403; signed-in browser and the fortnite/… URLs work.

1. User Interface Devices
2. UMG Widgets
3. Verse Fields Examples
4. Material Assets
5. Migrating Template Assets

`docsUrl` in `.uefnproject` is empty. Spreadsheet `Link in Desc` is the hub (`/en-us/uefn/…` form also redirects).

### What lines up

| Tutorial | Live / files |
|---|---|
| Feature-example template; UI under `UserInterfaces/` then per-device MaterialInstances / Textures / Widgets | `folders.txt` |
| Keys 1–4 jump rooms | `Device_CameraModes_FixedPoint` ×8 **Observed** |
| Devices: HUD Message, Pop-Up, HUD Controller (+ Quickbar), Stat Creator, Skilled Interaction, Timer, Conversation | actor histogram 3/3/25/3/3/3/4 |
| User Widget vs Modal Dialog Variant (Pop-Up / Conversation) | docs; Conversation Bank called out |
| Verse utilities at back of hall 2: material_block, text_size, player_input | files `ui_materialblock_*`, `ui_textsize_device`, `ui_hud_keybind_demo_device` exist |
| Three looks per device (materials / textures / both) | three instances of HUD Message / Pop-Up / Timer / SID |
| `UW_HCD01` HUD Controller: Overlay Root, Grid, Shield/Health containers, Viewmodel **Device - HUD Controller Team/Squad Player Info List**, Brush → `Progress` scalar | docs chapter 2 (not live-opened; **External**) |
| Shop: `shop_setup_device_template` (shipped name `ui_shop_setup_device_template`), `ui_shop`, `ui_shop_offer_item`, `link_volume_to_shop_device` | files match; extra **Demo02** (`ui_shop02*`) not in chapter 3 |
| Volume trigger for shop | `ui_link_volume_to_shop_device` + second volume for shop02 |
| Material collection + `material_collection_device`; UI-domain parent; SDF + SmoothStep | `ui_material_collection_device.verse`, `Shared/…/SDF`, `Shared/Icons` (366) |
| `overlay` / `stack_box` / `material_block` / `PopulateOffers` / `BuyOffer` | `ui_shop.verse` |

### What does **not** line up — do not train on the docs here

| Topic | Tutorial | Shipped |
|---|---|---|
| Hall count | “three halls” | **Five** outliner halls (01,02,03 SGUI, 04, 05) + Room 11 menu |
| Verse file names | `materialblock_basic_device.verse` (no `ui_` prefix) | `ui_materialblock_basic_device.verse` (all devices are `ui_*`) |
| Shop setup class | `shop_setup_device_template` / `var Shop` | `ui_shop_setup_device_template` |
| Shop Demo02 | omitted | full second shop (`ui_shop02`, buy button, offers_data) |
| Scene Graph in-world UMG | not in hub/chapter 1–4 | `Hall 03 - SGUI` + `ui_in_world_umg_device.verse` |
| Full Verse menu / vehicle / conversations devices | not in fetched chapters | `ui_menu_*`, `ui_conversations_*`, `ui_vehicle_device` |
| Chapter 3 overlay snippet | `Padding := margin {Left := 824.0, Bottom := 30.0}` on CloseButton | shipped `CreateUI` builds overlays differently (material_block backgrounds + `button_loud` Buy) — treat tutorial snippet as **illustrative**, not paste-ready |
| `verse_event_helpers.SubscribeToEvent` | not documented | required to listen to UMG `event(tuple())` dispatchers |
| Custom-button clickable-image (`button_slot`) | not in fetched chapters | `ui_simple_button_device` / `ui_custom_button_device` — the training gold for Verse clicks |
| Migrating assets | Content Browser → folder → Migrate… → Fortnite Projects → Content. **Only Textures can migrate alone;** device folders must migrate as a set or the widget/MI refs break. Destination is the other project's `Content/`, not a nested UI copy unless you choose one. | Same procedure. Stub `ui_helper.verse` modules still needed if Verse class paths should resolve. |

Typos in docs that must not become identifiers: **Verse UI Utilites**, **Modal Dialogue Variant**, **Assault Riffle**, **SImple** (`M_Texture_SImple`), **text_size propety**.

### Training rule

```
Tutorial prose (device table, viewmodel/Progress binding, material-collection vs custom parent, SDF)
    → valid outline (External)
Tutorial Verse / file names / “three halls”
    → untrusted on conflict
KB snippets/verse/*.verse + this digest’s Observed tables
    → ground truth
```

If both are in the corpus, snippets win. Emitting `materialblock_basic_device` (no `ui_` prefix) or assuming three halls is a doc-regurgitation failure.

## Negative findings (**Observed**)

- Tutorial hall-count and Verse filenames drifted; Scene Graph in-world UMG and the Verse menu are in the project and barely (or not) in the fetched chapters.
- `ui_helper.verse` is empty modules — do not treat it as gameplay.
- World Partition on, streaming off.
- Developer scratch folder ships.
- 65 volumes / 54 billboards / 46 demo displays: the *lesson* is UI, the *noise* is ContentHall navigation. Don't copy the hall.

## Capture backlog

1. **Billboard / BP_DemoDisplay captions** for Verse-button and Verse-fields booths (property values). Ranked high — they are the syllabus like Scene Graph boards. *Open.*
2. **`UW_HCD01` View Bindings live** (`ObjectTools` / UMG toolset) to confirm `Progress` scalar names. Docs are detailed; still Unverified until read. *Open.*
3. **Modal Dialog Variant vs User Widget** on the three Pop-Up devices — asset class of the assigned widget. *Open.*
4. Sequencer in conversation/menu `LS_*` — waive unless a technique needs timing.

Fetcher note (not a template defect): Pass B½ must use `/documentation/fortnite/<slug>-in-unreal-editor-for-fortnite?lang=en-US`, not `/en-us/uefn/<slug>`. Unsigned Chrome gets 403.

## Live-editor notes

Level confirmed. Folders named the curriculum (`Hall 01 - Creative Devices` … `Hall 03 - SGUI`). Map streaming disabled. `VerseToolset.ListFiles("")` returned package roots only (`bRecursive: false`); disk `find-files *.verse` is the 42-file inventory.
