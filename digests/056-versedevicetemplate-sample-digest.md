---
id: digest-056
kind: digest
---

# verse_device_template — sample digest

> Offline: `_raw/verse_device_template/`. Live UnrealMCP level
> `/b2869215-f52d-4092-8a77-5d42d136ef9b/verse_device_template` (**Observed**).
> Official docs: Verse Starter / Verse Commander, 7 chapters. `docsUrl` empty.

| | |
|---|---|
| Classification | **Playable Verse minigame** (Verse Commander) — not a walkable booth hall |
| Content type | kit of 8 Verse files + 5 `@editable` gameboards; 187 authored assets vs **1205** placed (sci-fi city set dressing dominates OFPA) |

## Scale

**Observed**

| | |
|---|---|
| Authored assets | 187 |
| OFPA actors | **1205** (592 static meshes — environment, not the lesson) |
| Verse | **8** |
| Fab | 0 |
| Data layers | 0 |
| `LevelHasStreamingDisabled` | 1 |
| Compatibility | 43.00 |
| `EnableVerseAssetReflection` | true |
| Plugin `Hidden` | true (**Observed** descriptor — template may not show in some browsers) |
| Live folders | `Verse_Commander_MiniGame/Gameboard1` … `Gameboard5` |

Authored-vs-placed looks like a catalog; it isn’t. The **game** is one `verse_commander_minigame` device + NPC behavior + 5 concrete `gameboard`s. The 1205 actors are a concert/sci-fi wrap.

## What this sample is *for*

**Verse Commander:** queue Forward / Turn Left / Turn Right on a custom Verse HUD, Execute, an NPC walks a tile grid, obstacles dissolve, five boards in order.

Spreadsheet note: *good quality quick verse tutorial, but requires the player to follow the documentation or they won’t know what to do.* Hub agrees: you must add gameplay tag `verse_commander_minigame_tag` on the minigame device’s VerseTagMarkup before play (**External** — Getting Started). That is a **shipped gotcha**, not optional flavor.

Live outliner (**Observed**): Gameboard1–5 each with Devices / Environment / Floors / Walls. Setup folder for the Verse device + NPC spawner.

## Project configuration

Mount `b2869215-f52d-4092-8a77-5d42d136ef9b`. Experimental Scene Graph / Python / toolsets **Unverified**. Content-side: no Scene Graph entities required for the minigame (NPC + Creative devices). `submodules.verse` exposes `Textures.MiniGameUI` and `VerseCommander.{Meshes,Materials}` for UI/arrow.

`Developers/felixkupis/Collections/` empty. Eighth first-party hit.

## Device inventory (the design)

| Count | Class | Role |
|---|---|---|
| 1 | `VerseDevice` | `verse_commander_minigame` |
| 1 | `Device_CharacterSpawner` | NPC with behavior override `verse_commander_character` |
| 5 | `Device_CameraModes_FixedPoint` | one per board, `Camera.AddToAll` |
| 16 | `Device_CinematicSequence` | barrier appear/dissolve + openings (`SEQ_Reset` / `SEQ_Trigger`) |
| 11 | `Device_Barrier_V2_Placed` | obstacles |
| 10 | `Device_Trigger_V2` | obstacle pads + end goals |
| 2 | `Device_VFXSpawner_V2` | teleport in/out on the NPC script |
| 1 | `Device_Button_V2` | Play Verse Commander |
| 1 | `Device_HUDController_V2` | hide default HUD during minigame |
| 19 | `Device_Billboard_V2` | tutorial billboards + `design_cinematic_flythrough` |
| 40 | `Device_AI_Navigation_Modification` | nav on the tiled floors |
| 592 | `FortStaticMeshActor` | city wrap — throwaway |

Tile size default `vector3{X:=512, Y:=512, Z:=384}` — Fortnite tile (**Observed** `gameboard.verse`). Note XYZ here, not Left/Up/Forward: `/UnrealEngine.com/Temporary/SpatialMath`.

## Verse (8 files)

Copied to `snippets/verse/`.

| File | Class / job |
|---|---|
| `verse_starter_tags.verse` | `verse_commander_minigame_tag := class(tag)` |
| `commands.verse` | abstract `command` + `Commands.{Forward,TurnLeft,TurnRight}` instances |
| `gameboard.verse` | `obstacle`, `gameboard<concrete>` |
| `ui_manager.verse` | `minigame_button`, `command_queue`, `verse_commander_minigame_ui` |
| `verse_commander_minigame.verse` | `creative_device` — game loop, events to NPC |
| `verse_commander_character.verse` | `npc_behavior` — navigatable / focus |
| `design_cinematic_flythrough.verse` | billboard cinematic before play |
| `submodules.verse` | empty public modules for reflection |

**Event bus (device ↔ NPC)** — this is the architecture:

```
ExecuteCommandsEvent: event(tuple([]command, vector3))
BoardResetEvent:      event(transform)
GameStartedEvent / GameEndedEvent / CharacterFinishedMovingEvent
```

NPC finds the device with `FindCreativeObjectsWithTag(verse_commander_minigame_tag{})` then `verse_commander_minigame[CreativeObject]` (**Observed** character `OnBegin`). Docs show `GetCreativeObjectsWithTag` and `CreativeObjects[0]` — **drift**, see alignment.

## What this sample actually teaches

1. **Split a minigame by responsibility, talk through events + a gameplay tag** — not one god device. Device owns boards/UI; `npc_behavior` owns movement. Tag is the lookup because the NPC cannot `@editable` the device until it exists. See techniques/device-npc-event-bus.md.
2. **Commands as `<computes><abstract>` class + module-scope instances, not `enum`.** Enums can’t grow after publish; subclasses can. Dispatch with failable casts `Commands.forward_command[Command]`, not `Command = Commands.Forward`. See techniques/command-as-class.md.
3. **`<concrete>` + `@editable` arrays make a Verse device a level editor.** `Gameboards:[]gameboard`, each with Camera, EndGoal, Obstacles[], StartingCharacterPosition prop. Designers add boards without code. See techniques/concrete-editable-gameboard.md.
4. **NPC “forward” is `navigatable.NavigateTo`; “turn” is `focus_interface.MaintainFocus` raced with `Sleep(FocusTime)`.** No raw yaw API. ReachRadius 0.1 not 0.
5. **Custom HUD = `button_quiet` under art overlays** (`minigame_button`), textures from reflected `Textures.MiniGameUI.*`. Queue is a growing `[]command_queue`. Same `button_slot` family as User Interfaces, quieter chrome.
6. **`race` is the game loop.** LevelLoop races command-loop vs `EndGoalReached`; outer races that vs Reset; character races AwaitReset vs AwaitCommands vs GameEnded.
7. **Play is gated on a tag the template does not guarantee is set.** Follow the docs Getting Started or the minigame never wires.

## Official tutorial vs shipped

**External** hub: Verse Starter Template

Chapters (working URL shape `/documentation/fortnite/verse-starter-0N-…-in-unreal-editor-for-fortnite?lang=en-US`):

1. Creating the NPC Behavior
2. Defining Boards
3. Designing Levels — fetched outline via hub (top-down + command design)
4. Representing Command Data
5. Controlling the NPC with UI
6. Managing the Game Loop — hub-linked (matches `verse_commander_minigame.verse` structure)
7. Final Result — dump of Verse (do **not** train on ch.7 if it duplicates drifted snippets)

Spreadsheet title “Verse Device”; hub “voice - Verse Device Starter Games” / Verse Commander. Same project.

### What lines up

| Tutorial | Shipped |
|---|---|
| Custom NPC definition + `verse_commander_character` as behavior | file exists, `class(npc_behavior)` |
| Tag `verse_commander_minigame_tag` must be added to play | `verse_starter_tags.verse`; hub Getting Started |
| `command` abstract + Forward/TurnLeft/TurnRight instances in `Commands` module, not enum | `commands.verse` |
| `gameboard<concrete>` tile 512³, fixed-point camera, trigger end, obstacles of barriers+cinematics+trigger | `gameboard.verse` |
| Five Gameboards editable on the minigame device | docs ch.2 step list; live Gameboard1–5 folders |
| Verse UI: command buttons, execute/remove/reset, dynamic queue; textures via `submodules.verse` | `ui_manager.verse` |
| Events: ExecuteCommands, BoardReset, GameEnded, CharacterFinishedMoving | minigame device |
| NavigateTo vs MaintainFocus+Sleep for turn | character `ExecuteCommand` |

### What does **not** line up

| Topic | Tutorial | Shipped |
|---|---|---|
| NPC find-device | `GetCreativeObjectsWithTag` then `CreativeObjects[0]` | `CharacterTeleportOutVFX.FindCreativeObjectsWithTag(...)` then `for` + cast. **Tag query is on a VFX device, not a global Get.** |
| Command compare | ch.01 `if (Command = Commands.Forward)` | `if (Commands.forward_command[Command])` failable class query. `=` on class instances is the wrong API here. |
| `command` specifiers | ch.04 `class<computes><unique><abstract>` | `class<computes><abstract>` — **no `<unique>`** on shipped `command` |
| Class names in prose | `turnright_command` | `turn_right_command` / `turn_left_command` |
| New NPC behavior filename | typo **`vese_commander_character`** | `verse_commander_character` |
| Ch.01 OnBegin | single `[0]` object | loop over tagged objects until cast succeeds |
| SpatialMath on boards | XYZ 512 (Temporary/SpatialMath) | same — do not “fix” to Left/Up/Forward |

Train on snippets. Emitting `GetCreativeObjectsWithTag` + `Command = Commands.Forward` + `<unique>` on `command` is doc regurgitation.

### Training rule

```
Tutorial prose (tag before PIE, abstract command vs enum, concrete gameboard, race loops)
    → outline
Tutorial Verse in ch.01 (GetCreativeObjectsWithTag, Command =)
    → UNTRUSTED
snippets/verse/*.verse
    → ground truth
```

## Negative findings

- Minigame is **inert until the tag is set** — template does not self-tag.
- Plugin `Hidden: true` — discovery issue for some browsers.
- 1205 OFPA vs 187 authored: don’t treat the city as reusable kit.
- Streaming disabled. Developer scratch folder.

## Capture backlog

1. Live `@editable` Gameboards count / tag present on the Verse device (`DeviceToolset` / VerseTagMarkup). High — confirms Getting Started. *Open.*
2. `ui_manager` `minigame_button` — confirm `button_quiet` + overlay (docs). Snippet is 20k; skim is enough. *Waive unless training a button-chrome note.*
3. Ch.03 / ch.06 / ch.07 full fetch — outline covered; ch.07 would only duplicate Verse. *Waive ch.07 as training source.*

## Live notes

Level confirmed. Five `Verse_Commander_MiniGame/GameboardN` folders. Streaming off.
