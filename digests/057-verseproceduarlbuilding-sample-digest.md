---
id: digest-057
kind: digest
---

# verse_proceduarl_building — sample digest

> **Extractor provenance.** `tools/extract-template.py` (Pass A) + UnrealMCP on
> `localhost:8000` (Pass B); `SceneTools.get_current_level` returned
> `/2eb45cdf-a7e0-4f65-8137-fd73c55ae1cd/verse_proceduarl_building`.
> `VerseToolset.BuildAll` on the shipped sources returned **zero diagnostics** (compiles clean
> on compat 42.30).

| | |
|---|---|
| Classification | **Verse algorithm showcase** — voxel building system + shape grammar + WFC |
| Picker name | Verse – Procedural Building (Game Examples, spreadsheet R37, Keep=**No**, "Needs update… I would depricate") |
| Origin | **Re: Imagine London** demo (island code 1442-4257-4418) (**External** — tutorial page) |

## Verdict first

Keep=No on the spreadsheet, but this packet is **the richest Verse source in the corpus**:
43 files, ~9,000 lines, implementing three non-trivial systems (voxel grid + SDF convex
decomposition, a shape-grammar operator tree, Wave Function Collapse) that **no other template
teaches at all**. Extracted anyway, Deserted-style. If Epic retires the template, the
`snippets/verse/` deposit below is the only surviving copy of this code.

**Name typo is shipped state:** folder, `.uefnproject` title, and Verse path are all
`verse_proceduarl_building` (missing an 'r') (**Observed**). Packet dir matches the folder
name so extracts reproduce.

## Scale (**Observed** — `summary.txt`)

| | |
|---|---|
| Authored assets | 34 (13 BP props, 14 meshes, 3 materials, 2 textures, map, GameFeatureData) |
| Placed actors | 29 OFPA packages; **16 distinct classes** |
| Verse files | **43** (largest: `voxel_grid` 32.5 KB, `build_system` 25.5 KB, `wfc_system` 22.7 KB, `player_data` 21.7 KB) |
| Fab `.uref` | **0** — no Fab dependencies |
| Data layers | **0** (live editor confirms empty) |
| Feature flags | SceneGraph allowed, Python, toolsets. `editorPermissions` empty (public template) |
| Grid (live `@editable`) | build zone `gridSize = 10,10,10` voxels; `BuildGridSize = 500,500,400` cm per voxel |

## Device / system inventory (**Observed** — live `find_actors`)

| Count | Device | Role |
|---|---|---|
| 1 | `root_device` (Verse) | Singleton entry: owns 4 input triggers, player join/leave, HUD, cursor prop |
| 1 | `build_zone` (Verse) | One 10³-voxel build site; owns `build_system`, `wfc_system`, volumes, buttons |
| 1 | `prop_spawner_device` (Verse) | Event-driven spawn queue (`CheckOut`/`SpawnPendingProps`) |
| 1 | `prop_spawner_references_device` (Verse) | Maps `piece_type` → `creative_prop_asset` (12 pieces) |
| 4 | Input Trigger | Place / Delete / PrevCategory / NextCategory |
| 2 | Volume | Enter / Exit site |
| 3 | Button (V2) | Add Park / Add Building / Reset |
| 1+1+1 | HUD Message, HUD Controller, Billboard | Messages, HUD setup, info sign |
| 2+2 | Player Spawner Prop, FortPlayerStartCreative | Spawns |
| 4 | GridPlane | Ground |
| 2 | Agency stair + balcony | Static dressing near the site |
| 1 | FortStaticMeshActor "Merlin VoxelPlaceGuide" | Ghost guide mesh |
| 1 | Island Settings (ExperienceSettings V2) | — |

All `@editable` wiring verified live (**Observed** — `DeviceToolset.GetDeviceProperties`):
root → 4 triggers, cursor BP, 1 build zone, HUD message device, references device;
build zone → root, spawner, both volumes, all 3 buttons. **No event bindings** — everything
is direct Verse references.

## Content organisation and naming (**Observed** — `folders.txt`, `naming-prefixes.txt`)

```
Content/
  verse_proceduarl_building.umap
  Lib/            build_system, mesh_grid, voxel_grid, vector3i, box, ray_caster …
    SpaceGrammar/ vo_base + 17 operators, rs_base, rs_building1, rs_wfc
    WFC/          wfc_model, wfc_model_factory, wfc_system
  root-level      root_device, build_zone, player_data, global_player_data,
                  spawner_device, spawner_asset_references, ui_manager,
                  build_hud, cursor_ui, config, Extensions
  Props/{B1,Park} BP_B1_{Box,Corner,Face,I,O,U}, BP_WFC_{Grass,Water,…Straight,InCorn,OutCorn}, BP_Cursor, BP_box
  Meshes/{B1,Park} SM_* counterparts
  Materials/{Cursor,Grid}, Textures/{T_Checker, T_Cursor}
```

Prefixes are textbook `SM_/BP_/M_/MI_/T_`. Verse modules mirror intent: `Lib` (grids +
build orchestration), `Lib.SpaceGrammar` (shape grammar), `Lib.WFC` (WFC), root (devices +
player + UI).

## What this sample actually teaches

1. **Shape grammar as a declarative operator tree** (`techniques/shape-grammar-volume-operators.md`).
   `vo_base.Operate(Box, Rot, BuildSystem)` composite; rules are *assembled*, not branched.
2. **Wave Function Collapse in pure Verse** (`techniques/wave-function-collapse.md`).
   Edge-label symmetry/flip pairs, rotation-expanded mesh options, entropy = option count,
   observe/propagate queue, **incremental** add/remove with `RecalcRadius`. Tile collapse
   is stochastic (`GetRandomInt`); deterministic per-position RNG via prime hashing is
   scoped to spawn-time cosmetic prop/rotation selection.
3. **Voxel editing at interactive rates** (`techniques/voxel-grid-sdf-remesh.md`).
   Flat 1D-array grid + signed-distance field → quicksort by depth → greedy convex box
   decomposition → per-box grammar meshing; spawn-queue decoupling; leaf-prop diffing so
   unchanged props are never respawned; DDA voxel raycast + virtual-plane "turbo build".

## Official tutorial vs shipped

URL (spreadsheet `Documentation`; `.uefnproject` `docsUrl` is empty):
`dev.epicgames.com/documentation/en-us/uefn/procedural-building-template-in-unreal-editor-for-fortnite`
Single page: Using the Template → Building with Voxels (grids, raycasting, inputs, spawning)
→ Procedural Generation (Shape Grammar, WFC) → Potential Extensions → Verse Performance.

| Lines up | Evidence |
|---|---|
| Architecture prose (two phases: edit voxels → respawn meshes; root/build-zone/spawner devices) | Matches shipped class layout |
| `vector3i` struct snippet | Identical to shipped `vector3i.verse` |
| `GetVoxelIndex` math `(X*(Size.Y*Size.Z))+(Y*Size.Z)+Z` | Identical in `voxel_grid`/`mesh_grid`/`wfc_system` |
| `vo_sizecheck` full snippet | Byte-equivalent logic to shipped |
| `rs_building1<public> := class(rs_base)` | Shipped class is `<public>` (unlike module-scoped `vo_*` operators) |
| WFC `wfc_edge_label` / `wfc_mesh` snippets (WaterEL, GrassToWater flip pair, `Park_Grass_Water_InCorn` edges) | Match `wfc_model_factory` |
| "Verse Performance" advice: init arrays with `for` comprehension, not element-by-element `+=` (O(N²)) | Shipped code does exactly this (`Cells`, `SDF`, `PropGrid`, `Tiles`) |

| Does not line up | Shipped truth |
|---|---|
| Doc's `build_category` enum shows `Building, Park, NewBuildingType` | Shipped: `Building1, Park, Null` (`voxel_grid.verse`). Doc snippet is from the *extension* section, not the shipped enum |
| Doc names input triggers "Fire / Aim" | Level has **four**: Place, Delete, PrevCategory, NextCategory |
| Tutorial references `MerlinExample` content paths implicitly via Re:Imagine London | Shipped project still carries `/MerlinExample` refs (redirected into project mount) |

**Training rule:** snippets win. The doc's enum/identifiers drifted; the algorithms did not.

## Negative findings

- **No Scene Graph anywhere** — `IsSceneGraphSystemAllowed` is set but every spawned thing
  is a classic `creative_prop` via `SpawnProp`. Don't cite this sample for SG.
- **No persistence, no rounds, no teams** — matchmaking is max-16/unrestricted boilerplate.
- **No Fab deps, no data layers, no Sequencer, no audio design.** The map is a flat grid
  plane with two Agency props as dressing.
- **Shipped WFC edge-case omissions** — in `wfc_system.verse`, the first-constraint branch
  (`Tile.AllOptions?`) sets `Tile.RemainingOptions` without updating `NumRemainingOptions`
  or entropy; `ResetWFC()` resets the `Tiles` array to false without clearing `RemainingTiles`
  or `PropagationQueue`. In `vo_wfc.verse`, zero-option overconstrained tiles fail the outer
  `Tile.NumRemainingOptions = 1` guard, so `FallbackPropOpt` is never reached.
- **Shipped lifecycle defects** — in `player_data.verse`, `Removed()` calls `HideCursor()`
  (hiding the prop) instead of `Dispose()`, and cursor material maps are declared empty. In
  `build_zone.verse`, `RemovePlayer` filters with `Key <> ExistingPlayerData`, comparing
  an `agent` map key to a `player_data` value, so the agent is never removed from `AgentDataMap`.
- **Verification boundary** — The original project compiles clean in UEFN 5.5+, but
  clean compilation of the project does not verify runtime behavior or absence of logical defects (such as the lifecycle/WFC edge-case bugs above). Extracted snippets and generalized files are derived reference excerpts.

## Capture backlog

1. ~~Device wiring / grid size~~ — **closed live** (values above).
2. Runtime perf of 10³ grid + WFC regen in an actual session — needs PIE playtest; waived
   (algorithmic content already captured).
3. `Materials/Cursor/M_CursorMaster` graph — minor; waived.

## Live notes

- Level: `/2eb45cdf-a7e0-4f65-8137-fd73c55ae1cd/verse_proceduarl_building` (project GUID
  mount, not `/Game`).
- `VerseToolset.BuildAll` → `[]` (clean).
- Verse `@editable`s read with friendly names via `DeviceToolset` — no `__verse_0x` mangling
  needed on this dev build.
