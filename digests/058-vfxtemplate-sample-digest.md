---
id: digest-058
kind: digest
---

# vfx_template — sample digest

> **Extractor provenance.** `tools/extract-template.py` (Pass A) + UnrealMCP on
> `localhost:8000` (Pass B); `SceneTools.get_current_level` returned
> `/4be6a9e9-1a5d-4c5b-89c4-cfa10d49150f/vfx_template`.
> Spreadsheet R16 Keep=**No** (“Time to depricate”) — extracted anyway (Deserted-style).
> Docs: VFX - Lava Cave.

| | |
|---|---|
| Classification | **VFX demo island** — lava cave walk; three VFX surfaces + no-Verse quake cinematic |
| Content type | small **kit** (authored Niagara/materials) inside a LuminousFortress **catalog** wrap |
| Verse | **0 files** |
| Fab `.uref` | **0** |

## Verdict first

The teaching is not the castle. 537 LuminousFortress pieces + 108 rocks are a gallery wrap. What survives deletion — and what the 54 billboards actually caption — is **how UEFN splits VFX work**: authored Niagara for custom lava, VFX Creator for sprite dust/smoke/debris, VFX Spawner for a tinted `AreaOfFog`, and a Cinematic Sequence device as the earthquake bus (no Verse). Spreadsheet Keep=No does not cancel that split.

## Scale (**Observed** — `summary.txt` + live)

| | |
|---|---|
| Authored assets | **33** |
| Placed actors | **954** OFPA, 72 classes |
| Verse files | **0** (live `VerseToolset.ListFiles` on the project GUID lists only `__ExternalActors__`) |
| Data layers | **0** (`get_data_layers` → `[]`) |
| Fab `.uref` | **0** |
| Compatibility | **43.00** (same generation as AmbientEnvironment) |
| Matchmaking | 1 player |
| Streaming | Partitioned **and** `LevelHasStreamingDisabled=1` |

Empty `Developers/felixkupis/Collections/`. Another first-party scratch folder.

## Authored kit (**Observed** — `assets.txt`)

```
Content/
  VFX/                 NS_LavaBurst, NS_LavaSplash_01
    Materials/         M_Lava_{Bubble,Burst,BurstLarge,Ripple,SplashSubUV}_01,
                       MI_Lava_{Bubble,SplashSubUV}_01,
                       M_PostProcess_Shimmer, MI_PostProcess_ShimmerHigh,
                       M_DebugCircle
    Textures/          T_waterSplash, T_waterSplash_R_N, T_Debug_SubUV, T_CP_Grid_01_M
  Sequences/           FallingRock_01, FallingRock_02
  CustomMeshes/        SM_Chain_01, CP_LuminousFortress_Statue_A + statue textures/MIs
  MasterMaterials/     CustomRotator (material function)
  Intro_camera_sequence, NewLevelSequence
  camshaker, NewCameraShakeBase, earthquake (SoundWave)
```

Copy `Content/VFX/` + the two camera-shake assets. Do not copy the LuminousFortress / coffin / barrel dressing.

## Level organisation (**Observed** — live `get_folders`)

```
Atmosphere/            Day Sequence, LumenExposureManager, PostProcessVolume_Shimmer
  Lights/              Device_PointLight_V2 ×8 + LuminousFortress light meshes
    NonLumen/          engine SpotLights on statues + LightingScalabilityManager
    Silhouette/        engine SpotLights (labels typo: Slhouette)
  VFX/                 NS_LavaBurst ×~26, NS_Janus_Campfire ×6, VFX Spawner Fog
    Falling/           VFX Creator debris ×6          (quake, not ambient)
    Falling Dust/      VFX Creator grit ×4            (looping)
    FirePitSmoke/      VFX Creator smoke ×10
Billboards/            Device_Billboard_V2 ×54        (the syllabus)
Gameplay/              8 triggers, 4 creature placers, 2 managers, cinematic,
                       camshaker, 6 audio players, 8 player spawners, End Game
Lava/                  CP_Lava_Tile_* (falls / rings / 2L)
Rocks/                 108 modular rocks
Rocks_Ref/             13 — one of each rock class (copy palette)
Sequence/              NS_LavaSplash_01/02 + RockChunkLG C 01/02
SetDressing/           barrels, coffins, statues, 5 Barriers
Structure/             537 LuminousFortress walls/floors/trims
```

Root (no folder): IslandSettings, HUD Controller, Radio, Class Designer, Item Placer (`BP_ItemContainer`).

`Rocks_Ref` is an Outliner **palette** — one instance per rock class parked for duplication. Composition lesson, not VFX.

## Device inventory (the design)

| Count | Class | Role |
|---|---|---|
| 54 | `Device_Billboard_V2` | captions — `text` is on the **actor**, not ToyOptions |
| 20 | `Device_VFX_Creator_V2` | dust / smoke / quake debris |
| 34 | `NiagaraActor` | lava burst/splash + stock Janus campfire |
| 8 | `Device_Trigger_V2` | invisible; chain rooms + Play cinematic |
| 8 | `Device_PointLight_V2` | torch/brazier glow (`lightType` Spot on the sample) |
| 6 | `Device_CRD_AudioPlayer` | door / quake-adjacent cues |
| 4 | `Device_CreaturePlacer_V2` | Brutes (Trigger4) / Ranged (Trigger6) |
| 5 | `Device_Barrier_V2` | invisible lava fences + one door (`Barrier Device5`) |
| 1 | `Device_VFXSpawner_V2` | `AreaOfFog` lava steam |
| 1 | `Device_CinematicSequence` | plays `NewLevelSequence` |
| 1 | `camshaker_C` | `CameraShakeSource` → `NewCameraShakeBase` |
| 1 ea | Day Sequence, Lumen Exposure Manager, HUD Controller, Radio, End Game, Class Designer, Item Placer, Island Settings | lighting / HUD-off / victory radio / sword |

## What this sample actually teaches

1. **Three VFX surfaces** — techniques/vfx-surface-split.md. Niagara for custom lava, Creator for sprites, Spawner for named presets.
2. **Trigger → Cinematic Sequence → shake + one-shot VFX**, no Verse — techniques/trigger-cinematic-quake.md.
3. **Heat is a unbound PPV + Day Sequence fog, not more particles.** `PostProcessVolume_Shimmer`: `bUnbound=true`, priority 0, `weightedBlendables[0]` = `MI_PostProcess_ShimmerHigh`. Day Sequence: Priority **2000**, FixedTime ~4.06, `overrideFog=true`, density **0.005**, orange fog `(1, 0.21, 0.019)`, sunlight override **off**. Same Day Sequence + Lumen Exposure Manager pairing as day_sequence_template; this island sets `useNonLumenPost = true` on the manager (**Observed**).
4. **Billboards are the syllabus** (54 captions). Numbered halls are not used; exhibits sit in the cave and talk. Verbatim dump: `_raw/vfx_template/billboard-texts.md`.

## Official tutorial vs shipped

**External:** VFX - Lava Cave

**Single page**, four in-page sections (not a chapter hub): Brazier VFX → Torch Fires → Falling Particles and Camera Effects → Lava Bursts.

Spreadsheet `/en-us/uefn/vfx---lava-cave-in-unreal-editor-for-fortnite` redirects to the unsuffixed fortnite slug (`vfx-lava-cave`, one hyphen). `docsUrl` empty. Retrieved 2026-09-14 headed Chrome; article body was readable; chrome still showed Sign in. Related **Visual Effects** / Niagara firework-dust pages are the discipline hub, not this island.

Verbatim capture: `_raw/vfx_template/tutorial.txt`.

### Lines up

| Tutorial | Shipped |
|---|---|
| Feature Examples lava-cave walk; lost-sword framing | Item Placer + Class Designer + creature rooms; billboard 32 welcome |
| Post-process shimmer / heat waves | `PostProcessVolume_Shimmer` → `MI_PostProcess_ShimmerHigh` |
| Brazier = VFX device + Niagara + light | VFX Creator smoke + `NS_Janus_Campfire` + `Device_PointLight_V2` |
| Statue orange glow from spotlights | `Atmosphere/Lights/NonLumen` + `Silhouette` engine `SpotLight`s |
| Torch Niagara + customizable light | `NS_Janus_Campfire` + Point Light devices |
| Day Sequence + Lumen Exposure Manager | both in `Atmosphere/` |
| VFX Creator for falling dust + fire-pit smoke | folders `Falling Dust/` + `FirePitSmoke/` |
| Trigger starts cinematic; falling particles + camera shake | Trigger/Trigger2 → Cinematic `PlayFunction`; `camshaker` + Falling/ debris |
| Lava Burst Niagara on lava tiles | `NS LavaBurst*` over `Lava/` tiles |

### Does not line up

| Topic | Tutorial | Shipped |
|---|---|---|
| “UEFN point light” / Customizable Light Gallery | Point lights; gallery path | `Device_PointLight_V2` with `lightType` **Spot** (5% intensity sample). Engine lights are `SpotLight` |
| “Niagara point lights” on braziers | implied | Campfire Niagara + a **separate** Creative light device |
| Typos | `Niagra` twice | — |
| Gameplay | “brief volcanic journey” only | Full trigger chain, creatures, barriers, HUD-off, Radio on End Game. Not in the four sections |
| Sequence asset | unnamed cinematic | Live device plays **`NewLevelSequence`**, not `FallingRock_01/02` or `Intro_camera_sequence` (those have empty referencers) |
| Fog | not named | VFX Spawner **AreaOfFog** (billboard 41) |
| HUD | not mentioned | HUD Controller hides UI |

Train on: billboard captions + the two technique notes + `Content/VFX/`. Tutorial *prose* is the syllabus outline; tutorial *device names* lose to live `text` / `ListEventBindings`. Emitting “point light gallery prefab with attached Niagara” as this island is doc regurgitation.

## Negative findings

- **No Verse**, SceneGraph flag on, unused.
- `FallingRock_01`, `FallingRock_02`, `Intro_camera_sequence`: LevelSequence assets with **empty** `get_referencers`. Do not cite as the played shot until Sequencer bindings are readable.
- `CineCameraActor` appears once in the OFPA class histogram; `find_actors` by name did not see it (likely sequence-possessed).
- Statue spotlight labels ship `Slhouette`. Campfire path ships `Enviroment`.
- `useNonLumenPost = true` on Lumen Exposure Manager — this cave prefers the non-Lumen PPV path. Do not copy as “always on Lumen.”
- LuminousFortress / GothicRegion coffins / VillageDressing barrels are dressing.

## Capture backlog

1. Sequencer tracks/keyframes in `NewLevelSequence` (which possessed actors, times). *Unanswerable with current tools — waived.*
2. HUD Controller / Island Settings / Class Designer / Item Placer option payloads. *Waive — billboards state the jobs; not the VFX lesson.*
3. Per-instance Niagara user-var overrides on every `NS LavaBurst Small` vs large. Sampled `NS LavaBurst 0`; waived rest.
4. Audio Player cue names. *Waive — NightEnvironment already taught local audio-player options.*

## Live notes

- Level `/4be6a9e9-1a5d-4c5b-89c4-cfa10d49150f/vfx_template`.
- Niagara `GetSystemSummary` needs the **full object path** (`/…/NS_LavaBurst.NS_LavaBurst`). Package path without the duplicate name → “not a valid object path for property System”.
- Billboard captions: `ObjectTools.get_properties` on the **actor** `text` field. `ToyOptionsComponent` is present but the syllabus is the actor property.
- `ListEventBindings` is **target-centric** — invert to get the trigger graph (`_raw/vfx_template/event-graph.md`).
- Engine `SpotLight` properties live on `LightComponent0`, not the actor.
- `get_referencers` on OFPA-placed Niagara returns `[]` even when the component Asset points at the system — do not treat empty referencers as “unused” for placed actors.
