---
id: digest-059
kind: digest
---

# village_starter_template — sample digest

> Offline `_raw/village_starter_template/`. Live
> `/a224a06b-1319-4f78-87ff-7ee8b58c5dfb/village_starter_template` (**Observed**).
> Spreadsheet Keep=Yes, Docs=**No** (link `-`). **No island tutorial.** Guessed `/documentation/fortnite/village-starter-template-in-unreal-editor-for-fortnite` redirects to the Fortnite docs hub.

| | |
|---|---|
| Classification | **Open-world village hub** — homestead claim, day/night clock, interior/mine teleports |
| Content type | full island: **68 authored vs 8801 placed** |

## Scale

| | |
|---|---|
| Authored | 68 (`WBP_*` UI, audio cues, intro sequence, 2 custom props) |
| OFPA | 8801 (0 unclassified) |
| Verse | **3** |
| Fab | **5** — `Village{Environment,Props,SFX,UI,VFX}` |
| Compatibility | **42.30** |
| Matchmaking | **4 players / 1 team** |

Empty `Developers/felixkupis/Collections/`. Thirtieth first-party hit.

Art is Fab + Fortnite prop soup (591 classes: trees, mines, tudor, ice castle…). Same Deserted/character_environment split: tiny `Content/`, huge OFPA.

## Fab (5 packs)

```
References/VillageEnvironment_….uref → /EpicGames@Fab.com/VillageEnvironment  860615fc-…
References/VillageProps_….uref       → /EpicGames@Fab.com/VillageProps        86648a3e-…
References/VillageSFX_….uref         → /EpicGames@Fab.com/VillageSFX          eab30f86-…
References/VillageUI_….uref          → /EpicGames@Fab.com/VillageUI           19c1c4f4-…
References/VillageVFX_….uref         → /EpicGames@Fab.com/VillageVFX          127e5a02-…
```

Deserted 3 (Audio/Props/VFX). character_environment 4 (+UI). Village **5** (Environment + SFX instead of Audio). `projectId` filled. Filename hash ≠ mount UUID.

## Live rooms (not numbered halls)

Zones: `Zone_Homestead_Town` (buildings / market / pet shop / town hall) + `Zone_Overworld` + `Camera_Example`.

Interiors: `Interior_{Homestead,GeneralStore,TownHall}` (homesteads 1–4, one per claimed pad).

Mines: `MineEntrance` → teleporter chain `Exterior<>Entrance` / `Entrance<>SideScroller` / `SideScroller<>Raptors` / `Raptors<>Wellbottom` / `WellBottom<>Exit` → `Mine Exit`. Raptor combat pit.

Also: `IntroSequence/SpawnCloset`, `Global/{Cameras,NPC Devices/GoodDoctor,Time of Day}`, `Patchwork`, `TitleSequence`.

## Verse (the design)

| File | Class | Role |
|---|---|---|
| `village_player_spawner.verse` | `village_player_spawner` | claim a homestead pad; intro cam + stasis + fade; unclaim on leave |
| `village_zonetransition.verse` | `village_zonetransition` | button **or** volume → stasis + `Teleport` + 2s fade + release |
| `village_time_manager.verse` | `village_time_manager` | timer ticks hours; switch day/night; `HourIncrementEvent` |

See techniques/homestead-claim-spawn.md and techniques/zone-transition-stasis-teleport.md.

## Device inventory (high-signal)

| Count | Class | Role |
|---|---|---|
| 24 | teleporter | interiors + mine chain |
| 18 | `Device_CameraModes_Perspective` | interior/overworld cams |
| 14 | `Device_CameraModes_FixedPoint` | homestead intro cams |
| 14 | `Device_PopUpDialog_V2` | welcome / tips |
| 27 | trigger | volume transitions |
| 11 | lock | interior doors |
| 8 | `BP_DaySequenceModifier_Device` | day/night look |
| 6 | `BP_LumenExposureManager` | Lumen parity |
| 1 | `BP_LightingScalabilityManager` | scalable lighting |
| 3 | `VerseDevice` | the three scripts |
| 1 | `Device_Character_V2` | GoodDoctor NPC |
| 178 | VFX spawner | day/night weather, mines |
| rest | Fortnite Creative props | village dressing |

Authored kit worth copying: `UI/{ConversationDialogWidget,HUDMessagingWidget,Intro,TrackerWidget}`, `Gameplay/TimeOfDay`, intro `LS_*`.

## What this sample actually teaches

1. **Homestead = claimed pad.** Disable on spawn, `[agent]player_spawner_device` map, re-enable on `PlayerRemovedEvent`. Counter `TransmitForAllCounted` is a **failsafe** if `SpawnedEvent` missed.
2. **Intro is stasis + fixed-point cam + cine with timeout.** `race { StoppedEvent.Await() / Sleep(6)+GoToEndAndStop }` so a stuck sequence can't freeze the player.
3. **Interior teleports stasis the pawn** then `Teleport` then `Sleep(2)` (fade) then `ReleaseFromStasis`. Filter `player[Agent]` so wildlife/chickens don't ride the teleporter.
4. **Hour clock is a Timer device**, not `GetSeconds()`. Morning/Evening hours flip a `switch_device`; other systems `HourIncrementEvent.Await()`.
5. **Five Fab packs**, Environment vs Props split (character_environment merged props).

## Official tutorial vs shipped

**None.** Spreadsheet Docs=No, `docsUrl` empty. Signed-in Chrome 2026-09-11: guessed island URL **redirects to Fortnite Documentation hub**. Do not train as if a walkthrough exists. Outliner + Verse comments *are* the syllabus.

## Negative findings

- Scratch `Developers/` folder.
- Compatibility 42.30.
- `IsDay()` uses `CurrentHour >= MorningHour or CurrentHour < EveningHour` — with Morning=6 / Evening=20 that is **almost always true**. Night HUD path looks broken unless hours wrap the other way. *Unverified in PIE.*
- 591 actor classes — not a lean kit; don't copy the prop soup.

## Capture backlog

1. Confirm `IsDay()` in PIE at hour 22. *Open, medium.*
2. GoodDoctor conversation vs `ConversationDialogWidget`. *Waive — UI assets exist, no Verse in this trio.*

## Live notes

Level + zone tree confirmed. Chrome still signed in. No file >80MB.
