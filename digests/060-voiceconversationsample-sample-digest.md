---
id: digest-060
kind: digest
---

# voice_conversation_sample — sample digest

> Offline `_raw/voice_conversation_sample/`. Live
> `/4cf59425-a81b-4e3e-a5ee-d3fa84822a48/voice_conversation_sample` (**Observed**).
> **Not on the Keep spreadsheet** (too new). Owner path `/jay_villanueva@fortnite.com/voice_conversation_sample`.
> Compatibility **42.30**. DocsUrl empty. **Persona / LLM NPC**, not Conversation Device trees.

| | |
|---|---|
| Classification | **Voice Persona car-sales haggling** — `npc_behavior` + `persona_component` + structured output → gold |
| Content type | tiny kit: **5 authored vs 98 placed** (office set dressing) |

## Scale

| | |
|---|---|
| Authored | 5 (`RandomCharacterDefinition`, `WB_Catalog`, `WB_Score`, HLOD, GFD) |
| OFPA | 98 |
| Verse | **2** |
| NPC | 1 `Device_CharacterSpawner` |
| VerseDevice | 1 (`ui_manager`) |
| Button | 1 (catalog) |
| End game | 2 |
| Matchmaking | 1 player (maxPlayers=1) |
| Compatibility | 42.30 |

Live folders: **`Gameplay` only**. Empty `Developers/{alarranaga,felixkupis,jayvillanueva}/Collections/`. Twenty-third packet.

## What it is *for*

One player. One NPC (car salesman) on a Character Spawner. On spawn, Verse picks a random **buyer persona** (family / heavy-duty / single / outdoors / collector), appends it + a gold budget to `persona_component.Personality`, puts player+NPC on a **voice channel**, and `PromptToTalk`s a greeting.

Two **structured-output** actions:

- *Final offer* → award agreed gold, despawn
- *Kick* → −200 gold, despawn

Captions follow `CommitSayEvent` / `StopSayEvent`. Catalog UMG (`WB_Catalog`) is a **button**, not voice. Fade UMG then `Spawner.DespawnAll`.

This is **runtime LLM Persona**, not the Conversation Device (scripted trees). Nightfall / MEMORY.md already treats `persona_component` as the experimental point.

## Device inventory (the design)

| Count | Class | Role |
|---|---|---|
| 1 | `Device_CharacterSpawner` | salesman; Verse `npc_behavior` override |
| 1 | `VerseDevice` | `ui_manager` (HUD + catalog + captions) |
| 1 | `Device_Button_V2` | open catalog |
| 2 | `Device_GameEnd` | win / lose? |
| 3 | billboards | captions |
| 1 | HUD controller | |
| rest | office / chairs / computers | throwaway |

Authored: `CharacterDefinitions/RandomCharacterDefinition`, `UIAssets/WB_Catalog`, `UIAssets/WB_Score`.

## Verse (2 files)

Copied to `snippets/verse/`.

| File | Job |
|---|---|
| `persona_behavior.verse` | `npc_behavior`: voice channel, personality join, RegisterAction structured output, PromptToTalk, captions, despawn |
| `ui_manager.verse` | `creative_device` + `ManagerTag`: catalog AddWidget InputMode.All, score/caption/fade via UMG view bindings |

**APIs to train (Persona):**

```
using { /UnrealEngine.com/Conversations }   # persona_component, PromptToTalk, SetPersonality, GetAISession, RegisterAction
using { /UnrealEngine.com/AgentGroup }      # agent_group(T), has_voice_member_info
using { /UnrealEngine.com/Chat }            # voice_channel, AddChatChannel, SetConversationTarget
```

`@ai_description` on struct fields + `prompt_binding_definition{Name, Description}` is how the LLM decides to fire `FinalOfferFunction` / `KickFunction`.

**Gameplay is deferred to `StopSayEvent`** so the NPC finishes the line before despawn/score.

Find UI: `FindCreativeObjectsWithTag(ManagerTag)` then `ui_manager[Manager]`.

## What this sample actually teaches

1. **Voice Persona loop:** `agent_group` + `voice_channel` + `SimEntity.AddChatChannel` + `Player.SetConversationTarget(Persona, Channel)` then `PromptToTalk`. See techniques/persona-voice-structured-output.md.
2. **Structured output is RegisterAction(binding, struct, callback)** — not Conversation Device pins. Flags on StopSay, not inside the LLM callback.
3. **Personality is Join(existing, character prompt, budget as “starting offer”)** — “budget” is a hard ceiling to the model; “starting offer” is negotiable.
4. **UMG catalog is additive:** `AddWidget(..., InputMode.All)` to steal mouse; RemoveWidget to give it back. View-bind `AnimationPlay` / `FadePlay` pulse true→false.
5. **Tag-locate sibling devices** (`ManagerTag`) so the NPC behavior doesn’t `@editable` the HUD.

## Official tutorial vs shipped

**No island tutorial.** `docsUrl` empty. Not on Templates list.xlsx.

Fetched (signed-in Chrome, 2026-09-09):

| URL | Result |
|---|---|
| `/documentation/fortnite/conversations-in-unreal-editor-for-fortnite` | **Conversation Device** trees — **different product** |
| `/uefn/persona-in-unreal-editor-for-fortnite` | **redirects to Fortnite docs home** |
| `/fortnite/persona-component-in-verse`, `/using-persona-in-…` | same redirect |

Snippet comment: *“To learn more about Structured Output … head to the Persona documentation page”* — that page is not public at those slugs.

**Official docs: none found for this island.** Conversation Device docs must **not** be trained as this sample.

## Negative findings

- 98 office actors are not the lesson.
- Scratch `Developers/` for three users.
- Compatibility 42.30 — Persona APIs may have moved by 43.
- `.lore/` is local Lore VCS — do not train, do not commit.

## Capture backlog

1. Live Character Definition / Persona component fields (model, voice, base Personality). *Open — editor Details, high value.*
2. `WB_Score` / `WB_Catalog` view bindings. *Open, medium — Verse field names are the contract.*
3. Public Persona docs URL if it ships later. *Open.*

## Live notes

Level confirmed. Folder `Gameplay`. Streaming off (not re-verified).
