---
id: digest-061
kind: digest
---

# horde_npc_sample — sample digest

> **Extractor provenance.** `tools/extract-template.py` (Pass A) + UnrealMCP on
> `localhost:8000` (Pass B). Default level
> `/a6e35c5f-014b-4b20-9f00-60866557bd08/Levels/horde_npc_sample` (**Observed**).
> Hub: TWDU Walker NPC Template.
> `docsUrl` empty; spreadsheet Documentation=Yes. Compat **43.00**. `featureSet`: **CustomGameplayFeature**.

| | |
|---|---|
| Classification | **theme learning island** — Walker setup room + horror corridor + action cafeteria |
| Content type | **catalog wrap** (26 authored vs 876 OFPA; meadow kit + TWDU devices) |
| Verse | **0 files** |
| Fab `.uref` | **0** |

## Verdict first

The picker name is the enemy type. The island is a **three-room syllabus**: how to author a Walker Character Definition, then two level-design beats (linear chokepoints vs open kiting) plus FPS / Heavy Linework / theme HUD toggles. Teaching that survives deletion is the definition + spawn-gating, not the prison geo. The sandbox prison itself is prison_sample (catalog wrap; **no** Walkers placed).

## Scale (**Observed** — `summary.txt` + live)

| | |
|---|---|
| Authored assets | **26** |
| OFPA | **876** |
| Maps | `horde_npc_sample.umap` (default). `LevelDesign_Iteration.umap` is an **ObjectRedirector** here |
| Verse files | **0** (`VerseToolset.ListFiles` on `/a6e35c5f-…` lists folders only; do not recurse `__ExternalActors__`) |
| Verse devices placed | **0** |
| Data layers | **0** |
| Fab `.uref` | **0** |
| Compatibility | **43.00** |
| Matchmaking | **1** player (`maxPlayers=1`; Island Settings `teamSize=Dynamic`) |
| Streaming | Partitioned + actor folders + OFPA. `LevelHasStreamingDisabled=1` |

Empty `Developers/felixkupis/Collections/`. First-party scratch folder.

`experimental.*` (Python / toolsets / Scene Graph) in `.uefnproject` is **Unverified** — do not treat as shipped. Content-side: **no** `.verse`, **no** `__ExternalEntities__`.

## Level organisation (**Observed** — live `get_folders`)

```
Room 1 - Walker Setup/          demo Walkers, RifleWeapon, flashlight, FPS/PPE/HUD boards
  TutorialBB/                   BB_Tutorial_1–8, 17–18, BB_WalkerSetup*, FPS/PPE toggles
  GameplaySystems/              4 NPC Spawners, 3 switches, 3 item spawners, 5 locks
Room 2 - Corridors and Chokepoints/    horror
  TutorialBB/                   LEVEL DESIGN TIP boards 9–16, 19
  WalkerSpawners/               13 NPC Spawners (gated)
  GameplaySystems/              12 locks, 3 volumes, 2 items, EndRoomTP, FPS switch
  Checkpoints/                  2 Player Checkpoint Pads
Room 3 - Open Areas and Kiting/        action cafeteria
  TutorialBB/                   ACTION / HORROR titles + arena tips
  GameplaySystems/              5 infinite spawners, 4 timers, 2 triggers, 3 volumes, 12 locks
Background/                     sky backdrops
External Corridors/             dressing
Landscape/  Lighting/           Day Sequence; point lights
```

Billboard captions: `_raw/horde_npc_sample/billboard-texts.md`. `text` is on the **actor**. Actors were loaded (`bIsSpatiallyLoaded=false`; streaming tag off).

## Device inventory (the design)

| Count | Class | Role |
|---|---|---|
| 39 | `Device_Billboard_V2` | syllabus |
| 29 | `Device_Keylock_V2` | cell / arena doors |
| 22 | `Device_CharacterSpawner` | all → `WalkerDefinition` |
| 12 | `Device_ItemSpawner_V3` | RifleWeapon, BatWeapon/bat, flashlight, shells, medkit |
| 7 | `Device_Volume` | horror wake + arena start/leave/checkpoint |
| 5 | `Device_HUDMessage_V2` | arena start / finish / escape / weapons |
| 5 | `Device_Switch_V2` | Spawn Walker, PPE, FPS ×3 |
| 4 | `Device_Timer_V2` | 120 s escape + weapon/catwalk delays |
| 4 | `Device_Prop_PlayerCheckpoint_V2` | Room 1 + Horror ×2 + arena |
| 2 | `Device_Trigger_V2` | arena start / finish bus |
| 2 | `Device_Teleporter_V2` | Horror exit (`EndRoomTP`) + one more |
| 2 | `Device_ItemRemover_V2` | arena cleanup (docs table omits) |
| 1 | `Device_Barrier_V2_Placed` | ceiling hole |
| 1 | `Device_Button_V2` | Give up |
| 1 | `Device_CameraModes_FirstPerson` | default 1P; three switches |
| 1 | `Device_PostProcess` | `PP_HeavyLinework`, starts off |
| 1 | `Device_HUDController_V2` | TWDU ExtraAssets widgets |
| 1 | Island Settings / Day Sequence / Lumen Exposure / player spawner | session |

## Authored kit (**Observed** — `assets.txt`)

```
Content/AI/            WalkerDefinition (live). WalkerCharacterDefinition = redirector, 0 refs
Content/Art/Materials/ instruction MIs + unused M_CustomHUD* / Material_Walker (0 refs)
Content/Art/Textures/  poster + HUD mockup textures (T_Player_Vitals → unused HUD MI)
Content/Meshes/        CubeGridToolOutput_* ×4 — 0 referencers (greybox leftover)
```

HUD that actually shows is **not** those materials. Live widget overrides: `/EnvironmentPack_ExtraAssets/UI/Widgets/UW_HUD_PlayersInfoStack` + `UW_HUD_EquippedInfoBlock`. Same ExtraAssets pattern as social-elimination (`/SOCIAL_WAVE_ExtraAssets/…`). Docs: you **cannot copy** the UI folder — reference it.

## What this sample actually teaches

1. **Walker = Character Definition Type, not Verse.** techniques/walker-character-definition.md. All 22 spawners point at `WalkerDefinition`. `nPCBehaviorScript` empty.
2. **Horde pacing is Enable/Spawn/Disable, not more definitions.** techniques/volume-checkpoint-gated-horde.md. Horror: volume + checkpoint, finite counts. Action: five infinite 8–10 count hoses, 120 s HUD timer.
3. **One First Person Camera, three switches.** techniques/switch-gated-first-person.md. `AddToPlayer` / `RemoveFromPlayer`. Priority 0, Add to Players on Start. BatWeapon still forces 3P while held (docs).
4. **Heavy Linework is a Post Process device + a switch**, not a volume. Live: `PP_HeavyLinework`, `enabledDuringPhase=None`, `EnableForAll`/`DisableForAll`. Day Sequence still owns sun/sky for the comic look (PPE chapter).
5. **theme HUD is HUD Controller widget overrides**, files stay in ExtraAssets. HideElement on minimap + build menu; health/shield **DoNotOverride** (custom widgets replace them). Text chat MiddleRight / CenterRight. Same recipe as social_deduction_sample (SOCIAL_WAVE widgets). Deeper UMG stays with user_interfaces_template.
6. **Hero weapons are Item Spawners**, not Item Granters. Room 1: RifleWeapon ×2 + flashlight pistol. BatWeapon live in Room 2 (`[Item] EnvironmentPack BatWeapon3`) and Room 3 (`Item Room 3 Bat`). Content Browser: zombie-survival > Weapons.
7. **Greybox then kitbash.** Docs: CubeGrid blockout → TWDU meadow pieces. Shipped CubeGrid meshes have **0** referencers — the lesson is the process, not the leftover SM.

## Official tutorial vs shipped

**External hub + companion chapters** (headed Chrome 2026-09-14; Sign in chrome still visible, article body readable):

1. TWDU Walker NPC Template
2. zombie-survival Weapons
3. zombie-survival Level Design
4. zombie-survival Cameras
5. TWDU Custom HUD: Vitals and Selected Item
6. Heavy Linework Post-Process Effect
7. Configuring Walkers (linked from hub; not a template-room walkthrough)

Verbatim capture: `_raw/horde_npc_sample/tutorial.txt`. Spreadsheet `docsUrl` empty; hub exists anyway.

### Lines up

| Tutorial | Shipped |
|---|---|
| theme Templates → zombie-survival Universe → Walker NPC | Project Browser; `featureSet` CustomGameplayFeature |
| Room 1 = Walker setup + weapons; then Horror / Action fork | Outliner Rooms 1–3 |
| NPC Spawner + `WalkerDefinition` in `Content/AI` | 22 spawners, that asset |
| Type Walker / Walker (Prisoner Uniform); prisoner is appearance | Type = Random Default; Prisoner Color Orange unused |
| Walker Gameplay Modifier defaults (headshot On, attack 12, bite 2, wander min 3) | Live matches; wander **max 15** (Configuring Walkers) |
| Horror: 13 NPC Spawners, 12 locks, 3 volumes, 1 teleporter | Exact |
| Action: 5 NPC Spawners, 3 volumes, 2 triggers, 4 timers, 7 item spawners, 5 HUD, 12 locks, 1 barrier, 1 checkpoint | Exact on those rows |
| Arena 2-minute countdown | Timer `duration=120`, HUD on |
| FPS: Priority 0 + Add to Players on Start; switches labeled FPS | Live camera + 3 switches |
| HUD: PlayersInfoStack + EquippedInfoBlock; cannot copy UI folder; chat Middle Right / Center Right | `/EnvironmentPack_ExtraAssets/UI/Widgets/UW_HUD_*` |
| Heavy Linework on Post Process device; toggle in starting area | `PP_HeavyLinework`; `Switch_PPEToggle` |
| CubeGrid then TWDU kit | Docs process; leftover CubeGrid assets unused |
| Single-player learning island | `maxPlayers=1` |

### Does not line up

| Tutorial | Shipped |
|---|---|
| Weapons + Level Design tables: wander max **16 s** | Live **15**; Configuring Walkers also **15.0** |
| First room: BatWeapon **and** RifleWeapon | Room 1 item spawners: **RifleWeapon ×2 + flashlight pistol**. BatWeapon is Room 2 / Room 3 Bat |
| Horror: **1** checkpoint, **1** item spawner | **2** checkpoints, **2** item spawners |
| Action device table | Extra live: FPS switch, **Item Remover**, **Give up** button |
| Hub: “new Character Definition, type Walker” as if `WalkerCharacterDefinition` is the file | Live file is `WalkerDefinition`; `WalkerCharacterDefinition` is a **redirector**, 0 refs |
| Docs typo “T he zombie-survival” (Cameras), “automictically” (Weapons) | — |

Training rule: billboards + live spawners/bindings = ground truth. Do not emit wander-max **16** or “BatWeapon in the setup room.”

## Negative findings

- **Name implies a Walker encounter; files show a three-room lab.** No round manager, no Verse, matchmaking 1.
- **`WalkerCharacterDefinition` is not a second definition.** Redirector → `WalkerDefinition`, **0** referencers.
- **`M_CustomHUD*` / `Material_Walker` / four `CubeGridToolOutput_*` have 0 actor referencers.** Do not cite as worked HUD/geo examples. HUD is ExtraAssets widgets; posters use `MI_Instructions_*`.
- **`GetDeviceProperties` does not apply** — no Verse `creative_device`s.

## Capture backlog

| Item | Status |
|---|---|
| Exact item-spawner loot rows (WID paths) beyond labels | **Waived** — labels + docs weapons chapter enough |
| Sequencer / keyframes | **Waived** — none are the lesson |
| PIE bite DoT / BatWeapon 3P Telekinetic | **Waived** — docs + modifier values |
| Room 2 volume *which* debris pile | **Waived** — bindings on spawners are the pattern |

## Live notes

- Mount `/a6e35c5f-014b-4b20-9f00-60866557bd08`. Project Verse path `/epictemplates@fortnite.com/horde_npc_sample`.
- `get_actors_in_folder` reports `folderPath: None` on OFPA descriptors; the query folder is still correct.
- Billboard `text` is on the actor (same as VFX / Social). Streaming disabled, so `load_actors` was not required this pass.
- Did **not** `load_level` the iteration redirector. Default map left as `horde_npc_sample`.
